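# Level Sequence that animates the TimeSquare cosmetic backpack's wings through a Control Rig.
# A Backpack skeletal mesh binding, parented to another binding, carries a parameter track for F_MED_TimeSquare_Pack_CtrlRig.
# In the visible left-wing channels, humerus, radius and metacarpal controls animate mainly through dense, irregularly timed Rotation.Y keys, producing repeated folding and unfolding.
# Location and scale on these wing controls hold at rest values.
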LevelSequence /BRCosmetics/Accessories/FORT_Backpacks/Backpack_F_MED_TimeSquare/Anim/TimeSquareAnimPACK_Puppet_LS
bindings [
 Backpack {
  id=#1
  parent_id=#2
  possessed_class=SkeletalMeshComponentBudgeted
  tracks [{
    type=MovieSceneControlRigParameterTrack
    display_name=F_MED_TimeSquare_Pack_CtrlRig
    sections [
     {
      type=MovieSceneControlRigParameterSection
      channels [
       pack_ctrl.Location.X {type=MovieSceneScriptingFloatChannel, default=0}
       pack_ctrl.Location.Y {type=MovieSceneScriptingFloatChannel, default=0}
       pack_ctrl.Location.Z {type=MovieSceneScriptingFloatChannel, default=0}
       pack_ctrl.Rotation.X {type=MovieSceneScriptingFloatChannel, default=0}
       pack_ctrl.Rotation.Y {type=MovieSceneScriptingFloatChannel, default=0}
       pack_ctrl.Rotation.Z {type=MovieSceneScriptingFloatChannel, default=0}
       pack_ctrl.Scale.X {type=MovieSceneScriptingFloatChannel, default=1}
       pack_ctrl.Scale.Y {type=MovieSceneScriptingFloatChannel, default=1}
       pack_ctrl.Scale.Z {type=MovieSceneScriptingFloatChannel, default=1}
       dyn_wings_humelus_l_ctrl.Location.X {type=MovieSceneScriptingFloatChannel, default=0, const [0,7]}
       dyn_wings_humelus_l_ctrl.Location.Y {type=MovieSceneScriptingFloatChannel, default=0, const [0,7]}
       dyn_wings_humelus_l_ctrl.Location.Z {type=MovieSceneScriptingFloatChannel, default=0, const [0,7]}
       dyn_wings_humelus_l_ctrl.Rotation.X {type=MovieSceneScriptingFloatChannel, default=0, const [0,7]}
       dyn_wings_humelus_l_ctrl.Rotation.Y {
        type=MovieSceneScriptingFloatChannel
        default=0
        thinned_from=808
        keys [[-121,-0.1457],[1,-0.1457],[5.6725,0],[7.5,0.2914],[7.8587,0.4371],[8.2425,0.8655],[8.9488,1.0112],[14.2988,0.8655],[17,1.0025],[18.0987,1.2852],[18.4563,1.868],[18.84,2.7423],[19.2138,3.7535],[19.6025,5.0474],[20.005,6.0586],[20.3913,6.7697],[20.7725,7.1982],[21.4825,7.4809],[23.8813,7.7723],[24.23,8.3464],[24.99,8.9205],[27.0412,9.0662],[27.4013,9.9317],[28.535,12.1085],[28.9188,12.2542],[31.3062,12.8284],[31.6725,14.8421],[32.055,17.1473],[32.4712,18.5956],[32.8525,19.1784],[40.945,19.3154],[51,20.0857],[52,20.2037],[55,19.94],[56,15.843],[57,10.8543],[58,5.7199],[59,3.3626],[60,1.7251],[61,-0.3582],[62,-0.7953],[65,-0.6583],[79.1888,-0.9915],[82.2513,-0.8458],[82.5925,-0.8371],[82.955,-0.4087],[83.325,-0.263],[83.685,-0.1173],[88.4438,-0.263],[89.1488,-0.4087],[89.4987,-0.5457],[89.86,-0.6914],[90.2188,-0.8284],[90.9312,-0.8197],[91.295,-1.085],[91.6788,-1.0502],[92.0738,-0.5783],[92.4675,0.1849],[92.875,1.2396],[93.2513,2.2856],[93.6388,3.4773],[94.0462,4.3689],[94.4262,4.8234],[94.8038,5.6977],[95.5538,6.5806],[95.915,6.872],[96.2875,7.3091],[96.66,7.3178],[97.0263,7.3091],[97.7313,7.6005],[98.8537,8.466],[100.2875,9.0488],[100.6837,9.4859],[101.9588,10.7972],[102.3413,11.517],[102.7062,11.9541],[103.0838,12.674],[103.4575,13.2568],[103.8525,13.8396],[104.2287,14.5681],[104.5963,15.4423],[104.99,16.3165],[105.3637,17.4734],[105.7437,18.4934],[106.1363,19.5046],[106.5113,20.5158],[106.9088,21.2443],[107.6688,22.984],[108.05,24.141],[108.4262,25.1609],[109.2,26.918],[109.5863,27.5008],[109.9513,27.7922],[110.315,27.9379],[111.6975,28.2293],[112.0638,28.6577],[112.8075,29.8233],[113.1762,30.6888],[113.9337,32.1372],[114.3,32.72],[114.6737,33.1484],[115.0475,33.7312],[115.4212,34.4597],[116.1825,36.2081],[116.5663,37.2281],[116.9525,38.1023],[117.3288,39.1135],[117.71,39.9877],[118.0838,40.9989],[118.4638,41.7274],[118.8512,42.3102],[119.21,42.7473],[119.585,42.8931],[121.6488,42.7473],[122.6762,42.6016],[123.03,42.3102],[123.3975,41.0076],[123.7862,38.5568],[124.1737,36.2342],[125.34,31.2891],[126.1225,27.8096],[126.5113,26.1982],[126.9112,25.0326],[127.2962,23.8582],[128.0713,22.1098],[128.4625,21.527],[128.8425,21.2356],[129.2138,21.0899],[133.105,21.2269],[133.5025,21.0725],[133.89,20.4723],[134.27,20.1548],[134.6688,19.4089],[135.0537,18.4999],[135.445,16.9907],[135.8663,14.8987],[136.2513,12.3696],[136.6762,9.7035],[137.0788,7.9203],[137.4812,6.5916],[137.8712,6.0001],[138.2387,5.7087],[138.6175,5.8457],[139.335,5.4086],[139.6962,5.0998],[140.065,4.654],[140.4538,4.5083],[143.5038,4.3452],[143.875,3.5819],[144.2625,2.2271],[144.6712,0.5635],[145.0738,-1.2631],[145.4875,-3.0811],[145.8813,-4.9078],[146.28,-7.4717],[146.6837,-9.9073],[147.0813,-11.897],[147.8925,-15.1981],[148.2962,-16.2267],[149.045,-17.4097],[149.42,-17.8469],[149.8012,-18.4471],[150.1812,-18.8929],[150.5675,-19.0559],[150.9388,-19.6388],[151.3088,-19.7932],[151.6812,-20.0933],[152.045,-20.2563],[152.4312,-20.7021],[152.8075,-20.5825],[153.1825,-20.9],[153.5663,-21.5002],[153.9488,-21.9634],[154.3462,-22.7267],[154.7375,-23.3704],[155.15,-24.1511],[155.5475,-24.6577],[155.9312,-25.5841],[156.3288,-26.3474],[156.735,-26.6562],[157.13,-26.8106],[165.1963,-26.6562],[165.5888,-26.2104],[165.9713,-25.4471],[166.3763,-24.3924],[166.775,-23.6291],[167.1762,-23.466],[167.9262,-22.8658],[168.32,-22.1286],[168.7275,-20.9369],[169.165,-19.3081],[169.5775,-17.962],[169.9875,-16.779],[170.3875,-16.3245],[170.7613,-15.7417],[171.1338,-15.4416],[171.5038,-14.9958],[171.875,-14.2586],[172.2775,-13.3844],[172.6537,-12.6385],[173.0412,-12.0557],[173.4262,-11.91],[177.5175,-11.6099],[177.8975,-10.727],[178.2763,-9.544],[178.6525,-7.6064],[179.065,-5.9602],[179.79,-4.7771],[180.1912,-2.9939],[180.6025,-1.5195],[180.9975,-0.6453],[181.7262,-0.4996],[191.565,-0.6453],[192.2912,-1.0911],[192.6525,-1.3999],[193.0412,-2.1371],[193.4137,-2.5829],[193.7962,-2.883],[194.1637,-3.1831],[194.5375,-3.3375],[194.9125,-3.929],[195.6913,-4.5292],[196.1838,-5.7383],[196.6762,-5.8927],[197.0775,-6.0471],[197.4475,-6.0558],[197.8188,-6.4929],[198.1912,-6.7843],[198.5562,-6.93],[199.2575,-7.0844],[199.62,-7.6759],[199.9975,-7.976],[200.3675,-8.1217],[205.12,-8.113],[205.4787,-7.5215],[205.8763,-6.7669],[206.2587,-5.1207],[206.6737,-4.8206],[207.0487,-4.0747],[207.4388,-1.7261],[207.8463,0.1854],[208.23,2.534],[208.65,4.874],[209.0437,7.5053],[209.4338,9.9996],[209.8412,11.7567],[210.2287,12.7767],[210.61,13.3595],[215.3575,13.2225],[217.0625,13.0768],[218.125,12.6397],[218.4925,12.3483],[218.87,12.3396],[219.5888,11.474],[219.9625,11.3196],[220.3325,10.8825],[222.7462,11.0282],[224.4438,11.1739],[226.4812,11.3109],[226.85,11.0022],[227.2487,9.5017],[227.7625,6.9552],[228.2337,4.7088],[228.665,3.217],[229.0512,1.8883],[229.44,0.5595],[229.8487,-0.4778],[230.2287,-1.3954],[230.6363,-2.6045],[231.4388,-4.7748],[231.8225,-5.9752],[232.2113,-7.0212],[233.3212,-7.4757],[234.705,-7.6301],[235.0638,-7.9389],[235.4312,-8.239],[236.8188,-8.3847],[237.5138,-8.239],[238.1988,-8.3847],[240.2375,-8.3934],[240.5975,-8.8392],[241.3338,-9.148],[241.7025,-9.4394],[243.7325,-9.3024],[246.1162,-9.3198],[246.4987,-9.3545],[246.8675,-8.9696],[247.2513,-8.4564],[247.64,-7.6691],[248.0462,-7.0275],[248.4737,-6.5143],[248.8675,-6.1294],[249.6437,-6.2424],[250.0263,-6.1836],[250.435,-5.8335],[250.83,-6.0401],[251.2088,-5.9378],[251.5987,-6.0835],[253.9975,-5.9378],[263.2862,-5.9291],[263.6375,-6.0661],[264.0013,-6.0574],[264.375,-5.9204],[270.0962,-5.7747],[270.45,-5.3202],[270.8225,-5.0027],[271.2025,-4.3852],[271.6175,-3.0217],[272.0375,-1.4864],[272.4588,0.3316],[272.8475,2.167],[273.2575,3.102],[273.645,3.4021],[274.01,4.2938],[274.3963,4.894],[274.765,5.1854],[275.1375,5.0483],[276.4887,5.1854],[280.5512,5.6312],[280.9112,6.8142],[281.3,7.6971],[281.665,8.58],[282.435,10.0457],[282.8075,10.3458],[283.1787,10.7829],[283.5462,10.9286],[284.5838,11.0656],[284.9375,11.2026],[289.4525,11.0569],[292.815,10.9025],[293.2025,10.7655],[293.5812,10.4741],[294.6575,10.6111],[296.7525,10.5937],[297.1375,10.5676],[297.52,10.3698],[297.9337,10.3002],[298.71,10.3829],[299.1213,10.1329],[299.5075,9.0086],[299.9137,7.6104],[300.6625,7.3016],[301.035,6.4013],[301.42,5.8098],[302.1775,5.5097],[308.2663,5.364],[308.6212,5.501],[308.9762,5.4749],[309.3537,5.1661],[309.7225,5.2683],[310.1225,5.3618],[310.515,5.7381],[310.9175,5.9686],[311.3162,6.3449],[311.7025,6.4732],[313.1,6.3362],[313.455,6.3449],[313.8325,7.0995],[314.2163,9.5089],[314.6212,12.5012],[315.4,19.0339],[315.805,21.5543],[316.21,23.1831],[316.6337,24.0573],[317.0025,24.3487],[318.3775,24.4944],[323.42,24.6401],[323.7738,25.9514],[324.175,27.8542],[324.5838,29.6114],[324.9825,31.6512],[325.3837,33.2626],[326.1625,35.1567],[326.535,35.6025],[326.9237,35.7482],[332.3413,36.4941],[332.7075,38.2513],[333.5013,42.9572],[333.91,44.4229],[334.29,45.2971],[337.3625,45.1514],[337.71,44.997],[338.535,42.7593],[338.93,41.1305],[339.3363,39.6474],[339.7188,38.173],[340.505,35.7983],[340.9137,35.0437],[341.295,33.8607],[341.7337,32.9604],[342.1513,31.7861],[342.5512,31.1859],[342.9188,30.4487],[343.2925,29.8572],[343.68,29.5571],[344.05,29.5484],[344.4175,28.9656],[344.7838,28.5198],[345.1538,28.2023],[345.545,27.6021],[346.3112,26.6931],[346.6825,26.0929],[347.06,25.6297],[347.4588,24.8838],[347.8425,24.4206],[348.2237,23.6834],[348.9875,23.0745],[351.0638,22.9288],[351.7538,22.7744],[353.6025,22.0111],[354.025,21.9937],[354.4875,21.8306],[355.28,21.8306],[356.63,21.9763],[356.9725,22.2677],[357.3438,24.1532],[357.7362,26.0473],[358.135,27.2216],[358.5625,28.0958],[359.33,28.6699],[360.4362,29.107],[369.2212,29.0984],[369.5788,29.2441],[370.2875,29.0984],[372.68,29.0897],[373.0325,28.9353],[373.4112,28.9266],[373.7775,28.6352],[374.5163,28.3351],[382.365,28.1894],[384.0738,28.3264],[386.2713,28.1807],[386.635,28.0263],[387.005,27.7175],[387.3975,27.2543],[387.7812,26.6367],[388.1938,25.4363],[388.5963,23.6357],[389.0025,21.5611],[389.7987,18.8863],[390.1988,18.4492],[390.5725,18.2948],[391.965,18.1404],[392.3425,17.5576],[392.72,17.3945],[393.0888,16.9487],[393.8288,16.9487],[395.1838,17.0857],[396.2225,17.0683],[396.585,16.8965],[396.97,16.4159],[397.365,16.081],[397.7437,15.5831],[398.1762,15.0851],[398.5825,14.4501],[398.9837,13.9869],[399.36,13.3954],[402.1438,13.352],[402.5187,13.1715],[402.9325,12.8366],[403.3112,12.5017],[404.4737,12.0037],[404.8438,11.995],[405.565,11.8406],[405.9413,11.8233],[406.3,11.6602],[406.6675,11.3514],[407.035,11.471],[407.4025,11.1796],[407.775,11.0252],[408.135,11.1535],[408.5125,10.8534],[408.875,10.5359],[409.2587,9.9183],[409.66,9.1289],[410.0675,8.0395],[410.4613,6.5129],[410.8863,4.9863],[411.2825,3.9229],[411.6762,2.7312],[412.0813,1.8396],[412.4725,1.0937],[412.8612,0.2021],[413.2337,-0.2524],[413.955,-0.7069],[414.3162,-2.0617],[414.7188,-3.7166],[415.1087,-5.8173],[415.52,-7.6179],[415.9062,-8.9553],[416.2962,-9.9839],[416.6975,-10.1296],[420.2937,-9.9839],[421.68,-9.8382],[422.0238,-9.5468],[422.395,-9.1097],[422.77,-8.964],[424.1538,-8.8183],[424.5075,-8.3812],[424.8775,-7.5069],[425.255,-5.9042],[425.6475,-3.573],[426.0562,-1.2418],[426.4575,0.9264],[426.865,1.7919],[427.24,2.2203],[427.6112,2.366],[430.3438,2.3573],[430.6925,2.4943],[431.0537,2.6053],[431.4438,2.4074],[431.8275,2.0464],[432.2337,0.8199],[432.6712,0.6134],[433.065,0.1676],[433.4388,0.1589],[433.8087,0.1676],[436.1862,0.0219],[439.2375,0.0132],[439.59,-0.2869],[439.9625,-0.6218],[440.3512,-1.2394],[440.7513,-1.857],[441.14,-2.4659],[441.5575,-2.7834],[441.9388,-3.2379],[442.305,-3.8294],[442.6988,-3.9925],[443.0713,-4.447],[443.4625,-5.0385],[444.2063,-5.6561],[444.5938,-5.8278],[444.9713,-5.9822],[445.3438,-5.9909],[445.7025,-5.9822],[447.3988,-6.1279],[448.125,-6.4541],[448.9787,-6.4889],[449.4125,-6.6346],[450.5025,-6.3519],[452.54,-6.2062],[453.5775,-6.0605],[453.9312,-5.1863],[454.325,-4.0207],[454.7062,-1.9982],[455.1162,-0.4042],[455.5113,0.8897],[455.8888,1.6182],[456.2537,2.201],[456.6313,3.0753],[457.0163,4.2409],[457.4075,5.8436],[457.7987,8.3205],[458.2625,11.0976],[458.6825,13.5658],[459.08,15.7339],[459.49,17.3106],[459.8737,18.7502],[460.2725,20.4726],[460.6662,23.4888],[461.0863,27.0705],[461.5025,29.504],[461.9062,30.0694],[462.965,30.052],[463.3537,29.8802],[463.7325,29.5627],[464.4987,29.2539],[465.5438,29.1082],[476.4025,29.0995],[476.76,28.9625],[477.125,29.0995],[478.4975,28.5167],[478.8637,28.645],[479.2437,28.6276],[479.6175,28.4646],[480.4225,28.4646],[483.2975,28.6016],[484.0462,30.0478],[484.4588,30.5436],[484.8737,30.9199],[485.2525,30.8851],[485.6313,30.722],[486.6688,30.5763],[488.74,30.7133],[489.4425,31.1243],[489.8062,32.2465],[490.1925,33.934],[490.585,36.6241],[490.9987,39.0402],[491.3938,40.7451],[491.795,41.0278],[495.9,40.8821],[496.2425,40.1536],[496.6187,38.8423],[497.4187,35.6282],[497.8188,33.8797],[498.2038,32.5684],[498.5875,31.6942],[498.9787,30.9657],[499.3512,30.0915],[499.7413,29.3543],[500.1213,28.7715],[500.495,28.3257],[502.0088,25.9683],[502.385,25.0854],[502.7875,23.9111],[503.5712,22.4367],[503.96,22.1366],[504.3238,21.6908],[504.6925,21.6821],[505.0662,21.5451],[508.1313,21.6908],[509.4925,21.9735],[509.86,22.4106],[510.225,22.9847],[510.6012,23.859],[510.995,24.5788],[511.3725,24.8702],[517.6313,24.7158],[518.3975,23.7981],[518.7825,22.7608],[519.1687,21.8692],[519.555,20.5318],[519.9588,19.0313],[520.3625,17.9679],[520.765,17.2133],[521.1475,16.7675],[526.5775,16.9045],[526.93,16.7588],[527.2912,16.3043],[527.6812,15.9868],[528.0625,15.2322],[528.465,14.906],[528.8438,14.2971],[529.2175,13.8426],[529.6125,13.2511],[529.9825,12.9423],[530.3725,12.4965],[530.7375,12.042],[531.1112,11.7333],[531.5,11.2875],[531.88,10.6959],[532.2513,10.3785],[532.6537,9.9327],[533.0288,9.4782],[533.4238,9.0324],[533.8087,8.8693],[534.18,8.4235],[534.5575,8.106],[534.9288,7.5145],[535.3,7.4971],[536.0437,7.197],[536.41,7.334],[536.7687,7.1883],[537.8475,7.1709],[538.5888,7.4101],[538.9613,7.3753],[539.35,7.2036],[539.7275,7.1775],[540.09,7.0231],[540.4538,7.0144],[540.83,7.0231],[541.5612,7.0057],[541.9475,7.271],[542.315,7.6646],[542.7013,7.7756],[543.0888,8.0409],[543.5087,8.1779],[544.3462,8.1692],[544.72,8.0322],[545.085,8.1431],[545.48,8.1083],[545.8675,8.3389],[546.2675,8.2954],[546.6425,7.8235],[547.0375,7.4799],[547.4588,6.8449],[547.845,6.7841],[548.2475,5.986],[548.715,5.5054],[549.11,4.7334],[549.5087,4.2529],[549.8888,3.6353],[550.2688,3.009],[550.6637,2.2196],[551.045,1.4128],[551.4512,0.4691],[551.8375,0.1255],[552.235,-0.1659],[553.6187,-0.0202],[553.965,0.1429],[554.3375,0.8975],[554.7088,1.6521],[555.09,2.535],[555.47,2.8351],[556.1925,3.1352],[556.5663,3.1439],[556.9312,3.8724],[557.3162,4.4726],[557.6913,5.9557],[558.0838,8.0216],[558.4825,9.8048],[558.8737,12.3165],[559.2775,14.5368],[559.6812,16.6115],[560.115,18.0946],[560.5163,18.8231],[560.885,19.2602],[566.9688,19.1145],[567.3225,18.8144],[567.6962,18.1881],[568.1037,17.1334],[568.4937,16.07],[568.8875,14.5695],[569.2962,13.4974],[569.6862,12.1513],[570.1012,11.1053],[570.4963,9.9136],[570.8875,8.7132],[571.2912,7.6585],[571.6737,6.7495],[572.0612,6.158],[572.4438,6.0123],[581.7337,0],[753,0]]
       }
       dyn_wings_humelus_l_ctrl.Rotation.Z {type=MovieSceneScriptingFloatChannel, default=0, const [0,7]}
       dyn_wings_humelus_l_ctrl.Scale.X {type=MovieSceneScriptingFloatChannel, default=1, const [1,7]}
       dyn_wings_humelus_l_ctrl.Scale.Y {type=MovieSceneScriptingFloatChannel, default=1, const [1,7]}
       dyn_wings_humelus_l_ctrl.Scale.Z {type=MovieSceneScriptingFloatChannel, default=1, const [1,7]}
       dyn_wings_radius_l_ctrl.Location.X {type=MovieSceneScriptingFloatChannel, default=0, const [0,9]}
       dyn_wings_radius_l_ctrl.Location.Y {type=MovieSceneScriptingFloatChannel, default=0, const [0,9]}
       dyn_wings_radius_l_ctrl.Location.Z {type=MovieSceneScriptingFloatChannel, default=0, const [0,9]}
       dyn_wings_radius_l_ctrl.Rotation.X {type=MovieSceneScriptingFloatChannel, default=0, const [0,9]}
       dyn_wings_radius_l_ctrl.Rotation.Y {
        type=MovieSceneScriptingFloatChannel
        default=0
        thinned_from=870
        keys [[-121,0.0832],[0.9438,0.0832],[1.26,0.0652],[1.5738,0.0317],[2,0],[2.2088,0.0034],[5.9212,-0.1952],[6.8612,-0.2188],[9.62,-0.6395],[9.935,-1.0367],[10.2663,-1.2354],[10.5975,-1.4575],[11.245,-1.8547],[11.8837,-2.0769],[12.2088,-2.4741],[12.5487,-2.8949],[12.8875,-3.9115],[13.2262,-4.9752],[13.5725,-5.6182],[13.9112,-6.6584],[14.2675,-7.675],[14.6388,-8.0958],[15.4162,-8.8902],[15.7862,-9.0888],[16.4462,-9.1124],[16.7775,-9.311],[24.2225,-9.5331],[24.5663,-10.1761],[24.9325,-10.5733],[25.2637,-10.7955],[25.5963,-10.9941],[25.9262,-11.0177],[26.245,-11.2163],[26.8963,-11.4149],[27.2113,-11.8121],[27.5475,-11.8356],[27.8763,-12.0343],[28.535,-12.2564],[28.855,-13.0509],[29.1963,-13.6938],[29.5587,-14.7575],[29.9062,-15.8213],[30.6125,-18.3224],[30.9638,-19.3626],[31.3262,-20.1806],[31.67,-20.8],[32.345,-21.618],[33.045,-22.4124],[34.045,-23.0553],[34.3837,-23.4525],[38.1825,-23.6983],[38.5187,-24.3177],[38.8512,-25.1357],[39.2063,-25.3343],[39.5362,-25.3579],[40.1912,-25.5565],[42.3525,-25.7786],[42.6962,-26.0008],[43.03,-26.4452],[43.3687,-27.1117],[43.7262,-27.381],[44.0638,-28.0475],[44.76,-28.4919],[49.2887,-28.6905],[49.9388,-29.1348],[50.2725,-29.5792],[50.6175,-30.1986],[50.9562,-30.6429],[51.295,-30.8651],[51.6212,-31.2859],[53.1925,-31.3094],[64,16.0176],[67,15.9941],[71,15.5497],[76,15.5733],[77,15.1996],[78,14.6509],[81,13.9508],[81,14.1258],[82,13.7993],[83,14.1057],[92.2325,-13.7775],[99.7088,-13.7539],[113,-40.9569],[122,-51.0483],[129,-40.1084],[141.7538,8.3696],[142.0938,8.5917],[142.4638,8.8139],[142.8487,9.2111],[143.2525,9.4097],[145.82,9.8541],[146.16,10.8942],[146.495,11.7358],[146.8363,12.1802],[147.1812,13.0217],[147.525,13.4425],[147.86,14.2841],[148.895,16.1422],[149.2337,16.5866],[149.5975,17.4046],[150.6037,18.024],[154.6837,17.8254],[155.6363,17.6267],[155.9475,17.4046],[156.62,17.0074],[157.3262,16.1658],[158.015,14.9034],[158.3637,13.8868],[158.7088,13.0924],[159.0375,12.2509],[159.405,10.8606],[159.745,9.844],[160.1037,9.0496],[160.775,8.6288],[161.7287,8.4302],[162.3663,8.4066],[162.6913,7.7872],[163.0375,7.5886],[163.36,7.3665],[163.6875,6.9457],[164.3525,6.7471],[166.2312,6.5249],[166.56,6.3263],[166.885,5.4847],[167.2487,4.6432],[167.6012,3.603],[167.9562,2.1185],[168.3225,0.8326],[168.685,-0.5005],[169.0512,-1.985],[169.4087,-3.4931],[169.77,-4.8497],[170.1313,-7.0007],[170.495,-8.5088],[170.8412,-9.9698],[171.545,-11.2086],[173.7337,-11.2321],[180.1012,-11.2557],[180.4238,-11.5014],[182.9475,-11.7],[187.28,-11.7236],[188.2312,-12.168],[188.5575,-13.0567],[188.9075,-13.9454],[189.2663,-14.4133],[189.6087,-14.6354],[190.2713,-14.6826],[190.6025,-14.8812],[198.6175,-15.1269],[199.2563,-15.3255],[200.8575,-15.3727],[201.18,-15.5713],[201.5113,-15.6184],[201.8438,-15.8406],[202.8875,-17.2443],[203.23,-17.5137],[203.8863,-17.7123],[206.1538,-17.7358],[209.06,-17.958],[209.6888,-18.2038],[210.3613,-18.6717],[210.6938,-19.1396],[211.7262,-19.9004],[214.61,-19.8297],[214.9425,-19.5839],[215.2713,-18.8231],[215.6375,-17.8166],[215.9862,-16.7865],[216.3475,-15.1134],[216.7237,-14.154],[217.0938,-13.9083],[217.4225,-13.6625],[218.41,-13.639],[218.7462,-13.5918],[219.7538,-12.9017],[220.085,-12.8782],[220.41,-12.656],[227.4975,-12.4338],[227.8338,-12.4103],[228.4737,-12.1881],[229.1137,-12.1409],[229.45,-11.8952],[229.78,-11.4037],[230.4913,-9.7879],[230.835,-8.4884],[231.1925,-7.2597],[231.5712,-6.4518],[231.9162,-5.9839],[232.2537,-5.9603],[234.125,-5.9367],[234.4488,-5.7146],[235.1062,-5.4924],[235.435,-5.4688],[235.7575,-5.2466],[236.0838,-4.5801],[236.4325,-4.3108],[237.1125,-3.3043],[237.5063,-2.5906],[237.9125,-2.3449],[244.4775,-2.3684],[245.0925,-2.7892],[245.4275,-3.2571],[245.7625,-4.6137],[246.1413,-5.7482],[246.5,-7.9464],[246.8587,-10.3903],[247.2375,-12.0633],[247.5912,-13.8879],[247.9488,-15.6653],[248.32,-16.9983],[248.6988,-17.7356],[249.0475,-18.695],[249.4187,-19.2336],[249.7638,-19.9472],[250.1087,-20.2166],[254.7375,-20.4152],[255.0612,-20.6845],[255.395,-21.1995],[255.7413,-21.8089],[256.1025,-22.536],[256.4712,-23.6032],[256.8275,-24.0274],[257.1812,-24.6974],[257.545,-25.2596],[258.8375,-25.2596],[259.1662,-25.6568],[259.8188,-26.1012],[260.16,-26.1247],[263.2337,-25.9261],[263.8863,-25.4111],[264.2237,-24.9432],[264.5775,-24.8489],[264.915,-24.381],[265.26,-24.2867],[265.605,-24.0409],[265.9375,-23.5495],[266.625,-22.9873],[266.9562,-22.7651],[267.5987,-22.7415],[269.6137,-22.4722],[270.0662,-21.91],[270.4588,-20.9742],[270.8087,-20.412],[271.16,-19.6512],[271.5225,-19.1361],[271.8563,-18.8668],[272.2088,-18.2003],[272.5512,-18.1767],[273.53,-17.9781],[285.2875,-17.931],[288.9913,-17.7088],[290.2325,-17.6853],[291.7825,-17.4631],[292.1125,-17.3924],[292.4412,-17.1466],[292.7838,-16.6551],[293.1225,-16.3858],[293.465,-16.3387],[293.7887,-16.1165],[294.1213,-16.0929],[300.6488,-16.0458],[301.0312,-15.8001],[301.8075,-15.7529],[305.2788,-15.7294],[305.6,-15.6587],[305.9362,-15.3893],[306.2713,-15.3186],[306.6112,-15.0493],[311.265,-14.8036],[313.12,-14.8272],[313.7538,-15.0493],[314.0713,-15.0965],[314.4,-15.8337],[314.77,-16.1737],[315.11,-16.5373],[315.465,-16.6551],[315.7987,-16.7258],[316.1287,-16.7494],[327.79,-16.773],[335.7262,-16.9952],[336.3825,-17.0659],[337.645,-17.2645],[340.13,-17.3116],[340.45,-17.5809],[341.1488,-18.5639],[344.605,-18.6346],[345.2625,-19.1261],[345.59,-19.419],[345.93,-19.934],[346.2875,-20.274],[346.6425,-20.439],[346.985,-20.8026],[347.3375,-20.9204],[347.6788,-21.4119],[349.625,-21.5062],[349.95,-21.7755],[350.2937,-21.8934],[351.345,-22.3883],[351.6837,-22.459],[352.02,-22.7047],[352.3562,-22.7283],[352.6825,-22.974],[353.0138,-22.9976],[353.3475,-23.2434],[353.6788,-23.3141],[354.7062,-24.0984],[355.045,-24.4149],[355.3988,-24.9771],[356.1037,-26.9801],[356.4625,-27.8958],[356.81,-28.485],[357.1662,-28.8857],[357.8537,-28.9799],[358.5163,-29.2021],[359.1838,-29.6936],[359.5163,-30.1144],[359.845,-30.1615],[360.5125,-30.6059],[363.71,-30.8752],[364.095,-30.9695],[364.5138,-32.141],[364.9375,-33.4303],[365.335,-34.8712],[365.6913,-36.0191],[366.0525,-37.0964],[366.765,-37.635],[368.0238,-37.6586],[368.66,-37.975],[369.685,-38.9479],[370.0387,-38.9951],[376.855,-38.9951],[377.1838,-39.2172],[377.8512,-39.2644],[378.5138,-39.7794],[379.5175,-39.8737],[379.845,-40.0959],[384.505,-40.143],[384.83,-40.3652],[386.0713,-40.3416],[386.72,-40.7017],[387.08,-40.6411],[387.43,-40.4862],[387.7825,-39.7926],[388.16,-38.7119],[388.515,-36.9277],[389.24,-34.0155],[389.5987,-32.3862],[389.9713,-31.3897],[390.3175,-30.336],[391.3363,-29.5752],[391.6725,-29.5516],[392.005,-29.2823],[392.3475,-28.5922],[392.7188,-28.0772],[393.0725,-27.34],[393.4225,-26.532],[393.7887,-25.549],[394.1525,-25.034],[394.4913,-24.3203],[394.835,-23.8288],[395.5888,-22.5194],[396.01,-22.0279],[396.3988,-21.5128],[396.78,-20.5299],[397.175,-19.2776],[397.54,-17.756],[397.9088,-16.5037],[398.275,-15.8136],[400.8188,-15.5443],[401.1463,-14.8306],[401.4963,-13.649],[401.8487,-12.1746],[402.2375,-10.7001],[402.5963,-8.8285],[402.9638,-7.2732],[403.3125,-6.6067],[403.6588,-6.1388],[404.0038,-5.2501],[404.345,-4.0685],[404.7138,-2.1026],[405.0725,0.8699],[405.425,3.9838],[405.7925,7.367],[406.1438,10.188],[406.5013,12.3189],[406.8587,13.7227],[407.2163,13.9449],[410.9237,13.9213],[411.855,13.8977],[412.1737,13.4062],[412.5238,12.2482],[412.8888,11.336],[413.2325,10.6223],[413.5775,10.1544],[413.91,9.9086],[414.2425,9.6393],[414.59,9.5922],[414.9188,9.1243],[415.26,8.6092],[415.9662,7.4377],[416.3138,6.2326],[416.675,4.886],[417.0437,3.0244],[417.3975,0.6477],[417.7663,-1.8703],[418.1225,-4.6577],[418.4775,-6.8022],[418.8438,-8.5359],[419.2025,-9.3438],[419.58,-9.5424],[421.23,-9.6131],[421.5912,-9.8824],[421.9475,-9.906],[422.27,-10.1517],[422.62,-10.1989],[422.9625,-10.4917],[423.3225,-11.0068],[424.01,-11.6161],[424.3512,-11.6633],[424.695,-11.909],[426.9475,-12.1312],[429.3225,-12.1547],[429.645,-12.3769],[430.9613,-12.5183],[431.29,-12.7169],[435.2937,-12.7405],[435.9275,-12.9862],[436.5963,-13.1041],[436.9212,-13.3498],[437.2663,-13.3734],[440.6513,-13.4441],[441.3162,-14.6628],[441.68,-15.4135],[442.0337,-16.41],[442.4,-17.4065],[442.7613,-17.9822],[443.1213,-18.6386],[443.5125,-18.9786],[443.8537,-19.0258],[444.195,-19.2715],[444.5375,-19.2951],[444.865,-19.3658],[445.2075,-19.4836],[445.5575,-19.8943],[445.9275,-20.1536],[446.2962,-20.3657],[446.9837,-21.0222],[448.0025,-21.1164],[448.9462,-21.3151],[450.8112,-21.2915],[453.6175,-21.3151],[453.935,-21.5608],[454.2988,-22.1937],[455.02,-24.4795],[455.7563,-25.9574],[456.1287,-26.5432],[456.485,-26.661],[457.51,-26.8496],[458.1888,-26.7352],[458.5525,-26.8194],[459.4225,-27.4356],[459.85,-28.0484],[460.25,-28.274],[460.9713,-28.4761],[461.305,-28.301],[462.5688,-28.3482],[462.895,-28.6546],[463.6537,-29.5806],[464.015,-30.1396],[464.7337,-31.7894],[465.0913,-32.0352],[466.0675,-31.9645],[466.3988,-32.0924],[466.7375,-32.0217],[467.075,-31.5773],[467.4125,-31.3551],[467.7437,-31.3316],[468.0775,-31.0858],[475.4362,-31.0151],[476.1225,-30.3687],[476.4725,-29.8738],[477.1975,-28.7054],[477.5625,-27.8334],[477.93,-26.7863],[478.295,-26.0322],[479,-25.4395],[479.37,-25.3317],[479.73,-25.0725],[480.0925,-25.1061],[480.77,-24.894],[481.1287,-24.8233],[481.4625,-24.4833],[481.8075,-24.3419],[482.16,-23.9547],[482.5087,-23.2982],[482.89,-22.7225],[483.6263,-21.8068],[484.3188,-21.7126],[486.4987,-21.5711],[487.2088,-21.7629],[487.56,-21.5002],[487.9475,-21.2846],[488.3112,-21.1633],[488.6637,-20.9613],[489.0375,-20.702],[489.39,-20.8299],[490.08,-20.7828],[490.4475,-20.9578],[490.8675,-21.0857],[491.2587,-20.9914],[491.625,-21.0722],[491.99,-21.1058],[492.325,-20.988],[492.6812,-20.648],[493.7313,-19.1028],[494.0838,-18.6113],[494.4525,-18.1434],[494.8,-17.4769],[495.1587,-16.0159],[495.5163,-14.7536],[495.8675,-14.3564],[496.1975,-14.1578],[496.5213,-13.6327],[496.87,-13.6462],[497.2337,-14.5248],[497.5912,-15.8242],[497.96,-17.7902],[498.32,-20.7155],[498.6913,-23.2908],[499.055,-25.7617],[499.4162,-28.1519],[499.7925,-30.0741],[500.145,-32.3364],[500.5238,-34.0601],[500.87,-36.3795],[501.2212,-38.6519],[501.5763,-40.8299],[501.9362,-42.4458],[502.2925,-43.3816],[502.64,-43.6274],[505.1413,-43.6509],[505.4688,-43.8495],[507.0412,-43.6509],[519.0013,-43.4288],[519.32,-43.2537],[519.66,-42.5064],[520.345,-41.7827],[520.6925,-41.6077],[526.1338,-41.5841],[527.1287,-42.1328],[527.4638,-42.8801],[527.805,-43.2302],[528.5,-45.0277],[528.8463,-45.307],[529.1812,-45.3878],[529.8625,-45.2464],[530.2013,-44.7785],[531.2138,-44.0412],[531.5475,-43.9941],[531.8775,-43.7248],[532.2075,-43.2569],[534.5738,-41.4424],[534.9075,-41.4188],[535.2312,-41.1731],[541.0888,-40.7017],[541.4288,-40.4796],[541.7675,-40.456],[542.0938,-40.2338],[549.91,-39.8467],[550.2513,-39.6009],[550.58,-39.5538],[550.9137,-39.2609],[551.28,-39.1431],[551.9737,-38.4395],[552.3288,-38.2981],[552.6775,-37.7359],[553.0425,-37.3723],[553.3975,-37.3487],[553.775,-37.1265],[554.2025,-35.8743],[554.6363,-34.1541],[555.0138,-32.4339],[555.785,-28.6198],[556.1438,-27.6133],[556.5225,-26.6774],[556.8663,-25.9638],[557.2375,-25.5194],[557.5875,-24.8058],[557.9375,-24.3614],[558.2825,-24.0685],[558.9613,-23.0856],[559.3062,-22.3483],[561,-2.4242],[562,-2.1549],[563,-2.1549],[565,-2.1313],[570,-2.1078],[571,-1.5692],[571,-1.862],[571,-2.0842],[572,-1.0305],[573,3.2885],[573,2.9956],[573,2.457],[574,3.5814],[575,4.0729],[576,5.8638],[577,11.3644],[578,13.401],[579,14.0104],[580,14.5019],[581,14.7005],[587,14.7712],[588,18.1578],[588,17.4542],[588,16.7742],[589,18.9793],[590,18.9557],[591,14.8387],[592,13.2128],[593,9.3651],[594,6.874],[595,4.4066],[596,2.1612],[598,1.4004],[599,0.4882],[600,-1.875],[601,-2.7872],[601,-2.7637],[606,-3.0094],[607,-5.2312],[608,-8.2508],[609,-8.9173],[610,-8.1801],[611,-6.81],[612,-6.002],[613,-5.7327],[615,-5.5105],[616,-4.982],[616,-5.2984],[616,-5.4634],[617,-4.5241],[618,-4.2548],[618,-4.5005],[619,-4.2077],[620,-3.9619],[621,-2.6389],[622,1.9394],[623,2.6766],[624,4.3732],[625,5.3798],[626,5.5784],[632,6.0463],[633,7.6722],[633,7.1807],[633,6.4906],[634,9.6246],[639,9.0153],[640,8.4531],[643,8.4296],[644,8.1131],[645,7.5745],[646,7.258],[646,7.3052],[646,7.5509],[647,7.2109],[648,7.0931],[650,6.8945],[660,6.1337],[661,5.1743],[662,1.3197],[663,-1.1783],[676,-1.1783],[677,-1.1547],[678,-0.9561],[681,-0.8854],[682,-0.8383],[685,-0.5925],[686,-0.5689],[687,-0.3468],[688,-0.3232],[689,0.2154],[690,0.239],[692,0.5083],[693,0.7069],[695,0.6833],[696,-0.9561],[697,-3.9792],[698,-5.3628],[699,-9.0524],[702,-9.1231],[703,-9.4631],[704,-9.8974],[705,-10.601],[706,-11.8064],[707,-11.9378],[708,-12.3384],[709,-12.5506],[710,-12.554],[711,-12.8673],[714,-12.8437],[715,-12.6316],[716,-11.7932],[717,-2.7812],[718,1.2113],[719,2.7094],[720,3.423],[721,5.8333],[722,6.5941],[723,6.8163],[724,8.3379],[725,8.8529],[726,9.3209],[727,9.8595],[728,10.1288],[730,10.3274],[732,10.3038],[733,10.1052],[743,0.8697],[753,-0.2125]]
       }
       dyn_wings_radius_l_ctrl.Rotation.Z {type=MovieSceneScriptingFloatChannel, default=0, const [0,9]}
       dyn_wings_radius_l_ctrl.Scale.X {type=MovieSceneScriptingFloatChannel, default=1, const [1,6]}
       dyn_wings_radius_l_ctrl.Scale.Y {type=MovieSceneScriptingFloatChannel, default=1, const [1,6]}
       dyn_wings_radius_l_ctrl.Scale.Z {type=MovieSceneScriptingFloatChannel, default=1, const [1,6]}
       dyn_wings_metacarpal_01_l_ctrl.Location.X {type=MovieSceneScriptingFloatChannel, default=0, const [0,7]}
       dyn_wings_metacarpal_01_l_ctrl.Location.Y {type=MovieSceneScriptingFloatChannel, default=0, const [0,7]}
       dyn_wings_metacarpal_01_l_ctrl.Location.Z {type=MovieSceneScriptingFloatChannel, default=0, const [0,7]}
       dyn_wings_metacarpal_01_l_ctrl.Rotation.X {type=MovieSceneScriptingFloatChannel, default=0, const [0,7]}
       dyn_wings_metacarpal_01_l_ctrl.Rotation.Y {
        type=MovieSceneScriptingFloatChannel
        default=0
        thinned_from=1028
        keys [[-121,-0.1672],[0,-0.1672],[7.5912,-0.3344],[13.6488,-0.4442],[13.9887,-1.1652],[14.3425,-2.3252],[14.7125,-4.039],[15.0775,-5.9723],[15.4475,-7.4092],[15.8312,-8.0728],[16.1838,-8.7363],[16.5287,-8.8461],[16.8775,-9.123],[17.23,-9.2902],[18.2513,-9.3999],[18.5863,-10.0635],[18.955,-10.8943],[19.315,-11.8348],[20.0325,-12.6081],[20.7188,-12.7179],[21.705,-12.6081],[29.9787,-12.8851],[30.3325,-12.9948],[31.0362,-13.7681],[31.7275,-14.322],[32.4275,-14.5414],[32.7763,-14.8183],[33.1313,-15.205],[33.485,-15.4244],[33.8375,-15.7014],[34.1838,-15.8686],[34.5225,-15.9783],[42.28,-16.088],[42.6288,-16.2552],[42.9713,-16.3649],[43.3138,-16.5322],[43.675,-16.8091],[44.02,-17.363],[44.3775,-17.6399],[44.7362,-18.3035],[45.0888,-18.7999],[45.4538,-19.0768],[46.1338,-19.1865],[48.0763,-19.2962],[48.8012,-19.4059],[49.13,-19.5157],[49.4787,-19.3484],[50.7938,-19.1812],[51.13,-19.2909],[51.4775,-19.2334],[51.8338,-19.2856],[52.205,-19.4476],[52.9088,-19.552],[53.275,-19.7139],[53.64,-20.1528],[53.9862,-20.3722],[54.345,-20.8111],[54.6988,-21.1977],[55.05,-21.4172],[55.4013,-21.5269],[55.755,-21.9135],[56.1075,-22.0232],[56.4563,-22.4099],[56.815,-23.0735],[57.1675,-23.899],[57.5375,-24.8917],[57.9075,-26.3233],[58.285,-27.6974],[58.6837,-28.5804],[59.7587,-29.1865],[60.4613,-29.4059],[60.8487,-29.6254],[61.2563,-30.3987],[62.0525,-31.2764],[62.4212,-31.4958],[62.775,-31.825],[63.1263,-32.2638],[63.4913,-32.593],[63.8388,-33.1416],[64.1988,-33.9671],[64.5663,-34.9545],[64.9337,-35.7225],[65.3012,-36.2711],[65.67,-36.6002],[66.0188,-36.7675],[67.3413,-36.8772],[68.02,-37.593],[68.3813,-37.8124],[68.7313,-38.4185],[69.4375,-39.3537],[69.7887,-40.0695],[70.525,-41.7205],[70.88,-42.823],[71.2525,-43.3193],[71.96,-43.706],[73.61,-43.9829],[74.32,-44.4793],[74.6637,-44.6465],[75.6562,-44.5943],[76.375,-44.4377],[78.9975,-44.5474],[80.3188,-45.096],[80.675,-45.864],[81.035,-47.0186],[81.4212,-47.7344],[81.7887,-48.1158],[82.53,-48.4449],[83.8788,-48.6644],[84.23,-48.4971],[86.53,-48.3874],[87.215,-48.168],[88.85,-47.9486],[89.205,-47.0133],[89.56,-45.2526],[89.9337,-42.2275],[90.3313,-39.6987],[90.7075,-37.6088],[91.105,-36.8931],[91.8025,-36.6161],[92.1562,-36.5586],[92.5175,-36.3914],[93.3212,-35.3881],[93.7287,-34.667],[94.46,-34.4998],[95.7663,-34.8289],[96.115,-35.2678],[96.8263,-36.3649],[97.5412,-37.2426],[97.9137,-37.4621],[98.6137,-38.3973],[98.9737,-38.7265],[100.6175,-38.1204],[100.9562,-37.2374],[101.3363,-36.4641],[102.4,-35.6907],[103.7487,-35.4138],[105.0587,-35.3041],[105.7362,-35.0272],[107.715,-35.1369],[108.3763,-35.2466],[108.7113,-35.3563],[109.3975,-35.1891],[110.7188,-35.2988],[111.0575,-35.4085],[111.4062,-36.0146],[111.7663,-36.4535],[112.145,-36.6729],[112.5,-37.1693],[112.86,-37.5559],[113.2113,-37.6656],[113.9112,-38.1045],[114.2637,-38.2142],[114.6087,-38.4336],[114.9662,-38.8203],[115.3225,-39.0397],[116.0362,-39.2591],[116.705,-39.3689],[117.0462,-39.6458],[119.0425,-39.8652],[119.38,-40.0847],[119.7387,-40.4138],[120.0913,-40.5235],[120.44,-40.9624],[120.805,-41.2915],[121.8575,-41.5632],[124.7713,-41.3959],[125.155,-41.0093],[125.555,-39.9644],[125.9975,-38.6947],[127.1213,-36.3801],[127.4825,-35.8837],[127.8563,-35.1104],[128.2025,-34.891],[128.5525,-34.2849],[129.2613,-33.846],[129.615,-33.5691],[135.85,-33.4594],[136.1925,-32.5764],[136.5663,-32.3569],[137.6537,-29.8698],[138.015,-28.4904],[138.4,-26.3899],[138.8,-23.7409],[139.1787,-22.7482],[141.2325,-23.0198],[144.2113,-22.6854],[144.8813,-22.2412],[145.2312,-21.5201],[145.5975,-20.6893],[145.9575,-19.801],[146.3325,-19.5241],[147.0075,-19.3568],[149.2837,-19.0799],[149.635,-17.0891],[150.0275,-15.0462],[150.4075,-12.7263],[150.7938,-11.2894],[151.165,-10.9027],[155,-1.1812],[160,-1.2909],[164,-1.4007],[165,-1.1237],[166,0.1512],[167,4.3],[168,6.6827],[169,7.0171],[170,7.1843],[174,6.9074],[175,5.3033],[176,3.3073],[177,2.4764],[178,0.867],[179,-0.0213],[180,-0.131],[187,1.4731],[188,2.5234],[189,3.793],[190,4.6238],[192,5.1777],[193,5.2874],[199,5.7316],[200,7.9418],[202,8.109],[206,7.9418],[206.4162,-0.9436],[206.765,-1.2206],[207.1213,-1.4975],[207.475,-1.9939],[207.845,-2.2708],[208.5462,-2.8247],[208.9,-3.3786],[209.2675,-4.3766],[209.6263,-5.7613],[210.0188,-6.9787],[210.385,-8.5881],[211.1637,-10.5842],[211.5287,-11.3628],[211.915,-12.0839],[212.28,-13.3588],[212.6488,-13.526],[213,-13.6358],[217.6037,-13.803],[218.3363,-14.3568],[218.6875,-14.5241],[220.0412,-14.6913],[220.7225,-14.8585],[221.0938,-15.3027],[221.5,-16.3007],[221.9513,-17.5757],[222.3663,-18.1295],[222.7487,-18.4065],[224.4025,-18.5737],[224.7462,-19.1276],[225.1125,-20.46],[225.5025,-22.733],[225.9112,-25.4554],[226.2937,-28.1204],[226.6637,-29.897],[227.0437,-30.5659],[231.6775,-30.8428],[232.0562,-31.8983],[232.4312,-33.5077],[233.23,-38.1687],[233.6313,-39.1668],[235.3075,-39.334],[235.6625,-39.7781],[236.0188,-40.8912],[236.3825,-43.3367],[236.785,-46.4458],[237.1787,-49.8893],[237.5875,-52.2773],[237.9737,-53.3328],[238.3562,-53.6098],[240.9913,-53.7195],[245.95,-53.5],[249.275,-53.777],[249.6363,-54.6078],[250.0038,-56.2694],[250.3763,-59.697],[250.775,-63.0097],[251.155,-66.7196],[251.555,-67.3831],[254.3813,-67.822],[254.765,-68.3706],[255.1213,-69.2536],[255.8512,-69.6403],[263.755,-69.9225],[264.1025,-69.8128],[264.4512,-69.4836],[264.8075,-69.4889],[265.1587,-69.1545],[267.7775,-69.1022],[268.1338,-68.6634],[268.4812,-68.6112],[268.8263,-68.282],[269.8725,-68.0104],[270.2163,-67.791],[270.5738,-67.4043],[270.9162,-67.1849],[274.7812,-67.2946],[275.4425,-67.4043],[275.785,-67.9529],[276.8763,-71.3594],[277.2537,-72.2371],[277.63,-72.5662],[277.9688,-72.676],[282.2013,-72.5662],[283.2387,-72.2893],[284.29,-71.1869],[284.6388,-70.4711],[285.43,-69.4261],[285.82,-68.7103],[286.98,-67.5557],[287.355,-67.3362],[288.71,-67.0071],[289.0525,-66.5107],[289.4038,-66.2913],[290.105,-65.4083],[290.45,-65.1888],[290.795,-64.6925],[291.155,-64.473],[291.87,-63.4803],[292.5838,-63.1512],[295.2525,-63.4803],[295.5812,-63.6997],[296.6437,-64.0289],[297.3438,-64.5775],[297.6862,-65.2933],[298.0512,-65.6224],[299.1112,-67.3309],[299.4575,-67.7698],[300.1625,-67.9892],[302.1,-67.8795],[302.4325,-67.9892],[303.1037,-67.8795],[303.4288,-67.3831],[303.7825,-66.3904],[304.1525,-63.9139],[304.5312,-59.781],[304.9175,-56.5259],[305.2887,-54.1538],[305.6725,-52.7169],[306.41,-51.0606],[306.7713,-50.397],[307.1263,-49.514],[307.48,-48.0771],[308.2312,-44.5397],[308.6,-42.9356],[308.9575,-42.1048],[309.3262,-41.7181],[310.0325,-41.3837],[312.955,-40.8298],[313.2975,-39.1735],[313.6837,-36.6289],[314.0512,-33.9692],[314.4275,-32.1404],[314.805,-31.6962],[318.5038,-31.8635],[319.525,-32.1926],[319.8775,-33.0756],[320.2312,-33.7914],[320.5838,-34.7842],[320.9588,-35.3902],[321.31,-36.106],[321.6788,-36.7121],[322.0475,-36.8218],[324.2962,-36.9472],[324.6437,-37.2371],[325.3512,-38.2753],[325.7075,-39.188],[326.0863,-40.0186],[326.445,-41.139],[327.1737,-42.0518],[327.525,-42.2594],[327.875,-42.3416],[328.2175,-42.5493],[328.5638,-42.8823],[328.9112,-43.3798],[329.2688,-43.7519],[329.63,-44.4571],[329.9787,-44.9546],[330.3325,-45.1622],[331.005,-45.2445],[331.3438,-45.3699],[332.33,-45.2876],[332.6588,-45.413],[335,-43.1021],[336,-40.7875],[337,-37.4279],[338,-35.7716],[339,-30.4107],[340,-27.0353],[341,-26.8681],[342,-25.7603],[343,-23.3254],[344,-22.7715],[346,-22.6043],[347,-21.1621],[348,-13.4654],[349,-2.2624],[350,3.5116],[351,4.8657],[361.4575,4.5671],[362.7525,4.7343],[363.0788,4.9016],[366.3162,4.7343],[367.9162,4.4574],[368.2625,4.2902],[368.9425,3.7363],[369.2912,3.6266],[369.96,3.3497],[370.6437,3.24],[370.975,2.963],[372.0188,2.1322],[372.3787,2.0225],[373.9962,1.8553],[374.6725,1.6358],[375.0275,1.5261],[375.3713,1.0298],[375.725,0.4237],[376.0913,0.037],[376.4362,-0.1824],[380.66,-0.0727],[381.6862,0.2042],[383.1712,-0.0727],[383.5,-0.4593],[383.87,-1.6768],[384.2275,-3.5578],[384.5963,-5.9405],[384.97,-8.3755],[385.37,-10.4812],[385.7538,-11.312],[386.145,-11.4793],[387.7763,-11.589],[388.4338,-11.7562],[388.7725,-12.3623],[389.1325,-13.0834],[389.4787,-14.1336],[389.8438,-15.5183],[390.2262,-16.8454],[390.5838,-19.2804],[390.9875,-21.8772],[391.355,-25.6393],[391.7375,-32.4424],[392.1338,-39.5799],[392.4975,-46.3308],[392.8763,-51.2006],[393.2537,-53.0241],[396.8238,-53.1339],[398.7663,-53.0241],[399.435,-52.8047],[399.775,-52.5853],[400.1388,-52.0889],[400.4825,-51.7598],[400.8325,-51.2112],[401.1938,-50.1662],[401.5538,-49.6177],[401.9075,-49.3982],[404.8375,-49.5079],[405.1687,-50.114],[405.5238,-51.7181],[405.9062,-53.5992],[406.2862,-55.8094],[406.7,-57.356],[407.0612,-58.239],[407.4175,-58.9548],[408.1363,-60.0572],[409.2013,-60.8253],[409.545,-60.935],[414.8888,-60.6058],[415.6025,-59.1742],[416.7013,-57.6329],[417.07,-57.2463],[417.4162,-56.5305],[418.1525,-54.545],[418.5075,-53.7717],[418.8788,-53.3851],[420.5538,-52.664],[420.9025,-52.1101],[421.2663,-51.0599],[421.6413,-50.2866],[421.99,-50.0096],[422.3388,-49.623],[423.0375,-49.346],[425.9438,-49.2363],[426.2812,-47.4128],[427.05,-41.6599],[427.4212,-38.6136],[427.8163,-36.2309],[428.1975,-34.2349],[428.5775,-32.683],[428.9462,-31.685],[429.305,-31.0214],[429.6662,-30.7444],[432.285,-30.9639],[434.2862,-31.3505],[435.0337,-32.8971],[435.3938,-33.5607],[436.45,-33.9474],[439.0475,-33.8377],[439.3925,-33.6704],[440.0938,-32.3433],[441.1275,-32.0663],[444.05,-32.3433],[444.3975,-33.0068],[444.76,-34.9976],[445.1388,-37.652],[445.545,-40.0294],[446.01,-41.6335],[446.4638,-42.574],[446.8788,-43.1801],[447.9675,-43.9534],[456.295,-43.8437],[456.9862,-42.4068],[457.7162,-41.3565],[458.07,-41.0796],[458.4288,-40.416],[458.7875,-38.9791],[459.1575,-36.492],[459.5438,-32.7299],[459.9175,-28.137],[460.3075,-24.6519],[460.6775,-23.6538],[461.0312,-23.4866],[463.6437,-23.3194],[464.6263,-22.985],[465.3212,-21.9816],[466.0163,-21.6472],[466.6962,-21.3703],[469.2625,-21.2606],[470.88,-20.8164],[471.2387,-19.9334],[471.6025,-18.6062],[471.98,-17.1118],[472.395,-16.5579],[473.0662,-16.3907],[474.695,-15.8891],[475.045,-15.7218],[478.3975,-15.9413],[478.78,-17.1012],[479.1675,-18.8151],[479.9312,-21.2447],[480.3125,-22.9585],[480.6888,-24.7298],[481.0825,-26.1145],[481.455,-26.6684],[483.0975,-26.0623],[483.4488,-24.406],[483.8238,-22.363],[484.2025,-19.8184],[484.57,-15.6121],[484.9575,-10.622],[485.3425,-5.2921],[485.715,-0.9603],[486.08,1.8196],[486.4712,2.4885],[490.3425,2.3788],[496.4038,2.1019],[496.7563,1.4383],[497.1287,0.3305],[497.4913,-0.5003],[497.8537,-0.61],[498.1975,-0.7197],[499.5075,-0.8869],[500.5325,-1.2736],[500.8788,-2.1044],[501.245,-2.8777],[501.5987,-3.8182],[501.99,-4.534],[502.3587,-5.3648],[502.7262,-6.415],[503.0962,-8.4058],[503.47,-10.8407],[503.87,-13.2181],[504.2575,-14.4931],[504.63,-15.047],[504.99,-15.2142],[505.3325,-15.6008],[505.6975,-15.7105],[506.405,-16.2644],[506.7563,-16.3741],[507.0938,-16.6511],[507.4425,-17.4244],[507.8075,-17.8685],[508.165,-18.2552],[508.51,-18.9188],[512.2862,-19.1957],[516.4438,-18.9188],[516.7875,-18.809],[517.14,-18.5321],[517.4812,-18.1455],[517.8375,-17.8685],[518.185,-17.7013],[518.5325,-17.2571],[518.8975,-16.7033],[519.2663,-15.6477],[519.6288,-14.1533],[520.0025,-12.8784],[520.3562,-11.2168],[520.7413,-10.386],[521.0938,-9.7224],[521.8,-9.3879],[522.1413,-9.2782],[522.8075,-9.111],[523.15,-8.6669],[523.4975,-8.4996],[523.845,-8.1652],[524.1912,-7.8883],[526.81,-8.0555],[527.1388,-8.1652],[527.4775,-8.6093],[527.84,-9.5499],[528.5788,-12.4811],[528.95,-13.4792],[529.31,-14.4772],[529.6862,-15.308],[530.0425,-16.1388],[530.4038,-16.583],[530.7538,-17.4713],[531.1325,-17.7482],[531.82,-17.8579],[541,-17.7482],[543,-8.1477],[549,-8.0316],[551,-15.7582],[553.5525,-15.591],[554.8725,-15.4237],[555.5387,-15.5335],[556.53,-15.6432],[556.8637,-15.7529],[557.545,-15.8626],[561.41,-16.1395],[561.7638,-16.2493],[562.1313,-16.6359],[564.435,-16.4687],[564.7663,-16.1917],[565.1112,-15.7476],[565.4663,-15.2512],[565.815,-14.6974],[566.1875,-14.5301],[566.5325,-14.1435],[566.8825,-13.6993],[567.2425,-12.9782],[567.5925,-11.9802],[567.95,-10.7575],[568.3225,-9.425],[569.0487,-6.4256],[569.4362,-5.4798],[569.7975,-4.7012],[570.1637,-4.1473],[570.5312,-3.7032],[570.8825,-3.4262],[573.7987,-3.259],[574.1812,-3.0918],[574.5688,-2.5901],[574.9862,-2.146],[575.3875,-1.8116],[576.0788,-1.5346],[576.7475,-1.3674],[577.08,-1.2002],[577.4325,-0.9232],[577.7725,-0.756],[578.1112,-0.4791],[578.4663,-0.3119],[578.8112,0.0226],[579.1775,0.357],[579.5213,0.5242],[579.8712,0.6915],[580.225,0.8587],[580.5587,1.3028],[580.91,1.9142],[581.2875,2.5256],[581.6388,3.137],[581.9962,3.5811],[582.345,3.9156],[582.6988,4.1925],[584.335,4.3597],[585.0038,4.5269],[585.38,4.5844],[586.1213,4.9189],[586.7987,5.0861],[589.4013,4.9764],[589.7313,4.8092],[594.265,4.6419],[594.5938,4.1978],[594.9575,3.367],[595.3425,2.0398],[595.7175,0.3207],[596.1125,-1.064],[596.4737,-1.6179],[596.8388,-1.7276],[597.1825,-1.8948],[597.5275,-2.4487],[597.8925,-2.6159],[598.5738,-2.9503],[598.9225,-3.2848],[599.2663,-3.6714],[599.6163,-4.1156],[599.9588,-4.3925],[603.8587,-3.9484],[604.5287,-3.6714],[604.8963,-3.2273],[605.245,-2.9503],[605.5938,-2.5062],[605.96,-2.2292],[606.355,-1.6754],[606.795,-1.2312],[607.195,-0.6773],[607.555,-0.1235],[607.9125,0.4304],[608.2812,0.9843],[608.6375,1.7629],[608.9962,2.0398],[609.3413,2.484],[609.695,2.7609],[615.165,2.5937],[615.4937,2.484],[615.8325,2.3167],[616.5063,2.1495],[616.8525,2.0398],[619.1162,1.9301],[619.455,1.6532],[619.8163,1.3762],[620.165,1.0993],[620.51,0.6551],[620.8763,0.1013],[621.2325,-0.8968],[621.6087,-2.1142],[622.4062,-5.6622],[622.7763,-7.3238],[623.1575,-9.1526],[623.5325,-13.3642],[623.9237,-17.351],[624.3,-20.8414],[624.6775,-23.7205],[625.0662,-25.1052],[625.43,-25.7688],[625.7938,-26.2129],[626.15,-26.3226],[632.9075,-26.2129],[633.89,-26.1032],[634.6037,-24.3424],[634.9588,-22.7436],[635.3313,-20.1468],[635.7275,-18.2657],[636.1037,-16.4422],[636.4925,-14.6709],[636.86,-13.3959],[637.6025,-11.1804],[637.9737,-10.6266],[638.375,-9.963],[638.8038,-9.686],[640.2038,-9.7958],[640.8675,-9.9055],[641.2013,-10.0727],[641.55,-10.2921],[641.9025,-11.0132],[642.6087,-12.3979],[642.9562,-13.2287],[643.325,-13.8923],[643.7125,-14.1692],[644.0575,-14.2789],[646.3375,-13.8348],[646.685,-12.6173],[647.045,-11.2902],[647.4137,-9.5188],[647.8087,-8.0819],[648.1775,-7.5281],[648.5362,-7.4184],[652.1413,-7.5281],[652.4863,-8.1341],[652.8463,-9.3516],[653.2088,-11.0654],[653.6163,-12.7793],[653.9937,-14.9372],[654.37,-17.3199],[654.7638,-17.8738],[655.11,-18.0932],[655.4712,-18.203],[660.6313,-18.8143],[660.9787,-21.3015],[661.3625,-24.123],[661.755,-26.9446],[662.1438,-28.9353],[662.5387,-29.8184],[662.8988,-30.3722],[663.2513,-31.0358],[663.6175,-31.9763],[664.3562,-34.9598],[664.7387,-36.7834],[665.115,-37.5567],[671.0962,-37.3895],[671.4812,-37.2797],[673.7725,-37.17],[674.4762,-35.7906],[674.8325,-34.6882],[675.2125,-34.0821],[676.2738,-32.9797],[676.6325,-32.3736],[676.9837,-31.5481],[677.7375,-29.0087],[678.095,-28.2354],[678.4625,-28.1257],[678.8025,-27.9585],[680.4438,-28.0682],[682.1263,-28.2876],[682.4613,-28.5646],[682.815,-28.6743],[683.525,-29.8865],[683.8813,-31.2083],[684.2775,-32.3108],[684.6375,-33.0266],[686.5987,-32.8594],[686.9388,-32.6399],[687.2988,-31.6994],[687.6612,-31.4225],[688.3575,-31.1455],[689.0325,-30.9783],[689.3813,-30.1475],[690.4262,-29.3167],[691.465,-28.8725],[693.1263,-28.9823],[693.46,-29.2017],[693.8112,-29.5308],[694.5312,-30.743],[694.8875,-31.7357],[695.2525,-32.2321],[695.6087,-32.3418],[695.97,-32.5612],[696.6675,-32.7807],[699.2637,-32.8904],[699.5925,-33.2195],[699.9425,-33.8256],[700.6662,-34.5414],[706.0175,-34.7086],[706.36,-35.0953],[706.7125,-35.9261],[707.0788,-36.8666],[707.4538,-38.0266],[707.8163,-39.244],[708.2,-39.6307],[708.8875,-39.7979],[709.225,-40.0748],[709.5738,-40.519],[709.935,-40.7959],[712.2287,-40.5765],[712.5688,-39.1396],[712.9675,-37.2063],[713.3462,-35.2156],[713.7513,-32.3418],[714.1237,-28.7469],[714.4937,-24.3788],[714.8813,-19.2895],[715.28,0],[753,0]]
       }
       dyn_wings_metacarpal_01_l_ctrl.Rotation.Z {type=MovieSceneScriptingFloatChannel, default=0, const [0,7]}
       dyn_wings_metacarpal_01_l_ctrl.Scale.X {type=MovieSceneScriptingFloatChannel, default=1, const [1,7]}
       dyn_wings_metacarpal_01_l_ctrl.Scale.Y {type=MovieSceneScriptingFloatChannel, default=1, const [1,7]}
       dyn_wings_metacarpal_01_l_ctrl.Scale.Z {type=MovieSceneScriptingFloatChannel, default=1, const [1,7]}
       dyn_wings_phalange_01_l_ctrl.Location.X {type=MovieSceneScriptingFloatChannel, default=0, const [0,5]}
       dyn_wings_phalange_01_l_ctrl.Location.Y {type=MovieSceneScriptingFloatChannel, default=0, const [0,5]}
       dyn_wings_phalange_01_l_ctrl.Location.Z {type=MovieSceneScriptingFloatChannel, default=0, const [0,5]}
       dyn_wings_phalange_01_l_ctrl.Rotation.X {type=MovieSceneScriptingFloatChannel, default=0, const [0,5]}
       dyn_wings_phalange_01_l_ctrl.Rotation.Y {type=MovieSceneScriptingFloatChannel, default=3.4398, const [3.4398,5]}
       dyn_wings_phalange_01_l_ctrl.Rotation.Z {type=MovieSceneScriptingFloatChannel, default=0, const [0,5]}
       dyn_wings_phalange_01_l_ctrl.Scale.X {type=MovieSceneScriptingFloatChannel, default=1, const [1,5]}
       dyn_wings_phalange_01_l_ctrl.Scale.Y {type=MovieSceneScriptingFloatChannel, default=1, const [1,5]}
       dyn_wings_phalange_01_l_ctrl.Scale.Z {type=MovieSceneScriptingFloatChannel, default=1, const [1,5]}
       dyn_wings_carpals_01_l_ctrl.Location.X {type=MovieSceneScriptingFloatChannel, default=0, const [0,6]}
       dyn_wings_carpals_01_l_ctrl.Location.Y {type=MovieSceneScriptingFloatChannel, default=0, const [0,6]}
       dyn_wings_carpals_01_l_ctrl.Location.Z {type=MovieSceneScriptingFloatChannel, default=0, const [0,6]}
       dyn_wings_carpals_01_l_ctrl.Rotation.X {type=MovieSceneScriptingFloatChannel, default=0, const [0,6]}
       dyn_wings_carpals_01_l_ctrl.Rotation.Y {
        type=MovieSceneScriptingFloatChannel
        default=0
        thinned_from=1186
        keys [[-121,0],[0,0],[5.9225,-0.1185],[6.2738,-0.635],[6.6125,-0.9903],[6.95,-1.2699],[8.815,-1.3884],[18.645,-1.5068],[18.9638,-1.3457],[20.5375,-1.8195],[20.88,-2.5302],[21.2212,-3.5536],[21.595,-4.6527],[22.0238,-5.7945],[22.4488,-6.2683],[22.8563,-6.3868],[26.0263,-6.2256],[29.7925,-6.581],[30.1162,-7.1732],[30.4538,-7.2917],[30.7912,-7.5286],[36.06,-7.4101],[37.9262,-7.1732],[38.2537,-6.8936],[38.6,-6.2587],[38.93,-6.1402],[39.2688,-5.9791],[39.605,-5.8606],[41.79,-5.6237],[42.4588,-5.1499],[43.1187,-4.913],[43.7825,-4.7945],[44.74,-4.6761],[46.02,-4.4392],[46.97,-4.5576],[47.6325,-4.7945],[47.9625,-5.2684],[48.3088,-5.6237],[48.9775,-6.3344],[49.33,-6.5713],[49.6612,-6.9267],[50.3188,-7.1636],[50.6537,-7.4762],[50.9962,-7.7131],[51.3338,-8.187],[51.6862,-8.4239],[52.0225,-8.5423],[52.3475,-8.7792],[53.6825,-8.7365],[54.07,-8.855],[54.4837,-9.2103],[54.8463,-9.5657],[55.1912,-9.6841],[55.8537,-9.8026],[57.7787,-9.921],[58.0925,-10.2764],[58.455,-10.6318],[59.1287,-10.7502],[61.345,-10.7075],[61.67,-11.1386],[62.0075,-11.8493],[62.3613,-12.5173],[63.04,-13.465],[63.3813,-13.3038],[65.8863,-13.4223],[66.2075,-13.5407],[66.545,-13.8961],[66.8737,-14.133],[67.5562,-15.0806],[67.9,-16.3836],[68.2725,-18.2457],[68.9962,-22.3585],[69.3875,-23.543],[69.73,-24.0168],[71.6637,-23.8984],[71.9862,-23.7372],[81.2738,-23.3392],[81.595,-22.8654],[82.67,-17.9468],[83.0188,-16.2458],[83.3875,-14.6632],[83.7513,-13.5972],[84.115,-12.8865],[84.49,-12.4127],[84.8412,-12.2942],[90.66,-12.1758],[90.9762,-11.465],[91.3338,-10.7116],[91.6762,-10.1194],[92.0275,-9.366],[92.3737,-8.8922],[94.04,-7.7076],[94.3675,-7.5892],[98.1,-7.2338],[98.4362,-6.5231],[98.7812,-6.092],[99.135,-5.7367],[99.4575,-5.6609],[99.7975,-5.424],[100.4538,-5.3055],[101.1,-5.4667],[101.4238,-5.1113],[101.7625,-5.154],[102.0962,-5.0356],[102.7462,-4.9171],[104.9125,-5.0356],[110.5238,-4.9171],[111.7675,-4.7987],[112.75,-4.4433],[113.095,-4.088],[113.4238,-3.9695],[113.7663,-3.7326],[114.105,-3.6142],[114.8012,-3.4957],[116.0625,-3.3773],[116.3887,-3.2588],[116.7625,-3.1831],[117.16,-2.8277],[117.57,-2.7519],[117.9438,-2.3966],[118.2975,-2.1597],[119.005,-2.0412],[119.3225,-1.8043],[119.6575,-1.449],[119.995,-0.8567],[120.34,-0.146],[120.6913,0.6832],[121.0537,1.3939],[121.3863,1.7492],[121.7287,1.9861],[122.0575,2.1046],[123.3462,2.3415],[123.6725,2.6211],[124.005,2.858],[124.6688,3.0949],[124.9987,3.3318],[125.6725,3.5687],[126.3075,3.6872],[127.565,3.8056],[127.8913,4.3979],[128.2237,5.1086],[128.5725,6.2173],[128.9362,7.165],[129.2775,7.9941],[129.6488,8.4679],[129.9787,8.7806],[130.6463,8.899],[131.91,9.0175],[132.5888,9.4486],[132.92,9.7613],[133.2575,10.1166],[133.6075,10.3866],[134.6288,11.3246],[134.9625,11.4761],[135.305,11.7887],[136.0088,12.6509],[136.3525,12.8025],[136.7125,12.9967],[137.0462,13.2336],[137.3775,13.3093],[137.725,13.4278],[138.0487,13.3851],[138.3825,13.4608],[138.7287,13.4939],[139.0675,13.5696],[139.4112,13.4842],[139.7563,13.56],[140.0938,13.5173],[140.735,13.6357],[146.955,13.4746],[148.9925,13.3134],[149.38,13.3892],[150.0375,13.2281],[154.0612,13.3892],[154.7225,13.2708],[155.065,13.1523],[155.4238,13.3134],[155.7613,13.3988],[156.1213,13.3231],[156.455,13.3658],[156.7862,13.5269],[157.1213,13.4512],[157.455,13.4939],[158.1112,13.655],[160.615,13.7735],[160.93,13.8919],[161.2637,14.0861],[161.6012,14.1619],[161.9362,14.3561],[162.285,14.4746],[162.6112,14.4319],[163.275,14.228],[163.5963,14.3465],[163.9175,14.2611],[165.55,14.133],[165.87,14.2514],[166.21,14.3272],[166.8788,14.2418],[173.4112,14.1233],[174.04,14.2845],[174.7013,14.166],[175.3587,14.2514],[175.6938,14.3368],[176.3775,13.9484],[176.7138,13.7969],[177.06,13.6027],[177.4,13.6454],[177.7375,13.5696],[182.3763,13.7308],[182.695,13.6123],[183.3375,13.655],[183.6637,13.7404],[184.3562,13.4374],[184.7025,13.5655],[185.05,13.2528],[185.39,13.2955],[185.7313,13.2198],[186.065,13.1013],[188.255,13.2625],[188.9175,13.111],[189.2663,13.0779],[189.61,13.0022],[189.9362,13.0876],[190.2875,13.1303],[190.615,13.0118],[190.9388,13.0545],[191.6075,13.0972],[191.935,13.2583],[192.2625,13.1826],[192.6075,12.9884],[193.29,13.0738],[193.6225,13.2776],[194.2875,13.0407],[195.2463,13.0834],[195.5788,13.0077],[195.92,12.8135],[196.2613,12.8989],[196.9288,12.7473],[197.605,12.8327],[198.26,12.7143],[198.9112,12.8754],[206.6762,12.757],[213.715,12.9181],[214.045,12.9608],[214.3675,12.8851],[214.7138,12.9705],[215.375,13.0559],[215.7138,12.9801],[216.04,13.0228],[216.375,12.9471],[216.7213,12.7956],[217.0575,12.8052],[217.3988,12.2983],[217.7625,11.4127],[218.1263,10.7309],[218.5087,10.0395],[218.8712,9.6607],[219.58,9.1965],[219.9237,9.2392],[222.1263,8.647],[222.4575,7.8178],[223.1875,5.34],[223.5725,4.3593],[223.9237,3.8097],[224.2675,3.4971],[224.6212,3.1417],[224.945,3.066],[225.285,2.3552],[225.6388,1.3746],[226.0063,0.1143],[226.3687,-1.7382],[226.7262,-3.4723],[227.075,-4.4957],[227.4288,-5.0879],[227.7725,-5.2064],[228.7162,-5.3248],[229.0425,-6.3579],[229.3863,-8.4142],[229.7675,-11.5036],[230.1263,-14.195],[230.4837,-16.7252],[230.8563,-18.232],[231.2088,-18.63],[232.4837,-18.7912],[232.8025,-19.4688],[233.1488,-20.663],[233.5187,-22.331],[233.8687,-24.2028],[234.2362,-25.3542],[234.9137,-25.4727],[235.225,-25.6338],[235.8712,-25.795],[236.5125,-26.0746],[236.845,-26.8376],[237.185,-27.7191],[237.5312,-28.4395],[237.8663,-28.5579],[238.2063,-30.3967],[238.5738,-33.0743],[238.9262,-35.5576],[239.275,-38.3633],[239.6337,-41.7708],[239.995,-44.2638],[240.37,-46.4772],[240.73,-48.7237],[241.0938,-51.4109],[241.485,-54.1738],[241.8312,-56.098],[242.1787,-56.8183],[242.5125,-56.9795],[246.7475,-57.6998],[247.0838,-59.4628],[247.4475,-62.4296],[247.8025,-66.2353],[248.1862,-70.202],[248.5487,-72.6854],[248.905,-74.4484],[249.6125,-74.8891],[252.6988,-75.0503],[253.0175,-75.491],[253.6862,-76.7802],[254.69,-78.2306],[255.02,-78.9936],[255.3663,-80.0363],[255.7013,-81.4012],[256.4038,-84.7757],[256.7538,-85.8611],[257.4538,-86.9891],[259.105,-87.956],[259.7738,-88.719],[260.4737,-90],[262.19,-88.3885],[263.1775,-87.9051],[263.8525,-87.0993],[264.5338,-86.777],[267.3438,-87.0566],[267.66,-87.5401],[267.9937,-88.872],[268.37,-89.3981],[269.0375,-89.7204],[269.9862,-89.8816],[270.6588,-88.9573],[271.0063,-88.1185],[271.3413,-87.6778],[271.68,-86.9574],[272.0312,-86.3982],[272.6962,-85.7536],[273.0263,-85.1517],[273.3763,-84.2702],[274.0487,-82.948],[275.1575,-81.2277],[275.5738,-80.2704],[276.6575,-78.5928],[277.655,-78.1947],[278.3238,-77.2802],[278.6612,-76.2045],[279.0163,-75.4084],[279.3525,-74.4938],[280.0462,-73.7404],[282.2688,-73.3424],[283.2788,-72.2336],[283.9438,-71.8782],[288.8825,-71.7598],[289.5187,-71.4802],[290.2,-70.2957],[290.5362,-69.4238],[291.2412,-67.0451],[291.965,-64.0315],[292.325,-62.8043],[292.675,-61.2547],[293.0612,-59.9848],[293.7525,-59.3499],[295.9462,-59.2314],[296.2812,-59.113],[296.6075,-58.6392],[297.295,-56.4974],[297.9937,-53.678],[298.3525,-52.4507],[298.715,-51.3847],[299.0813,-50.2002],[299.425,-48.7787],[299.8075,-47.7554],[300.5087,-45.396],[300.8813,-44.2541],[301.245,-43.2308],[301.97,-41.3782],[302.6962,-40.5587],[303.04,-40.2461],[303.3763,-40.4499],[303.7113,-40.1799],[304.0675,-40.1469],[304.4075,-39.8769],[305.1125,-38.7778],[305.4438,-38.1525],[305.7975,-36.8922],[306.23,-35.6416],[306.7362,-34.3152],[307.27,-33.0219],[308.0875,-32.6858],[308.4312,-32.4158],[309.1313,-31.7478],[309.4688,-31.6721],[309.8075,-31.4352],[310.4675,-31.2836],[310.81,-31.0894],[311.4737,-30.9379],[311.81,-30.7437],[312.1575,-30.668],[312.4913,-30.2369],[312.84,-30.2038],[313.1825,-29.8911],[313.5312,-29.3843],[314.2387,-28.6502],[314.6025,-28.3471],[314.9638,-27.6461],[315.3212,-27.4284],[315.6988,-26.8128],[316.4112,-25.4106],[316.7825,-24.9038],[317.1225,-24.3542],[317.4837,-24.0842],[317.8312,-23.7716],[318.515,-23.4686],[318.8612,-23.0802],[319.5562,-22.6256],[319.91,-22.6353],[320.2437,-22.1945],[320.59,-22.0857],[320.9488,-22.0953],[321.295,-21.9769],[321.6662,-22.1477],[322.01,-22.2427],[322.365,-22.2951],[322.7175,-22.8639],[323.0562,-23.1531],[323.3988,-23.3239],[323.7687,-23.5801],[324.1363,-23.9975],[324.5138,-23.7798],[324.8687,-24.7564],[325.2287,-25.4629],[325.5987,-26.2453],[326.3225,-27.8953],[326.675,-28.5592],[327.0387,-29.5027],[327.3938,-30.3277],[327.755,-31.2382],[328.1112,-31.9874],[328.4825,-33.6278],[328.8612,-34.6994],[329.2188,-35.5244],[329.6087,-36.2641],[329.9525,-36.7572],[330.6775,-37.4637],[331.02,-37.3976],[331.745,-38.7157],[332.1025,-38.7681],[335.5462,-38.4127],[335.8863,-37.702],[336.2275,-37.3136],[337.275,-35.3523],[337.6488,-34.1677],[338.9388,-29.0316],[339.3587,-27.4587],[340.0863,-24.5497],[340.4588,-23.0195],[340.82,-21.2096],[341.1925,-19.7882],[341.5562,-19.594],[341.895,-19.3571],[342.905,-18.4756],[343.245,-17.433],[343.9237,-15.9923],[344.2625,-14.9496],[344.6263,-13.9496],[344.965,-13.0681],[345.31,-12.1439],[345.6562,-11.7032],[345.9875,-11.3809],[346.3325,-11.1013],[346.6625,-10.6179],[346.995,-10.2956],[347.33,-9.8121],[347.665,-9.1248],[348.0038,-8.4802],[348.3613,-8.1579],[348.6875,-7.9968],[349.0138,-7.556],[349.6625,-7.3949],[352.7587,-7.2337],[353.4137,-6.9114],[353.7425,-6.6318],[354.4162,-6.3095],[354.745,-5.9872],[355.0738,-5.9445],[355.4162,-5.6222],[355.7462,-5.2145],[356.0975,-4.6457],[356.4362,-4.0768],[356.7838,-3.2284],[357.125,-2.9818],[357.4675,-3.1003],[357.7987,-3.0576],[358.1287,-2.8965],[358.78,-2.8538],[359.1037,-2.9295],[359.4475,-2.8868],[359.7775,-3.0053],[361.0425,-3.1237],[361.3737,-3.3179],[362.0437,-3.5548],[365.4775,-3.6733],[365.7987,-4.1471],[366.1313,-5.0947],[366.5013,-5.8812],[366.8512,-6.355],[367.1812,-6.7103],[367.5187,-7.1841],[367.8537,-7.421],[368.1925,-7.5395],[368.515,-7.7764],[368.85,-7.8948],[372.1875,-8.2929],[372.5175,-9.089],[372.8512,-10.0036],[373.2075,-10.5628],[373.5575,-11.3589],[373.8925,-12.0365],[374.2387,-12.553],[374.5788,-12.9511],[374.9062,-13.0695],[375.2387,-13.3491],[377.5138,-14.1783],[377.8438,-14.4152],[379.8312,-15.1259],[380.4863,-15.922],[380.8175,-16.7181],[381.1875,-17.9123],[381.5287,-18.5473],[382.22,-20.4521],[382.5863,-21.722],[382.9388,-23.4658],[383.3175,-26.124],[383.68,-29.4173],[384.0638,-33.1843],[384.4775,-36.9844],[384.8712,-39.9884],[385.2537,-41.7652],[385.6212,-42.5944],[385.9737,-42.7128],[395.255,-42.7886],[395.5838,-42.9401],[395.925,-42.9304],[396.63,-43.1577],[396.9937,-43.6315],[397.3275,-43.4704],[399.8263,-43.5888],[400.1475,-43.8257],[400.8012,-43.9442],[401.8837,-43.8257],[405.69,-43.1481],[406.0288,-41.7927],[406.3713,-39.7598],[406.7587,-38.1249],[407.1112,-36.8453],[407.83,-35.3619],[408.175,-35.0823],[408.505,-34.6416],[409.4663,-34.0824],[409.8062,-33.1678],[410.15,-31.8979],[410.51,-30.3484],[410.8612,-29.3581],[411.2038,-28.7231],[413.73,-27.9697],[414.065,-26.1406],[414.4488,-23.2358],[414.8363,-21.0086],[415.185,-19.696],[418.0263,-18.8241],[418.3587,-17.7153],[418.71,-16.1328],[419.08,-14.389],[419.435,-12.8492],[419.795,-11.5462],[420.1463,-10.5228],[420.5063,-9.1441],[420.8813,-8.1634],[421.2362,-6.8701],[421.585,-5.7379],[421.9513,-4.5203],[422.315,-4.1319],[422.9987,-3.9804],[423.3438,-3.862],[423.675,-3.9047],[423.9925,-3.592],[424.3363,-3.6774],[424.6888,-3.3648],[425.0263,-2.7725],[425.3837,-2.6967],[425.7162,-2.2229],[426.0462,-2.0287],[426.3775,-1.7161],[426.7175,-1.5219],[427.0625,-1.4461],[427.3975,-1.4888],[427.7313,-1.3704],[428.38,-1.1334],[428.7138,-0.9819],[429.05,-0.4324],[429.3925,0.4298],[429.7675,1.1736],[430.115,2.0358],[430.48,2.2204],[430.845,2.5661],[431.1925,2.4807],[431.55,2.7934],[431.885,2.9118],[432.2138,2.708],[432.565,2.9022],[432.9275,3.0964],[433.3288,3.2052],[433.7462,3.6694],[434.1413,4.1335],[434.5013,4.5977],[434.8737,4.6638],[435.2138,5.0095],[435.5938,5.3552],[435.9438,5.4641],[436.2887,5.8098],[436.6463,6.1128],[436.9987,6.6527],[437.3625,7.7849],[437.7375,8.6375],[438.095,9.8881],[438.4538,10.7407],[438.8012,11.4321],[439.1612,11.9624],[439.8837,12.9141],[440.6263,13.7805],[440.98,13.9554],[441.35,14.0973],[441.695,14.0877],[442.0288,14.173],[442.3587,14.1304],[443.0425,13.6373],[443.3775,13.0257],[443.7362,13.0161],[444.0688,12.7792],[444.4425,13.0491],[444.7987,12.9637],[445.14,12.9541],[445.4887,12.7172],[445.8225,12.7503],[446.16,12.3426],[446.8612,11.5272],[447.2138,11.2379],[447.565,10.6691],[447.9075,10.176],[448.2625,9.6402],[448.6363,9.1802],[448.9825,8.0756],[449.3438,7.0894],[449.72,6.4255],[450.0713,5.449],[450.4438,4.8801],[450.785,4.7093],[451.1263,4.2589],[451.4875,3.3678],[451.84,3.0358],[452.1925,2.7466],[452.5525,2.5758],[452.8875,2.5331],[455.0962,2.2962],[455.4162,2.1778],[455.7437,1.6186],[456.1025,1.3817],[456.425,1.1021],[456.755,0.9836],[457.0838,1.1117],[457.435,0.9602],[457.7925,0.7329],[458.1612,0.4203],[458.55,0.3018],[458.9013,0.1834],[459.2312,0.0222],[459.5638,-0.0962],[460.515,-0.3758],[460.8412,-0.4943],[461.1737,-0.7312],[461.5025,-1.0866],[461.8425,-1.3992],[462.1838,-1.4323],[462.5187,-1.6265],[462.85,-1.4226],[463.195,-1.4557],[463.5325,-1.2945],[463.86,-1.2518],[464.2063,-1.0053],[464.5788,-1.0714],[464.9937,-1.3648],[465.4488,-1.8193],[465.8375,-2.1981],[466.1938,-3.1788],[466.5587,-4.2779],[466.9112,-5.8081],[467.6125,-9.5793],[467.9913,-11.1522],[468.665,-11.5406],[469.0088,-12.3932],[469.3687,-12.9331],[469.725,-13.5158],[470.09,-14.1411],[470.4325,-14.7664],[470.7875,-16.1355],[471.1737,-17.5803],[471.5263,-18.9728],[471.8875,-19.9246],[472.25,-20.1229],[472.6025,-19.9328],[472.9538,-19.6766],[473.3138,-19.5486],[473.6488,-19.3874],[473.9713,-19.0224],[474.3438,-17.6148],[474.7275,-15.4248],[475.12,-11.9029],[475.4712,-7.7364],[475.82,-3.1196],[476.1912,0.81],[476.55,2.8719],[476.8975,4.0853],[477.2625,4.7726],[477.6,5.4599],[477.9475,6.6306],[478.2975,7.5645],[478.6788,9.541],[479.055,10.9156],[479.4025,11.3233],[482.2088,11.4844],[482.8312,11.366],[485.01,11.4844],[485.325,11.3233],[485.6525,11.2379],[486.3363,10.9817],[486.7,10.3701],[487.06,9.9624],[487.4013,9.7586],[487.7337,9.5975],[488.0612,9.5548],[488.73,9.2325],[489.3988,8.8248],[489.74,8.2986],[490.0838,8.3744],[490.415,8.3317],[491.3737,8.4501],[497.485,8.3317],[497.8112,7.7725],[498.1562,7.2987],[498.4863,6.7822],[499.165,6.3083],[499.4887,6.1899],[502.9075,6.2657],[503.2275,6.8152],[503.5712,7.6347],[503.9538,8.0135],[505.1037,8.9033],[505.445,9.2917],[505.8087,9.3247],[506.1463,9.4005],[506.4787,9.7559],[507.1425,9.9928],[511.175,10.5093],[511.505,11.1015],[511.8563,12.0918],[512.2113,13.2764],[512.5587,14.7405],[513.27,18.2183],[513.6263,20.0805],[514.0088,21.3077],[514.36,21.5446],[514.695,21.7057],[519.635,21.5873],[519.9613,20.8766],[520.3062,19.5309],[520.6888,18.1095],[521.0375,17.0434],[521.3837,16.3327],[521.7337,15.9016],[522.0675,15.352],[522.4087,14.5656],[522.7713,13.8218],[523.1263,13.4334],[523.4887,13.2061],[523.8438,13.0973],[524.1838,12.9127],[524.5512,12.8893],[524.8975,12.7474],[525.2475,12.9279],[525.6175,12.4307],[525.9688,12.1703],[526.3288,11.9431],[526.6788,11.91],[527.0188,11.9527],[530.6737,11.8343],[530.99,11.7585],[531.32,11.2943],[532.0362,10.7544],[532.38,10.8825],[532.7175,10.9252],[533.3737,10.8068],[535.2475,10.9679],[535.5763,10.8495],[538.7113,10.6552],[539.0375,9.7503],[539.3988,9.1677],[539.7587,8.3482],[540.115,7.8083],[540.4887,7.6995],[540.8425,7.3868],[541.1812,7.4722],[541.515,7.3537],[545.245,7.2353],[545.5712,7.0411],[545.905,6.5673],[546.26,6.0935],[546.5938,6.0177],[547.2613,5.7808],[552.2275,4.3555],[552.8612,4.2592],[553.2287,4.3793],[553.5587,4.4705],[553.8937,4.6282],[554.2537,4.7413],[554.6087,4.9227],[554.9562,5.1722],[555.3313,5.3472],[555.685,5.5642],[556.0425,5.8826],[556.7462,6.3122],[557.0913,6.6294],[557.455,6.5987],[557.8012,6.7839],[558.4863,6.5503],[559.1513,6.4985],[559.5213,6.3794],[559.92,6.3188],[560.3275,6.4192],[560.6975,6.8583],[561.05,7.2278],[561.42,7.3957],[561.7775,7.7604],[562.15,8.0924],[562.4963,8.1379],[562.8425,8.0252],[563.185,8.045],[563.5038,7.8437],[563.8338,7.8668],[564.7987,7.7513],[565.1225,7.5527],[565.4588,7.669],[565.785,7.565],[566.1125,7.5885],[568.6075,7.2411],[568.9225,7.5391],[569.275,7.7029],[569.6137,7.9589],[569.9713,8.118],[570.3112,8.4021],[570.6475,8.5624],[570.9937,8.6291],[571.3413,8.8728],[571.6825,9.2046],[572.0238,9.7109],[572.3825,10.265],[573.1075,11.2931],[573.4775,11.7103],[573.84,11.9266],[574.1888,11.9146],[577.6087,10.8157],[578.9362,10.2968],[579.275,10.1049],[580.2713,9.8627],[580.9288,9.5859],[581.265,9.4974],[581.59,9.4397],[581.9325,9.1359],[582.2825,8.8577],[582.6313,8.7145],[583,8.5837],[583.3462,8.5305],[584.3413,8.0158],[584.9775,7.8608],[585.2937,7.6538],[585.6325,7.6391],[585.9725,7.3971],[586.305,7.2572],[586.6475,6.9839],[587.01,6.9304],[587.3763,6.844],[587.7237,6.736],[588.095,6.535],[588.43,6.3937],[588.7687,6.1514],[589.1037,6.0031],[589.435,5.6706],[589.7812,5.3099],[590.5075,4.7169],[590.92,4.4569],[591.3338,4.1848],[591.7663,3.9132],[592.1825,3.7669],[592.5263,3.571],[593.1838,2.9955],[593.5238,2.6076],[593.885,1.9229],[594.2537,1.3023],[594.665,0.7511],[595.03,0.4052],[595.3887,0.0746],[595.7487,-0.1556],[596.0975,-0.4064],[596.43,-0.6528],[597.0813,-1.3012],[597.3975,-1.6079],[597.7313,-1.9692],[598.0875,-2.4067],[598.43,-2.7414],[598.7775,-3.0446],[599.125,-3.4392],[599.4613,-3.7717],[600.13,-4.5701],[600.4638,-5.1639],[600.825,-5.7221],[601.1787,-6.3244],[601.5425,-6.9758],[601.8988,-7.6285],[602.2625,-8.1279],[602.605,-8.6474],[602.9588,-9.1467],[603.3025,-9.598],[603.955,-10.656],[604.2875,-11.3356],[604.6388,-12.1012],[605.0025,-12.9327],[606,-14.8633],[607,-15.5345],[613,-15.565],[615,-16.225],[619,-16.2461],[624,-15.8989],[625,-14.4943],[626,-13.1405],[627,-12.1341],[628,-11.3377],[629,-7.9711],[630,-7.3627],[631,-7.0729],[631,-7.1525],[631,-7.2034],[635,-6.9933],[636,-6.3849],[637,-5.7544],[638,-5.1016],[639,-4.8979],[644,-4.8756],[649,-4.5571],[650,-4.159],[651,-3.6303],[653,-3.5793],[654,-2.6525],[655,-2.5219],[657,-2.283],[658,-2.1238],[659,-1.7543],[660,-1.6173],[661,-1.6617],[662,-1.8145],[663,-2.4738],[664,-2.8367],[670,-3.0627],[671,-3.983],[672,-4.3238],[673,-4.6358],[674,-6.0275],[675,-6.1071],[678,-6.1868],[679,-6.0784],[681,-6.1581],[682,-6.6072],[683,-6.709],[689,-6.7599],[691,-8.209],[692,-13.5656],[693,-14.8267],[694,-14.5369],[695,-13.7406],[696,-12.8137],[699,-12.7341],[701,-12.8137],[703,-13.1322],[705,-13.0526],[706,-11.1924],[707,-6.9053],[708,-5.2942],[709,-5.5173],[712,-5.4886],[715,-4.8516],[716,-3.2432],[717,-2.6192],[718,-2.2849],[719,-2.0145],[729,-2.0145],[730,-2.2469],[731,-3.0238],[732,-3.1988],[733,-3.721],[734,-3.8229],[735,-3.9025],[736,-4.0331],[737,-4.0044],[739,-4.1349],[740,-4.2655],[741,-4.6858],[746,-4.6571],[747,-4.6062],[748,-3.9979],[749,-2.4469],[750,-1.0043],[751,-0.0617],[752,0.0179],[753,-0.0617]]
       }
       dyn_wings_carpals_01_l_ctrl.Rotation.Z {type=MovieSceneScriptingFloatChannel, default=0, const [0,6]}
       dyn_wings_carpals_01_l_ctrl.Scale.X {type=MovieSceneScriptingFloatChannel, default=1, const [1,6]}
       dyn_wings_carpals_01_l_ctrl.Scale.Y {type=MovieSceneScriptingFloatChannel, default=1, const [1,6]}
       dyn_wings_carpals_01_l_ctrl.Scale.Z {type=MovieSceneScriptingFloatChannel, default=1, const [1,6]}
       dyn_wings_metacarpal_02_l_ctrl.Location.X {type=MovieSceneScriptingFloatChannel, default=0, const [0,6]}
       dyn_wings_metacarpal_02_l_ctrl.Location.Y {type=MovieSceneScriptingFloatChannel, default=0, const [0,6]}
       dyn_wings_metacarpal_02_l_ctrl.Location.Z {type=MovieSceneScriptingFloatChannel, default=0, const [0,6]}
       dyn_wings_metacarpal_02_l_ctrl.Rotation.X {type=MovieSceneScriptingFloatChannel, default=0, const [0,6]}
       dyn_wings_metacarpal_02_l_ctrl.Rotation.Y {
        type=MovieSceneScriptingFloatChannel
        default=0
        thinned_from=839
        keys [[-121,-0.15],[1,-0.15],[1.305,-0.1164],[1.6187,-0.0523],[2,0],[2.255,-0.0028],[8.8075,-0.1527],[14.2763,-0.3027],[14.5912,-0.435],[14.9112,-1.1673],[15.2487,-1.8995],[15.5812,-2.6141],[15.9112,-3.0463],[22.68,-3.1786],[23.0025,-3.7432],[23.3375,-4.5724],[23.68,-5.2693],[24.0238,-5.6839],[24.6812,-5.8162],[32.0763,-6.0985],[32.4075,-6.3808],[32.7638,-6.6631],[33.0863,-7.0777],[33.4162,-7.36],[33.7675,-7.6423],[34.0913,-8.3391],[34.46,-8.8861],[34.805,-9.3007],[35.145,-10.1299],[35.515,-10.4121],[35.8988,-10.5445],[39.3938,-10.8267],[39.7262,-11.2413],[40.0512,-11.506],[43.15,-11.6559],[44.0975,-11.9382],[44.4187,-12.8821],[45.1037,-13.9759],[45.7613,-14.1082],[46.71,-14.3905],[47.355,-15.4843],[47.6988,-16.1635],[48.0188,-16.8604],[48.36,-17.5396],[48.705,-17.9542],[49.0337,-18.6511],[49.3813,-19.198],[50.4075,-22.0649],[50.7575,-22.7441],[51.095,-23.141],[51.4488,-23.9525],[51.8038,-24.4994],[52.135,-25.311],[52.4688,-26.4048],[52.8238,-27.3663],[53.17,-27.9132],[53.5087,-28.3278],[53.8412,-28.4601],[55.41,-28.5924],[55.7337,-29.1393],[56.0662,-29.5362],[56.74,-29.8009],[64.795,-29.6685],[66.3562,-29.8009],[67.0063,-30.3124],[67.3587,-30.2771],[68.1025,-30.5417],[68.8288,-31.3532],[69.5013,-31.6178],[73.9037,-31.7501],[74.2188,-32.0324],[75.1725,-32.1647],[80.1363,-31.5032],[80.4638,-30.2947],[80.8087,-28.9363],[81.16,-27.8602],[81.5063,-27.4456],[81.8338,-27.2956],[83.115,-27.1633],[83.4462,-26.1841],[83.785,-24.6403],[84.1375,-23.5289],[84.4837,-22.832],[84.82,-22.6997],[87.6137,-22.832],[88.2412,-23.2466],[88.56,-23.6612],[88.8963,-24.2081],[89.235,-24.8873],[89.5738,-25.3019],[90.8475,-25.4519],[91.1712,-25.5842],[100.6163,-25.7165],[100.9725,-26.6426],[101.3138,-27.4542],[101.6575,-28.1157],[102.0063,-28.5303],[105.0987,-28.6626],[105.41,-28.7949],[105.7325,-29.0772],[106.0675,-29.2095],[111.6075,-29.6241],[111.9337,-30.7179],[112.2825,-31.3972],[112.6125,-31.6618],[115.7287,-31.7941],[116.05,-32.341],[116.7062,-32.6056],[124.4512,-32.7379],[125.715,-32.8702],[128.5087,-32.5879],[128.8363,-32.4556],[129.1575,-32.1733],[129.8238,-31.891],[132.085,-31.3265],[132.46,-30.0827],[133.24,-27.1805],[133.605,-25.9367],[133.9613,-25.0898],[134.2975,-24.1106],[134.99,-22.6992],[135.3288,-21.8346],[135.6775,-20.8731],[136.0288,-19.1971],[136.3837,-17.2387],[136.765,-15.7126],[137.1137,-14.7334],[137.4512,-14.4511],[140.57,-14.5835],[144.2688,-14.3012],[144.59,-14.1512],[144.93,-13.7189],[145.2563,-13.4366],[145.5888,-12.872],[145.9188,-12.4398],[146.2563,-12.1575],[146.5888,-12.0252],[146.9137,-11.7429],[147.2412,-11.6106],[148.5063,-11.4783],[148.8188,-11.3283],[153.1712,-11.1783],[153.8163,-10.896],[154.1438,-10.6137],[154.4725,-10.4814],[155.1413,-9.0523],[155.4837,-8.0554],[155.8238,-7.1909],[156.1625,-6.494],[156.5025,-6.0617],[156.84,-5.7794],[157.1888,-5.3472],[157.525,-5.1972],[158.17,-5.0472],[158.8212,-4.7649],[159.1463,-4.4826],[159.8025,-4.3326],[162.5863,-4.1827],[166.4887,-4.315],[167.12,-4.5973],[167.445,-5.2942],[167.7938,-5.7088],[168.1263,-6.4233],[168.4712,-7.4202],[168.8225,-8.3993],[169.1675,-8.9639],[169.5075,-9.1139],[170.18,-9.9784],[170.8475,-10.543],[171.5013,-10.8253],[175.21,-10.543],[175.5325,-10.393],[175.855,-10.1108],[176.1925,-9.6785],[176.5287,-9.1316],[177.2,-8.267],[177.8687,-7.4378],[178.2063,-7.1556],[179.4887,-6.591],[179.8212,-6.3087],[180.1575,-6.1587],[180.4825,-5.8764],[187.3088,-5.7441],[188.5825,-5.5941],[189.1988,-5.4441],[189.5275,-5.3118],[190.4787,-5.1618],[191.1187,-4.7296],[191.7638,-4.5796],[192.0888,-4.4473],[193.6463,-4.2973],[195.5113,-4.015],[195.8575,-3.865],[196.2325,-3.5827],[196.6237,-3.0005],[196.9925,-2.5682],[197.9575,-2.4359],[200,-4.054],[201,-5.5971],[202,-5.7976],[204,-10.2113],[206,-12.8296],[208,-20.088],[209,-20.6651],[238.9787,-15.1127],[239.2912,-14.6804],[239.9638,-12.7221],[240.3038,-11.4783],[240.6725,-10.9314],[240.9987,-10.6491],[241.3262,-10.4991],[242.9175,-10.2168],[243.2287,-9.9345],[243.5587,-9.5199],[243.91,-9.2376],[244.56,-8.3731],[244.9013,-7.8085],[245.2325,-7.2262],[245.5763,-6.794],[245.905,-6.644],[246.23,-6.5117],[246.8788,-6.3617],[247.5312,-5.7971],[248.1912,-4.9679],[248.53,-4.8179],[248.85,-4.6856],[258.3913,-4.271],[258.7287,-3.7064],[259.06,-2.5773],[259.4013,-0.8835],[259.78,0.9778],[260.1762,1.6924],[265.8425,1.8424],[266.155,1.9747],[266.4837,2.1247],[266.8087,2.557],[267.1475,2.9892],[267.4737,3.5891],[267.815,4.7713],[268.1575,5.3713],[268.5,5.8035],[268.8263,5.9535],[273.7875,5.8035],[274.105,5.5212],[274.4388,5.089],[274.77,4.789],[275.0938,4.3567],[275.4362,3.9245],[275.7675,3.4922],[276.4325,3.1922],[284.7313,3.3422],[285.0562,3.7745],[285.385,4.3391],[286.0437,4.9036],[286.3725,5.0536],[287.625,4.9036],[288.2675,4.6037],[288.5963,4.1891],[288.92,3.4745],[289.265,2.7599],[289.61,1.2162],[289.9575,-0.6099],[290.3425,-1.3244],[290.6675,-1.6067],[290.9913,-1.7567],[294.87,-1.889],[295.1825,-2.0213],[297.645,-2.1713],[298.2825,-2.3213],[298.6012,-3.1505],[298.9325,-3.5827],[299.2688,-4.1473],[299.595,-4.5619],[299.9425,-4.6942],[301.8062,-4.8442],[304.8913,-4.5619],[305.5187,-4.2796],[305.8438,-3.7151],[306.1875,-3.3005],[306.5175,-3.1681],[307.1712,-3.0358],[309.3175,-3.1681],[309.6313,-4.165],[309.9737,-5.8587],[310.3225,-9.1493],[310.715,-12.1575],[311.0625,-14.8481],[311.4225,-15.4127],[311.7525,-15.9772],[312.71,-16.1095],[314.2575,-15.6773],[314.58,-14.5658],[314.9288,-14.0012],[315.2637,-13.3043],[315.5987,-13.022],[316.2475,-12.7397],[317.4963,-12.5897],[317.8238,-12.4574],[318.1438,-11.6106],[318.4812,-10.7637],[318.8163,-9.7845],[319.1612,-9.3699],[319.4837,-9.2199],[319.8062,-8.9376],[320.1475,-8.7877],[321.4062,-8.5054],[321.7225,-8.0908],[322.0612,-7.8085],[322.3875,-7.3762],[322.735,-6.5117],[323.0725,-5.9471],[323.4087,-5.1002],[323.7525,-4.668],[324.1237,-4.1034],[324.4925,-3.6711],[324.885,-3.2388],[337.5863,-3.3888],[338.2175,-3.5388],[338.5387,-3.6711],[341.3275,-2.9565],[341.6525,-0.4159],[341.9987,4.9829],[342.3887,10.2671],[342.7763,13.2754],[343.1425,14.7222],[343.48,15.3044],[348.7325,15.0221],[349.9787,14.7398],[351,14.6559],[352.4913,14.5899],[353.4338,14.4576],[353.75,14.3076],[365.6075,14.4576],[365.9175,14.7398],[366.5963,15.8867],[366.925,16.3013],[367.5788,16.6012],[368.8263,16.7335],[371.9175,16.5836],[372.8888,15.2867],[373.2175,14.7222],[373.5562,14.2722],[373.8837,13.69],[374.2237,12.8254],[374.5775,12.0755],[374.9075,11.3433],[375.2475,11.0433],[375.6037,10.761],[378,3.5977],[379,-1.1219],[380,-4.3771],[381,-5.3739],[382,-5.3562],[383,-6.0355],[384,-6.5824],[388,-6.8647],[389,-11.3283],[390,-16.2065],[391,-23.7576],[392,-33.0732],[393,-33.2232],[397,-32.5263],[398,-32.394],[399,-33.4878],[400,-41.4889],[401,-58.5763],[402,-59.4231],[408.275,-59.4055],[408.5912,-59.5378],[408.92,-59.3878],[409.245,-59.1055],[409.575,-58.6909],[409.9137,-58.4086],[410.2375,-58.2586],[412.1175,-58.9025],[412.4575,-59.6964],[412.8087,-60.6402],[413.1662,-61.4694],[413.5013,-62.0163],[413.8425,-62.1486],[414.17,-62.4309],[414.4925,-62.8632],[415.165,-64.257],[415.5,-65.2361],[415.8512,-65.783],[421.235,-65.933],[422.875,-66.1976],[423.535,-68.4822],[423.9013,-69.0115],[424.2287,-69.1438],[426.1237,-68.7468],[426.8012,-66.9385],[427.1525,-65.1478],[427.5138,-62.6602],[427.8925,-59.3257],[428.2475,-55.0121],[428.6137,-50.1692],[428.9737,-45.8732],[429.325,-43.068],[429.6962,-42.5211],[430.0213,-42.3711],[432.5013,-42.5034],[433.1525,-42.6357],[433.4737,-42.768],[433.8,-43.1649],[434.125,-43.9941],[435.16,-47.2579],[435.51,-49.084],[435.8788,-50.6101],[436.22,-52.5507],[436.58,-54.9237],[436.935,-56.5998],[437.29,-57.2966],[437.6325,-57.429],[442.235,-57.2966],[442.8687,-57.0144],[445.9513,-56.882],[446.9413,-55.6559],[447.5938,-55.5236],[451.91,-55.2413],[452.74,-51.8099],[453.1825,-48.943],[453.6163,-45.9439],[453.995,-44.0209],[454.4,-42.2125],[454.7862,-40.8187],[455.1475,-39.1427],[455.5362,-37.6166],[455.9025,-35.279],[456.29,-31.5299],[456.6938,-27.7808],[457.0638,-25.5579],[457.425,-25.011],[457.7812,-24.861],[463.2575,-24.7287],[464.3025,-23.8818],[464.9962,-22.7526],[465.3525,-21.7558],[466,-20.5665],[471,-20.502],[482,-20.4795],[484,-21.191],[490.7775,-15.2722],[491.1187,-15.1222],[491.7975,-14.5753],[494.0525,-14.8576],[494.3913,-15.2899],[494.735,-16.2867],[495.0987,-17.5835],[495.4538,-19.0303],[495.8087,-20.3095],[496.1575,-21.2886],[496.5125,-21.7209],[496.85,-21.8709],[499.3587,-22.0032],[500.625,-22.4355],[500.955,-23.0001],[501.2887,-23.8469],[501.625,-24.8438],[501.9825,-26.1052],[502.3338,-27.799],[502.6962,-28.2136],[503.0362,-28.4959],[503.3863,-29.4751],[503.7275,-30.3043],[504.0763,-30.5865],[507.2175,-30.4366],[507.5438,-30.3043],[507.8775,-29.8897],[508.2138,-29.6074],[508.5475,-28.7782],[508.915,-28.2136],[509.2475,-27.5167],[510.2387,-27.1021],[510.5612,-26.8198],[511.54,-26.3875],[516.95,-26.5198],[527.2962,-26.6522],[527.995,-28.1606],[528.3487,-31.0628],[528.705,-35.5264],[529.0688,-39.0108],[529.415,-40.9692],[529.7725,-41.2515],[531.34,-41.1192],[532.5925,-41.2515],[532.9025,-41.1015],[533.84,-40.9692],[534.1475,-40.8192],[535.7325,-40.5369],[536.0412,-40.4046],[537.2962,-40.2546],[539.1612,-39.69],[539.495,-39.1255],[540.4863,-38.6932],[541.7413,-38.4109],[542.06,-38.1286],[542.415,-37.4317],[542.7763,-36.5849],[543.1137,-35.1734],[543.465,-33.0651],[543.8375,-31.2213],[544.1875,-29.113],[544.56,-26.5547],[544.9062,-22.9996],[545.2663,-17.733],[545.6375,-13.5957],[545.9825,-12.0342],[548.225,-12.1665],[548.905,-12.4488],[549.6237,-12.5811],[549.985,-12.7311],[550.6612,-13.2957],[550.99,-13.8603],[551.3238,-14.5748],[551.6725,-15.1394],[551.9925,-16.1363],[552.3325,-16.7008],[552.6788,-17.1331],[553.3162,-17.4154],[561.62,-17.2831],[562.9088,-16.7185],[564.7687,-16.5685],[565.7237,-16.4186],[566.0312,-16.1363],[566.6913,-15.854],[570.7138,-15.7217],[572.6112,-15.5717],[573.2563,-15.2894],[573.585,-15.0071],[573.91,-14.8748],[574.2325,-14.5925],[577.0213,-14.4425],[578.8813,-14.1602],[579.2138,-13.8779],[581.5113,-13.728],[581.8775,-13.2957],[582.2287,-13.1457],[582.5525,-12.8634],[583.1912,-12.7134],[589.3512,-13.0134],[589.665,-13.578],[589.9937,-14.2925],[590.3388,-15.5717],[590.6825,-17.1508],[591.0462,-17.5831],[591.3725,-17.7154],[596.0275,-17.4331],[596.355,-17.2831],[596.6825,-17.0008],[597.015,-16.5862],[597.6913,-16.004],[598.3575,-15.1394],[598.6913,-14.4249],[599.0337,-13.8603],[599.3763,-13.128],[599.7188,-12.5635],[600.3825,-11.9989],[600.7138,-11.8666],[601.3775,-11.302],[601.715,-10.8697],[602.0512,-10.1728],[602.4087,-9.4759],[602.765,-8.6291],[603.1012,-8.0468],[603.4375,-7.6322],[603.7725,-7.1999],[604.1062,-6.9177],[604.4275,-6.7677],[604.7538,-6.3354],[605.0938,-6.0354],[605.415,-5.6032],[605.7462,-4.8709],[606.0962,-4.1387],[606.7575,-3.2742],[611.4013,-2.8419],[611.7313,-2.5419],[612.0638,-2.3919],[612.3837,-1.9597],[612.7325,-1.6597],[613.1225,-1.0775],[613.5038,-0.3452],[613.9075,0.237],[614.265,0.537],[614.605,0.8546],[614.9525,1.5869],[615.2837,2.0368],[615.6187,2.769],[616.2837,3.3513],[616.6187,3.5013],[618.475,3.3513],[619.4162,3.069],[619.7375,2.7867],[620.0788,2.2045],[620.4087,1.2076],[620.75,0.0608],[621.1,-0.9361],[621.4412,-1.3683],[621.775,-1.5183],[622.43,-1.8183],[624.2912,-1.6683],[624.9225,-1.5183],[625.2412,-1.0861],[625.5675,-0.5038],[626.2312,0.0784],[626.55,0.2284],[626.8875,0.5107],[627.2262,0.793],[627.56,1.5076],[627.8975,1.7899],[628.5587,2.3544],[629.2088,2.5044],[629.8537,2.7867],[630.8,2.9367],[635.4438,2.7867],[635.7663,2.6544],[636.4125,2.3544],[636.7475,1.9398],[637.0713,1.5076],[637.405,1.2076],[637.74,0.9253],[638.0638,0.643],[639.6413,0.3607],[640.2675,0.0608],[640.5938,-0.3715],[640.9225,-0.7861],[641.2525,-1.5007],[641.5963,-2.0652],[641.9362,-2.6298],[642.2625,-3.1944],[642.6,-3.759],[642.9413,-3.8913],[643.2625,-4.0413],[651.4762,-4.1912],[651.7925,-5.2058],[652.1287,-6.4849],[652.48,-7.9317],[652.8225,-9.0609],[653.17,-10.3577],[653.5638,-11.3545],[653.8975,-12.7483],[654.2475,-13.8775],[654.5987,-14.592],[654.935,-15.4389],[655.2788,-15.7212],[657.4388,-16.0035],[657.7675,-17.0973],[658.47,-19.7349],[658.8138,-21.261],[659.1662,-23.2016],[659.53,-24.86],[659.89,-25.9715],[660.235,-26.2538],[667.3963,-26.8184],[667.7225,-28.0799],[668.0662,-29.3413],[668.4238,-30.1705],[669.1162,-32.9758],[669.4725,-34.6695],[669.8263,-37.1748],[670.1975,-39.68],[670.5587,-40.3769],[676.7575,-40.077],[677.1112,-39.5124],[677.4975,-37.0071],[677.9225,-33.9373],[678.3537,-30.3559],[678.7437,-27.286],[679.1475,-24.2338],[679.525,-21.5786],[679.9125,-18.941],[680.275,-18.0941],[680.6187,-17.8119],[684.0075,-17.6619],[686.4987,-17.9442],[687.7587,-18.2265],[688.0738,-18.9233],[688.4175,-20.0525],[688.765,-22.2755],[689.1225,-25.213],[689.4863,-27.4537],[689.84,-28.4328],[692.0337,-28.5651],[692.3825,-28.6975],[692.7138,-29.5267],[694.06,-31.185],[694.3913,-31.8819],[694.7313,-32.9758],[695.0775,-34.2196],[695.4375,-34.8811],[695.7812,-35.0311],[697.06,-35.2957],[697.72,-36.3895],[698.06,-37.0687],[698.4,-37.598],[698.7275,-37.7303],[700.2837,-37.448],[700.6,-36.7688],[700.9413,-36.2395],[701.2725,-34.9957],[701.6212,-33.7696],[701.9775,-32.2435],[702.6888,-28.0799],[703.045,-26.1215],[703.405,-24.7454],[703.77,-24.1808],[704.1062,-23.7662],[704.4338,-23.4839],[704.7687,-23.3516],[707.2525,-23.4839],[707.5675,-23.6162],[707.895,-24.0308],[708.9025,-24.4454],[711.5888,-23.3163],[711.9388,-21.7548],[712.2837,-19.9111],[712.6463,-18.3497],[713.3462,-16.9206],[713.6825,-16.7706],[714.0113,-16.3383],[714.3438,-16.056],[714.675,-15.1915],[715.0113,-14.1946],[715.3562,-13.4801],[715.7,-12.6155],[716.0288,-12.051],[716.3663,-11.3364],[716.7125,-11.2041],[717.0375,-11.0541],[720.1162,-10.9041],[720.4338,-10.4542],[720.7812,-10.0219],[721.1087,-9.4397],[721.4462,-9.0074],[721.7787,-8.8574],[722.11,-8.5751],[722.7887,-8.4251],[742.87,-8.2928],[743.1825,-7.7282],[743.51,-6.7491],[743.8587,-6.1845],[744.19,-5.4876],[744.535,-4.9053],[744.8675,-4.4907],[745.195,-4.3408],[749.7975,-4.1908],[750.1187,-3.9085],[750.445,-3.7585],[750.7763,-3.0263],[751.1087,-2.1617],[751.4438,-0.7149],[751.8,0.3173],[752,0.2938],[753,0]]
       }
       dyn_wings_metacarpal_02_l_ctrl.Rotation.Z {type=MovieSceneScriptingFloatChannel, default=0, const [0,6]}
       dyn_wings_metacarpal_02_l_ctrl.Scale.X {type=MovieSceneScriptingFloatChannel, default=1, const [1,6]}
       dyn_wings_metacarpal_02_l_ctrl.Scale.Y {type=MovieSceneScriptingFloatChannel, default=1, const [1,6]}
       dyn_wings_metacarpal_02_l_ctrl.Scale.Z {type=MovieSceneScriptingFloatChannel, default=1, const [1,6]}
       dyn_wings_phalange_02_l_ctrl.Location.X {type=MovieSceneScriptingFloatChannel, default=0, const [0,6]}
       dyn_wings_phalange_02_l_ctrl.Location.Y {type=MovieSceneScriptingFloatChannel, default=0, const [0,6]}
       dyn_wings_phalange_02_l_ctrl.Location.Z {type=MovieSceneScriptingFloatChannel, default=0, const [0,6]}
       dyn_wings_phalange_02_l_ctrl.Rotation.X {type=MovieSceneScriptingFloatChannel, default=0, const [0,6]}
       dyn_wings_phalange_02_l_ctrl.Rotation.Y {
        type=MovieSceneScriptingFloatChannel
        default=0
        thinned_from=645
        keys [[-121,0.5787],[-1,0.5787],[8.6313,0.7268],[12.1737,0.8612],[12.86,1.13],[13.1988,1.4126],[13.5462,1.8295],[14.2537,2.9458],[14.6025,3.3628],[15.5963,3.5109],[18.2025,3.6453],[23.7213,3.9141],[24.0688,4.7479],[24.435,5.7025],[24.82,6.2539],[25.1687,6.6708],[25.5175,7.2221],[25.8837,7.6527],[26.2312,8.3385],[26.6075,8.8898],[27.025,9.1723],[32.3188,9.0379],[32.6525,8.8898],[34.6225,8.7554],[35.9513,8.4592],[36.2912,8.0286],[36.6562,7.7461],[37.33,7.6117],[37.6688,7.4636],[42.2825,7.3155],[42.635,6.8985],[42.9837,5.9029],[43.35,4.921],[43.7162,4.504],[44.0713,4.2215],[50.2975,4.0734],[50.6463,3.6428],[51.0075,2.3647],[51.3775,1.6515],[51.735,1.2209],[52.0913,1.0865],[52.4438,0.9384],[54.085,0.7903],[64.645,0.6422],[70.14,0.4941],[70.81,0.6285],[78.6012,0.4941],[78.9375,0.346],[79.2825,0.0635],[79.6313,-0.2327],[81.93,-0.3671],[82.2713,-0.5152],[82.6275,-0.6633],[82.9775,-0.7977],[89.1275,-1.2283],[89.4638,-1.807],[89.8212,-2.2376],[90.5063,-2.3857],[92.3262,-2.8163],[92.6688,-2.9644],[93.02,-3.3951],[93.3763,-3.5432],[94.045,-3.6913],[96.0038,-3.8394],[97.9638,-4.4181],[98.3238,-6.2612],[98.6988,-8.6694],[99.0775,-9.517],[102.0337,-10.0957],[102.4187,-11.5082],[102.795,-12.0869],[103.1637,-12.5039],[104.2475,-12.7864],[110.4075,-12.652],[111.3988,-12.5175],[115.3363,-12.3831],[116.6712,-11.8455],[117.025,-11.3215],[118.1125,-9.2916],[118.4688,-9.1572],[136.8663,-9.0228],[137.1938,-8.4715],[137.5525,-7.7721],[137.9212,-7.3551],[138.2587,-7.2207],[138.9462,-7.0726],[143.1775,-7.207],[154.895,-7.0589],[155.2988,-6.9245],[163.5663,-6.7901],[170.0325,-6.642],[171.9987,-6.3595],[172.3413,-6.2251],[173.035,-5.1087],[173.405,-4.1405],[173.76,-3.2929],[174.1275,-2.311],[174.4937,-1.4771],[174.8475,-0.9121],[175.2113,-0.0646],[175.5675,0.4731],[175.9212,1.4413],[176.2837,2.4233],[176.6513,3.8085],[177.0412,4.7904],[177.4137,5.9068],[177.7738,6.8887],[178.1375,7.4537],[178.8487,8.0187],[179.55,8.8526],[179.91,9.1351],[180.2613,9.8345],[180.62,10.1171],[180.9737,10.6821],[181.34,11.2334],[181.6938,11.5159],[182.3887,11.7984],[193.94,11.9328],[195.575,12.0673],[196.2387,12.2017],[196.5938,12.753],[196.9538,13.0218],[197.3025,13.5732],[197.6562,13.9901],[198.0125,14.2726],[198.3663,14.407],[198.705,14.6758],[199.0487,14.8239],[199.7425,15.3753],[200.1025,15.5097],[200.445,15.7922],[200.8025,16.2091],[201.1587,16.3435],[201.5225,16.7605],[201.8788,17.3118],[202.24,17.7287],[202.6075,18.28],[203.3575,19.9614],[203.7162,20.3784],[204.44,20.5128],[204.795,20.6609],[206.185,20.9434],[206.895,21.2259],[207.9325,21.374],[208.29,21.5084],[210.0138,21.7909],[210.3663,22.2215],[210.7287,22.504],[211.095,22.6384],[214.1363,22.3559],[214.4913,21.4947],[214.8975,19.0728],[215.3,16.7991],[215.6875,15.9515],[216.0688,15.669],[216.43,15.5346],[218.8863,14.9833],[219.2875,14.1631],[219.7325,12.926],[220.1338,12.2403],[220.5087,11.8234],[221.2425,11.5545],[221.6012,10.5863],[222.005,9.0667],[222.4,6.2963],[222.7975,3.2708],[223.1963,1.6305],[223.5875,1.2136],[229.6413,0.8103],[230.0188,-0.6819],[230.3988,-2.3085],[230.79,-3.115],[231.165,-3.5319],[234.1975,-3.3975],[234.5575,-2.9943],[234.915,-1.2195],[235.2962,1.9267],[235.7062,5.9341],[236.1162,9.6865],[236.535,14.2863],[236.93,18.7654],[237.3288,23.1238],[237.7138,26.9309],[238.1237,29.7423],[238.5138,31.5718],[238.9262,31.9887],[253.3162,31.8406],[253.66,31.7062],[254.345,30.8723],[254.7138,30.0385],[255.1,28.9084],[255.51,28.0746],[255.9062,27.3752],[256.27,26.8101],[257.3425,25.9763],[257.6913,25.5594],[258.0512,25.008],[258.43,24.7392],[259.4812,24.3223],[272.3725,24.7392],[272.7088,24.8736],[273.4187,26.2725],[273.7713,27.1063],[274.4937,27.9402],[274.84,28.0883],[275.8363,28.3708],[278.1388,28.2227],[279.4925,27.8058],[280.2975,26.8375],[281.0962,25.7348],[281.475,24.7666],[281.8338,24.4978],[282.1975,23.9464],[282.6187,23.7983],[288.7337,23.3814],[289.09,22.9508],[289.4412,22.3995],[289.7938,21.7],[290.5063,20.5973],[290.8512,19.9116],[291.2025,18.7952],[291.5812,17.827],[291.94,16.7243],[292.3212,15.4872],[292.6775,14.5327],[293.0288,12.8786],[293.4087,11.9241],[293.7638,11.3727],[294.13,10.9558],[297.7425,10.8214],[298.7538,10.4045],[301.0512,10.5389],[301.3875,10.6733],[302.765,11.7486],[303.125,11.883],[303.4638,12.1518],[303.8188,12.407],[304.5163,13.1998],[304.8788,13.8718],[305.225,14.4095],[305.5775,15.0815],[305.9337,15.4848],[306.29,15.6192],[307,16.1568],[307.3487,16.5738],[308.0788,17.9316],[308.4575,18.4692],[308.8038,19.155],[314.74,19.0205],[315.1438,18.6036],[315.5538,17.7835],[315.9325,17.5146],[316.2775,16.9633],[316.63,16.6945],[316.9875,16.5601],[326.9737,16.1431],[327.3188,14.9061],[327.695,13.6553],[328.0562,11.7325],[328.4663,10.0921],[328.835,8.3037],[329.1988,7.4836],[329.5562,6.9322],[329.9088,6.7978],[340.6137,6.6634],[340.9513,6.2465],[341.3038,5.5607],[341.665,4.7406],[342.025,3.7723],[342.4125,3.221],[342.7625,2.4008],[343.1237,1.9839],[343.48,1.8495],[346.3975,1.9839],[346.7787,2.2527],[347.1637,2.9385],[347.5763,3.7723],[347.9513,4.0411],[349.9337,3.9067],[350.2675,3.6379],[350.6325,2.697],[350.9862,1.7698],[351,1.7356],[351.365,0.9633],[351.7175,0.2913],[352.0625,0.0225],[355.3088,0.1569],[360.4575,0.0225],[360.7938,-0.6633],[361.1463,-1.2009],[361.505,-1.7386],[361.8587,-2.0074],[362.205,-2.4106],[362.5688,-2.5451],[362.9062,-2.6795],[366.485,-2.5451],[367.4663,-2.4106],[367.8075,-2.1418],[368.175,-1.7386],[368.5612,-1.2009],[368.91,0.0225],[369.2675,1.0978],[369.6525,2.8725],[370.0263,5.4674],[370.4225,7.9415],[370.7975,9.5956],[371.1612,10.2813],[374.4288,10.4157],[374.7625,10.5501],[375.1075,10.967],[375.465,11.5184],[375.8163,11.9353],[376.1675,12.6211],[376.53,13.1724],[376.8788,13.8581],[377.2375,14.6783],[377.6037,15.2296],[377.9662,15.5121],[379,15.5121],[379.4575,15.3777],[380,16.3597],[381,21.1624],[382,28.3048],[383,29.8518],[385,29.9862],[389,30.5239],[390,32.6881],[391,32.9706],[392,32.6881],[393,31.1685],[394,29.1113],[395,26.3272],[396,24.9284],[398,24.3633],[399,23.5295],[407.5338,23.3814],[409.1562,23.247],[409.49,22.9782],[409.8375,22.2787],[410.2088,22.0099],[410.5963,21.8755],[411.0063,21.3242],[411.7663,20.4903],[412.1162,19.7909],[412.8338,18.6882],[413.175,18.4194],[413.525,17.868],[414.2113,17.5992],[423.2188,17.7336],[426.7812,17.4511],[427.1162,17.3167],[427.47,15.7834],[427.8463,13.4026],[428.2475,11.1836],[428.6337,9.5159],[429.0138,7.2969],[429.3963,5.6292],[429.7875,4.4058],[430.1625,3.7063],[430.5187,3.5719],[438.9037,3.4375],[439.2362,3.0343],[439.5963,2.3622],[439.9513,1.5284],[440.305,0.8563],[440.6688,0.1706],[441.0138,-0.6633],[441.3725,-1.0665],[441.7413,-1.6179],[442.1413,-2.438],[442.565,-2.9757],[442.99,-3.1101],[443.3525,-3.2582],[451.4212,-3.3926],[451.7462,-3.527],[452.0913,-4.0647],[452.45,-5.1536],[452.835,-6.9147],[453.2125,-9.2271],[453.6012,-11.3913],[453.9725,-12.7492],[454.3325,-12.8836],[455.3262,-12.7492],[456.9538,-12.6147],[458.61,-12.3459],[458.9488,-11.6602],[459.3075,-10.6919],[459.6913,-9.7373],[460.0462,-8.7828],[460.4238,-7.5457],[460.7987,-6.4567],[461.1562,-4.3858],[461.5462,-2.3422],[461.9225,-0.1506],[462.3175,1.5035],[462.7038,2.1892],[464.05,2.3373],[464.3975,2.4717],[466.015,2.3373],[467.315,2.1892],[468.9413,2.3236],[469.275,2.7268],[469.6313,3.2645],[469.9937,4.0847],[470.3562,5.0392],[470.7062,6.1282],[471.0738,7.7685],[471.46,9.705],[471.8288,11.8967],[472.1988,13.9813],[472.57,15.6353],[472.9375,16.3211],[473.2937,16.5899],[473.64,16.7243],[475.1087,16.5899],[483.7875,16.4555],[485.4137,16.3074],[485.7413,16.173],[486.0925,15.3254],[486.4475,13.2271],[486.8238,10.295],[487.2113,8.0486],[487.5925,7.0666],[487.9562,6.7841],[493.4887,7.0666],[494.4775,7.201],[494.8062,7.3355],[495.16,7.6043],[495.5075,8.1556],[495.8687,8.6933],[496.2163,9.3927],[496.5725,10.1992],[496.9375,10.468],[497.28,11.0193],[497.9775,11.2882],[498.3313,11.8395],[498.6725,12.2564],[499.3837,13.0766],[501.6562,12.9422],[501.9975,12.6597],[502.6875,12.3908],[509.6012,12.2427],[509.9438,11.8395],[510.2887,10.992],[510.6475,10.0237],[511.0125,9.338],[511.3763,8.2216],[511.74,7.6566],[512.0913,6.8364],[512.455,6.0163],[512.8212,5.5993],[513.165,4.8999],[513.5238,4.7655],[519.0475,4.2141],[519.3975,3.2459],[519.7675,1.4712],[520.1438,-0.4517],[520.5125,-1.4062],[520.87,-1.5407],[522.5587,-1.4062],[523.545,-1.003],[523.8925,0.2204],[524.245,2.1295],[524.6137,5.5172],[525.005,9.1737],[525.3763,12.9646],[525.7875,14.9945],[526.1475,16.0835],[526.5025,16.2179],[527.495,16.0835],[527.8312,15.8146],[528.17,15.6802],[528.5163,15.3977],[528.8712,15.2633],[529.22,14.9945],[529.5712,14.4432],[529.9375,13.3405],[530.29,12.1171],[530.6562,10.4631],[531.05,9.226],[531.425,8.5402],[531.7862,8.1233],[532.1375,7.9889],[532.8125,7.7064],[533.145,7.4376],[533.5125,6.7518],[533.865,5.7836],[534.2312,5.2322],[534.6,5.0978],[534.9413,4.9634],[535.9312,4.8153],[542.2312,4.9497],[542.9013,5.0841],[543.235,5.353],[543.5963,5.4737],[543.9425,5.8769],[544.2837,6.4146],[544.6375,7.221],[545.01,8.7269],[545.3813,11.1874],[545.7663,14.0511],[546.1425,16.2428],[546.5075,18.7307],[546.9125,19.8333],[547.2788,20.1158],[547.64,20.2502],[550.8963,20.1158],[552.54,19.9677],[554.8025,19.8333],[556.7713,19.5508],[558.0987,19.1339],[558.4775,18.8514],[558.8525,18.4345],[559.555,18.3],[560.2088,18.0175],[561.19,17.735],[561.5387,17.1837],[561.885,16.9012],[562.2287,16.7668],[563.2212,16.6187],[582.13,15.9055],[582.525,15.058],[582.8775,14.2105],[583.2537,12.8253],[583.6325,11.2783],[583.9975,9.4899],[584.4013,8.1047],[584.7638,6.3026],[585.1488,5.1999],[585.52,4.3661],[585.8763,4.0836],[590.745,4.218],[596.52,4.3524],[596.8663,5.1589],[597.2212,6.5167],[597.595,8.6947],[597.9837,11.5584],[598.3775,13.0643],[598.7375,13.1987],[600.4388,13.0643],[606.2775,12.7818],[606.6413,12.096],[607.005,11.5447],[607.3537,10.7108],[607.7075,10.4283],[608.0838,10.2939],[608.7825,9.7289],[609.1413,9.5945],[609.4887,9.312],[609.85,8.7469],[610.2,8.33],[610.5575,7.7787],[610.9162,7.6443],[611.2537,7.4962],[613.915,7.6306],[614.255,7.9131],[614.5975,8.3163],[615.2725,8.4507],[615.6175,8.5852],[625.595,8.7196],[625.9337,9.1228],[626.6237,9.3916],[640.0063,9.2572],[640.3388,8.854],[640.6925,8.1545],[641.3963,7.6169],[641.7362,7.4825],[644.3475,6.9311],[644.7088,6.0973],[645.065,5.8285],[645.4125,5.6804],[645.7613,5.546],[649.9413,5.6804],[656.0337,5.2771],[656.3938,4.1745],[656.7462,3.0718],[657.1062,2.386],[657.4737,2.2516],[667.6175,1.7003],[667.9625,0.8938],[668.3288,0.2081],[668.685,-0.2088],[669.035,-0.3433],[670.0263,-0.6258],[670.3763,-1.0427],[670.715,-1.1771],[671.055,-1.3252],[680.7413,-1.1908],[681.0863,-1.0564],[681.425,-0.6668],[681.78,-0.2636],[682.1313,0.6773],[682.5138,1.4838],[682.865,2.5591],[683.2513,3.648],[683.61,4.4682],[683.9638,4.8851],[698.2075,5.719],[703.745,0],[753,0]]
       }
       dyn_wings_phalange_02_l_ctrl.Rotation.Z {type=MovieSceneScriptingFloatChannel, default=0, const [0,6]}
       dyn_wings_phalange_02_l_ctrl.Scale.X {type=MovieSceneScriptingFloatChannel, default=1, const [1,6]}
       dyn_wings_phalange_02_l_ctrl.Scale.Y {type=MovieSceneScriptingFloatChannel, default=1, const [1,6]}
       dyn_wings_phalange_02_l_ctrl.Scale.Z {type=MovieSceneScriptingFloatChannel, default=1, const [1,6]}
       dyn_wings_carpals_02_l_ctrl.Location.X {type=MovieSceneScriptingFloatChannel, default=0, const [0,6]}
       dyn_wings_carpals_02_l_ctrl.Location.Y {type=MovieSceneScriptingFloatChannel, default=0, const [0,6]}
       dyn_wings_carpals_02_l_ctrl.Location.Z {type=MovieSceneScriptingFloatChannel, default=0, const [0,6]}
       dyn_wings_carpals_02_l_ctrl.Rotation.X {type=MovieSceneScriptingFloatChannel, default=0, const [0,6]}
       dyn_wings_carpals_02_l_ctrl.Rotation.Y {
        type=MovieSceneScriptingFloatChannel
        default=9.5891
        thinned_from=429
        keys [[-121,0.7463],[1,0.7463],[7,0.6699],[8,0.6699],[9,0.7463],[10,0.6699],[13,0.4851],[14,-0.3307],[16,-1.07],[30,-1.2549],[37,-1.4397],[38,-3.1033],[39,-4.5821],[40,-7.6481],[41,-10.7905],[42,-13.085],[43,-14.5637],[44,-15.3031],[45,-15.6728],[46,-17.1516],[47,-20.8485],[55,-22.1424],[56,-24.8066],[57,-26.4381],[58,-30.6253],[59,-32.9962],[73,-31.6259],[74,-30.8102],[75,-30.6253],[76,-29.0381],[77,-28.2987],[84,-28.7448],[86,-29.006],[92,-29.1908],[93,-29.7133],[94,-29.8981],[95,-31.5296],[96,-32.6066],[97,-34.0532],[98,-34.9453],[99,-35.1302],[100,-35.3914],[101,-35.4678],[102,-35.9138],[103,-35.9902],[104,-36.9587],[105,-37.4811],[107,-37.8187],[108,-37.8951],[109,-38.0799],[112,-38.1563],[113,-38.2327],[125,-38.1563],[126,-37.7102],[128,-34.7527],[129,-23.4008],[130,-20.7045],[131,-20.1499],[132,-19.4106],[134,-19.4869],[146,-19.1172],[147,-18.0524],[148,-15.8785],[153,-15.4004],[157,-15.107],[158,-14.8137],[159,-14.8137],[165,-14.6289],[166,-13.5962],[167,-13.2265],[168,-12.9332],[169,-11.2375],[173,-10.9441],[175,-10.7593],[176,-9.4654],[177,-9.2805],[179,-8.5411],[180,-8.4327],[181,-8.2478],[182,-7.3236],[183,-5.1055],[184,-4.997],[188,-4.6273],[189,-4.5188],[190,-3.04],[191,-2.3007],[192,-1.4528],[193,-0.9746],[194,1.4605],[196,1.8301],[197,2.1998],[198,3.7871],[200,4.5265],[213,4.3416],[214,3.1241],[215,1.352],[216,-9.4133],[217,-13.1545],[218,-13.709],[232,-13.8939],[235,-11.6436],[236,-8.4691],[237,-7.2516],[238,6.1857],[239,8.6208],[257,11.3492],[258,13.6758],[259,13.8607],[265,13.3061],[266,12.7516],[267,11.4577],[268,8.8699],[269,7.3147],[270,5.6832],[271,5.4984],[272,5.422],[273,5.2372],[279,4.2687],[279,5.1608],[280,1.8214],[281,-1.6264],[282,-3.0731],[283,-3.6276],[286,-3.8888],[287,-4.4434],[288,-4.8894],[289,-5.444],[290,-8.9239],[291,-10.479],[292,-13.2196],[293,-13.959],[300,-17.4068],[301,-20.2559],[302,-21.365],[303,-23.1813],[305,-23.1813],[306,-23.2256],[307,-21.2365],[308,-16.2579],[309,-15.7033],[315,-15.41],[316,-13.007],[317,-10.9737],[318,-11.0501],[319,-10.7889],[321,-10.8652],[323,-11.0501],[325,-14.5231],[326,-17.4717],[327,-20.7599],[328,-22.7437],[329,-23.1869],[330,-23.3717],[331,-23.1869],[333,-21.6121],[334,-12.7881],[335,-6.7242],[336,-6.1137],[337,-5.3442],[338,-4.866],[339,-4.6812],[342,-2.3225],[343,-1.8443],[344,-1.2576],[351,-1.0241],[359,-1.0728],[364,-1.2576],[373,-1.1813],[374,-1.551],[375,-2.4752],[376,-9.0532],[377,-25.1669],[378,-26.276],[384,-26.4608],[385,-28.1687],[386,-31.4317],[387,-32.802],[388,-37.9898],[389,-45.6249],[392,-46.2558],[393,-47.626],[395,-48.4418],[396,-50.0412],[397,-51.8254],[399,-54.2085],[400,-56.6236],[401,-57.5157],[402,-57.777],[403,-58.223],[405,-60.1157],[407,-60.7145],[408,-60.8994],[409,-62.1291],[410,-63.8048],[411,-64.3273],[413,-64.7412],[414,-66.0152],[415,-66.5376],[416,-66.614],[417,-66.8752],[418,-67.6267],[419,-68.7159],[420,-69.3147],[422,-70.0662],[424,-70.2953],[427,-71.3523],[432,-71.276],[433,-69.7286],[434,-69.4674],[436,-68.4104],[439,-67.5061],[441,-65.7418],[442,-64.3151],[443,-63.8248],[444,-63.7484],[445,-62.9969],[446,-62.4302],[447,-60.5132],[448,-59.0101],[450,-53.4682],[451,-52.0415],[452,-42.3002],[453,-38.6232],[454,-29.4851],[455,-10.807],[456,-7.098],[457,-5.9125],[459,-4.6186],[460,-4.0641],[461,-3.1398],[462,-1.4762],[463,0.2195],[464,1.437],[465,1.8388],[466,2.0237],[467,2.6867],[474,2.6103],[475,0.3679],[476,-0.2309],[477,-2.1679],[479,-6.6285],[480,-8.6297],[481,-8.7061],[482,-9.783],[483,-10.4139],[484,-10.9685],[486,-11.2297],[487,-13.4157],[488,-15.7866],[489,-17.6029],[490,-19.6041],[491,-21.4205],[492,-22.8671],[493,-24.2374],[494,-30.904],[495,-32.2743],[496,-33.3513],[497,-33.5361],[507,-33.7209],[508,-34.167],[518,-35.1676],[519,-38.0609],[520,-41.1391],[521,-42.6621],[522,-44.4463],[523,-44.9688],[527,-45.4148],[534,-45.8609],[536,-45.6318],[537,-42.3124],[538,-36.3166],[546,-36.1317],[547,-35.5772],[548,-31.0967],[549,-28.6937],[551,-28.1391],[553,-27.7694],[554,-27.1064],[555,-23.3132],[556,-21.6939],[564,-21.3242],[565,-20.9545],[566,-20.4763],[567,-20.2915],[571,-19.1182],[572,-18.9333],[575,-18.7485],[582,-18.5636],[583,-17.7922],[584,-17.4988],[585,-16.8358],[586,-16.3577],[587,-15.771],[588,-15.2928],[589,-14.6298],[595,-14.445],[596,-14.2601],[598,-9.3014],[599,-8.2687],[600,-7.5293],[603,-6.9748],[604,-5.3876],[605,-5.2027],[606,-5.0943],[607,-4.9094],[608,-4.1379],[609,-2.6392],[610,-2.2053],[615,-2.0205],[616,-1.0442],[617,-0.5504],[618,-0.6268],[620,-0.7031],[629,-1.0286],[630,-1.2776],[631,-1.9285],[632,-1.7758],[636,-1.8842],[637,-2.1776],[638,-2.286],[639,-2.6878],[641,-2.7199],[642,-3.4914],[643,-4.7411],[644,-5.6974],[645,-10.3506],[646,-12.7857],[647,-20.1231],[648,-22.6346],[649,-25.3309],[650,-28.3969],[651,-30.4302],[652,-31.5393],[653,-34.6495],[654,-39.544],[655,-39.7289],[658,-39.8053],[660,-39.958],[661,-40.3398],[662,-42.8678],[663,-47.1601],[664,-49.2298],[665,-49.3826],[666,-49.6438],[667,-51.2232],[669,-50.962],[670,-47.4578],[671,-43.7886],[672,-41.8837],[673,-41.0237],[674,-38.4237],[675,-37.6844],[677,-37.4995],[678,-37.2947],[679,-37.1784],[684,-37.9621],[685,-38.7778],[686,-41.9002],[687,-45.3038],[688,-46.3044],[689,-46.9353],[690,-47.7511],[693,-48.7196],[694,-48.1208],[695,-45.8905],[696,-41.4099],[706,-41.0402],[707,-34.3216],[708,-21.3061],[709,-14.2377],[710,-13.1286],[714,-13.205],[715,-13.3898],[723,-13.4662],[724,-12.988],[725,-13.0644],[727,-12.9238],[728,-12.719],[729,-12.3615],[730,-12.3294],[731,-11.7427],[732,-11.5379],[733,-11.301],[734,-10.835],[735,-10.3247],[736,-10.152],[737,-9.8266],[738,-9.7181],[740,-8.6976],[741,-8.2958],[743,-8.111],[744,-6.817],[745,-1.8141],[746,-0.0298],[748,-0.0619],[751,0.1672],[753,0.2435],[754,0.0787]]
       }
       dyn_wings_carpals_02_l_ctrl.Rotation.Z {type=MovieSceneScriptingFloatChannel, default=0, const [0,6]}
       dyn_wings_carpals_02_l_ctrl.Scale.X {type=MovieSceneScriptingFloatChannel, default=1, const [1,6]}
       dyn_wings_carpals_02_l_ctrl.Scale.Y {type=MovieSceneScriptingFloatChannel, default=1, const [1,6]}
       dyn_wings_carpals_02_l_ctrl.Scale.Z {type=MovieSceneScriptingFloatChannel, default=1, const [1,6]}
       dyn_wings_metacarpal_03_l_ctrl.Location.X {type=MovieSceneScriptingFloatChannel, default=0, const [0,19]}
       dyn_wings_metacarpal_03_l_ctrl.Location.Y {type=MovieSceneScriptingFloatChannel, default=0, const [0,19]}
       dyn_wings_metacarpal_03_l_ctrl.Location.Z {type=MovieSceneScriptingFloatChannel, default=0, const [0,19]}
       dyn_wings_metacarpal_03_l_ctrl.Rotation.X {type=MovieSceneScriptingFloatChannel, default=0, const [0,19]}
       dyn_wings_metacarpal_03_l_ctrl.Rotation.Y {
        type=MovieSceneScriptingFloatChannel
        default=-4.1902
        thinned_from=650
        keys [[-121,0.7462],[0,0.7462],[13.0738,0.9262],[13.4262,1.3734],[13.7725,2.1806],[14.115,2.4477],[14.78,2.6277],[21.9975,2.1358],[24.57,2.2492],[24.9037,2.5601],[25.2513,2.955],[25.6175,3.3226],[26.0487,3.7468],[26.47,4.2277],[26.9,4.7378],[27.3175,5.2205],[27.7,5.449],[28.0625,5.6776],[28.41,5.8202],[28.77,5.8768],[30.0537,5.8494],[36.0362,5.7927],[36.3637,5.566],[36.7113,5.1143],[37.08,4.8876],[37.4288,4.6901],[38.8038,4.4634],[39.1438,4.35],[39.4875,4.2934],[39.815,4.18],[40.1525,3.9825],[40.87,3.7558],[41.21,3.6424],[41.5538,3.5583],[42,3.4622],[42.2212,3.5016],[49,-1.3449],[50.7462,-1.8027],[52.0225,-1.6227],[52.68,-1.4315],[53.6488,-1.5928],[53.9925,-2.5674],[54.3725,-3.486],[55.0888,-5.0006],[55.4625,-5.4319],[55.81,-5.5932],[56.4675,-5.7019],[60.19,-5.6198],[60.5225,-5.5376],[60.8563,-5.3747],[61.2013,-5.2317],[61.8988,-5.0277],[62.2475,-4.9667],[62.5812,-4.9257],[63.2663,-4.6396],[63.6112,-4.3323],[63.965,-4.1283],[64.3138,-4.0051],[64.6612,-3.9031],[65.635,-3.8621],[65.9588,-3.8409],[66.3038,-3.7177],[66.6625,-3.6554],[67.0138,-3.6144],[71.9362,-3.5272],[72.2688,-3.6201],[72.6337,-3.8059],[72.9862,-4.1716],[73.3562,-4.7174],[73.7188,-5.1703],[74.0863,-5.9833],[74.47,-6.4362],[74.8275,-7.1563],[75.19,-7.5163],[75.55,-8.0564],[76.245,-8.2364],[81.1775,-8.0564],[82.1675,-7.8763],[82.8663,-7.6963],[83.2212,-7.1563],[83.5812,-5.449],[83.9638,-2.6673],[84.3787,0.3873],[84.7775,1.8217],[85.1538,2.0017],[86.83,1.8217],[89.1675,1.6417],[89.5663,0.5616],[90.0125,-1.8658],[90.4925,-4.2932],[90.9013,-6.4477],[91.3188,-8.3407],[91.7262,-9.6937],[92.1475,-10.3266],[92.5275,-10.6924],[92.885,-10.9653],[93.2587,-11.0582],[93.6137,-11.0639],[93.9775,-11.1511],[94.33,-11.3311],[94.6775,-11.3368],[95.4013,-11.8769],[95.7563,-11.9697],[96.1075,-12.1497],[96.475,-12.2426],[98.1363,-12.4226],[99.4588,-12.9627],[100.165,-13.6828],[101.595,-14.4028],[102.2788,-14.5829],[102.62,-14.8557],[102.9813,-15.3958],[103.35,-15.5758],[103.6938,-15.9358],[104.0512,-16.1159],[106.35,-16.2959],[109.195,-16.1159],[112.6612,-15.9358],[113.3363,-15.5758],[119.6775,-15.4829],[120.3363,-15.3029],[128,-14.9732],[129,-14.346],[136,-14.1676],[139,9.4729],[140,9.5238],[142,9.8781],[144,11.3068],[145,12.4683],[146,12.8226],[147,13.8912],[147,13.624],[147,13.3569],[148,14.0712],[150,14.3384],[151,14.6055],[156,14.4255],[157,13.8912],[158,11.2081],[159,9.9538],[160,9.3266],[161,7.8922],[162,5.7377],[162,5.9177],[162,6.0977],[163,5.5577],[176,5.1105],[179.34,0.4641],[189.0013,0.6441],[191.5675,1.1841],[191.9088,2.4385],[192.275,3.8729],[192.6325,4.6802],[192.9887,4.8602],[193.3238,5.0402],[195.2713,4.8602],[199.0638,4.6802],[200,4.6802],[201,3.2457],[202,-1.5046],[203,-7.7766],[204,-16.1358],[205,-21.7871],[206,-23.4944],[207,-34.1316],[208,-32.4755],[209,-32.473],[213,-26.6524],[215,-27.1809],[217.31,-20.1462],[218.02,-21.3249],[218.3663,-21.9578],[218.7075,-22.2307],[219.0512,-22.7708],[219.3975,-23.1308],[219.745,-23.578],[220.1025,-24.118],[220.4412,-24.4781],[220.8,-25.0181],[221.1538,-25.1982],[229,-25.5728],[232,-7.3936],[234,-8.3499],[237.77,9.3789],[239.0813,9.4661],[240.36,9.3732],[241.025,9.0132],[241.36,8.4731],[241.7,8.1131],[242.0575,7.7531],[242.3988,7.573],[252.7938,7.7531],[255.945,7.8402],[256.6112,8.1074],[257.5912,8.3745],[257.9262,8.7288],[258.2688,9.4431],[258.6237,9.9775],[258.96,10.7789],[259.3025,11.1332],[259.9913,11.6676],[260.34,11.8476],[260.6712,11.9347],[263.2038,11.5804],[263.5375,10.8661],[263.8837,9.6174],[264.2412,8.6359],[264.6,7.7473],[264.9475,6.7658],[265.2912,5.3372],[265.6675,4.2685],[266.015,3.0199],[266.3988,2.0384],[266.7462,1.5912],[267.0913,1.504],[270.2775,1.324],[280.525,1.504],[280.8625,1.5912],[281.5225,1.8583],[283.74,2.1255],[284.0863,2.4855],[284.4312,2.7527],[290.695,2.8398],[291.0325,3.1941],[291.3687,3.8213],[292,20.9055],[293,20.9927],[299,21.0798],[302,20.8127],[303,13.7567],[304,14.7073],[306,14.7985],[308,14.8314],[310,12.2235],[313,10.5544],[327.32,10.3146],[327.9875,10.4946],[328.6375,10.5817],[330.2425,10.8489],[330.9237,11.7432],[331.2587,11.8304],[331.6075,12.0104],[331.9362,12.0975],[332.2675,12.2776],[333.26,12.5447],[333.9488,13.4391],[334.285,13.7062],[334.625,14.1534],[334.97,14.5134],[337.1737,14.2463],[337.5275,13.7991],[337.865,13.0848],[338.2088,12.5505],[338.5587,11.3889],[338.9262,11.0346],[339.2688,10.8546],[339.9288,10.7675],[340.25,10.5875],[342.5587,10.7675],[343.3262,11.0346],[344.0188,11.569],[344.37,12.2833],[345.0913,12.8176],[345.7587,13.0848],[346.0875,13.2648],[346.77,13.7991],[351,13.6965],[351.5225,13.4505],[351.86,13.4563],[352.195,13.6363],[352.5512,13.7292],[352.8875,13.9092],[353.2188,14.2692],[354.2025,14.4492],[354.845,14.6293],[355.1888,15.1636],[355.5375,15.8779],[355.885,16.5051],[356.2237,16.9523],[356.5688,17.2194],[356.9212,17.6666],[357.2625,17.9338],[357.6112,18.6481],[358.2937,19.3624],[358.6525,19.8096],[358.99,20.0768],[359.3275,20.2568],[361.2537,20.0768],[361.8937,19.8096],[362.2337,19.4553],[362.58,18.6538],[363.2713,17.2309],[363.6275,16.8766],[363.9625,16.4294],[364.3038,16.3423],[364.9638,16.2551],[372.2675,15.8137],[372.6175,15.2794],[372.96,14.5708],[373.3088,13.1478],[373.6888,11.5449],[374.095,9.0476],[374.5688,5.4816],[375.0263,1.3756],[375.43,-1.656],[375.7912,-2.7246],[376.155,-2.9047],[380.5538,-2.7246],[381.8325,-2.5446],[389.0238,-2.6318],[389.3687,-2.626],[389.7075,-2.2603],[390.0825,-1.3487],[390.4462,-0.71],[390.7925,-0.5243],[391.15,-0.3443],[393.6925,-0.5243],[394.3487,-1.2386],[394.7125,-2.853],[395.1012,-5.0947],[395.5075,-6.169],[395.8725,-6.8891],[396.2188,-7.2491],[397.5275,-7.6092],[397.8575,-7.9692],[398.54,-9.4094],[399.5625,-9.9494],[400.23,-10.1294],[400.565,-10.4023],[400.9325,-10.4952],[401.2675,-10.9481],[401.6112,-11.4882],[401.9513,-11.6682],[405.83,-11.761],[407.57,-11.6739],[408.5338,-11.6796],[408.8763,-11.7725],[409.2212,-12.1383],[409.5825,-12.6841],[409.9337,-12.957],[410.275,-13.4099],[410.9462,-13.7699],[411.605,-13.9499],[411.9462,-13.8628],[412.2812,-14.0428],[412.62,-14.1357],[412.9775,-14.3157],[416.755,-14.2286],[417.0875,-12.8813],[417.4475,-9.6467],[417.81,-4.5305],[418.2013,1.1199],[418.5638,6.869],[418.9475,10.2779],[419.3075,12.5195],[419.665,13.3267],[420.0362,13.4139],[423.2075,13.2339],[424.1762,13.141],[424.8275,12.7809],[425.1775,12.0609],[425.5287,11.0679],[425.8975,10.0807],[426.2412,9.1806],[426.5963,8.1062],[426.9588,7.299],[427.2962,7.119],[427.6288,6.8518],[437.4837,5.0517],[437.92,1.3699],[438.3625,-2.9391],[438.7625,-7.2423],[439.1425,-10.204],[439.5325,-11.9984],[439.895,-12.7185],[440.235,-13.3457],[440.5775,-14.0657],[440.9362,-14.7858],[441.2738,-15.1458],[441.6087,-15.233],[441.955,-15.413],[443.2375,-15.233],[469.7287,-15.413],[473.205,-15.3259],[474.4775,-15.1458],[475.115,-14.9658],[475.7887,-14.6058],[476.1388,-14.3386],[477.7387,-13.7114],[478.0838,-12.9971],[478.4262,-11.0226],[478.79,-8.961],[479.1637,-6.7194],[479.5562,-5.0179],[479.9162,-4.7507],[482.7675,-4.6636],[483.1075,-4.4835],[485.3562,-3.6821],[485.695,-2.7877],[486.0487,-2.0734],[486.4062,-1.6262],[486.7425,-1.4462],[493.0425,-1.6262],[498.7337,-1.4462],[499.0763,-1.0919],[499.7538,-0.5576],[500.8487,-0.3775],[501.9037,-0.2904],[515.065,-0.2033],[516.0275,-0.0232],[516.97,0.0639],[519.49,0.151],[519.8325,0.3311],[520.1637,0.4182],[520.5087,0.5982],[522.1175,0.7782],[522.4512,1.3126],[522.8112,2.0269],[523.1575,2.6541],[523.505,3.6356],[523.8687,3.9956],[524.2038,4.0828],[525.8062,4.2628],[526.4638,5.07],[526.8138,5.4243],[527.15,5.6043],[530.6363,5.8715],[540.2513,5.6915],[540.5812,5.3372],[540.9312,5.07],[541.2637,4.6228],[541.6075,3.9085],[541.965,2.9213],[542.3188,2.2069],[542.6825,1.7597],[543.0312,1.3126],[543.38,1.1325],[543.725,0.8654],[544.0612,0.2382],[544.4062,-0.389],[544.7763,-0.9233],[545.12,-1.1033],[555.9425,-0.9233],[556.2688,-0.8362],[556.61,-0.4761],[556.9575,-0.029],[557.295,0.6854],[557.65,1.4054],[557.995,1.6726],[558.3325,1.8526],[560.2362,1.9398],[563.405,1.8526],[564.5275,1.5855],[565.2113,1.4054],[565.5538,1.1383],[565.8913,0.5111],[566.2325,-0.0232],[566.5863,-0.6504],[566.9262,-1.0105],[567.27,-1.2776],[567.6125,-1.4576],[567.9575,-2.172],[568.315,-2.352],[568.6513,-2.532],[572.09,-2.6192],[578.1012,-2.7992],[578.7538,-3.2463],[579.0938,-3.5135],[580.0625,-3.6935],[584.7975,-3.7807],[588.2675,-3.9607],[588.5987,-4.1407],[589.2575,-4.3207],[591.1463,-4.1407],[591.4737,-3.6935],[591.8188,-2.712],[592.185,-0.9176],[592.5638,0.3368],[592.9325,2.0384],[593.2975,3.467],[593.6475,4.6286],[594.0125,5.43],[594.3737,6.5044],[594.7175,6.7716],[595.0788,7.2187],[595.4875,7.3059],[595.9,7.4859],[597.8663,7.3059],[602.5812,7.2187],[603.5375,7.0387],[604.185,6.9516],[605.1513,6.6844],[605.4863,6.5044],[606.18,5.9701],[606.5312,5.7901],[606.88,5.7029],[607.21,5.5229],[608.2125,5.4358],[608.55,5.2557],[608.8988,4.7214],[609.275,4.0071],[609.6288,2.9327],[609.9987,1.504],[610.3938,0.1625],[610.775,-1.5333],[611.1687,-2.962],[611.5187,-4.9192],[611.9,-6.8822],[612.265,-8.2238],[612.615,-8.4909],[612.9562,-8.6709],[614.9488,-8.4909],[616.8375,-8.4038],[617.4837,-8.2238],[617.8087,-7.9566],[618.1513,-7.7766],[618.4875,-7.1494],[618.8537,-6.8822],[619.5038,-6.7022],[624.5125,-6.8822],[625.7887,-7.0623],[626.1187,-7.1494],[627.1625,-7.3294],[627.9413,-7.5094],[634.58,-7.3294],[635.2237,-6.8822],[635.5663,-5.9007],[635.925,-4.5649],[636.28,-3.0434],[636.6562,-2.0619],[637.0038,-1.7947],[653.9962,-1.5276],[654.3425,-1.2604],[654.6788,-0.7261],[655.0238,-0.2789],[655.365,0.0754],[655.715,0.2554],[656.0475,0.5226],[657.6437,0.3425],[657.9662,-0.3718],[658.33,-2.1605],[658.7625,-4.6636],[659.2212,-7.7881],[659.6737,-10.5583],[660.07,-13.4156],[660.4512,-15.3786],[660.8087,-16.0058],[663.6725,-16.1858],[664.3112,-16.0987],[665.2688,-16.0115],[665.6075,-15.7444],[669.375,-15.6572],[671.92,-15.21],[672.2563,-14.4086],[672.62,-13.4271],[672.9787,-11.9984],[673.3587,-10.2097],[673.7325,-8.5081],[674.0987,-7.1666],[674.465,-6.0922],[674.8263,-5.8251],[678.6012,-6.0051],[679.2575,-6.5394],[682,9.4498],[689,9.0027],[691,14.9438],[701,15.3039],[705,21.4294],[707.5888,3.0073],[710.425,2.8272],[711.0775,2.7401],[711.7325,2.2058],[712.085,2.0258],[712.42,1.5786],[712.785,1.3114],[713.1213,1.0443],[713.4787,0.7771],[713.8163,0.69],[714.1488,0.2428],[714.4887,-0.7387],[714.855,-1.9874],[715.2063,-2.7888],[715.5825,-3.5032],[716.2425,-3.6832],[717.5075,-3.7703],[717.8412,-3.9504],[718.1762,-4.2175],[718.8463,-4.4847],[719.1712,-4.6647],[720.46,-4.7518],[722.1625,-4.9318],[725.7413,-4.7518],[726.72,-4.839],[727.05,-4.659],[733.0625,-4.4789],[734.025,-4.2989],[737.1712,-4.0318],[737.5075,-3.8517],[737.8438,-3.5846],[738.175,-3.4974],[738.5225,-3.3174],[739.1625,-3.2303],[740.7487,-3.0503],[741.0775,-2.7831],[741.4325,-2.5159],[743.3512,-2.2488],[752,-0.4204],[753,-0.3787]]
       }
       dyn_wings_metacarpal_03_l_ctrl.Rotation.Z {type=MovieSceneScriptingFloatChannel, default=0, const [0,19]}
       dyn_wings_metacarpal_03_l_ctrl.Scale.X {type=MovieSceneScriptingFloatChannel, default=1, const [1,19]}
       dyn_wings_metacarpal_03_l_ctrl.Scale.Y {type=MovieSceneScriptingFloatChannel, default=1, const [1,19]}
       dyn_wings_metacarpal_03_l_ctrl.Scale.Z {type=MovieSceneScriptingFloatChannel, default=1, const [1,19]}
       dyn_wings_phalange_03_l_ctrl.Location.X {type=MovieSceneScriptingFloatChannel, default=0, const [0,10]}
       dyn_wings_phalange_03_l_ctrl.Location.Y {type=MovieSceneScriptingFloatChannel, default=0, const [0,10]}
       dyn_wings_phalange_03_l_ctrl.Location.Z {type=MovieSceneScriptingFloatChannel, default=0, const [0,10]}
       dyn_wings_phalange_03_l_ctrl.Rotation.X {type=MovieSceneScriptingFloatChannel, default=0, const [0,10]}
       dyn_wings_phalange_03_l_ctrl.Rotation.Y {
        type=MovieSceneScriptingFloatChannel
        default=0
        thinned_from=835
        keys [[-121,4.6701],[0,4.6701],[12.5825,4.0985],[12.9275,1.8891],[13.2862,-0.5108],[13.6575,-1.5851],[14.0163,-2.0269],[14.3625,-2.2783],[15.6337,-2.4688],[16.2887,-2.6593],[21.7175,-2.7201],[23.0113,-3.1012],[23.3425,-3.8633],[23.6888,-5.1362],[24.0638,-7.4915],[24.4412,-12.0115],[24.8125,-17.4234],[25.195,-22.5839],[25.5675,-25.9607],[25.9525,-26.4715],[33.3413,-26.5323],[33.7237,-26.2729],[34.0875,-26.1351],[34.4412,-25.8756],[34.7975,-25.8067],[35.145,-25.5473],[35.5025,-25.4784],[36.195,-24.8298],[36.905,-24.6231],[37.2613,-24.1731],[37.62,-24.0353],[37.98,-23.5164],[38.335,-23.1273],[38.7075,-22.7989],[39.0587,-22.1584],[39.405,-21.2585],[39.79,-20.6707],[40.1463,-20.0221],[40.515,-19.1222],[40.8788,-18.8627],[41.23,-18.6722],[41.5838,-18.733],[41.9188,-18.6033],[42,-18.536],[42.5987,-18.4128],[42.9275,-18.4736],[43.2625,-18.3438],[43.6187,-17.8939],[43.9688,-17.3142],[44.3262,-16.6737],[44.6775,-16.1629],[45.03,-15.9035],[45.39,-15.7129],[45.7262,-15.5832],[49,-15.6066],[51.1725,-15.4535],[51.5175,-14.7441],[51.8775,-13.2037],[52.26,-11.1525],[52.6363,-8.2094],[53.0375,-4.4961],[53.43,-0.9126],[53.7975,3.9358],[54.1812,7.7018],[54.5612,10.4462],[54.9413,12.1083],[55.3325,13.5717],[55.7075,14.2122],[56.0575,15.2946],[56.4175,16.8877],[56.7862,19.4335],[57.15,23.1306],[57.5425,26.698],[57.9088,29.5032],[58.285,31.1572],[58.655,31.3477],[63.4338,31.1572],[64.4362,30.7072],[65.2663,29.1749],[65.6888,27.6425],[66.0888,26.0494],[66.475,25.3481],[67.4775,25.4089],[68.4512,25.1494],[68.7925,24.8292],[69.1438,24.059],[69.5213,23.4185],[69.8737,22.6483],[70.2362,21.4362],[70.6063,20.7957],[70.96,19.9646],[71.3325,19.2633],[71.6788,18.7525],[72.0325,18.3715],[72.6988,18.2418],[73.035,18.112],[73.3787,17.6013],[73.7188,17.0216],[74.085,15.9311],[74.445,13.7583],[74.825,12.0273],[75.22,11.4476],[75.5775,11.0666],[75.9225,11.1274],[76.2575,10.9368],[76.9088,10.6166],[77.2537,10.2274],[77.6125,9.7775],[77.9675,9.5788],[79.9212,9.518],[83.1187,9.5788],[83.4488,9.1978],[83.8087,8.2898],[84.1825,5.9102],[84.5512,2.3713],[84.9675,-0.4502],[85.3637,-1.4798],[88.24,-1.419],[91.0962,-1.4271],[91.445,-1.9095],[91.8338,-3.1174],[92.2175,-5.2253],[92.59,-6.6156],[92.9625,-7.7223],[93.335,-7.852],[95.2663,-7.9128],[95.6075,-7.7223],[96.5963,-7.7831],[96.975,-7.5237],[97.8175,-7.3697],[98.5538,-7.5967],[98.9175,-7.5805],[99.27,-7.4954],[99.6488,-7.2725],[100.0188,-6.9807],[100.41,-6.5754],[100.7763,-6.4376],[103.6988,-6.4984],[104.06,-6.4903],[104.4137,-6.2998],[104.7687,-6.17],[107.6313,-6.0403],[107.975,-5.7039],[108.3375,-5.3755],[108.7113,-4.7877],[109.12,-4.5202],[109.4725,-4.1311],[109.8475,-3.7419],[110.205,-3.2919],[110.57,-2.392],[110.9462,-1.6137],[111.305,-0.8354],[111.6913,-0.3165],[112.0437,0.324],[112.3963,0.4537],[129.7075,0.2713],[130.085,0.3726],[130.485,1.1143],[130.8937,2.3669],[131.265,3.6114],[131.6525,4.4586],[132.0375,4.5883],[137.1275,4.6491],[139.7175,4.5883],[141.655,4.6491],[142.3162,4.5194],[142.6637,4.5802],[143.0163,4.26],[143.3713,4.2519],[143.715,4.0005],[144.0625,3.8627],[144.4288,3.5344],[144.7887,2.748],[145.1988,1.3211],[145.5688,0.0766],[145.9513,-1.2206],[146.3413,-1.6097],[149.2362,-1.6706],[149.5725,-1.48],[149.9062,-1.5408],[150.2463,-1.2814],[150.96,-0.6247],[151.3225,-0.3653],[151.6837,-0.2964],[152.0263,-0.2275],[152.3837,0.2914],[152.7362,0.49],[153.0987,0.7576],[153.475,1.2156],[153.83,1.6048],[154.1812,1.8723],[154.5525,2.3223],[154.9075,2.5209],[159.055,2.4601],[159.3963,2.6506],[159.7337,2.5898],[162.7787,2.6506],[163.1175,2.5209],[163.4675,2.2615],[163.81,2.0629],[164.1612,1.9331],[164.5125,1.9939],[165.4825,1.8642],[166.7713,1.7345],[167.12,1.6048],[167.805,1.0859],[168.1687,0.567],[168.525,-0.1505],[168.9475,-0.418],[169.3038,-0.8071],[169.68,-0.8761],[170.02,-1.0058],[170.3663,-1.1963],[170.705,-1.326],[171.045,-1.3949],[171.3988,-1.6544],[171.74,-2.0435],[172.095,-2.4935],[172.4512,-2.5624],[172.8012,-2.6921],[173.4675,-2.8826],[174.1375,-2.8299],[174.48,-3.1664],[174.8463,-4.1514],[175.2275,-5.3959],[175.5987,-6.9606],[175.9813,-7.8078],[176.3487,-8.2578],[179.8438,-8.197],[182.7,-8.2578],[184.935,-8.0672],[185.6112,-8.1889],[185.9625,-8.3632],[186.325,-8.2173],[186.68,-8.0106],[187.0475,-7.6822],[187.3963,-7.4228],[187.745,-7.3539],[188.1112,-7.4147],[188.4488,-7.285],[188.7838,-7.4674],[189.145,-7.0783],[189.4937,-6.8188],[189.8525,-6.6202],[190.2013,-6.4905],[190.5425,-6.5513],[190.8888,-6.3608],[191.2237,-6.4216],[191.885,-6.2919],[192.23,-6.3446],[192.6112,-6.1379],[193.0375,-5.542],[193.5025,-4.6259],[194.28,-2.7936],[194.6625,-1.4964],[195.0512,-0.0087],[195.4212,1.479],[195.7887,3.0965],[196.1875,5.1558],[196.5587,7.1462],[196.9575,7.665],[197.3062,7.8556],[205.58,7.7948],[205.9413,8.7555],[206.3,10.1054],[206.6938,11.0053],[207.4,11.3945],[209.6225,11.2648],[210.59,11.1351],[210.9325,10.8148],[211.2713,10.2959],[211.63,9.769],[212.0113,8.6015],[212.3787,6.5341],[212.7587,4.2762],[213.1225,2.2169],[213.4925,1.1183],[213.87,0.4008],[214.235,-0.2478],[214.6187,-0.7748],[215.0025,-1.7517],[215.365,-2.7286],[215.7275,-3.584],[216.1137,-4.3623],[216.4775,-4.4393],[216.8338,-4.569],[219.6825,-4.6379],[220.0188,-4.7676],[220.3575,-4.9662],[220.7175,-5.1041],[221.065,-5.3108],[221.4462,-5.9675],[221.795,-6.8228],[222.16,-7.4186],[222.5225,-7.7632],[222.9075,-8.0307],[223.2675,-8.0996],[223.615,-8.2293],[223.9675,-8.4199],[226.3713,-8.2293],[227.0288,-8.2902],[227.3537,-8.351],[227.6925,-8.2212],[228.0325,-8.4037],[228.3775,-8.4483],[228.7587,-8.4767],[229,-8.2859],[232,23.6043],[234,29.8679],[237.5163,-11.6954],[238.2287,-11.8739],[238.595,-11.4362],[238.9775,-10.4958],[239.3425,-9.0365],[239.7213,-7.3948],[240.0825,-5.4328],[240.46,-3.2276],[240.8388,-1.6102],[241.2,-1.0913],[241.5487,-0.9616],[242.5438,-0.9007],[249.165,-0.8318],[249.51,-0.5116],[249.8512,-0.3819],[250.1938,-0.4427],[250.5387,-0.2522],[253.0738,-0.313],[253.4075,-0.1832],[253.7487,0.137],[254.0987,0.2667],[254.4512,0.587],[254.7925,0.8464],[255.14,1.6166],[255.515,2.4557],[255.8737,4.1867],[256.2337,6.2379],[256.6375,8.6094],[257.0575,11.7349],[257.5475,13.328],[257.9413,13.5186],[258.295,13.6483],[259.9188,13.8388],[264.955,13.5186],[265.2925,13.2591],[265.6363,13.1902],[266.665,12.8011],[267.0088,12.7322],[267.3475,12.7241],[267.6975,12.5335],[268.0575,12.0147],[268.415,11.6863],[268.77,11.4269],[269.12,11.2972],[269.4712,11.358],[270.7462,11.1674],[272.3413,10.8472],[272.6737,10.5878],[273.0238,10.3891],[273.3787,9.6797],[273.7487,8.8406],[274.12,8.192],[274.4787,7.6731],[274.8263,7.0934],[275.19,6.834],[275.875,6.5746],[276.235,6.2543],[276.5775,6.0557],[276.925,5.6665],[277.2812,5.4679],[277.6325,4.6369],[278.005,4.4302],[278.42,3.8424],[278.7925,3.2465],[279.195,2.2615],[279.5738,1.2156],[279.955,0.3765],[280.3175,-0.0126],[282.2437,0.0482],[282.905,-0.0816],[300,-0.341],[301,-1.0585],[302,-4.3096],[303,-10.4063],[304,-13.4303],[307.04,-13.7506],[308.0225,-13.6209],[308.35,-13.6128],[309.055,-13.4222],[309.4013,-13.2236],[309.745,-13.2155],[310.1037,-12.8953],[310.46,-12.6885],[310.835,-11.9102],[311.195,-11.0549],[311.5638,-10.001],[311.9513,-8.6876],[312.325,-7.7634],[312.6962,-7.1067],[313.06,-6.7784],[313.4175,-6.6486],[313.7625,-6.5797],[314.115,-6.1906],[314.4613,-5.923],[314.8188,-5.3353],[315.1875,-4.2367],[315.5663,-3.6489],[315.9288,-3.58],[321.8475,-3.3895],[322.1775,-2.6801],[322.5375,-1.3221],[322.9262,0.1737],[323.2962,1.5925],[323.6612,2.1722],[329.3225,2.5614],[329.66,2.7843],[330.0312,3.8301],[330.405,5.0057],[330.785,5.8529],[331.1525,6.8826],[331.5175,7.6609],[331.9013,8.6905],[332.26,9.7891],[332.6388,10.5674],[333.35,11.0863],[340.2937,10.9566],[340.6388,10.8269],[340.9737,10.0485],[341.3413,8.4392],[341.7113,7.6001],[342.0662,7.4704],[343.975,7.6609],[344.625,7.6001],[345.27,7.7906],[345.6063,7.7298],[345.945,8.05],[346.2925,9.0189],[346.68,10.0566],[347.045,11.2241],[347.4575,11.743],[347.805,12.0632],[349.735,11.9335],[350.3913,11.743],[351,11.6363],[351.0425,11.6133],[351.39,11.293],[351.7287,10.6444],[352.1175,10.0566],[352.5087,8.8892],[352.9488,7.6001],[353.405,6.9596],[360.535,6.6393],[360.87,6.0515],[361.2275,5.2124],[361.605,3.9152],[361.9762,1.5965],[362.3713,-0.7912],[362.7413,-2.3478],[363.1062,-3.2558],[363.485,-3.7666],[368.5638,-3.8963],[369.2088,-4.026],[369.5462,-4.3463],[369.89,-4.476],[374.9638,-4.9341],[375.3075,-6.5758],[375.68,-8.9958],[376.0525,-13.5441],[376.4475,-19.3571],[376.815,-26.3985],[377.2188,-30.6103],[377.5975,-32.6858],[377.9562,-33.5249],[380.8675,-33.3952],[381.5312,-33.1357],[381.88,-33.0668],[382.2237,-32.6777],[383.295,-32.2196],[383.635,-32.2804],[383.9725,-32.0899],[386.6925,-32.1507],[387.03,-31.9521],[387.375,-31.2873],[387.7487,-30.1725],[388.5,-27.2702],[388.8837,-25.446],[389.6388,-22.1707],[390.3938,-19.0331],[390.7687,-18.117],[391.14,-18.0481],[392.7475,-17.9873],[397.1538,-18.0481],[397.4913,-17.8576],[397.835,-17.5981],[398.1812,-17.3995],[398.5362,-17.0103],[399.2525,-16.8725],[399.5963,-16.682],[404.3425,-16.3009],[404.6888,-16.1023],[405.3925,-14.1646],[405.78,-12.608],[406.1325,-10.6541],[406.5087,-8.0517],[406.87,-5.3884],[407.25,-3.3129],[407.62,-1.3144],[407.9987,-0.3375],[408.36,0.9435],[408.7337,1.7827],[409.1413,2.3624],[409.4975,2.7515],[409.8575,3.1326],[411.7663,2.942],[412.1012,3.0029],[414.6513,2.942],[415.6137,3.0718],[415.9562,3.3312],[416.3012,3.842],[416.6712,3.7812],[418.11,4.2919],[418.8288,6.0391],[419.1938,7.389],[419.5725,8.6092],[419.9475,9.3186],[420.305,9.8294],[420.6625,9.9591],[420.9962,10.1496],[421.3325,10.4699],[421.6762,11.2401],[422.0412,12.5211],[422.4087,15.1439],[422.8025,18.4072],[423.18,21.6705],[423.5425,24.8568],[423.9188,26.8999],[424.2912,27.2202],[428.3913,27.0296],[429.36,27.0905],[430.08,26.8999],[430.44,26.5797],[430.8125,26.4419],[431.1687,25.6027],[431.525,24.6258],[431.9088,23.4664],[432.275,22.5665],[432.6363,22.3071],[432.9775,22.3679],[433.6488,22.1773],[436.1725,22.1165],[441.855,22.1773],[442.5338,21.9868],[443.265,21.4071],[446.245,21.2166],[452.045,21.2774],[452.405,21.0869],[452.7587,20.7666],[453.4638,20.8274],[453.8075,20.6369],[454.5187,20.1261],[454.875,19.737],[455.25,19.0276],[455.6413,18.1276],[456.0225,17.1061],[456.41,16.7858],[456.7687,16.6561],[459.4525,16.4656],[459.7962,16.3358],[460.16,16.0156],[460.5287,15.8859],[460.8813,15.6953],[461.58,15.5048],[469.4112,15.5656],[469.7475,15.3751],[506.855,15.1846],[507.1912,15.2454],[507.5287,14.086],[507.9,12.3469],[508.2587,9.5174],[508.6237,5.0785],[509.0075,0.2586],[509.3787,-4.3181],[509.76,-6.8354],[510.1263,-7.9421],[510.4937,-8.0799],[515.1037,-8.1407],[515.4325,-7.9502],[522.6562,-7.8894],[526.75,-8.0799],[532.9638,-8.6596],[533.3275,-10.1392],[534.125,-12.5188],[534.5325,-12.9687],[534.8837,-13.1674],[535.5863,-13.6862],[535.9362,-13.7632],[536.3,-14.1524],[536.655,-14.3591],[537.0163,-14.7483],[537.3637,-14.878],[546.6612,-14.9388],[547.3062,-15.1212],[547.6562,-15.097],[548.0425,-14.943],[548.4312,-14.6673],[548.7925,-14.2782],[551.365,-14.4687],[551.6988,-14.4079],[552.3938,-14.5376],[552.73,-14.7362],[553.0788,-15.2551],[553.4588,-15.6443],[553.8087,-15.7213],[554.1637,-16.1712],[554.8737,-16.6901],[555.2225,-17.0104],[555.5712,-16.9495],[558.7587,-17.0104],[559.73,-16.8806],[560.065,-16.9414],[564.1938,-16.8806],[564.52,-17.0104],[564.8625,-17.2698],[565.2175,-17.2779],[565.5612,-17.3387],[565.9,-17.5981],[566.9638,-18.194],[567.6588,-18.7129],[568.3462,-18.8507],[568.685,-18.9804],[594.2912,-19.0412],[594.6288,-18.9115],[594.9725,-18.4535],[595.325,-17.9265],[595.6837,-17.0712],[596.4375,-13.9579],[596.84,-12.7215],[597.2088,-11.5622],[597.5788,-10.6622],[597.95,-10.0136],[598.2988,-8.9232],[598.6737,-8.0841],[599.0462,-6.7422],[599.4137,-6.6733],[602.28,-6.8639],[602.9413,-6.8031],[610.6313,-6.6125],[611.2738,-6.4828],[611.6137,-6.1626],[611.9488,-6.2234],[613.545,-6.0328],[613.8888,-5.8423],[614.225,-5.7126],[614.5663,-5.5221],[615.2463,-5.2626],[620.3413,-5.3234],[620.675,-4.7965],[621.0462,-3.6128],[621.4225,-1.6589],[621.8163,0.8057],[622.1925,2.7353],[622.5612,3.5055],[622.9312,3.6961],[626.3863,3.5055],[636.3887,3.5663],[637.0612,3.3758],[637.4,3.4366],[640.8937,3.3758],[641.2375,3.5055],[641.6437,4.146],[642.0587,5.4959],[642.49,7.608],[642.8875,10.0321],[643.2613,11.3131],[643.6225,11.8847],[643.9675,12.0144],[647.7713,12.205],[649.71,12.2658],[650.3875,11.8847],[650.7213,11.5564],[651.1025,10.5875],[651.4675,8.7796],[651.83,6.7203],[652.1975,5.942],[652.5763,5.6825],[654.81,5.492],[655.1538,5.3623],[655.5075,4.7826],[655.8675,4.0043],[656.2513,2.3341],[656.6263,0.4653],[657.0113,-1.9831],[657.415,-3.9127],[657.78,-4.2411],[658.13,-4.3708],[680.3625,-4.8208],[680.7025,-5.1572],[681.075,-5.745],[681.445,-6.0733],[681.8038,-6.5233],[682.1538,-6.9124],[682.5125,-7.5529],[682.8775,-8.5218],[683.2475,-9.5595],[683.6187,-10.8648],[683.9975,-11.9107],[684.38,-12.8187],[684.7437,-13.147],[685.1025,-13.3376],[685.4438,-13.2767],[688.2775,-13.3376],[693.0013,-13.3984],[693.6475,-13.2078],[693.98,-13.2687],[696.1988,-13.3903],[696.5362,-13.3822],[696.9,-13.5119],[697.2425,-13.3741],[697.6075,-13.1755],[697.9613,-13.2282],[698.3062,-13.0985],[698.6413,-13.1593],[698.9975,-13.0296],[699.6737,-12.8917],[700.0275,-12.7539],[700.3825,-12.4175],[700.7413,-12.2554],[701.1287,-12.0244],[701.4937,-11.161],[701.8788,-10.6259],[702.2587,-10.4881],[702.6087,-10.4192],[702.9625,-10.2895],[703.2988,-10.3503],[703.6413,-10.2206],[703.9925,-10.2814],[704.3438,-10.2733],[704.6988,-9.8922],[705.0537,-9.6855],[705.8625,-9.0977],[706.2575,-9.1585],[710.4275,-9.0977],[714.5612,-9.2883],[715.225,-9.2274],[715.55,-9.418],[715.8937,-9.5477],[716.2337,-9.6166],[716.5825,-9.8152],[716.945,-10.2044],[717.2962,-10.853],[717.67,-11.8218],[718.0525,-12.8515],[718.7787,-13.6298],[719.1438,-13.8203],[719.4837,-13.95],[720.1425,-14.1406],[724.2063,-14.0108],[724.8712,-13.9419],[725.2125,-13.796],[725.585,-13.261],[725.9613,-12.3367],[726.3225,-11.0396],[726.7088,-10.7112],[727.0662,-10.772],[734.1375,-10.5815],[734.4787,-10.5126],[734.8175,-10.1923],[735.16,-10.0545],[735.535,-9.7951],[735.88,-9.787],[736.2337,-9.7262],[736.9513,-9.0776],[737.3188,-8.1168],[737.7387,-6.775],[738.155,-5.8062],[738.5975,-5.1576],[739.015,-4.8982],[741.8663,-4.7076],[742.2,-4.7684],[742.545,-4.5779],[742.9,-4.509],[743.2475,-4.3185],[749.515,-3.8685],[749.8563,-3.0983],[750.2138,-2.3889],[750.565,-1.4281],[750.925,-0.4674],[751.2887,-0.1471],[751.63,-0.0174],[751.9713,0.1731],[752,0.173],[752.94,0.1123],[752.9987,3.2662],[753,3.2662]]
       }
       dyn_wings_phalange_03_l_ctrl.Rotation.Z {type=MovieSceneScriptingFloatChannel, default=0, const [0,10]}
       dyn_wings_phalange_03_l_ctrl.Scale.X {type=MovieSceneScriptingFloatChannel, default=1, const [1,10]}
       dyn_wings_phalange_03_l_ctrl.Scale.Y {type=MovieSceneScriptingFloatChannel, default=1, const [1,10]}
       dyn_wings_phalange_03_l_ctrl.Scale.Z {type=MovieSceneScriptingFloatChannel, default=1, const [1,10]}
       dyn_wings_metacarpal_04_l_ctrl.Location.X {type=MovieSceneScriptingFloatChannel, default=0, const [0,15]}
       dyn_wings_metacarpal_04_l_ctrl.Location.Y {type=MovieSceneScriptingFloatChannel, default=0, const [0,15]}
       dyn_wings_metacarpal_04_l_ctrl.Location.Z {type=MovieSceneScriptingFloatChannel, default=0, const [0,15]}
       dyn_wings_metacarpal_04_l_ctrl.Rotation.X {type=MovieSceneScriptingFloatChannel, default=0, const [0,15]}
       dyn_wings_metacarpal_04_l_ctrl.Rotation.Y {
        type=MovieSceneScriptingFloatChannel
        default=0
        thinned_from=965
        keys [[-121,0],[0,0],[3,-0.1992],[11.4975,-0.3985],[13.5612,-0.5977],[13.8288,-0.797],[14.1012,-1.1782],[14.3887,-1.7586],[14.67,-2.3564],[15.2312,-3.5172],[15.5175,-4.3142],[16.09,-5.4578],[16.3763,-6.0555],[17.23,-7.1991],[18.085,-8.8884],[18.3787,-9.6508],[18.685,-9.7981],[18.9837,-9.9973],[22.9275,-9.7981],[25.8288,-9.8154],[26.0888,-9.6162],[26.385,-9.0877],[26.685,-8.7757],[26.9787,-8.2646],[27.2713,-7.7361],[27.5738,-7.3722],[27.8525,-7.2422],[28.1625,-7.043],[31.7938,-7.2422],[32.3325,-7.6234],[32.6112,-7.788],[32.9,-8.1519],[33.1787,-8.1346],[33.7262,-8.3165],[34.5263,-8.5158],[35.0462,-8.715],[35.31,-8.6977],[35.59,-8.897],[35.8575,-8.8796],[36.1338,-9.0789],[36.4013,-9.0616],[36.6725,-9.2435],[40.06,-9.2608],[41.1187,-9.0616],[42,-8.8395],[49,-9.0913],[51.5263,-9.5901],[51.8012,-10.4651],[52.1025,-11.9986],[52.4038,-15.5854],[52.7188,-18.9902],[53.0662,-21.6499],[53.4125,-23.6339],[53.7713,-24.6735],[54.1363,-25.1847],[54.485,-25.3667],[62.625,-25.3667],[63.9275,-25.1847],[64.205,-25.202],[64.4725,-25.0374],[64.7738,-25.072],[65.0525,-24.8728],[80.065,-24.8728],[80.335,-25.2194],[80.6187,-25.9212],[80.9475,-27.0824],[81.2513,-28.72],[81.5512,-30.5831],[81.8537,-32.9399],[82.16,-35.4441],[82.4812,-37.2725],[82.7912,-38.6763],[83.1037,-39.2743],[83.415,-39.517],[85.0875,-39.517],[87.105,-39.7162],[90.22,-39.6989],[91.0412,-39.595],[101.9375,-39.6469],[102.7437,-39.7508],[103.0387,-39.6208],[103.885,-39.7247],[104.17,-39.5428],[107.1625,-39.5947],[107.47,-39.3955],[113.1912,-39.4128],[113.775,-39.2309],[119.2113,-39.2828],[119.5438,-39.0228],[119.8913,-38.9967],[120.225,-38.6068],[120.9962,-37.463],[121.3512,-36.6918],[121.7,-35.3575],[122.0612,-33.2436],[122.4162,-31.6928],[122.7462,-30.1246],[123.0863,-29.2842],[123.4238,-28.5738],[123.7713,-28.4092],[124.0888,-28.4265],[124.4112,-28.2272],[125.605,-28.0453],[125.9075,-27.8807],[126.5225,-27.1529],[126.8525,-26.971],[135.66,-26.4252],[137,-19.6738],[139,-2.1876],[140,-3.4203],[141,-3.7495],[142,-4.4599],[145.7713,12.4124],[146.0525,12.4297],[147.11,12.2304],[147.9362,11.6327],[149.0288,11.4334],[149.63,10.8703],[149.9312,10.1079],[150.2688,9.7268],[150.585,9.5448],[152.7825,9.5275],[153.8438,9.7268],[154.6988,10.186],[155.76,10.1687],[158.3725,9.9694],[158.9137,10.004],[159.4675,9.2417],[160.0437,9.2936],[160.325,9.0944],[167.5925,8.6959],[167.8663,8.1154],[168.16,7.1711],[168.475,6.4261],[168.79,5.1006],[169.1162,3.6624],[169.4338,2.5882],[169.7387,1.8777],[170.0387,1.3492],[170.3325,1.1673],[170.6125,0.9681],[170.8875,0.9854],[171.9513,0.7861],[173.5225,0.8034],[173.7812,0.6388],[174.07,0.4396],[176.425,0.2403],[176.6875,0.0584],[176.9688,0.0757],[177.2537,-0.3055],[177.5375,-0.4528],[177.8238,-0.6347],[178.12,-0.8167],[184.415,-0.834],[184.9438,-0.8513],[186.0725,-0.9898],[186.3537,-1.0417],[186.6313,-1.0591],[186.9025,-0.9118],[187.2025,-0.9464],[187.4812,-0.9637],[188.0275,-0.981],[189.3425,-0.9983],[189.8663,-1.0156],[190.1313,-1.033],[190.6762,-0.8337],[194.8487,-0.851],[195.9237,-0.8857],[196.7237,-0.903],[197.7563,-0.9203],[198.0238,-0.9549],[198.3062,-1.1715],[198.5825,-1.388],[199.195,-1.8558],[199.4725,-2.055],[199.7462,-2.4709],[200.04,-3.0686],[200.3188,-3.6837],[200.6125,-5.0784],[200.9262,-7.2702],[201.2387,-9.6611],[201.5362,-13.213],[201.8537,-16.5829],[202.155,-19.7363],[202.4675,-23.3054],[202.8,-26.4415],[203.155,-30.1406],[203.4613,-34.6022],[203.7763,-40.0426],[204.0875,-45.717],[204.405,-50.9582],[204.6975,-56.6325],[205.0175,-61.3279],[205.6237,-69.3932],[205.93,-72.9105],[206.235,-76.2284],[206.53,-79.9449],[206.83,-84.2245],[207,-86.3444],[207.1513,-87.9237],[207.45,-89.72],[207.7513,-89.538],[209.8563,-90],[211.2013,-89.2896],[211.4812,-88.943],[211.7713,-88.7957],[212.3413,-88.1026],[213.2563,-87.3488],[213.5763,-86.8376],[214.2525,-85.1741],[215.6388,-80.3396],[215.9638,-78.5722],[216.2675,-77.1687],[216.56,-74.4397],[216.8725,-71.7107],[217.1975,-69.5275],[217.505,-67.8554],[217.81,-66.7465],[221.4887,-66.9631],[222.0475,-67.6735],[222.34,-68.4359],[222.6212,-69.5275],[222.94,-70.4718],[223.2275,-70.671],[225.6012,-70.671],[225.8687,-70.5064],[226.1463,-70.0298],[226.4475,-69.2241],[226.7663,-67.7425],[227.0688,-65.8451],[227.685,-61.4526],[227.9875,-59.1567],[228.2912,-56.5142],[228.6,-53.1441],[228.9462,-49.4793],[229.2663,-44.8529],[229.5638,-39.4814],[229.8788,-32.5333],[230.49,-19.3216],[230.8125,-15.7783],[231.1125,-13.3266],[231.415,-13.1619],[231.97,-13.1446],[232.2375,-13.3266],[233.3238,-13.6904],[233.6025,-13.8897],[233.8763,-13.8724],[234.1575,-14.0716],[234.44,-14.0543],[237.305,-14.2882],[237.5788,-14.3401],[238.155,-14.5133],[238.7437,-14.3572],[239.0337,-13.8979],[239.3425,-13.638],[239.6488,-12.9622],[239.9538,-12.4683],[240.2425,-12.3729],[240.5562,-12.0437],[240.84,-11.7145],[241.7375,-10.848],[242.0312,-10.718],[242.315,-10.5533],[248.01,-10.3714],[248.285,-10.0595],[248.5825,-9.9295],[248.865,-9.6176],[249.4675,-9.4269],[249.7538,-9.1669],[250.6562,-9.002],[250.955,-8.9239],[251.2487,-8.9759],[251.5325,-9.2444],[251.8375,-9.279],[252.1087,-9.4956],[252.3813,-9.7468],[252.6662,-9.9634],[252.9462,-10.3965],[253.2362,-10.4138],[253.5275,-10.8469],[253.8,-11.0808],[254.0775,-11.7305],[254.39,-12.3802],[254.6825,-13.0299],[255.2525,-14.7451],[255.5512,-15.8106],[256.1325,-17.5604],[256.44,-18.2101],[256.7213,-18.8598],[257.6,-20.2804],[258.1862,-20.8866],[259.105,-21.3195],[259.4038,-21.4061],[259.7075,-21.2934],[260.28,-21.4319],[260.57,-21.3019],[263.5063,-21.3539],[263.77,-20.99],[264.3363,-21.0593],[264.89,-20.73],[265.4462,-20.7647],[266.225,-20.5827],[268.3613,-20.6001],[269.4075,-20.4008],[269.67,-20.4181],[270.2125,-20.2362],[270.4937,-20.2535],[271.3363,-19.7423],[271.6075,-19.3611],[271.89,-19.1965],[272.455,-19.2485],[272.7237,-18.8673],[273.0125,-18.9019],[273.5663,-18.538],[274.3675,-18.5554],[281.2763,-18.3734],[281.54,-18.2088],[282.1125,-17.1691],[282.4162,-16.8226],[282.7025,-16.6407],[282.9813,-16.0949],[283.2788,-16.1122],[284.1075,-15.549],[285.5213,-14.7086],[285.795,-14.5267],[286.34,-14.544],[287.6375,-14.3448],[288.1762,-14.1628],[289.3125,-13.4524],[296.7675,-13.0539],[297.0325,-12.1269],[297.34,-10.2037],[297.635,-7.7173],[297.9225,-3.2731],[298.235,1.9161],[298.545,7.9024],[298.835,13.1263],[299.15,16.6089],[299.465,17.7697],[304.13,17.5705],[304.6475,17.5878],[305.4375,17.3886],[307.5125,17.1893],[308.045,17.2066],[308.3062,17.0074],[308.5812,17.0247],[308.8637,16.8428],[309.4462,16.0458],[310.0775,15.2834],[312.4638,13.7067],[312.7325,13.3082],[313.02,12.927],[313.295,12.7278],[313.5612,12.3293],[314.1175,11.9481],[321.8263,11.9308],[322.87,12.13],[323.4075,12.5112],[324.1938,12.7105],[326.5225,12.5112],[326.7875,12.312],[327.0537,11.9481],[327.3425,11.3677],[327.6263,10.4407],[327.93,9.3144],[328.225,8.1709],[328.5238,6.8281],[328.8288,5.8665],[329.1213,5.1214],[329.4175,4.5237],[329.725,4.3417],[332.5788,4.541],[333.1037,4.9395],[333.37,5.3034],[333.6612,5.8665],[333.955,6.6115],[334.2437,7.1746],[334.5338,7.3393],[334.8112,7.5385],[335.08,7.5212],[343.8225,7.7031],[344.0987,8.0843],[344.3863,8.2489],[351.1338,8.2663],[351.3988,8.067],[351.9325,7.8851],[352.1912,7.6858],[352.4863,7.14],[352.7675,6.5596],[353.0662,5.6153],[353.3587,4.2725],[353.6675,3.5101],[353.9475,2.947],[354.2262,2.7824],[354.5238,2.4185],[362.5375,2.0373],[362.8,2.0546],[363.0763,1.8727],[363.3663,1.6735],[363.6313,1.4742],[363.9075,1.4915],[364.1838,1.2923],[364.455,1.093],[364.7237,0.6945],[365.0063,-0.3017],[365.3088,-1.4799],[365.6037,-3.4378],[365.93,-5.9761],[366.2463,-9.294],[366.5587,-12.9586],[366.8587,-17.8013],[367.1812,-22.8259],[367.49,-31.0473],[367.9625,-35.405],[368.2625,-39.1995],[368.5825,-41.4866],[368.8813,-43.3579],[369.185,-44.666],[369.7912,-46.8666],[370.1037,-47.5943],[370.4087,-48.1574],[370.965,-48.4693],[376.1463,-48.4867],[376.6988,-48.2874],[376.9625,-48.3047],[377.2412,-48.1228],[377.5238,-47.7762],[377.7987,-47.2477],[378.1112,-46.7539],[378.4112,-46.0435],[378.7062,-44.9519],[379.9737,-41.8243],[380.2613,-41.2959],[380.5512,-40.932],[386.55,-40.9147],[387.09,-41.062],[387.3613,-41.0274],[387.64,-41.1573],[388.2275,-41.7119],[388.5287,-42.1885],[389.1187,-42.4485],[392.3425,-42.2753],[398.2537,-42.258],[399.0962,-41.9809],[399.3963,-42.0244],[399.6888,-41.8685],[400.315,-42.0939],[403.1675,-42.0592],[406.0913,-41.8773],[407.2387,-41.8946],[407.85,-41.4961],[408.1313,-41.5481],[408.4175,-41.2015],[408.7113,-40.6903],[409.97,-37.2075],[410.2675,-36.679],[410.5575,-35.934],[410.8537,-35.3882],[411.1488,-34.4785],[411.4438,-33.3869],[411.7563,-32.1133],[412.065,-31.0217],[412.3613,-30.4586],[412.66,-30.2767],[416.02,-30.4759],[421.9513,-30.2767],[422.22,-29.9301],[422.515,-29.4189],[422.8175,-28.5093],[423.1087,-28.05],[423.4013,-27.7208],[423.6925,-27.5562],[426.8687,-28.1539],[427.1587,-28.7517],[427.46,-29.5314],[427.7538,-30.6922],[428.06,-31.8358],[428.355,-33.7071],[428.665,-36.2888],[428.9837,-39.8148],[429.28,-44.0512],[429.5812,-47.3953],[429.88,-49.05],[430.1762,-49.8124],[432.0462,-49.8297],[432.8925,-50.5314],[433.7787,-52.1597],[434.3963,-53.8227],[434.705,-54.2124],[435,-54.7321],[435.2875,-54.8187],[435.5825,-55.0526],[446.945,-54.9053],[447.2188,-54.628],[447.5275,-54.5846],[447.8312,-54.29],[448.1313,-54.16],[450.4975,-54.0127],[451.3212,-53.5015],[451.8837,-53.5534],[452.1587,-53.3715],[455.0688,-53.4234],[455.6375,-53.1635],[455.9162,-53.2327],[456.2025,-53.0854],[460.4175,-53.172],[462.9987,-52.9727],[465.1037,-53.0766],[465.3787,-53.1632],[465.665,-53.4144],[466.245,-53.5529],[466.5438,-53.8387],[467.7287,-54.3928],[468.3275,-54.1848],[468.6063,-54.2367],[468.9025,-54.0721],[469.4837,-54.1414],[469.7887,-53.7948],[470.1062,-53.6995],[470.44,-53.4741],[470.7638,-52.9022],[471.4112,-52.9368],[472.0612,-51.4726],[472.3763,-51.2299],[472.6925,-50.7014],[473.0025,-50.3895],[473.6187,-49.3845],[474.2587,-48.0675],[474.5938,-47.1925],[474.8837,-46.1355],[475.1963,-45.2431],[475.8012,-44.2035],[476.7025,-42.0376],[476.9987,-40.9633],[477.31,-39.1439],[477.6137,-36.7788],[477.9025,-33.452],[478.2013,-29.5622],[478.5075,-25.4903],[478.8175,-22.4755],[479.135,-20.9681],[479.44,-19.8245],[479.7362,-19.2614],[480.0275,-19.0795],[480.3175,-18.5164],[480.6175,-17.5894],[480.9062,-16.4805],[481.2038,-15.77],[481.5075,-15.5708],[482.3075,-15.77],[482.8463,-16.1512],[483.1287,-16.7143],[483.43,-17.2601],[483.9987,-17.9706],[484.29,-18.1698],[491.5675,-18.1698],[492.1187,-17.9879],[492.385,-17.4248],[492.67,-17.0782],[492.9613,-16.9309],[493.2437,-16.5843],[493.5287,-16.2724],[493.8425,-15.7439],[494.1338,-14.8689],[494.4325,-14.1758],[494.7275,-13.6473],[495.03,-13.4827],[506.4462,-13.2834],[506.9887,-13.3007],[513.9062,-13.3354],[515.555,-13.4912],[515.8363,-13.3266],[516.1275,-13.3958],[516.4262,-13.4997],[517.0275,-13.2917],[517.3112,-13.2656],[517.6187,-13.1876],[517.915,-12.9276],[518.5287,-12.4596],[518.825,-12.0696],[519.1187,-11.8789],[519.7287,-11.047],[520.0462,-10.9516],[520.345,-10.657],[520.6337,-10.527],[520.9375,-10.1978],[521.2375,-9.9032],[521.5375,-9.6432],[521.8537,-9.1667],[522.1587,-8.5255],[522.465,-7.919],[522.775,-7.4078],[523.0775,-7.4424],[531.0587,-7.6243],[531.3238,-8.2048],[531.6125,-8.7506],[531.9013,-9.6429],[532.2,-11.1157],[532.5187,-13.0824],[533.1175,-17.5442],[533.435,-19.5281],[533.74,-21.1655],[534.0612,-22.7856],[534.425,-24.423],[534.7663,-25.61],[535.09,-26.485],[540.6988,-26.5197],[540.9688,-26.3377],[541.5263,-26.3724],[544.1012,-26.1731],[545.3925,-26.0085],[545.9538,-25.7485],[546.2475,-25.3846],[546.8188,-25.09],[547.1,-24.7608],[547.4013,-24.4835],[547.695,-24.0243],[547.9987,-23.2531],[548.31,-22.6293],[548.615,-21.9708],[548.9125,-21.8235],[552.825,-21.8235],[556.185,-21.6242],[556.4538,-21.6762],[556.7325,-21.3643],[557.035,-21.4508],[557.6187,-21.2255],[557.9075,-21.0609],[561.6087,-21.2774],[572.545,-21.2948],[579.1425,-21.6152],[579.42,-21.6672],[579.6975,-21.9183],[580.27,-22.0915],[580.83,-22.23],[581.7038,-22.4897],[582.5638,-22.1517],[582.8487,-21.8744],[583.1575,-21.4152],[583.4488,-20.8087],[583.7663,-20.0029],[584.075,-18.8332],[584.3637,-17.369],[584.6688,-15.307],[584.9937,-13.6261],[585.5863,-8.7571],[585.88,-6.4612],[586.1787,-5.0923],[586.4663,-4.4165],[586.75,-3.888],[587.0425,-3.9053],[589.1175,-4.1046],[590.18,-4.2865],[590.9625,-4.4684],[591.245,-4.6677],[592.8062,-5.0662],[593.0962,-5.6639],[593.3737,-5.6812],[593.6513,-6.3136],[593.9462,-6.7814],[594.5438,-7.7861],[594.8463,-7.8727],[595.1275,-8.1759],[595.4125,-8.2451],[595.9837,-8.4356],[596.5512,-8.678],[596.84,-8.5999],[597.1363,-8.5392],[597.4362,-8.6258],[597.7437,-8.6777],[598.055,-8.695],[598.3475,-8.8943],[600.0425,-9.0935],[600.33,-9.2755],[600.94,-10.7656],[601.2287,-11.9091],[601.535,-13.0354],[601.8363,-14.4908],[602.4162,-16.3275],[602.7162,-16.8733],[602.9987,-17.2372],[603.5587,-17.5664],[606.1775,-17.5664],[606.44,-17.3845],[606.7125,-17.4364],[607.005,-17.1245],[607.2937,-16.9945],[607.575,-17.0464],[607.865,-17.0638],[608.1375,-16.8645],[609.1825,-16.6826],[609.7237,-16.518],[610.2613,-16.3187],[610.5375,-16.336],[625.2513,-16.1541],[625.5312,-15.8075],[626.3825,-15.3483],[626.6575,-15.3656],[626.9225,-15.1837],[629.2725,-15.2183],[629.5525,-15.0883],[630.1912,-14.8803],[630.5125,-14.5337],[630.825,-14.5857],[635.8625,-14.3864],[644.855,-14.5857],[645.1162,-14.7676],[645.395,-15.1315],[645.68,-15.6773],[645.9688,-16.0412],[646.2525,-16.9508],[646.5725,-17.6439],[646.8675,-19.0994],[649,-19.2539],[650,-19.6959],[655,-19.8778],[657,-19.8951],[661,-19.7736],[662,-19.2449],[663,-17.9884],[664,-16.0215],[665,-14.7565],[666,-14.5919],[670,-14.7738],[676,-14.5746],[684,-14.228],[685,-13.6388],[686,-13.6734],[689,-13.4915],[692,-13.3961],[693,-13.0669],[694,-11.6892],[695,-6.0922],[696,2.5283],[697,5.0407],[698,4.8588],[699,4.5469],[700,4.3476],[701,4.183],[702,3.3426],[703,3.0134],[704,3.0307],[705,2.7014],[706,2.1903],[707,1.3498],[708,1.1679],[709,0.856],[710,-0.0017],[716,-3.0472],[724,-3.2465],[727,0.9502],[735,1.1668],[748,-0.0971],[753,-0.0971]]
       }
       dyn_wings_metacarpal_04_l_ctrl.Rotation.Z {type=MovieSceneScriptingFloatChannel, default=0, const [0,15]}
       dyn_wings_metacarpal_04_l_ctrl.Scale.X {type=MovieSceneScriptingFloatChannel, default=1, const [1,15]}
       dyn_wings_metacarpal_04_l_ctrl.Scale.Y {type=MovieSceneScriptingFloatChannel, default=1, const [1,15]}
       dyn_wings_metacarpal_04_l_ctrl.Scale.Z {type=MovieSceneScriptingFloatChannel, default=1, const [1,15]}
       dyn_wings_phalange_04_l_ctrl.Location.X {type=MovieSceneScriptingFloatChannel, default=0, const [0,10]}
       dyn_wings_phalange_04_l_ctrl.Location.Y {type=MovieSceneScriptingFloatChannel, default=0, const [0,10]}
       dyn_wings_phalange_04_l_ctrl.Location.Z {type=MovieSceneScriptingFloatChannel, default=0, const [0,10]}
       dyn_wings_phalange_04_l_ctrl.Rotation.X {type=MovieSceneScriptingFloatChannel, default=0, thinned_from=10, keys [[-121,0],[0,0],[5.9862,-0.0444],[753,-0.0444]]}
       dyn_wings_phalange_04_l_ctrl.Rotation.Y {
        type=MovieSceneScriptingFloatChannel
        default=18.1973
        thinned_from=490
        keys [[-121,-0.0721],[1,-0.0721],[5,0.37],[5.9862,0.9828],[6,0.8735],[7,1.1252],[8,1.5673],[9,3.1453],[10,3.9005],[11,4.0909],[18,4.0295],[23,4.2812],[24,11.7414],[25,17.9367],[26,20.0857],[34,20.2761],[35,21.2278],[36,21.6085],[37,21.7988],[40,21.4181],[41,21.3567],[42,21.5155],[46,21.1664],[48,17.5744],[49,14.1378],[51,14.4859],[59,13.35],[60,-1.4414],[61,-11.6278],[62,-17.3257],[63,-17.5161],[65,-19.2291],[66,-21.7035],[68,-21.5132],[69,-21.3229],[71,-21.1325],[77,-17.1968],[78,-13.7707],[79,-13.39],[80,-10.3446],[81,-2.1047],[82,2.8381],[83,7.1483],[84,7.3387],[85,7.4001],[86,7.2098],[89,7.4001],[90,11.649],[91,17.2241],[92,22.4861],[93,22.8668],[96,23.0571],[97,24.7702],[98,25.3412],[103,25.9122],[104,26.4832],[105,26.8639],[110,26.6736],[116,26.6122],[119,26.4218],[121,25.9183],[122,25.5377],[123,24.3403],[124,23.0141],[125,21.6264],[126,20.3001],[127,19.416],[136,19.1642],[137,14.289],[138,13.4663],[143,13.6566],[144,16.3152],[145,26.8086],[146,28.1963],[147,8.1123],[148,-0.4408],[150,-4.5546],[151,-11.0078],[161,-10.8175],[170,-10.2464],[171,-4.8494],[172,-1.1715],[179,5.4904],[180,8.4683],[181,10.298],[182,18.0099],[183,23.7018],[184,24.9052],[185,27.1771],[186,30.7813],[187,32.9917],[204,32.4207],[205,31.9786],[206,31.7269],[207,23.886],[208,8.0631],[209,-1.8776],[213,-2.0679],[225,-0.6742],[226,14.9215],[227,24.3526],[228,32.1259],[229,36.4854],[230,36.7371],[232,39.6536],[233,41.7412],[237,36.7985],[238,26.8517],[239,14.369],[240,11.3911],[241,7.394],[242,-1.9327],[243,-5.1132],[244,-5.6842],[247,-5.0518],[248,-4.6711],[253,-4.2904],[254,-0.8704],[255,1.7882],[256,2.9917],[258,3.3723],[259,6.2213],[260,10.022],[261,0.4308],[262,-2.6086],[263,-7.803],[264,-13.501],[265,-15.9079],[266,-16.1596],[272,12.4289],[274,14.5226],[275,16.6778],[276,17.8198],[278,18.0102],[281,18.2005],[282,20.6074],[283,22.9529],[285,23.3335],[286,24.537],[287,24.7273],[291,24.2852],[292,23.4011],[293,19.9811],[294,16.119],[295,14.2831],[296,13.3314],[297,12.699],[301,11.9376],[302,10.7956],[303,10.6053],[305,9.7825],[306,8.579],[307,6.4853],[308,3.1881],[309,-4.2904],[310,-6.4456],[311,-6.8262],[312,-9.7427],[313,-13.1689],[314,-13.5495],[316,-13.9302],[317,-14.3109],[324,-15.7047],[325,-18.8729],[326,-19.5054],[329,-20.2667],[330,-20.6474],[331,-21.0895],[332,-23.2446],[333,-25.8419],[334,-27.8681],[336,-29.7715],[337,-31.7363],[338,-33.83],[339,-35.2238],[340,-36.1755],[342,-41.0568],[343,-41.6278],[344,-42.6409],[345,-44.1637],[346,-46.7609],[348,-52.9071],[349,-56.2656],[350,-58.6111],[351,-60.6819],[352,-61.1469],[354,-61.3373],[357,-60.7663],[358,-59.4339],[359,-57.9787],[360,-56.0139],[361,-54.4912],[362,-51.5133],[363,-46.8837],[365,-41.37],[366,-39.0245],[367,-36.2983],[368,-32.9397],[369,-30.7846],[370,-28.7584],[371,-26.1611],[372,-22.9929],[373,-19.1922],[374,-16.595],[375,-14.7591],[376,-11.7812],[377,-9.6261],[378,-7.2806],[379,-4.8737],[380,-0.6248],[381,2.3531],[382,5.4599],[383,8.1861],[384,13.3253],[385,18.3969],[386,21.6941],[387,24.3589],[388,26.7044],[389,30.1305],[390,31.2725],[395,31.2111],[402,31.0208],[404,30.8304],[406,28.3683],[407,25.9675],[408,25.5254],[409,25.3351],[410,24.4509],[411,24.2606],[412,24.1992],[413,23.7571],[414,22.744],[418,21.6695],[419,19.0783],[420,17.8135],[421,17.5617],[422,15.4741],[423,13.6383],[424,11.4217],[425,8.8245],[426,4.4528],[427,4.2624],[434,4.0721],[435,3.6914],[436,3.8818],[437,4.4528],[438,6.1658],[439,8.5113],[440,10.0955],[441,12.0603],[442,13.8347],[445,14.9154],[446,17.5801],[447,19.8642],[448,21.0062],[449,21.1352],[458,20.7545],[460,20.5641],[461,18.4765],[462,13.4724],[463,10.6234],[464,10.2428],[465,10.1813],[466,9.8007],[467,9.6103],[479,9.8007],[480,10.3717],[481,12.2137],[482,16.3336],[483,21.0921],[484,29.5285],[485,31.1126],[487,30.9223],[488,27.6312],[489,20.0421],[490,15.6151],[491,12.8951],[492,7.0743],[493,4.606],[502,-15.9693],[503,-15.9079],[509,-15.7175],[511,-14.5755],[514,-13.9431],[515,-13.7527],[516,-12.7396],[517,-10.2713],[518,-3.3699],[519,-1.534],[520,-1.3437],[522,-1.1534],[527,-0.5209],[528,0.6825],[529,1.6956],[530,4.0411],[531,8.4128],[532,11.2004],[533,12.2749],[534,12.4652],[535,13.288],[536,13.4783],[538,13.288],[551,13.0362],[554,8.6707],[555,0],[560.5438,-0.988],[561.03,-2.5401],[561.4837,-4.1474],[561.9425,-5.8096],[562.4125,-6.6344],[562.8663,-8.7006],[563.3175,-10.5633],[563.77,-11.8455],[564.2675,-12.4693],[565.855,-12.8118],[566.2725,-14.5027],[566.72,-16.634],[567.1888,-17.8079],[580.695,-16.7763],[581.1538,-15.4974],[581.6075,-14.1121],[582.08,-13.454],[582.5462,-13.089],[592.7,-12.5422],[593.525,-11.5136],[593.9688,-10.837],[594.4288,-10.1598],[595.34,-7.4411],[595.7962,-4.1558],[596.285,-2.0027],[596.7362,-0.843],[597.5912,-0.4845],[605.445,0.1842],[605.885,1.6832],[606.3537,1.9224],[609.065,2.31],[611.0125,2.5466],[611.4425,3.9938],[611.92,5.9801],[612.405,8.6154],[612.895,11.7604],[613.3813,14.7705],[613.8637,16.57],[614.3438,17.3761],[630.3913,16.6606],[630.8225,16.1829],[631.2738,15.4525],[631.7362,14.332],[632.2075,13.8743],[639.0063,13.1695],[639.8338,12.7051],[640.67,12.3524],[641.1087,11.9169],[641.9525,11.5778],[642.3913,11.1378],[642.8475,10.8901],[649.64,10.555],[650.0688,9.4509],[650.5275,8.4809],[650.995,5.653],[651.5038,0.0903],[651.9713,-7.4065],[652.47,-11.9579],[652.9425,-13.5939],[655.08,-15.7251],[655.5912,-20.3499],[656.135,-26.8609],[656.6363,-28.9431],[657.09,-29.5595],[667.9975,-27.8584],[668.43,-25.3261],[668.8975,-23.1935],[669.3388,-21.2333],[669.8375,-20.4201],[670.2837,-18.8903],[670.7475,-17.8683],[671.1838,-17.3599],[671.6388,-17.1908],[672.5587,-16.2679],[673.005,-15.3708],[673.4663,-14.0641],[673.925,-12.0121],[674.4038,-9.0126],[674.8837,-6.3379],[675.3813,-5.3788],[677.3863,-4.9216],[677.8288,-4.332],[678.2862,-2.8951],[678.76,-1.2685],[679.22,0.0249],[679.7013,1.4621],[680.175,3.7146],[680.665,6.7341],[681.165,9.6441],[681.6475,11.9234],[682.1475,13.4558],[682.62,14.0061],[683.0925,14.3605],[684.345,14.6094],[689.235,14.6616],[689.6938,14.7508],[690.1712,15.1015],[690.62,15.1874],[691.0775,15.5331],[691.535,16.3737],[692.4963,19.7403],[692.9887,20.8611],[693.4675,21.59],[693.9513,22.1641],[694.41,22.6554],[694.8788,22.6914],[695.3338,22.9515],[695.7862,22.9858],[697.17,23.8487],[697.6388,24.4031],[698.1287,25.1575],[698.6012,26.1638],[699.0975,27.1251],[700.0875,28.837],[700.5738,29.0296],[705.5525,28.2884],[706.4362,27.7849],[706.9137,27.2963],[707.39,27.0148],[707.8775,26.9755],[708.74,26.7298],[710.0175,26.6363],[710.4787,26.1715],[710.9362,26.1162],[711.3963,25.8723],[711.8637,25.8297],[712.3175,25.5509],[712.8,24.4715],[713.2775,22.195],[713.7825,18.5238],[714.3088,10.9658],[714.8275,3.4747],[715.3112,-1.5759],[715.8075,-4.6294],[716.3475,-6.9369],[716.8325,-8.0857],[717.3,-8.6683],[723.375,-7.7287],[723.8112,-7.1185],[724.2688,-6.7929],[724.7225,-6.2465],[725.175,-5.9371],[725.6337,-5.4118],[726.105,-4.6418],[726.575,-2.9452],[727.055,-1.5586],[727.5263,-0.8414],[727.985,-0.4321],[728.44,-0.23],[728.9013,0.0488],[729.3587,0.5932],[729.8288,1.3845],[730.2937,2.7417],[730.7613,3.7708],[731.2287,4.7447],[731.7162,6.4343],[732.2075,7.8838],[732.7025,9.9313],[733.1963,12.3412],[733.6913,13.9405],[734.1888,14.7561],[734.6862,14.9938],[735.14,15.0207],[737.1075,14.5685],[737.9375,14.5435],[746,0],[753,0]]
       }
       dyn_wings_phalange_04_l_ctrl.Rotation.Z {type=MovieSceneScriptingFloatChannel, default=0, thinned_from=10, keys [[-121,0],[0,0],[5.9862,0.0184],[753,0.0184]]}
       dyn_wings_phalange_04_l_ctrl.Scale.X {type=MovieSceneScriptingFloatChannel, default=1, const [1,10]}
       dyn_wings_phalange_04_l_ctrl.Scale.Y {type=MovieSceneScriptingFloatChannel, default=1, const [1,10]}
       dyn_wings_phalange_04_l_ctrl.Scale.Z {type=MovieSceneScriptingFloatChannel, default=1, const [1,10]}
       dyn_wings_humelus_r_ctrl.Location.X {type=MovieSceneScriptingFloatChannel, default=0, const [0,2]}
       dyn_wings_humelus_r_ctrl.Location.Y {type=MovieSceneScriptingFloatChannel, default=0, const [0,2]}
       dyn_wings_humelus_r_ctrl.Location.Z {type=MovieSceneScriptingFloatChannel, default=0, const [0,2]}
       dyn_wings_humelus_r_ctrl.Rotation.X {type=MovieSceneScriptingFloatChannel, default=0, const [0,2]}
       dyn_wings_humelus_r_ctrl.Rotation.Y {
        type=MovieSceneScriptingFloatChannel
        default=0
        thinned_from=908
        keys [[-1,0],[1,0],[5.6413,-0.0292],[6.32,-0.618],[6.6337,-1.0574],[6.9613,-1.495],[7.2788,-1.9307],[7.61,-2.0754],[7.92,-2.22],[8.5512,-2.3644],[11.6137,-2.5085],[11.9237,-3.1512],[12.2937,-4.4409],[12.6562,-5.4073],[12.9875,-5.7748],[26.9137,-5.732],[27.2188,-5.4598],[27.5575,-4.8718],[27.9075,-4.1453],[28.2437,-3.3798],[28.5912,-2.6164],[28.925,-2.0271],[29.255,-1.6841],[29.5863,-1.3748],[29.9188,-1.205],[30.2362,-1.0663],[30.5487,-0.9273],[31.165,-0.7585],[36.4775,-0.6195],[36.7787,-0.341],[37.0913,-0.1734],[37.4075,-0.034],[38.0013,0.1055],[38.3162,0.2452],[38.6363,0.5514],[38.9638,0.9715],[39.3162,2.0991],[39.6913,3.1134],[40.0512,4.1294],[40.4125,4.5585],[43.5587,4.717],[44.2275,5.4334],[45.7337,5.4467],[49.0312,6.164],[49.5175,6.6059],[57.135,7.0373],[57.4563,7.1813],[59.2188,7.3254],[60.12,7.6139],[60.7675,7.9027],[61.7225,8.0473],[62.0437,8.1978],[62.3725,8.4865],[62.7075,8.9241],[63.0537,9.7925],[63.4225,10.6584],[63.7812,11.6656],[64.125,12.3867],[64.485,12.9572],[64.9325,13.0939],[65.2812,13.3812],[65.65,13.5249],[66.0125,13.8032],[71.6637,13.9465],[71.9725,14.2229],[72.3325,14.7948],[72.6612,15.0807],[73.3225,15.3534],[73.9525,15.496],[74.5925,15.9236],[74.9325,16.2086],[75.2713,16.6359],[75.61,17.0279],[75.95,17.311],[76.2937,17.4525],[76.9225,17.5939],[78.3963,17.7161],[78.715,17.8571],[79.39,18.4208],[81.9813,18.6807],[82.3175,18.9613],[82.6588,19.3819],[82.9913,19.4982],[87.6225,20.0315],[87.9575,20.561],[88.2775,21.2525],[88.6413,22.4303],[89.0387,23.4874],[89.3763,24.533],[89.73,25.3001],[90.075,25.6617],[91.6588,26.0206],[103.1375,25.9821],[103.4462,25.8502],[104.0487,25.6805],[108.71,25.5487],[110.1112,25.2116],[110.4288,24.6147],[110.76,24.3519],[111.0788,23.9235],[111.73,23.6278],[113.9725,23.4963],[118.1087,23.333],[118.4125,22.8071],[119.0713,21.987],[119.7162,21.6954],[126.065,21.4048],[126.3825,20.6957],[126.7413,20.1463],[127.0662,19.7296],[127.7038,19.4453],[131.105,20.1225],[131.4737,21.482],[131.8813,22.4253],[132.29,23.2386],[132.995,23.7594],[133.6212,23.8893],[144.5163,23.4992],[144.8375,22.7866],[145.1687,21.791],[145.5125,20.9857],[145.8625,20.4173],[146.19,20.264],[150.6625,20.1114],[166.9887,19.9823],[167.29,19.1649],[167.9675,18.0964],[168.6288,17.3249],[169.235,17.18],[169.5512,17.0516],[170.21,16.5377],[170.5387,15.8817],[170.8663,14.3228],[171.54,11.258],[171.865,9.864],[172.2088,8.9843],[172.535,8.4821],[172.8675,8.2313],[173.1912,8.1007],[173.7925,7.9752],[180.3425,7.8497],[180.6425,7.6055],[180.9613,7.4805],[181.2763,7.2307],[181.5975,7.106],[182.7487,6.9891],[183.0475,6.8648],[183.3687,6.4922],[183.6862,6.2442],[184.0063,5.8728],[184.6625,5.3787],[184.9787,5.2554],[185.3062,4.886],[185.62,4.6402],[186.2637,4.3948],[198.1313,4.4085],[199.0863,4.0419],[199.7162,3.828],[200.0375,3.4801],[200.355,3.2388],[200.6862,2.8945],[201.02,2.535],[201.34,2.3141],[201.66,1.9758],[201.9962,1.8573],[202.305,1.7388],[205.86,1.6205],[211.3587,1.7581],[211.675,2.032],[211.9925,2.5394],[212.3238,2.7751],[212.6363,3.0113],[212.9538,3.2637],[213.275,3.3819],[214.155,3.3974],[214.45,3.5155],[216.1662,3.7518],[216.4812,3.8701],[216.7975,4.1069],[217.1062,4.1209],[219.84,4.3576],[220.465,5.0698],[220.7962,5.557],[221.43,5.7952],[229.1812,6.0433],[229.5412,6.5201],[229.9425,6.6395],[230.3062,6.8785],[230.665,6.998],[231.2788,7.0056],[231.8763,7.1174],[232.1762,7.3564],[235.5163,7.4757],[235.8175,7.834],[236.135,8.1928],[236.47,8.4323],[237.1062,8.672],[239.625,8.7919],[245.0863,8.9119],[246.485,9.0319],[246.8012,9.3945],[247.7838,10.8366],[248.4375,11.3184],[249.035,11.4389],[249.6225,11.4365],[252.0888,11.5592],[252.3925,11.9202],[252.7125,12.161],[253.0288,12.5175],[253.6775,12.9985],[254.28,13.1188],[256.7513,13.3527],[257.3763,13.8329],[260.4462,13.953],[260.75,14.3132],[261.0688,14.9137],[261.4362,15.2739],[262.1737,15.7541],[266.1562,15.3717],[266.4825,14.6354],[266.8075,14.024],[267.1425,13.6613],[267.8188,13.4249],[272.9925,13.1761],[273.6225,12.7004],[273.9288,12.5782],[274.24,12.3441],[274.5675,11.9859],[274.8913,11.5201],[275.215,10.9372],[275.55,10.2417],[275.8687,9.78],[276.2013,9.3231],[276.5688,8.8644],[277.8438,8.4064],[278.1637,8.1777],[278.4775,7.7209],[278.8112,7.0374],[279.1575,6.3645],[279.8,5.4615],[280.1388,5.3489],[283.6888,5.2364],[283.9962,4.7989],[284.3462,4.3644],[284.6662,4.0305],[284.9862,3.822],[285.3138,3.7113],[288.09,3.6007],[288.3837,3.2692],[289.3288,2.9385],[289.6662,2.8285],[291.6488,2.6086],[291.9513,2.2795],[292.29,1.9512],[292.6163,1.6237],[292.9362,1.2972],[293.2625,1.0799],[293.5825,0.9714],[293.9462,0.8631],[296.4062,0.8826],[296.7062,0.5587],[297.0225,0.2564],[297.3663,-0.0651],[297.7088,-0.2789],[298.0425,-0.3638],[301.3775,-0.4702],[304.3875,-0.4482],[304.6975,-0.4264],[308.3288,-0.4046],[308.6337,-0.2988],[313.0587,-0.4666],[313.3675,-0.6548],[313.685,-0.9689],[314.0163,-1.1548],[314.3575,-1.3629],[314.7038,-1.6741],[315.06,-2.0877],[315.38,-2.4994],[315.7038,-3.2156],[316.0475,-3.9261],[316.3763,-4.5304],[316.7125,-4.9309],[317.0337,-5.1004],[317.675,-5.4973],[318.3012,-5.5962],[324.3363,-5.695],[324.655,-6.1237],[324.9887,-6.5137],[325.6562,-7.2543],[325.9775,-7.4462],[326.3275,-7.6376],[326.6975,-7.7943],[328.1787,-8.1739],[328.495,-8.3629],[328.8288,-8.7393],[329.1525,-9.1706],[329.4863,-9.5417],[329.815,-9.8186],[330.1363,-9.9652],[330.4575,-10.1484],[330.7812,-10.2397],[331.0975,-10.422],[331.4187,-10.513],[332.5612,-10.6038],[332.8637,-10.8755],[333.1875,-11.2358],[333.5138,-11.6832],[333.8425,-12.2155],[334.1775,-12.4798],[334.51,-12.8303],[334.8388,-13.0917],[335.155,-13.137],[335.4712,-13.3101],[335.7912,-13.3964],[336.1025,-13.6546],[336.4162,-13.7404],[336.7425,-13.9969],[337.0612,-14.2941],[337.3913,-14.7995],[337.8087,-15.256],[338.16,-15.5422],[338.4863,-15.7065],[351.1488,-15.6622],[356.3325,-15.4982],[356.6375,-15.2511],[356.9613,-14.8365],[357.3,-14.5024],[357.6413,-14.2504],[358.64,-13.9972],[360.595,-13.6159],[360.9413,-13.189],[361.285,-12.5857],[361.625,-11.7123],[361.955,-10.7363],[362.2925,-9.9258],[362.6225,-9.1965],[362.9588,-8.6442],[363.3125,-8.3319],[363.6375,-8.1123],[363.9562,-7.8334],[364.2837,-7.366],[364.6225,-6.6747],[364.9625,-5.4361],[365.3288,-3.7866],[365.6637,-1.6047],[365.9962,0.524],[366.3162,2.2768],[366.6675,3.3191],[367.0013,4.1707],[367.3388,4.813],[367.6888,5.0445],[368.0225,5.3694],[368.3637,5.5794],[368.9975,6.0086],[369.65,6.6401],[369.9713,7.1741],[370.315,7.5963],[370.6375,8.0239],[370.9625,8.2353],[374.3062,8.341],[374.605,8.6618],[374.9212,9.193],[375.2812,9.6174],[375.6075,9.9351],[375.925,10.3573],[376.2563,10.5677],[376.8937,11.4064],[377.225,11.9318],[377.5512,12.7685],[377.8863,13.9244],[378.2212,15.1672],[378.5487,15.9938],[378.8737,16.6082],[379.2088,16.8166],[384.4875,16.7001],[386.1112,15.1394],[387.4013,14.7229],[391.06,14.5147],[391.3863,14.1949],[391.7738,13.7792],[392.1838,13.2542],[392.5625,12.6237],[392.9562,12.1044],[393.31,11.8986],[393.635,11.7929],[396.7125,11.6852],[398.1413,11.5833],[398.4462,11.2778],[398.7775,10.5648],[399.1213,10.0593],[399.4762,9.4558],[399.8175,9.0561],[400.48,8.4551],[401.1075,8.3551],[401.69,8.3588],[402.2837,8.1555],[402.5925,7.956],[402.9225,7.5576],[403.58,6.5717],[403.9162,5.496],[404.27,4.3222],[404.6,2.9638],[404.9362,1.9055],[405.2837,1.0461],[405.6225,0.5713],[405.9662,0.1928],[406.3,-0.0902],[406.625,-0.2785],[406.9425,-0.5602],[407.265,-0.7477],[407.5912,-1.2148],[407.9312,-1.6083],[408.2663,-2.3497],[408.5963,-4.1124],[408.9325,-6.1692],[409.26,-8.139],[409.6012,-9.9626],[409.9525,-11.2901],[410.2837,-12.2452],[410.6425,-12.8995],[410.9688,-13.2666],[411.2988,-13.3483],[411.605,-13.4695],[411.9225,-13.6326],[412.2587,-13.9165],[412.9275,-15.002],[413.2637,-15.9126],[413.61,-16.5784],[413.94,-17.0076],[414.5963,-17.4346],[414.925,-17.5108],[417.6237,-17.6453],[420.185,-15.8912],[420.7675,-14.4467],[421.4375,-11.1749],[421.79,-10.6179],[422.1075,-10.5308],[424.3913,-10.4436],[424.8012,-10.0579],[425.18,-9.3184],[425.5738,-8.305],[425.945,-7.2802],[426.6575,-5.3845],[426.9887,-4.2106],[427.3075,-2.8399],[427.6575,-1.4809],[428.0225,-0.4761],[428.365,0.1427],[428.6913,0.822],[429.0662,1.3264],[429.3938,1.831],[429.7437,2.3505],[430.0925,2.7563],[430.4212,3.2599],[430.7738,3.6526],[431.1025,4.2551],[431.4262,4.5617],[431.7638,4.8677],[432.08,5.0649],[432.3875,5.1732],[433.025,5.3703],[433.6762,5.7822],[434.0025,6.1843],[434.3238,6.7762],[435.0025,7.7712],[435.33,8.1717],[435.6612,8.2708],[439.2825,8.3699],[443.1325,8.3735],[445.365,8.4688],[445.9588,8.5678],[446.6125,8.7688],[448.3525,8.8704],[448.9488,8.9691],[454.9975,9.0679],[455.5925,9.3663],[455.9137,9.6625],[456.2587,9.9598],[456.6087,10.4534],[456.9987,10.8473],[457.4162,11.0445],[458.1375,11.1431],[461.1825,11.0445],[461.4937,10.9459],[462.1338,10.3548],[463.5875,10.2569],[464.7138,9.9641],[465.0275,9.4779],[465.3737,9.0912],[466.0975,8.3271],[466.425,8.1347],[466.7487,8.1426],[467.0838,8.0389],[470.1413,8.043],[470.4438,7.9433],[471.6012,7.8522],[471.9112,7.6663],[472.2375,7.5761],[472.5538,7.3916],[472.875,7.1134],[473.2,6.7426],[473.535,6.1936],[473.8937,5.5571],[474.2362,4.9179],[474.5963,4.3752],[474.9112,4.0999],[475.2788,3.9166],[477.515,3.7335],[477.8425,3.4593],[478.1637,3.0945],[478.4937,2.9125],[478.8238,2.653],[479.16,2.5624],[479.7787,2.3814],[481.7425,2.2006],[482.0387,2.0201],[482.365,1.7498],[482.685,1.48],[483.01,1.2108],[483.3438,0.9423],[483.67,0.6743],[483.98,0.407],[484.3138,0.2291],[484.64,-0.0372],[484.9538,-0.3209],[485.3,-0.6746],[485.615,-1.2908],[485.9688,-1.9033],[486.2962,-2.5986],[486.6237,-3.2027],[486.9725,-3.5461],[487.305,-3.8882],[487.63,-4.0588],[488.905,-4.3988],[489.6437,-4.7375],[490.0288,-5.1588],[490.44,-5.4364],[491.1425,-5.604],[494.1963,-5.8548],[494.81,-6.0215],[501.4388,-6.298],[502.7387,-7.2888],[503.0537,-7.4526],[503.3725,-7.645],[504.6463,-7.9712],[509.1075,-7.8083],[509.4275,-7.5632],[509.7638,-7.1246],[510.0975,-6.6022],[510.4275,-6.1048],[510.78,-5.745],[511.1025,-5.4943],[511.4187,-5.1335],[511.7538,-4.8813],[512.0775,-4.4592],[512.4013,-4.0963],[512.7337,-3.7562],[513.0638,-3.3067],[513.38,-3.0278],[513.705,-2.7707],[514.3325,-2.5989],[515.4887,-2.5128],[520.9413,-2.168],[521.2575,-1.8218],[521.62,-1.3679],[522.0013,-0.739],[522.4025,-0.0208],[522.8012,1.0528],[523.2225,1.9544],[523.5987,2.3997],[523.9425,2.4889],[524.2563,2.5782],[527.32,2.5913],[530.6187,2.6805],[530.925,2.9485],[531.6,3.8565],[531.9288,4.2162],[532.2487,4.3964],[532.5675,4.667],[532.885,4.7573],[533.1975,4.7668],[533.51,4.8475],[533.8175,4.9378],[534.9725,5.0281],[535.5763,5.1185],[535.8875,5.2089],[536.195,5.3982],[536.8288,5.579],[537.1388,5.76],[537.45,5.8506],[537.8062,6.0318],[543.5875,6.1295],[543.8963,6.4915],[544.225,6.9506],[544.5725,7.4041],[545.2337,7.9536],[545.5712,8.3207],[545.9025,8.4116],[546.2287,8.6843],[546.5625,9.0505],[546.8837,9.2324],[547.545,9.961],[547.86,10.1433],[548.5025,10.3258],[552.38,10.0526],[552.705,9.4175],[553.0362,8.2475],[553.3938,7.4433],[553.7362,6.7276],[554.1075,6.1175],[554.5025,5.682],[554.885,5.4252],[557.075,5.3371],[557.3763,5.2574],[557.7038,5.0027],[558.02,4.9151],[560.8,4.924],[561.1037,4.6531],[561.44,4.4882],[561.75,4.3141],[562.07,4.2272],[564.3088,4.0535],[564.6112,3.8039],[564.9375,3.3944],[565.28,2.8913],[565.6263,2.6472],[565.9588,2.562],[566.2738,2.3917],[566.605,2.3066],[573.435,2.1497],[573.7475,1.8955],[574.07,1.6418],[574.4013,1.5574],[574.7262,1.3886],[575.05,1.3044],[581.0375,1.2202],[581.3413,0.7999],[581.6637,0.4648],[581.9875,-0.036],[582.3587,-0.2856],[582.6788,-0.6174],[583.005,-0.9482],[583.3613,-1.1313],[584.005,-1.2963],[584.635,-1.3787],[590.825,-1.5433],[591.14,-1.6255],[591.4762,-1.7076],[592.0863,-1.7897],[593.2563,-1.9536],[593.5725,-2.0355],[593.8813,-2.2806],[594.2025,-2.5453],[594.5225,-2.7895],[594.855,-2.9519],[595.1775,-3.1951],[595.4925,-3.276],[595.8087,-3.4376],[596.1375,-3.5182],[596.4688,-3.6794],[596.795,-3.8623],[597.43,-4.023],[597.7675,-4.2863],[598.1087,-4.6065],[598.4375,-4.9254],[598.7575,-5.3465],[599.1012,-5.8208],[599.4325,-6.2392],[599.7563,-6.4744],[600.0925,-6.7349],[601.0537,-6.9689],[601.9262,-7.1245],[602.2312,-7.2798],[602.5512,-7.4618],[602.8788,-7.6166],[603.5138,-7.771],[604.6962,-7.8481],[605.3062,-8.002],[605.6187,-8.1556],[605.94,-8.3371],[606.5763,-8.6429],[606.8988,-8.8714],[607.5412,-9.0233],[607.865,-9.1283],[608.1737,-9.2799],[608.505,-9.3555],[608.8288,-9.5065],[609.1488,-9.5819],[609.4787,-9.7624],[609.8062,-9.9127],[610.44,-10.0627],[610.7575,-10.2123],[611.0638,-10.243],[611.7,-10.3924],[612.2988,-10.4669],[612.6063,-10.5414],[617.8587,-10.4669],[618.1625,-10.1681],[618.4812,-9.7926],[618.8075,-9.2631],[619.1463,-8.7582],[619.4737,-8.2966],[619.8338,-8.0647],[620.215,-7.9096],[627.2862,-7.8319],[627.6288,-7.4423],[627.9975,-6.9718],[628.3338,-6.4452],[628.665,-5.8077],[629.0138,-5.3261],[629.3487,-5.0275],[629.6862,-4.8654],[630.3162,-4.703],[632.0075,-4.5403],[632.3188,-4.3773],[632.6475,-4.1322],[632.9625,-3.9683],[633.285,-3.8863],[633.6075,-3.6398],[634.2188,-3.5574],[634.54,-3.3925],[634.8637,-3.31],[635.455,-3.1446],[636.095,-2.9789],[637.2337,-2.896],[637.8288,-2.7298],[638.14,-2.5634],[638.4762,-2.3132],[638.7875,-2.0624],[639.1225,-1.6428],[639.4488,-1.3059],[640.0813,-1.1371],[643.3938,-0.9679],[643.7038,-0.8832],[650.9237,-0.7985],[651.5263,-0.6287],[651.835,-0.4588],[652.1612,-0.2885],[652.5113,-0.118],[652.88,0.3095],[653.2513,1.0829],[653.6612,1.7747],[654.0525,2.296],[654.4162,2.4702],[661.8775,2.5574],[662.1875,2.8193],[662.8275,3.1693],[664.055,3.3447],[664.6662,3.5202],[664.9625,3.696],[665.29,3.96],[665.625,4.1362],[666.2513,4.4893],[666.885,4.6662],[667.4913,4.7547],[668.6325,4.8432],[669.565,5.1091],[670.2088,5.4643],[670.5312,5.8277],[670.8712,6.2727],[671.2063,6.6295],[671.525,6.8081],[671.8463,6.8974],[672.4563,6.9031],[674.6712,7.076],[674.9787,7.1653],[678.2862,7.16],[678.5975,7.2548],[679.2463,7.4194],[689,7.1494],[690,5.0288],[691,3.0538],[692,2.5466],[702,2.5416],[703,2.3295],[704,1.7004],[705,1.1675],[706,1.0774],[707,1.0646],[708,1.0131],[709,0.9744],[713,0.9486],[714,0.9228],[715,0.8708],[716,0.6883],[717,0.4892],[718,0.3256],[719,0.2729],[720,0.2597],[721,0.2186],[724,0.1922],[725,0.1642],[726,0.1079],[729,0.0947],[730,0],[752,0]]
       }
       dyn_wings_humelus_r_ctrl.Rotation.Z {type=MovieSceneScriptingFloatChannel, default=0, const [0,2]}
       dyn_wings_humelus_r_ctrl.Scale.X {type=MovieSceneScriptingFloatChannel, default=1, const [1,2]}
       dyn_wings_humelus_r_ctrl.Scale.Y {type=MovieSceneScriptingFloatChannel, default=1, const [1,2]}
       dyn_wings_humelus_r_ctrl.Scale.Z {type=MovieSceneScriptingFloatChannel, default=1, const [1,2]}
       dyn_wings_radius_r_ctrl.Location.X {type=MovieSceneScriptingFloatChannel, default=0, const [0,7]}
       dyn_wings_radius_r_ctrl.Location.Y {type=MovieSceneScriptingFloatChannel, default=0, const [0,7]}
       dyn_wings_radius_r_ctrl.Location.Z {type=MovieSceneScriptingFloatChannel, default=0, const [0,7]}
       dyn_wings_radius_r_ctrl.Rotation.X {type=MovieSceneScriptingFloatChannel, default=0, const [0,8]}
       dyn_wings_radius_r_ctrl.Rotation.Y {
        type=MovieSceneScriptingFloatChannel
        default=0
        thinned_from=1190
        keys [[0,0],[9.9462,-0.2],[18.0213,-0.4],[18.7487,-0.6],[19.1213,-1.2009],[19.5187,-1.6017],[19.9088,-2.4026],[20.3188,-3.0034],[20.7337,-4.806],[21.2138,-6.4086],[21.7,-7.8104],[22.18,-8.4112],[22.595,-9.0121],[22.9837,-9.6138],[23.3825,-10.6173],[23.785,-11.2199],[24.1775,-11.6207],[24.5775,-11.6216],[24.95,-11.8207],[30.3813,-12.0216],[30.7538,-12.4225],[31.1287,-12.6233],[31.5187,-13.0242],[32.2988,-13.4268],[32.6962,-14.0285],[33.8612,-14.6329],[35.26,-14.8338],[36.0263,-15.2372],[36.415,-15.239],[36.785,-15.4372],[40.6837,-15.4372],[54.1112,-15.6372],[54.5063,-16.6398],[54.92,-17.6415],[55.3238,-18.0433],[55.7075,-18.2433],[57.8438,-18.2433],[58.57,-18.4433],[59.64,-18.4441],[60.035,-18.6433],[60.4425,-18.6441],[61.1925,-19.0441],[61.5763,-19.445],[61.9588,-19.6459],[62.35,-20.0476],[62.745,-20.6493],[63.1463,-21.0511],[63.5425,-21.6528],[63.945,-22.0545],[64.7387,-23.258],[66.2937,-24.0615],[66.6737,-24.0623],[67.0875,-24.2623],[67.4987,-24.6632],[67.9175,-24.6641],[68.6525,-24.8632],[69.0275,-25.0641],[69.4137,-25.0649],[69.79,-25.2649],[70.1687,-25.2658],[70.57,-25.6667],[70.9762,-25.8684],[71.375,-26.471],[71.7725,-26.6727],[72.5362,-27.4753],[74.8463,-28.6823],[75.2212,-28.6832],[75.6012,-28.8831],[75.985,-29.284],[76.37,-29.2858],[77.52,-29.8884],[77.9088,-30.2901],[78.2962,-30.291],[78.6825,-30.691],[79.0763,-30.8918],[79.4725,-31.2936],[79.865,-31.2944],[80.95,-31.4944],[81.3225,-31.6944],[81.7013,-31.6962],[82.0938,-31.8944],[82.4775,-31.8953],[82.86,-32.0953],[83.235,-32.097],[83.6212,-32.2962],[84.005,-32.2979],[84.785,-32.6979],[85.2038,-32.6996],[86.095,-33.1022],[86.515,-33.504],[86.9125,-33.7074],[87.325,-34.31],[87.7162,-34.7126],[88.1125,-35.5144],[88.5138,-35.7161],[88.9025,-36.3187],[89.29,-36.3204],[89.6775,-36.7204],[90.4125,-36.9222],[91.1575,-37.3239],[91.9162,-37.3239],[93.3575,-37.7256],[93.7337,-37.9265],[94.125,-38.3274],[95.28,-38.93],[95.665,-39.5308],[96.4525,-40.3326],[97.6825,-42.1378],[98.0838,-42.5395],[98.4812,-43.1404],[98.8725,-43.3404],[99.255,-43.3412],[99.6375,-43.5404],[100.365,-43.5412],[100.7325,-43.7412],[101.1525,-44.1412],[101.5612,-44.3421],[102.2875,-44.5421],[104.4125,-44.5412],[104.77,-44.3421],[105.1413,-44.3412],[105.5225,-43.9395],[105.9212,-43.3334],[106.3275,-42.5273],[106.735,-42.3256],[107.1213,-42.3247],[107.5,-42.1247],[109.64,-42.1239],[110.01,-42.323],[110.4013,-42.3213],[110.7825,-42.5221],[114.1662,-42.5195],[114.9462,-42.1152],[115.335,-42.1117],[115.7262,-41.9134],[120.7613,-41.9126],[121.7925,-42.1126],[123.595,-42.1108],[125.3988,-41.91],[127.2763,-41.9091],[127.67,-41.71],[128.45,-41.7082],[128.8375,-41.5082],[129.2262,-41.5065],[129.6125,-41.3065],[130,-41.3047],[130.3925,-41.103],[130.7775,-41.1004],[131.1612,-40.8995],[131.5788,-40.8969],[132.0188,-40.6961],[132.4212,-40.6943],[133.5025,-40.4935],[134.275,-40.0856],[134.6812,-40.0813],[136.295,-39.2622],[136.685,-39.2596],[137.0738,-39.0604],[138.9575,-39.0587],[139.7175,-39.4552],[142.07,-39.45],[142.4675,-39.2491],[142.8725,-39.2413],[143.2788,-39.4387],[145.2825,-39.4352],[146.0325,-39.2352],[146.4175,-39.03],[147.25,-39.0196],[148.0738,-38.8205],[148.9013,-37.9962],[150.4087,-36.1536],[150.8875,-35.7432],[151.7162,-35.3301],[152.1287,-35.324],[152.9525,-34.9119],[153.3825,-34.3015],[155.0825,-32.658],[155.5,-32.6502],[156.3075,-32.245],[157.585,-31.0155],[158.015,-30.805],[158.4462,-30.3946],[158.8687,-30.1842],[160.135,-28.9407],[162.765,-25.2426],[163.1963,-24.8295],[163.6187,-24.0148],[164.0387,-23.3983],[164.4675,-22.58],[164.9025,-21.5566],[165.3363,-20.1331],[165.7525,-18.9088],[166.1862,-17.2827],[166.61,-16.0576],[167.0487,-14.4315],[167.4925,-13.8228],[167.9225,-13.1278],[168.3687,-12.5187],[169.23,-11.4741],[169.6688,-11.0385],[170.5425,-10.3404],[170.975,-9.9052],[171.4062,-9.3837],[171.8537,-8.7765],[172.2862,-8.0822],[172.72,-7.3],[173.625,-5.392],[174.0725,-4.4386],[174.505,-3.8299],[174.9362,-3.1371],[175.38,-2.5306],[175.8263,-1.6673],[176.275,-0.7168],[176.7437,0.0613],[177.1975,0.5813],[177.6837,0.8419],[178.1175,0.9294],[178.5375,1.0162],[178.9362,1.1033],[179.7362,1.4495],[180.1538,1.6223],[180.9588,1.7965],[184.015,1.7965],[185.3837,1.7109],[188.0875,1.6252],[188.4525,1.5389],[188.8375,1.3672],[189.2325,1.0236],[190.0525,-0.0065],[190.4638,-0.3501],[190.8637,-0.5218],[191.25,-0.6938],[192.0325,-1.2086],[192.43,-1.3806],[192.83,-1.6383],[193.64,-1.9822],[194.8175,-2.4982],[195.2125,-2.7566],[195.635,-3.3593],[196.0775,-4.2201],[196.51,-5.0819],[196.9462,-5.8578],[197.3863,-6.7196],[197.8325,-7.4962],[198.2675,-8.0152],[199.1363,-8.7114],[199.5825,-9.1466],[200.4837,-9.8483],[200.9325,-10.0244],[202.185,-10.2872],[203.6888,-10.2876],[204.0675,-10.3755],[204.91,-10.8985],[205.3487,-11.2477],[205.8038,-11.5124],[206.6962,-11.8646],[207.14,-11.9525],[207.5638,-11.9551],[207.9775,-12.1278],[208.8775,-12.6587],[209.735,-13.3616],[210.1825,-13.624],[211.0587,-13.7994],[211.4875,-13.7997],[211.91,-13.8865],[212.325,-13.8884],[213.6012,-14.149],[214.5125,-14.149],[215.585,-13.9715],[216.8512,-13.9733],[217.5912,-14.1724],[219.7862,-14.1741],[220.5238,-14.3759],[221.3062,-14.3794],[222.065,-14.5802],[222.4625,-14.5872],[222.8675,-14.788],[223.28,-14.7941],[223.68,-14.995],[224.5087,-14.9959],[224.9,-15.1959],[226.65,-15.1976],[227.0437,-15.4019],[227.4525,-15.4115],[227.9088,-16.0132],[228.7713,-16.4306],[232.3138,-16.4358],[232.7225,-16.6367],[234.77,-16.6393],[235.1762,-16.8384],[236.3737,-16.8402],[236.765,-17.0402],[237.18,-17.0428],[237.595,-17.2419],[237.995,-17.2463],[239.2463,-18.4688],[240.12,-19.6932],[241.4225,-22.1227],[241.8625,-23.1322],[242.2875,-23.7409],[242.71,-24.5522],[244.0075,-26.3739],[244.8312,-26.7782],[246.3687,-26.7791],[247.1875,-26.9791],[250.8238,-26.9791],[251.1938,-27.1791],[255.0662,-27.1756],[255.4675,-26.9782],[255.8512,-26.9765],[256.595,-27.1774],[257.3737,-27.1748],[258.1287,-26.9739],[258.515,-26.3687],[258.9413,-25.7617],[260.1988,-25.1522],[260.59,-25.1505],[260.9813,-24.9496],[261.3763,-24.9452],[261.7738,-24.5426],[262.1988,-24.3366],[262.6087,-23.9322],[263.0138,-23.7287],[263.4212,-23.7279],[263.825,-23.5279],[264.2225,-23.527],[265.3038,-23.3261],[267.7113,-22.1088],[275.6012,-22.1088],[276.4,-21.2992],[276.82,-20.4905],[277.2812,-18.8732],[277.795,-16.8471],[278.3075,-15.2237],[278.8038,-14.0063],[279.2812,-13.1985],[280.1087,-12.7907],[280.505,-12.7881],[280.8988,-12.3889],[281.68,-11.982],[282.0875,-11.3742],[282.5125,-10.5655],[282.9362,-9.3533],[283.3737,-8.7455],[283.7887,-8.5447],[284.5725,-8.5438],[284.9638,-8.3429],[285.3938,-8.1403],[285.805,-7.7377],[286.61,-7.3308],[287.02,-6.929],[287.8238,-6.5238],[288.225,-5.9204],[288.64,-5.3134],[289.0575,-4.1056],[289.4937,-2.8952],[289.9388,-2.29],[290.355,-2.0882],[291.0913,-2.0874],[291.4725,-1.8882],[292.1963,-1.8874],[292.5775,-1.4865],[293.0038,-1.4856],[293.4,-1.2865],[293.7938,-1.2856],[294.5213,-1.0856],[294.9088,-0.8839],[295.3025,-0.6822],[295.7062,-0.6813],[296.1025,-0.6822],[296.4863,-0.6813],[296.87,-0.4822],[297.2688,-0.4813],[297.66,-0.2822],[298.0512,-0.0813],[298.4488,0.1204],[298.8463,0.123],[299.235,0.3222],[299.6225,0.523],[300.015,0.5239],[300.7587,0.7239],[301.1425,0.7248],[301.5275,0.9248],[301.9225,0.9265],[302.7,1.3274],[303.5013,1.7317],[303.8975,1.7335],[305.065,2.3369],[305.8663,3.1413],[306.26,3.3439],[306.6562,3.7456],[307.0675,4.1491],[307.8612,4.5525],[308.665,5.3569],[309.0675,5.9604],[309.4987,6.363],[309.9725,6.7664],[310.4613,7.1682],[310.9075,7.3699],[311.7025,7.3699],[312.4288,7.5699],[318.185,7.7708],[318.5688,8.5742],[318.9837,9.1777],[319.4125,9.3803],[319.8225,9.7812],[320.2362,9.982],[320.975,10.1829],[321.3575,10.3838],[321.7487,10.7855],[322.5338,11.1872],[322.93,11.5881],[324.1037,12.1907],[324.5038,12.9933],[324.9162,13.5959],[325.7413,15.1994],[326.1425,15.8011],[326.5438,16.002],[326.9325,16.4037],[327.3675,16.8046],[327.7587,17.2063],[328.155,17.808],[328.5538,18.2098],[330.5462,21.2158],[331.345,22.0193],[333.2125,22.0193],[335.615,22.2202],[336.3963,22.6219],[336.795,23.2236],[337.5975,24.0254],[338.0013,24.628],[338.8038,25.4314],[339.2125,26.0323],[339.6075,26.234],[340.0013,26.2349],[340.3887,26.4349],[341.8363,26.6349],[342.2688,26.8358],[342.7188,27.2357],[343.5587,27.2357],[343.9413,27.4366],[344.3313,27.4375],[345.115,27.8375],[345.5125,27.8392],[346.3025,28.2401],[347.1125,29.0436],[347.5138,29.2444],[348.585,29.2453],[349.3587,29.647],[354.1637,29.6462],[354.5487,29.0453],[354.96,28.2427],[355.3625,27.0392],[355.785,25.2349],[356.2237,23.228],[357.1463,18.4089],[357.625,15.5968],[358.0763,12.9811],[358.5312,10.5673],[358.9688,8.1508],[359.4462,6.1325],[359.8863,3.5091],[360.3338,1.4874],[360.7663,-0.9395],[361.3587,-1.7543],[361.7975,-2.5656],[362.2312,-3.1743],[362.6975,-3.3821],[363.1712,-3.7838],[363.6112,-3.9856],[365.3487,-3.9864],[365.725,-4.189],[366.5575,-5.0012],[366.995,-5.2055],[367.4087,-5.2081],[368.215,-5.6142],[368.63,-6.0229],[369.0512,-6.8377],[369.4913,-8.0542],[369.925,-9.0724],[370.3587,-9.6863],[370.8038,-10.0941],[371.2138,-10.2959],[372.7713,-10.2976],[373.1637,-10.4985],[373.5738,-10.5011],[374.03,-10.7002],[374.9538,-10.7002],[403,-10.1003],[404,-10.095],[405,-9.8941],[406,-9.8933],[407,-9.4889],[408,-8.2759],[410,-7.0646],[411,-5.4516],[412,-4.8446],[413,-4.442],[414,-4.4403],[415,-3.4307],[416,-2.2064],[417,0.2309],[418,1.0483],[419,1.6526],[421,5.2987],[422,8.536],[423,10.1568],[424,10.5612],[425,11.5751],[426,14.8063],[427,21.2706],[428,30.1504],[429,38.006],[430,41.419],[431,48.4441],[432,52.0562],[433,57.6797],[435,57.4788],[436,47.4563],[437,37.2286],[438,23.5827],[439,9.7341],[440,-2.3092],[441,-12.1595],[442,-14.3821],[443,-14.989],[444,-15.8038],[445,-22.3159],[446,-23.7367],[449.3975,-30.9949],[450.7787,-30.794],[451.16,-29.9914],[451.585,-28.9879],[452,-27.7836],[452.4175,-26.9801],[453.2175,-26.5793],[453.6137,-26.1784],[454.005,-26.1775],[454.79,-25.7749],[456.0063,-24.5654],[456.4137,-23.7628],[456.8238,-23.5602],[457.2237,-23.1584],[457.6263,-22.9567],[458.0375,-22.555],[458.8463,-22.1497],[459.655,-21.3419],[460.065,-20.7359],[460.4963,-20.3298],[460.9137,-19.7237],[461.7387,-19.3194],[462.1388,-19.3168],[462.5338,-19.1176],[462.9325,-19.1159],[464.555,-18.3029],[464.9625,-17.8985],[466.5525,-17.8968],[466.9562,-17.6977],[467.8012,-17.6933],[468.6275,-17.2864],[469.0475,-17.2811],[469.4875,-17.0829],[469.9125,-17.0811],[471.7538,-16.2664],[473.0425,-15.0438],[474.67,-14.2299],[475.5163,-13.4195],[475.9625,-13.2134],[476.42,-12.6073],[477.3475,-11.7952],[477.7812,-11.1874],[478.2375,-10.3752],[478.7125,-9.5648],[479.61,-8.7544],[480.0462,-8.5509],[480.46,-8.5483],[480.8687,-8.1483],[481.3025,-8.1457],[481.7237,-7.9448],[482.5863,-7.1326],[483.0288,-6.5257],[483.4762,-6.1205],[484.7425,-5.5135],[485.1587,-5.5109],[485.5838,-5.3109],[485.995,-4.9057],[486.4288,-4.9005],[486.865,-4.7014],[487.3313,-4.4996],[491.2663,-4.4996],[492.71,-4.2996],[493.4775,-4.0996],[493.8712,-3.8979],[494.2775,-3.497],[494.6938,-3.0962],[498.2312,-3.0988],[498.6425,-3.9014],[499.0788,-4.9083],[499.5338,-5.9179],[500.4525,-7.5335],[500.9062,-8.1396],[501.355,-8.5422],[501.785,-8.9448],[502.2725,-9.1483],[502.7775,-9.7526],[503.295,-9.9552],[503.7563,-10.3587],[504.1938,-10.363],[505.045,-11.1656],[505.485,-11.3691],[505.9075,-11.7726],[508.4625,-12.9952],[509.29,-13.8117],[509.7575,-14.0203],[510.2025,-14.4273],[511.8725,-15.2429],[512.2788,-15.2464],[512.6875,-15.4455],[513.9088,-15.4481],[514.3338,-15.6464],[516.6163,-15.6481],[517.0337,-15.8499],[517.4438,-15.8516],[517.8363,-16.0516],[519.69,-16.0525],[520.0725,-16.2533],[520.4688,-16.256],[520.8775,-16.4568],[522.82,-16.4594],[524.4588,-17.2777],[524.88,-17.2811],[525.2788,-17.4803],[525.6812,-17.482],[527.1862,-17.889],[527.5975,-17.8933],[528.435,-18.302],[528.8575,-18.3081],[529.645,-18.5072],[530.045,-18.5107],[530.4462,-18.7107],[530.875,-18.7133],[531.275,-18.9133],[532.0813,-18.915],[532.4812,-19.1141],[537.7775,-18.9133],[538.1625,-18.3107],[538.57,-17.1055],[539.0075,-15.9011],[539.4262,-15.2977],[539.8425,-14.2925],[541.13,-11.8716],[541.59,-11.063],[542.0213,-10.456],[542.9,-8.4386],[543.3413,-7.2274],[543.8,-5.6126],[544.2575,-4.1978],[544.715,-2.9839],[545.1812,-2.1718],[545.6537,-1.7657],[546.0813,-1.5614],[546.495,-1.3596],[546.9237,-1.3579],[547.3288,-1.157],[547.7362,-1.1544],[548.1375,-0.9536],[548.5362,-0.5501],[548.9475,-0.1449],[549.3637,0.0595],[549.7725,0.4656],[550.1963,0.8708],[551.0362,1.6812],[551.4488,1.8838],[551.8475,1.8855],[552.6437,2.2864],[553.045,2.6899],[553.4638,3.2959],[553.8913,4.3038],[554.7725,6.322],[555.235,7.1315],[555.6913,7.7359],[556.1137,7.9402],[556.5338,8.3446],[556.9538,8.5472],[557.3713,8.9506],[557.7738,9.5532],[559.4375,11.1671],[559.8475,11.3697],[560.5888,11.5706],[572.1175,11.5715],[572.5113,12.3741],[573.395,13.1793],[573.7987,13.181],[574.2,13.3801],[575.645,13.381],[577.2412,14.1871],[577.6463,14.7879],[578.06,14.7897],[579.6562,15.594],[580.0625,15.9984],[581.7225,18.4157],[582.1425,18.6192],[582.555,19.0218],[583.3763,19.4287],[583.785,19.8322],[584.1912,20.034],[589.055,20.0331],[589.4362,19.2314],[589.85,18.227],[590.7125,15.4166],[591.5675,14.2105],[591.9813,13.6045],[592.91,11.1889],[593.3562,10.3819],[593.7862,9.7784],[594.6063,9.3767],[595.01,9.3758],[595.8138,8.5698],[596.2312,7.9619],[596.6775,7.1507],[597.1438,5.9333],[597.6463,4.7185],[598.0962,4.3098],[599.0612,3.9064],[599.905,3.9038],[600.6562,3.7046],[601.0437,3.502],[602.6225,3.5003],[603.865,2.8907],[604.2925,2.4829],[604.7313,1.876],[605.1725,1.4699],[605.5925,1.0664],[606.0113,0.8638],[606.4087,0.6603],[606.8112,0.6586],[607.6212,0.2551],[608.0263,0.0508],[608.4512,-0.1544],[608.8612,-0.3596],[609.27,-0.7657],[609.6837,-1.1718],[610.1,-1.5779],[610.525,-2.184],[610.9662,-2.3892],[611.3863,-2.7952],[611.8038,-3.4022],[612.2325,-4.0117],[613.1375,-5.6361],[613.595,-6.0439],[614.025,-6.2482],[614.4512,-6.65],[616.4688,-6.6517],[619.3275,-8.0725],[619.7287,-8.0751],[620.9662,-8.6882],[621.3787,-9.0943],[621.8,-9.3012],[622.2262,-9.7047],[622.6413,-9.7064],[625.0888,-10.9255],[625.5013,-11.3316],[625.9275,-11.5368],[626.375,-11.9412],[626.8112,-12.1438],[627.2362,-12.1455],[627.65,-12.3472],[628.07,-12.7507],[628.945,-13.9681],[629.8312,-15.5898],[630.2937,-16.1993],[630.815,-16.6046],[632.2013,-17.2132],[634.3088,-19.2384],[636.3625,-20.2523],[636.7687,-20.6575],[638.5225,-22.287],[639.345,-22.694],[639.7513,-22.6966],[640.565,-23.1001],[640.98,-23.5062],[641.4225,-23.714],[642.6913,-24.9331],[643.1062,-25.1374],[643.515,-25.1409],[644.3363,-25.9496],[644.7538,-26.1565],[645.1875,-26.5626],[646.0387,-26.9678],[648.3338,-26.9687],[650.105,-27.1704],[650.8963,-27.1704],[655.31,-26.7696],[655.7075,-25.967],[656.5438,-24.3626],[656.9638,-24.1618],[657.7925,-23.3574],[658.2025,-22.3522],[658.6537,-21.3479],[659.0725,-19.9091],[659.4925,-19.0799],[660.305,-18.5263],[661.0962,-18.5263],[662.3212,-17.6934],[662.7725,-17.1386],[663.2525,-17.1363],[663.715,-16.8606],[664.4875,-16.5838],[664.8737,-16.5814],[665.2688,-16.3034],[665.7,-15.7451],[666.6087,-14.6295],[667.4688,-14.0723],[668.2525,-14.0699],[668.6513,-13.7955],[669.055,-13.7919],[669.4588,-13.5163],[669.8663,-12.9591],[671.1112,-12.1167],[671.5125,-12.1131],[672.7162,-11.2791],[673.13,-10.7207],[683,1.841],[703,1.841],[706,1.0617],[721,1.0617],[726,1.7007],[739,1.7007],[744,0],[752.9987,0]]
       }
       dyn_wings_radius_r_ctrl.Rotation.Z {type=MovieSceneScriptingFloatChannel, default=0, const [0,7]}
       dyn_wings_radius_r_ctrl.Scale.X {type=MovieSceneScriptingFloatChannel, default=1, const [1,7]}
       dyn_wings_radius_r_ctrl.Scale.Y {type=MovieSceneScriptingFloatChannel, default=1, const [1,7]}
       dyn_wings_radius_r_ctrl.Scale.Z {type=MovieSceneScriptingFloatChannel, default=1, const [1,7]}
       dyn_wings_metacarpal_01_r_ctrl.Location.X {type=MovieSceneScriptingFloatChannel, default=0}
       dyn_wings_metacarpal_01_r_ctrl.Location.Y {type=MovieSceneScriptingFloatChannel, default=0}
       dyn_wings_metacarpal_01_r_ctrl.Location.Z {type=MovieSceneScriptingFloatChannel, default=0}
       dyn_wings_metacarpal_01_r_ctrl.Rotation.X {type=MovieSceneScriptingFloatChannel, default=0}
       dyn_wings_metacarpal_01_r_ctrl.Rotation.Y {type=MovieSceneScriptingFloatChannel, default=0}
       dyn_wings_metacarpal_01_r_ctrl.Rotation.Z {type=MovieSceneScriptingFloatChannel, default=0}
       dyn_wings_metacarpal_01_r_ctrl.Scale.X {type=MovieSceneScriptingFloatChannel, default=1}
       dyn_wings_metacarpal_01_r_ctrl.Scale.Y {type=MovieSceneScriptingFloatChannel, default=1}
       dyn_wings_metacarpal_01_r_ctrl.Scale.Z {type=MovieSceneScriptingFloatChannel, default=1}
       dyn_wings_phalange_01_r_ctrl.Location.X {type=MovieSceneScriptingFloatChannel, default=0}
       dyn_wings_phalange_01_r_ctrl.Location.Y {type=MovieSceneScriptingFloatChannel, default=0}
       dyn_wings_phalange_01_r_ctrl.Location.Z {type=MovieSceneScriptingFloatChannel, default=0}
       dyn_wings_phalange_01_r_ctrl.Rotation.X {type=MovieSceneScriptingFloatChannel, default=0}
       dyn_wings_phalange_01_r_ctrl.Rotation.Y {type=MovieSceneScriptingFloatChannel, default=0}
       dyn_wings_phalange_01_r_ctrl.Rotation.Z {type=MovieSceneScriptingFloatChannel, default=0}
       dyn_wings_phalange_01_r_ctrl.Scale.X {type=MovieSceneScriptingFloatChannel, default=1}
       dyn_wings_phalange_01_r_ctrl.Scale.Y {type=MovieSceneScriptingFloatChannel, default=1}
       dyn_wings_phalange_01_r_ctrl.Scale.Z {type=MovieSceneScriptingFloatChannel, default=1}
       dyn_wings_carpals_01_r_ctrl.Location.X {type=MovieSceneScriptingFloatChannel, default=0, const [0,3]}
       dyn_wings_carpals_01_r_ctrl.Location.Y {type=MovieSceneScriptingFloatChannel, default=0, const [0,3]}
       dyn_wings_carpals_01_r_ctrl.Location.Z {type=MovieSceneScriptingFloatChannel, default=0, const [0,3]}
       dyn_wings_carpals_01_r_ctrl.Rotation.X {type=MovieSceneScriptingFloatChannel, default=0, const [0,3]}
       dyn_wings_carpals_01_r_ctrl.Rotation.Y {
        type=MovieSceneScriptingFloatChannel
        default=0
        thinned_from=1047
        keys [[-10,0],[0,0],[8.6263,-0.1619],[13.4488,-0.3238],[13.82,-0.5303],[14.2088,-0.8987],[14.6025,-0.9432],[14.9775,-1.1051],[18.1825,-0.9877],[28.5362,-0.8258],[31.5812,-0.6639],[31.9475,-0.5019],[32.33,-0.2226],[32.7025,-0.0607],[33.075,0.2186],[33.4725,0.3805],[33.8575,0.5425],[34.9262,0.7044],[35.295,1.1456],[35.6775,1.5868],[36.06,1.9107],[36.44,2.0726],[37.1575,2.2345],[40.8712,2.3965],[41.6163,2.7203],[42.69,3.0441],[43.0688,3.3234],[43.8112,3.6473],[44.8525,3.8092],[51.8312,3.6473],[52.2287,3.3234],[52.675,2.6758],[53.1388,2.4248],[53.5763,2.0564],[53.9675,2.0119],[54.3613,1.8055],[54.7613,1.7609],[55.5013,1.4371],[56.21,1.2752],[56.5825,1.1133],[56.955,1.0687],[57.6975,0.9068],[58.0725,0.7004],[58.4575,0.6558],[58.84,0.4494],[59.225,0.1255],[59.6087,0.081],[59.985,-0.1254],[60.37,-0.2873],[60.745,-0.3319],[61.1225,-0.5383],[61.5025,-0.7002],[61.885,-1.0686],[62.29,-1.437],[62.6837,-1.6879],[63.0788,-2.0563],[63.4675,-2.2627],[63.85,-2.3073],[65.5975,-2.4692],[66.2975,-2.6311],[66.6688,-2.6756],[67.0537,-2.8376],[67.4338,-2.8821],[67.8175,-3.2059],[68.2038,-3.4124],[68.595,-3.7362],[68.9775,-4.1046],[69.365,-4.7968],[69.7625,-5.3716],[70.17,-6.0638],[70.5788,-6.2703],[72.3038,-6.4322],[73.3687,-6.5941],[73.7437,-6.756],[74.12,-6.6386],[74.5,-6.8005],[75.2225,-6.9625],[75.5987,-7.007],[76.3112,-7.1689],[77.685,-7.0515],[78.0487,-7.2134],[78.425,-7.6992],[78.81,-7.9056],[79.1988,-8.3914],[79.5875,-8.9217],[79.9813,-9.6584],[80.3913,-9.9823],[80.7875,-10.0268],[83.8625,-10.1887],[84.79,-10.5125],[88.5475,-10.6745],[88.9112,-10.8809],[89.2875,-11.2493],[89.68,-11.4112],[90.0587,-11.6176],[90.4438,-11.7796],[95.4488,-11.6176],[95.8112,-11.2938],[96.19,-10.808],[96.5775,-10.2778],[96.9737,-9.541],[97.3675,-9.0553],[98.1025,-8.8933],[103.4787,-8.8488],[103.8412,-8.6869],[104.6388,-7.8611],[105.0337,-7.3308],[105.4175,-6.8005],[105.815,-6.4322],[106.1988,-5.9019],[106.5975,-5.2542],[106.9875,-5.0923],[107.3663,-5.0478],[107.7575,-4.8858],[108.8913,-3.9143],[109.2763,-3.5459],[109.6688,-3.0602],[110.0625,-2.6918],[110.4488,-2.3235],[110.8312,-1.9996],[111.21,-1.8377],[111.595,-1.7932],[112.3125,-1.6312],[112.6762,-1.5867],[113.4475,-1.2629],[113.84,-1.2184],[114.22,-1.0564],[114.6087,-1.0119],[115.0025,-0.85],[115.4025,-0.6881],[118.9575,-0.85],[119.66,-1.0119],[120.0288,-1.1738],[120.4137,-1.4977],[120.7925,-1.9834],[121.1712,-2.3518],[121.5625,-2.8376],[121.9425,-2.9995],[122.3175,-3.4852],[122.7075,-4.1329],[123.095,-4.9425],[123.4825,-5.6347],[123.8813,-6.4443],[124.2937,-6.8127],[125.6938,-6.9746],[128.3875,-7.1365],[129.5225,-7.6223],[129.895,-7.6668],[138.9312,-7.8288],[139.3038,-7.8733],[139.685,-8.1971],[140.07,-8.4036],[140.4575,-8.7274],[140.8412,-9.4196],[141.2412,-9.9499],[141.6375,-10.3628],[142.0275,-10.7312],[142.4137,-10.7757],[142.7975,-10.9376],[143.5263,-11.0995],[144.2513,-11.4679],[144.655,-11.9982],[145.0512,-12.3665],[145.4338,-13.1033],[145.84,-14.0019],[146.245,-14.5767],[146.6413,-15.3135],[147.0638,-15.8883],[147.535,-17.2282],[148.0238,-18.4507],[148.4775,-19.4667],[148.8813,-19.997],[149.2713,-20.1589],[152.3687,-20.3654],[152.7362,-20.5273],[153.1137,-20.8511],[153.5113,-21.3814],[153.91,-21.8672],[154.3125,-22.5149],[154.705,-22.8387],[155.0863,-23.0006],[156.7987,-22.8832],[157.8388,-23.0452],[158.965,-23.5309],[160.995,-24.0167],[161.7675,-25.5914],[162.5612,-26.8867],[162.9462,-27.2106],[163.335,-27.3725],[166.0537,-27.5344],[166.425,-27.6963],[166.81,-28.0202],[167.5963,-29.3155],[167.9875,-29.8013],[168.395,-29.9632],[170.1838,-30.1251],[174.9462,-30.2871],[175.7038,-30.6109],[176.4787,-31.5824],[176.8663,-31.7443],[177.245,-32.0682],[180.58,-32.2301],[180.9513,-32.7887],[182.1125,-34.1124],[182.8825,-35.084],[183.2725,-35.8936],[184.065,-36.8651],[184.4613,-37.1889],[184.8463,-37.0715],[186.2188,-37.2335],[186.9275,-37.3954],[187.3012,-37.7192],[188.8138,-38.3669],[189.54,-38.2495],[193.5725,-38.205],[194.3313,-37.954],[195.1162,-37.5411],[195.5063,-37.4966],[195.89,-37.1283],[196.3112,-36.5534],[196.7225,-35.6548],[197.1287,-35.2419],[197.9162,-34.7845],[201.045,-34.578],[201.8125,-34.2542],[202.1975,-34.3716],[202.59,-34.2097],[206.6988,-34.489],[209.415,-34.4445],[209.795,-34.5902],[210.2163,-34.574],[211.1662,-33.8211],[212.0412,-32.716],[212.8312,-32.0238],[215.9212,-31.8174],[216.2925,-31.449],[216.6875,-31.3599],[217.0863,-31.0361],[217.4725,-31.1535],[217.8487,-31.109],[218.2412,-31.2264],[218.6275,-31.1818],[219.0075,-31.2992],[219.3938,-31.2547],[219.785,-31.445],[220.18,-31.194],[220.5812,-30.4128],[221.8225,-28.1702],[222.2425,-27.7573],[222.645,-27.7856],[223.0475,-27.6237],[223.8075,-27.5347],[224.2013,-27.3282],[224.605,-26.9599],[224.99,-26.7089],[225.795,-25.8386],[226.6163,-24.5554],[227.4338,-24.2599],[227.8338,-24.3773],[229.3562,-24.2437],[229.74,-24.1547],[230.1513,-23.9766],[230.57,-23.6811],[231.0113,-23.2236],[231.4238,-23.252],[231.8238,-23.001],[232.2163,-22.912],[232.6163,-23.1022],[233.0125,-23.0132],[233.4175,-23.1589],[233.8312,-22.7744],[234.27,-22.4182],[234.76,-22.1349],[235.2075,-21.5722],[235.6525,-21.1877],[236.0675,-21.3617],[236.5,-21.7423],[237.3062,-22.3131],[238.0938,-22.5479],[238.4712,-22.7827],[238.8637,-23.1348],[239.2537,-23.7218],[239.66,-24.4707],[240.0662,-25.0577],[240.4575,-25.5273],[240.8525,-26.2317],[241.2688,-26.7741],[241.6788,-27.3894],[242.1475,-28.4015],[242.6862,-29.2961],[243.5788,-30.3527],[243.9837,-30.7049],[244.8125,-32.9799],[245.235,-35.5059],[245.6775,-38.4286],[246.1263,-40.6025],[246.5675,-41.91],[246.9887,-42.3067],[249.3938,-42.1731],[250.2163,-39.8455],[250.6513,-38.0927],[251.515,-32.6441],[251.9388,-30.5675],[252.3813,-28.8147],[252.8225,-27.5922],[253.2475,-26.5762],[253.6637,-25.4427],[254.485,-24.0583],[255.2775,-23.4107],[256.7275,-23.0868],[257.1162,-23.0423],[257.5375,-22.1882],[257.9588,-21.1721],[258.375,-19.7148],[258.7875,-18.4478],[259.215,-17.5937],[259.6325,-16.5776],[260.035,-15.8854],[260.835,-14.7803],[261.2387,-14.0881],[261.6475,-13.3514],[262.065,-12.1289],[262.4762,-11.0238],[262.92,-10.1697],[263.3338,-9.522],[263.7462,-9.4775],[264.1338,-9.1536],[264.5487,-8.7852],[264.9688,-8.3724],[265.3675,-7.8421],[265.7675,-7.1499],[266.1687,-6.5751],[266.5688,-5.8828],[266.965,-5.3526],[267.3787,-4.4984],[267.7862,-2.9966],[268.2075,-1.4948],[268.6388,0.1689],[269.0638,2.5977],[269.5013,4.4679],[269.9175,6.7348],[270.3562,8.1192],[271.195,9.7829],[271.605,10.1513],[272.0013,10.6087],[272.4087,11.4629],[272.8175,12.1551],[273.2175,12.2441],[273.6463,12.568],[275.255,12.406],[275.9737,12.2441],[276.3438,12.0822],[276.7313,11.8029],[277.9225,9.9934],[278.3212,9.5076],[278.7175,8.9045],[279.12,8.7426],[279.505,8.1394],[279.89,7.8156],[280.2837,7.5362],[280.6812,7.4189],[281.06,7.2569],[281.44,6.8157],[281.83,6.2571],[282.23,5.4191],[282.6413,4.9779],[284.3687,4.8605],[289.7887,4.6986],[297.8463,4.5812],[299.6012,4.6419],[300.0075,5.3503],[300.435,5.7793],[300.8575,5.751],[301.2587,5.6336],[311.8525,5.9129],[312.2387,6.0748],[312.63,6.1922],[313.3537,6.4715],[316.7687,6.3096],[317.1438,6.2651],[317.53,5.9412],[318.335,5.6174],[318.7413,5.2936],[319.16,4.8078],[319.5612,4.3666],[320.385,4.0427],[320.8125,3.8808],[321.8813,3.7189],[322.255,3.557],[322.645,3.0712],[323.0325,2.63],[323.4262,2.3061],[324.235,1.9823],[324.6413,1.703],[325.0325,1.541],[325.4162,1.3791],[325.8,1.0553],[326.1988,0.4076],[326.6037,-0.0336],[326.9913,-0.3575],[327.3787,-0.5194],[328.12,-0.6368],[330.145,-0.7987],[330.5175,-0.9606],[330.9013,-1.2845],[331.2975,-2.0496],[331.6975,-2.3734],[332.1,-2.8146],[332.495,-3.2559],[333.2862,-3.9035],[333.675,-4.3448],[334.065,-4.8305],[334.47,-5.4337],[334.875,-5.8749],[335.2738,-6.0369],[341.525,-5.8749],[342.2387,-5.5511],[342.6275,-5.3447],[343.0213,-4.8144],[343.4125,-4.0048],[344.2088,-3.0332],[344.61,-2.7094],[346.0662,-2.3856],[349.4312,-2.5475],[350.17,-2.7094],[350.55,-2.8713],[351.3212,-3.519],[351.7125,-3.6809],[352.1037,-3.5635],[352.8325,-3.7254],[353.8825,-3.8874],[354.6938,-4.8144],[355.0925,-5.1382],[355.4925,-5.4175],[355.885,-5.5794],[357.6475,-5.7414],[359.04,-5.9033],[359.435,-6.2271],[360.615,-6.7129],[361.3487,-6.8748],[361.7538,-7.0367],[362.1463,-7.5225],[363.33,-8.494],[364.13,-8.8179],[364.5163,-8.9353],[368.645,-9.2591],[369.5975,-11.0686],[370.0225,-11.9956],[370.4325,-13.04],[370.8487,-13.805],[371.2525,-14.1289],[374.3575,-13.9224],[375.48,-13.4367],[376.2688,-12.6999],[376.6725,-12.3761],[377.0713,-12.1696],[383.145,-12.3316],[383.875,-12.4935],[385.3825,-12.6999],[385.7587,-12.8619],[386.1562,-13.1857],[386.55,-13.3476],[386.9388,-13.3921],[393.98,-13.5541],[394.705,-13.716],[395.095,-13.967],[395.4963,-14.4527],[395.8963,-15.1449],[396.3038,-15.3514],[396.6988,-15.3959],[397.09,-15.5578],[397.4825,-15.7643],[397.8837,-15.9262],[398.2812,-16.2945],[398.6862,-16.8694],[399.1,-17.6506],[399.5175,-19.0796],[399.9613,-20.3466],[400.4412,-22.0548],[400.9813,-24.6172],[401.9275,-30.7418],[402.3837,-33.0087],[402.8288,-33.8629],[403.235,-34.0248],[405.3062,-34.1422],[406.0238,-34.2596],[406.3988,-34.0531],[406.8038,-33.3893],[407.2262,-32.3571],[407.6588,-30.75],[408.095,-29.3049],[408.5375,-27.9043],[408.9713,-26.9611],[409.3938,-26.3863],[409.8087,-26.1799],[410.2138,-26.1353],[410.975,-25.767],[411.365,-25.4431],[412.1925,-24.9857],[412.5987,-24.4554],[413.0088,-23.6742],[413.43,-22.3626],[413.8612,-20.1524],[414.31,-16.3959],[414.8075,-14.1857],[415.2412,-13.2425],[415.6688,-12.9187],[417.4725,-12.7122],[418.205,-12.0646],[418.6063,-11.4897],[419.0312,-10.4737],[419.4475,-9.7815],[419.855,-9.1338],[420.2563,-9.0893],[420.6513,-8.6035],[421.045,-8.0732],[421.4575,-7.1017],[421.8813,-5.7618],[422.3062,-3.7297],[422.74,-1.815],[423.1737,0.0552],[423.6075,0.7474],[424.01,1.2777],[424.4338,1.3222],[424.8238,1.9699],[425.2225,2.9859],[425.64,4.2084],[426.0638,5.7993],[426.49,7.6694],[426.9337,8.1552],[427.33,8.3171],[429.745,8.1997],[430.4663,8.0378],[430.845,7.8759],[431.2362,6.787],[431.6938,4.933],[432.1975,3.5202],[432.6925,2.7106],[433.14,1.4598],[433.57,0.3263],[433.9787,-1.0136],[434.39,-1.9567],[434.8238,-2.8109],[435.2412,-3.2238],[435.6537,-3.7095],[436.4462,-4.3572],[436.84,-5.1668],[437.2437,-5.9764],[437.6537,-6.6686],[438.0487,-7.2434],[438.47,-7.6118],[438.8663,-8.1421],[439.2738,-8.3931],[439.6725,-9.2472],[440.0838,-10.4697],[440.5063,-11.6922],[440.9475,-12.0605],[441.7775,-12.3844],[442.87,-12.5908],[443.2637,-13.1211],[443.6737,-14.1372],[444.1012,-15.0358],[444.555,-16.0964],[444.9737,-16.7441],[447.12,-16.9505],[447.51,-17.2744],[447.915,-17.4808],[448.3188,-18.3349],[448.7287,-19.351],[449.1325,-20.2051],[449.545,-20.5735],[452.9275,-20.7354],[453.3062,-20.8973],[457.0263,-21.0147],[457.7987,-20.7192],[458.5938,-20.1727],[459.4125,-20.023],[459.8175,-19.8165],[460.21,-19.9339],[461.2713,-19.772],[464.1812,-19.6101],[465.435,-18.505],[465.8388,-18.254],[466.24,-18.0476],[467.0387,-17.3109],[467.4325,-17.2663],[467.8225,-16.736],[468.2188,-16.6915],[468.61,-16.4851],[469.0063,-16.3231],[469.4125,-16.2786],[469.8163,-16.1167],[471.2287,-16.2341],[472.0163,-16.0722],[473.24,-15.5864],[473.635,-15.5419],[474.38,-15.38],[474.7738,-15.0561],[475.1712,-14.8497],[475.5688,-14.5258],[475.975,-14.3639],[476.36,-14.4813],[477.5663,-13.9956],[477.955,-13.6717],[478.3512,-13.5098],[484.36,-13.6717],[487.0662,-13.8336],[487.4475,-14.1575],[488.2275,-14.5704],[488.635,-14.9387],[489.0362,-15.2626],[489.7625,-15.1452],[490.8125,-15.3071],[491.535,-15.3516],[491.915,-15.6755],[492.3188,-15.72],[492.7188,-15.8819],[493.1025,-16.2058],[493.4987,-16.0884],[494.2287,-16.2503],[495.8,-16.4122],[496.2212,-16.5741],[496.6288,-16.4567],[497.4,-16.7806],[498.1375,-16.8251],[498.88,-16.987],[500.325,-17.1935],[501.0587,-17.3554],[501.4425,-17.5618],[502.1787,-17.7237],[502.9162,-18.0476],[504.03,-18.2095],[505.2175,-18.6953],[505.605,-18.9017],[506.0013,-19.2256],[506.39,-19.3875],[507.1237,-19.5494],[508.4987,-19.432],[509.5487,-19.5939],[510.2637,-19.7558],[513.3088,-19.9178],[514.7825,-20.1242],[515.6275,-20.448],[516.0275,-20.7719],[516.4262,-20.9338],[516.82,-21.1403],[517.235,-21.4641],[517.6388,-21.9499],[518.4525,-22.6866],[518.8563,-22.893],[519.6163,-23.0995],[520.02,-23.2614],[525.5362,-23.4233],[525.915,-23.4678],[526.32,-24.1155],[526.7638,-24.6458],[527.255,-25.2935],[527.7213,-25.338],[528.1325,-25.4999],[529.5263,-25.6619],[541.31,-25.4999],[541.6875,-25.6173],[542.485,-25.249],[542.875,-24.8806],[543.285,-24.6742],[543.685,-24.6296],[544.1175,-24.3058],[546.8888,-24.1439],[548.3875,-23.4517],[549.8512,-23.2452],[550.2225,-23.0833],[550.6137,-23.0388],[551.0038,-22.8769],[555.0462,-22.7149],[555.43,-22.7878],[555.8238,-22.302],[556.22,-22.0956],[556.6237,-21.7718],[557.0163,-21.5653],[557.4187,-21.2415],[557.8138,-21.1969],[558.2425,-21.035],[565.1213,-20.9905],[565.8475,-20.8286],[567.2362,-20.6667],[567.6175,-20.5047],[568.015,-20.1809],[568.8112,-19.768],[569.2175,-19.4442],[569.6275,-19.3996],[570.02,-19.2377],[571.4137,-19.1932],[571.8038,-18.8694],[572.195,-18.6629],[572.5925,-18.2945],[573.0063,-18.0881],[573.4112,-18.0436],[573.8125,-17.8816],[576.2212,-17.7197],[576.6075,-17.5133],[577.0138,-17.5862],[577.41,-17.3352],[577.8062,-17.4081],[578.2038,-17.3635],[578.6,-17.2016],[578.99,-17.319],[580.7113,-17.2745],[581.4812,-17.5093],[581.8725,-17.3473],[584.2988,-17.0964],[584.7038,-16.8454],[585.1162,-16.639],[585.9125,-16.5499],[586.3038,-16.388],[587.035,-16.4609],[587.4225,-15.9751],[587.8263,-16.048],[588.2262,-15.797],[589.04,-15.5298],[589.45,-15.2789],[589.8963,-15.3072],[590.86,-14.4814],[591.2887,-13.9066],[591.7025,-13.6557],[592.1062,-13.773],[592.5013,-13.6111],[595.6137,-13.4047],[596.0013,-13.2428],[596.3913,-13.1982],[596.79,-13.0363],[597.5325,-12.9918],[598.2637,-12.6679],[598.6562,-12.417],[599.0625,-12.0486],[599.46,-11.8422],[600.2437,-11.5183],[600.6425,-11.4738],[601.0387,-11.15],[601.4325,-11.1054],[601.8475,-10.7371],[602.2688,-10.5306],[602.6625,-10.2068],[603.0562,-10.1177],[603.8663,-9.7048],[604.2738,-9.3365],[604.6825,-9.0855],[605.0888,-8.3933],[605.4937,-7.9804],[605.8975,-7.612],[606.6962,-7.523],[607.095,-7.3611],[607.815,-7.1991],[608.86,-7.0372],[609.2387,-7.1546],[609.6363,-6.9927],[610.0238,-7.1101],[612.4175,-6.9482],[612.8038,-6.9765],[613.2088,-6.8875],[613.6063,-6.592],[614.0113,-5.9726],[614.4288,-5.3533],[614.8463,-5.1023],[615.2487,-4.6166],[616.0375,-4.2037],[616.435,-3.6289],[616.85,-2.6412],[617.28,-0.9896],[617.7262,0.8239],[618.1687,2.2974],[618.6025,2.5483],[619,2.5928],[619.4013,2.9167],[619.8,3.2567],[620.2387,4.2889],[620.6737,6.6327],[621.1388,9.021],[621.6588,9.9804],[623.2775,9.8184],[626.1338,9.3772],[626.525,9.0534],[626.9162,8.5676],[627.3125,8.2438],[627.7088,8.3612],[628.1075,8.1992],[628.5013,8.1547],[628.8913,7.9928],[629.2825,7.507],[629.6875,6.6974],[630.0938,6.0781],[630.515,5.6652],[630.9337,5.3414],[633.0188,5.4588],[633.3988,5.2523],[633.805,4.8394],[634.2188,4.2201],[634.6375,3.4833],[635.0537,3.0259],[635.4675,2.864],[635.8475,2.8195],[640.54,2.6576],[640.9137,2.4511],[641.3125,2.3621],[641.7088,2.0382],[642.1025,1.9937],[642.5025,1.7873],[642.8913,1.3015],[643.2887,1.0505],[643.7025,0.7267],[644.0962,0.5648],[644.485,0.6822],[651.4913,0.8441],[651.875,1.1679],[652.2725,1.6537],[652.6837,2.1394],[653.1237,2.6252],[653.6175,3.0664],[658.4975,2.9045],[658.875,2.2123],[659.2937,1.3137],[659.745,-0.3946],[660.1938,-1.1313],[660.6063,-1.3378],[660.9962,-1.2204],[661.3837,-1.3823],[661.7862,-1.5887],[662.1862,-2.0745],[662.5938,-2.2809],[662.995,-2.5319],[663.3913,-2.9003],[663.7962,-3.548],[664.2025,-3.7544],[664.605,-4.0054],[665.0063,-4.5357],[665.4112,-4.7866],[665.8212,-5.3169],[666.2225,-5.5679],[666.6288,-5.7743],[667.025,-5.9363],[672.0763,-6.0982],[672.4625,-6.1427],[673.2487,-6.7904],[673.6437,-6.9523],[674.0387,-6.8349],[677.7663,-6.9968],[678.4887,-7.3207],[678.87,-7.851],[679.2713,-8.4987],[679.6925,-9.4702],[680.1012,-10.0005],[680.4963,-10.4862],[680.8937,-10.6481],[683.3288,-10.8101],[686.7,-10.972],[687.0913,-11.1784],[687.5,-11.8706],[688.3062,-12.9312],[688.705,-13.2996],[689.1087,-13.4615],[690.8275,-13.6234],[691.55,-13.506],[705.2788,-13.4149],[711.98,-13.2142],[712.7713,-12.3729],[713.5638,-11.5728],[713.96,-11.1972],[714.3825,-10.7515],[714.8038,-10.37],[715.6063,-9.5979],[715.9962,-9.2731],[716.385,-9.0366],[723.5925,-5.7767],[723.985,-5.7362],[724.3887,-5.8678],[724.795,-6.0341],[725.2212,-6.3338],[725.625,-6.5658],[726.8062,-6.5842],[727.6175,-6.8021],[728.4062,-7.0483],[728.8025,-7.0822],[729.5863,-7.3598],[729.9913,-7.3314],[730.4025,-7.5797],[731.6075,-8.0555],[731.995,-8.0433],[732.3863,-8.2143],[732.78,-8.2824],[733.1737,-8.2748],[734.3438,-8.4976],[738.7125,-8.0227],[739.0838,-7.7522],[739.4675,-7.3237],[739.89,-6.8857],[740.2887,-6.4442],[740.6988,-6.1525],[741.0962,-5.9482],[742.2688,-5.4258],[742.66,-5.1191],[743.0512,-4.9365],[748.1263,-3.9672],[749.6925,-3.241],[750.1225,-2.5839],[750.5375,-2.0659],[750.9325,-1.8039],[751.32,-1.5369],[751.71,-1.2619],[752.1,-1.0223],[753,0]]
       }
       dyn_wings_carpals_01_r_ctrl.Rotation.Z {type=MovieSceneScriptingFloatChannel, default=0, const [0,3]}
       dyn_wings_carpals_01_r_ctrl.Scale.X {type=MovieSceneScriptingFloatChannel, default=1, const [1,3]}
       dyn_wings_carpals_01_r_ctrl.Scale.Y {type=MovieSceneScriptingFloatChannel, default=1, const [1,3]}
       dyn_wings_carpals_01_r_ctrl.Scale.Z {type=MovieSceneScriptingFloatChannel, default=1, const [1,3]}
       dyn_wings_metacarpal_02_r_ctrl.Location.X {type=MovieSceneScriptingFloatChannel, default=0, const [0,2]}
       dyn_wings_metacarpal_02_r_ctrl.Location.Y {type=MovieSceneScriptingFloatChannel, default=0, const [0,2]}
       dyn_wings_metacarpal_02_r_ctrl.Location.Z {type=MovieSceneScriptingFloatChannel, default=0, const [0,2]}
       dyn_wings_metacarpal_02_r_ctrl.Rotation.X {
        type=MovieSceneScriptingFloatChannel
        default=0
        thinned_from=805
        keys [[0,0],[7.8863,0.0001],[32.0938,0],[34.6363,0.0001],[36.4863,0.0005],[39.9375,0.0002],[41.24,-0.0008],[42.8263,-0.0014],[47.865,-0.0015],[48.225,-0.0017],[52.5912,-0.002],[53.755,-0.0029],[54.51,-0.0031],[58.8975,-0.0036],[80.1612,-0.0029],[80.5263,-0.0017],[80.9112,-0.0001],[81.3025,0.0012],[81.6875,0.0016],[86.7475,0.0012],[89.4563,0.0007],[90.8175,0.0002],[91.1862,-0.0003],[92.3487,-0.0032],[92.72,-0.0037],[93.5113,-0.0066],[93.9062,-0.0089],[94.3138,-0.0103],[94.7,-0.0107],[95.7575,-0.0111],[103.59,-0.0115],[104.75,-0.0135],[105.1988,-0.0146],[105.6575,-0.0154],[106.09,-0.0156],[108.1762,-0.0158],[108.9112,-0.0165],[113.1687,-0.0188],[124.675,-0.0186],[126.0725,-0.0183],[126.8625,-0.0178],[127.245,-0.0173],[128.7663,-0.0162],[129.1675,-0.0156],[129.545,-0.0153],[129.94,-0.0141],[130.7188,-0.0135],[131.1,-0.0129],[131.8425,-0.0126],[132.2237,-0.0121],[132.615,-0.0118],[133.3887,-0.0106],[133.7775,-0.0103],[134.1513,-0.0097],[138.0962,-0.0106],[138.5,-0.0121],[138.8975,-0.0127],[139.6163,-0.013],[150.29,-0.0122],[150.6788,-0.0113],[151.4588,-0.0063],[151.8637,-0.0055],[156.305,-0.008],[156.6938,-0.0115],[157.1125,-0.016],[157.8963,-0.0229],[158.3038,-0.0246],[159.0562,-0.0265],[159.7938,-0.0274],[160.555,-0.0327],[160.9438,-0.0363],[161.3313,-0.0381],[161.7275,-0.0389],[162.1,-0.0408],[171.3713,-0.042],[171.7513,-0.0441],[172.1475,-0.0476],[172.9438,-0.0522],[174.055,-0.0555],[174.4275,-0.0578],[175.1637,-0.0589],[176.2312,-0.0599],[177.2812,-0.0612],[178.56,-0.0684],[179.3487,-0.0709],[182.3675,-0.072],[182.7462,-0.0746],[183.1313,-0.0759],[183.5187,-0.0786],[184.2962,-0.0868],[184.6875,-0.0908],[185.8438,-0.099],[186.6237,-0.1018],[189.0188,-0.1046],[189.3925,-0.1088],[189.7887,-0.1119],[190.1825,-0.1161],[190.5812,-0.1189],[190.9625,-0.122],[191.3525,-0.1262],[191.7425,-0.1292],[192.5375,-0.1381],[193.7275,-0.147],[194.12,-0.1512],[194.5163,-0.1528],[194.895,-0.1556],[195.285,-0.1572],[195.6775,-0.1602],[196.075,-0.1643],[196.475,-0.1672],[197.6263,-0.1712],[201.0763,-0.1701],[203.6437,-0.1688],[204.0225,-0.1675],[204.4013,-0.1649],[204.7838,-0.1598],[205.195,-0.1559],[205.5987,-0.1534],[205.99,-0.1521],[207.4075,-0.151],[209.5512,-0.1471],[209.95,-0.1458],[210.3512,-0.1433],[210.7738,-0.1418],[211.185,-0.1393],[212.3375,-0.1356],[212.72,-0.1355],[213.8888,-0.1319],[214.2862,-0.1297],[217.005,-0.1214],[218.1575,-0.1148],[218.5425,-0.1115],[218.945,-0.1092],[219.325,-0.1059],[219.7313,-0.1047],[220.1175,-0.1024],[221.2825,-0.096],[222.0738,-0.0939],[222.4525,-0.091],[222.8412,-0.0889],[223.2337,-0.0858],[223.635,-0.0817],[224.0487,-0.0785],[224.4475,-0.0744],[225.2287,-0.0724],[227.9987,-0.0713],[228.375,-0.0695],[228.765,-0.0666],[229.9975,-0.0554],[230.3963,-0.0536],[230.79,-0.0534],[231.5412,-0.0526],[232.2637,-0.0508],[232.6475,-0.0492],[233.5263,-0.0474],[237.4863,-0.0466],[237.8712,-0.044],[238.2663,-0.0405],[238.6625,-0.0361],[239.0675,-0.0326],[239.47,-0.0281],[240.285,-0.0262],[242.0362,-0.0273],[242.795,-0.0295],[243.1825,-0.0297],[243.9713,-0.0322],[244.37,-0.0358],[244.7613,-0.0407],[245.1675,-0.0471],[246.345,-0.0515],[248.4613,-0.053],[248.83,-0.0545],[249.62,-0.0603],[250.0125,-0.0618],[250.4062,-0.0647],[251.1313,-0.0659],[251.5087,-0.0707],[251.9162,-0.0786],[252.3225,-0.0914],[252.7475,-0.1015],[253.1625,-0.1096],[253.5812,-0.1145],[254.0088,-0.1163],[256.1,-0.1172],[256.8575,-0.1179],[257.9688,-0.1154],[258.3575,-0.1138],[259.1812,-0.1073],[259.6037,-0.1032],[260.0337,-0.0995],[260.4425,-0.0969],[260.8575,-0.0959],[266.5187,-0.0955],[267.3225,-0.095],[268.5263,-0.0952],[271.6575,-0.0987],[272.45,-0.1018],[272.8475,-0.1047],[274.0688,-0.1152],[274.4762,-0.1175],[274.8813,-0.1187],[275.275,-0.121],[276.49,-0.1263],[276.905,-0.1285],[277.3038,-0.1301],[277.7062,-0.1322],[278.105,-0.136],[278.5338,-0.1398],[278.95,-0.1411],[282.4087,-0.1418],[285.1237,-0.1415],[286.6775,-0.1404],[287.93,-0.1382],[289.1438,-0.1367],[289.5438,-0.1364],[289.935,-0.1366],[290.68,-0.1364],[291.07,-0.1366],[291.4725,-0.136],[292.6837,-0.1354],[295.4462,-0.1357],[295.8225,-0.1368],[296.6575,-0.1429],[297.11,-0.1449],[298.0625,-0.1466],[298.5438,-0.1472],[299.9787,-0.1475],[301.1575,-0.1503],[301.9575,-0.1512],[303.3625,-0.1516],[311.4688,-0.1523],[311.8675,-0.1541],[312.2613,-0.1574],[312.6788,-0.1599],[313.0938,-0.1615],[313.5,-0.1619],[319.37,-0.1632],[319.76,-0.1653],[320.1637,-0.1668],[327.3413,-0.1671],[328.5312,-0.1678],[331.26,-0.1683],[334.085,-0.1681],[334.8988,-0.1676],[338.3913,-0.1675],[340.4575,-0.1672],[341.6788,-0.1662],[342.0838,-0.165],[342.9237,-0.1641],[348.6175,-0.1636],[355.89,-0.164],[357.1137,-0.1655],[362.2287,-0.1658],[362.6888,-0.1667],[363.1175,-0.1682],[363.5587,-0.1703],[363.9813,-0.1732],[364.4375,-0.1742],[364.8463,-0.1745],[372.6225,-0.175],[375.0263,-0.1754],[375.4137,-0.1776],[376.2513,-0.1809],[376.6737,-0.1818],[377.5063,-0.1841],[378.3112,-0.1868],[379.1213,-0.1887],[379.5263,-0.1891],[379.92,-0.1902],[380.3388,-0.1921],[380.76,-0.1948],[381.2013,-0.197],[381.6225,-0.1986],[382.03,-0.1993],[384.4825,-0.1997],[387.5587,-0.1993],[388.6337,-0.1985],[389.0325,-0.1965],[389.4675,-0.1917],[389.9075,-0.1877],[390.355,-0.1864],[391.1875,-0.1867],[393.1475,-0.1854],[393.97,-0.1858],[397.1988,-0.1863],[397.5875,-0.1879],[398.4288,-0.1935],[398.85,-0.1953],[399.2763,-0.1965],[400.775,-0.1981],[401.18,-0.199],[402.3913,-0.2054],[402.7887,-0.2044],[403.195,-0.2056],[408.6475,-0.2066],[409.055,-0.2112],[409.475,-0.2168],[410.3587,-0.2296],[410.8038,-0.2356],[411.24,-0.2391],[411.6513,-0.2415],[415.4475,-0.2419],[423.4212,-0.2417],[425.5987,-0.2424],[426.0475,-0.2434],[426.555,-0.2461],[427.0813,-0.2485],[427.93,-0.2503],[429.9713,-0.2525],[431.2237,-0.2572],[432.0537,-0.2583],[437.0462,-0.2585],[438.2412,-0.2593],[438.6612,-0.2603],[439.5425,-0.264],[439.9775,-0.2652],[440.4137,-0.2657],[447.945,-0.2662],[448.7563,-0.2701],[449.1725,-0.2708],[450.325,-0.2739],[451.1562,-0.2801],[451.5738,-0.282],[452.02,-0.2831],[459.6625,-0.2856],[460.0763,-0.2904],[460.52,-0.2946],[460.9837,-0.2982],[461.415,-0.2994],[462.2237,-0.3],[467.1062,-0.2989],[468.6575,-0.2961],[469.0612,-0.2948],[470.3188,-0.2926],[471.1513,-0.289],[471.58,-0.2863],[472.0013,-0.285],[474.0375,-0.2819],[475.7175,-0.2762],[476.5625,-0.275],[478.0425,-0.2742],[478.8788,-0.2717],[479.3038,-0.2699],[479.7287,-0.2674],[480.5675,-0.2642],[481.01,-0.2638],[487.23,-0.2641],[489.445,-0.2629],[501.1225,-0.2625],[502.2663,-0.2616],[503.1263,-0.2588],[503.55,-0.2559],[503.9962,-0.2514],[504.42,-0.2483],[504.835,-0.2463],[505.645,-0.2453],[506.0487,-0.2442],[506.8125,-0.2431],[508.5612,-0.2426],[508.955,-0.2409],[509.7787,-0.2325],[510.24,-0.2284],[510.6625,-0.2242],[511.525,-0.217],[512.4075,-0.2046],[513.3,-0.1871],[513.76,-0.1753],[514.2475,-0.1665],[514.69,-0.1581],[515.1488,-0.1566],[516.2925,-0.1558],[517.0662,-0.1567],[517.46,-0.1576],[517.8763,-0.158],[518.28,-0.162],[518.7225,-0.1651],[519.155,-0.1691],[519.5825,-0.1739],[520.0175,-0.1795],[520.4737,-0.1842],[520.895,-0.1862],[521.7313,-0.188],[526.0575,-0.1889],[526.4563,-0.1898],[526.8687,-0.19],[528.0975,-0.1929],[529.525,-0.1938],[530.2613,-0.1948],[530.6525,-0.1958],[531.7487,-0.1959],[532.4812,-0.1969],[533.6413,-0.1997],[534.04,-0.2016],[534.8825,-0.2075],[535.3,-0.213],[535.7213,-0.2168],[536.1675,-0.2196],[536.5925,-0.2235],[537.4563,-0.2343],[537.9088,-0.2387],[538.7462,-0.2405],[541.5538,-0.2397],[542.2975,-0.2388],[542.6975,-0.2371],[543.125,-0.2363],[544.2387,-0.2355],[545.0587,-0.2339],[545.4688,-0.2324],[545.8988,-0.2292],[546.3325,-0.2284],[547.215,-0.2251],[547.6337,-0.2227],[548.0462,-0.221],[548.4737,-0.2202],[549.695,-0.2153],[550.5213,-0.2137],[551.6063,-0.2145],[552.0075,-0.2153],[552.415,-0.2177],[552.8425,-0.2223],[553.2637,-0.2276],[553.7088,-0.2313],[554.1787,-0.2365],[554.69,-0.2408],[555.1888,-0.2459],[555.6525,-0.2494],[556.0763,-0.2501],[564.1325,-0.2508],[564.8788,-0.2521],[565.2812,-0.2534],[565.6913,-0.2553],[566.1187,-0.2559],[566.525,-0.2572],[568.595,-0.2597],[575.4575,-0.2602],[576.7125,-0.2648],[578.7825,-0.2672],[583.865,-0.2663],[585.525,-0.2612],[589.85,-0.255],[592.2825,-0.2553],[592.6788,-0.2568],[593.5525,-0.2629],[594.025,-0.265],[594.4588,-0.266],[597.9987,-0.2666],[605.515,-0.2663],[606.3562,-0.2647],[606.8125,-0.2633],[607.7487,-0.2618],[612.5138,-0.2627],[613.7862,-0.2636],[617.0487,-0.2631],[617.9062,-0.2614],[618.865,-0.2608],[623.4175,-0.2595],[624.1625,-0.2583],[624.5825,-0.2571],[625.4425,-0.2535],[626.2763,-0.2523],[628.05,-0.2533],[628.455,-0.2602],[628.9212,-0.2752],[629.375,-0.2924],[629.8537,-0.297],[634.755,-0.2963],[640.2437,-0.2966],[640.6475,-0.2973],[644.1413,-0.298],[645.39,-0.3],[649.2463,-0.3],[650.0888,-0.2976],[651.09,-0.2932],[652.4837,-0.289],[652.9088,-0.2887],[657.3887,-0.289],[657.7925,-0.2897],[658.6363,-0.2941],[659.0713,-0.295],[662.0525,-0.2976],[676.3637,-0.2966],[681.8875,-0.2968],[682.8062,-0.2978],[685.5612,-0.3049],[686.425,-0.3058],[695.9475,-0.3064],[698.94,-0.3094],[701.5438,-0.31],[702.7925,-0.3125],[706.75,-0.3148],[711.5788,-0.3161],[713.735,-0.3197],[718.6437,-0.3216],[724.5712,-0.3212],[725.0063,-0.3207],[725.9112,-0.3182],[729.9625,-0.3151],[733.425,-0.3147],[736.3613,-0.3113],[738.3375,-0.3104],[743.2763,-0.3112],[751.115,-0.3114],[753,-0.3121]]
       }
       dyn_wings_metacarpal_02_r_ctrl.Rotation.Y {
        type=MovieSceneScriptingFloatChannel
        default=0
        thinned_from=776
        keys [[0,0.4908],[1.2212,0.4908],[6.6325,0.2088],[7.06,-1.1781],[7.4625,-2.4128],[7.8863,-3.2586],[8.2937,-4.1043],[8.7325,-4.3863],[9.1912,-4.6682],[9.6475,-4.7978],[28.54,-4.6455],[30.2663,-4.5159],[30.9787,-4.3636],[31.3388,-4.2113],[31.7075,-3.952],[32.0938,-3.6701],[32.8363,-3.5178],[33.55,-3.3882],[34.2637,-3.1062],[35.0125,-2.5424],[35.3925,-2.1308],[36.1225,-2.0012],[36.4863,-1.8489],[38.9025,-2.0012],[39.9375,-2.3058],[40.3425,-2.7174],[40.8038,-3.2813],[41.24,-3.8451],[41.7013,-3.9974],[42.0925,-4.2794],[42.8263,-4.5613],[47.865,-4.6909],[48.225,-4.8432],[52.5912,-5.2548],[52.9862,-5.5367],[53.755,-6.6644],[54.1388,-6.8167],[54.51,-7.0987],[54.8937,-7.2283],[56.6275,-7.5102],[58.1388,-8.6379],[58.8975,-9.461],[59.2713,-9.743],[59.9987,-9.8953],[61.3587,-10.0249],[62.0638,-10.1772],[62.43,-10.4591],[63.515,-10.5887],[80.1612,-10.7411],[80.5263,-11.198],[80.9112,-11.7846],[81.3025,-12.2415],[81.6875,-12.3938],[86.7475,-12.2415],[89.4563,-12.0892],[90.8175,-11.9369],[91.1862,-11.7846],[91.5812,-11.5026],[92.3487,-10.8934],[92.72,-10.7411],[93.1037,-10.2841],[93.5113,-9.8272],[93.9062,-9.0883],[94.3138,-8.6541],[94.7,-8.5017],[95.7575,-8.3721],[103.59,-8.2198],[104.3462,-7.6106],[104.75,-7.3059],[105.1988,-6.6967],[105.6575,-6.2397],[106.09,-6.0874],[108.1762,-5.9351],[108.5362,-5.6532],[108.9112,-5.2189],[109.2937,-5.0666],[109.6737,-4.762],[110.0537,-4.4801],[110.4388,-4.1754],[110.8112,-3.5889],[111.2125,-3.2842],[111.5925,-2.85],[111.985,-2.1111],[112.39,-1.3722],[112.7725,-0.7857],[113.1687,-0.3514],[113.56,-0.1991],[113.9312,-0.0468],[124.675,-0.1991],[126.0725,-0.3514],[126.45,-0.5038],[126.8625,-0.6561],[127.245,-0.938],[127.6375,-1.0903],[128.01,-1.2426],[128.3925,-1.3722],[128.7663,-1.5246],[129.1675,-1.8292],[129.545,-1.9588],[129.94,-2.5454],[130.3375,-2.6977],[130.7188,-2.85],[131.1,-3.1319],[131.8425,-3.2842],[132.2237,-3.5662],[132.615,-3.7185],[133.005,-4.0004],[133.3887,-4.305],[133.7775,-4.4573],[134.1513,-4.7393],[138.0962,-4.2823],[138.5,-3.5662],[138.8975,-3.2615],[139.6163,-3.1092],[150.29,-3.2615],[150.6788,-3.4365],[151.065,-3.8935],[151.4588,-4.3731],[151.8637,-4.5254],[156.305,-4.0912],[156.6938,-3.5046],[157.1125,-2.7658],[157.515,-2.1792],[157.8963,-1.5927],[158.3038,-1.3108],[159.0562,-1.0061],[159.7938,-0.8538],[160.1675,-0.4196],[160.555,0.0147],[160.9438,0.6012],[161.3313,0.8831],[161.7275,1.0127],[162.1,1.2947],[171.3713,1.447],[171.7513,1.7289],[172.1475,2.1858],[172.5512,2.4904],[172.9438,2.7951],[174.055,3.2293],[174.4275,3.5339],[175.1637,3.6862],[176.2312,3.8158],[177.2812,3.9681],[178.56,4.882],[179.3487,5.1866],[182.3675,5.3162],[182.7462,5.6208],[183.1313,5.7732],[183.5187,6.0778],[184.2962,6.9916],[184.6875,7.4258],[185.8438,8.3397],[186.6237,8.6443],[189.0188,8.9489],[189.3925,9.4058],[189.7887,9.7332],[190.1825,10.1901],[190.5812,10.4947],[190.9625,10.822],[191.3525,11.2789],[191.7425,11.6062],[192.5375,12.5655],[193.7275,13.5247],[194.12,13.9816],[194.5163,14.1567],[194.895,14.4613],[195.285,14.6363],[195.6775,14.9636],[196.075,15.4205],[196.475,15.7478],[197.6263,16.2048],[201.0763,16.0751],[203.6437,15.9228],[204.0225,15.7705],[204.4013,15.4659],[204.7838,14.8567],[205.195,14.3997],[205.5987,14.0951],[205.99,13.9428],[207.4075,13.8132],[209.5512,13.3563],[209.95,13.204],[210.3512,12.8994],[210.7738,12.7243],[211.185,12.4197],[212.3375,11.9628],[212.72,11.9401],[213.8888,11.4832],[214.2862,11.2013],[215.8637,10.592],[217.005,10.1351],[218.1575,9.2666],[218.5425,8.8324],[218.945,8.5278],[219.325,8.0935],[219.7313,7.9412],[220.1175,7.6366],[220.89,7.0728],[221.2825,6.7681],[222.0738,6.4635],[222.4525,6.0293],[222.8412,5.7247],[223.2337,5.2677],[223.635,4.6585],[224.0487,4.1789],[224.4475,3.5696],[224.8475,3.4173],[225.2287,3.265],[227.9987,3.09],[228.375,2.7854],[228.765,2.3057],[229.1725,1.6965],[229.5812,1.0645],[229.9975,0.4326],[230.3963,0.128],[230.79,0.1053],[231.5412,-0.047],[232.2637,-0.3517],[232.6475,-0.6336],[233.5263,-0.9382],[237.4863,-1.0905],[237.8712,-1.5475],[238.2663,-2.1567],[238.6625,-2.9183],[239.0675,-3.5048],[239.47,-4.2437],[240.285,-4.5483],[242.0362,-4.396],[242.795,-4.0914],[243.1825,-4.0687],[243.575,-3.9164],[243.9713,-3.7414],[244.37,-3.2844],[244.7613,-2.6752],[245.1675,-1.9364],[246.345,-1.4794],[248.4613,-1.3271],[248.83,-1.1748],[249.2212,-0.8929],[249.62,-0.5883],[250.0125,-0.436],[250.4062,-0.1314],[251.1313,0.0209],[251.5087,0.6302],[251.9162,1.6963],[252.3225,3.5241],[252.7475,5.0472],[253.1625,6.3111],[253.5812,7.118],[254.0088,7.4226],[256.1,7.575],[256.8575,7.7046],[257.9688,7.2249],[258.3575,6.8749],[258.7675,6.1133],[259.1812,5.3291],[259.6037,4.2629],[260.0337,3.2421],[288,-3.978],[289,-4.2055],[290,-5.2521],[292,-5.309],[293,-5.3573],[295,-5.3005],[296,-5.2267],[297,-4.9653],[298,-4.6071],[299,-4.3228],[300,-4.3143],[301,-4.2405],[304,-4.1836],[305,-5.6029],[306,-6.1546],[308,-6.7633],[308,-6.6011],[308,-6.4874],[309,-7.096],[309,-7.0391],[309,-6.9339],[311,-7.1529],[317,-7.2666],[318,-7.5425],[319,-8.6375],[324,-9.3115],[324,-8.9219],[325,-9.5874],[331,-9.6926],[332,-9.977],[333,-10.2045],[334,-10.3182],[335,-10.6594],[336,-10.9438],[338,-11.3019],[340,-11.3588],[342,-11.4726],[343,-11.7085],[344,-12.5785],[346,-12.7007],[347,-12.6439],[348,-11.5404],[349,-10.0956],[354,-9.5923],[355,-9.4217],[356,-8.8046],[357,-8.7562],[361,-8.1221],[362,-6.7633],[363,-6.7064],[369,-6.6495],[371,-6.5926],[372,-5.8533],[373,-5.552],[374,-4.9664],[375,-4.6082],[376,-3.7213],[377,-2.9964],[379,-2.9396],[382,-2.9964],[383,-3.3861],[384,-4.7739],[385,-4.734],[386,-4.8562],[387,-4.9131],[388,-4.8647],[390,-4.8078],[391,-4.0261],[392,-3.7248],[393,-3.6026],[394,-3.3183],[395,-3.1961],[400,-2.8464],[401,-1.6497],[402,-1.0071],[403,-0.8365],[406,-0.7797],[409,0.5852],[412,-0.6091],[413,-0.6574],[414,-0.7143],[425.5987,-0.9116],[426.0475,-1.2616],[426.555,-2.0913],[427.0813,-2.7687],[427.5113,-2.9664],[427.93,-3.2937],[428.3613,-3.446],[428.7575,-3.4687],[429.9713,-3.9256],[430.3713,-4.4053],[431.2237,-5.3419],[431.6363,-5.6465],[432.0537,-5.7988],[433.8175,-5.8215],[434.8937,-5.9738],[436.3012,-5.8215],[437.4388,-5.3646],[437.8438,-4.9076],[438.2412,-4.2984],[438.6612,-3.1252],[439.0987,-1.0155],[439.5425,0.4849],[439.9775,1.2465],[440.4137,1.5511],[447.945,1.7034],[448.3363,2.1604],[448.7563,2.7469],[449.1725,2.8992],[449.5888,3.2039],[450.325,3.6608],[450.7237,4.3997],[451.1562,5.1385],[451.5738,5.5955],[452.02,5.8774],[454.7875,6.0297],[459.6625,6.4867],[460.0763,7.7052],[460.52,8.7714],[460.9837,9.6625],[461.415,9.9672],[462.2237,10.1195],[465.6537,9.9672],[467.1062,9.8149],[467.5025,9.5102],[468.2563,9.076],[468.6575,8.9237],[469.0612,8.4894],[469.5,8.3371],[470.3188,7.7506],[471.1513,6.5547],[471.58,5.6636],[472.0013,5.2066],[472.8288,4.902],[473.2412,4.5974],[473.6413,4.445],[474.0375,4.1404],[474.4475,3.6608],[474.8712,3.2038],[475.2812,2.5719],[475.7175,1.9626],[476.1338,1.658],[476.5625,1.5057],[477.6525,1.3534],[478.0425,1.1784],[478.46,0.7214],[478.8788,0.0895],[479.3038,-0.6948],[479.7287,-1.761],[480.1513,-2.5226],[480.5675,-3.1318],[481.01,-3.2842],[487.23,-3.1545],[487.9675,-3.3069],[488.3562,-3.4592],[489.445,-3.6115],[501.1225,-3.7411],[501.8637,-3.8707],[502.2663,-4.0003],[502.6913,-4.4119],[503.1263,-4.8007],[503.55,-5.6011],[503.9962,-6.8357],[504.42,-7.6588],[504.835,-8.2],[505.645,-8.4592],[506.0487,-8.7411],[506.8125,-9.0004],[508.5612,-9.13],[508.955,-9.5415],[509.36,-10.4942],[509.7787,-11.5765],[510.24,-12.5292],[510.6625,-13.5046],[511.525,-15.1508],[511.9488,-16.515],[512.4075,-17.8793],[513.3,-21.5605],[513.76,-23.9001],[514.69,-27.1082],[515.1488,-27.3901],[516.2925,-27.5198],[517.0662,-27.3674],[517.46,-27.2151],[517.8763,-27.147],[518.28,-26.5151],[518.7225,-26.0128],[519.155,-25.3581],[519.5825,-24.5738],[520.0175,-23.6373],[520.4737,-22.853],[520.895,-22.503],[521.7313,-22.1984],[526.0575,-22.0461],[526.4563,-21.8937],[526.8687,-21.871],[528.0975,-21.4141],[529.525,-21.2618],[530.2613,-21.1095],[530.6525,-20.9572],[531.7487,-20.9345],[532.4812,-20.7822],[533.6413,-20.3252],[534.04,-20.0206],[534.8825,-19.0613],[535.3,-18.1475],[535.7213,-17.5155],[536.1675,-17.0359],[536.5925,-16.3812],[537.0325,-15.4674],[537.4563,-14.5081],[537.9088,-13.7238],[538.32,-13.5488],[538.7462,-13.3965],[541.5538,-13.5488],[542.2975,-13.7011],[542.6975,-14.0058],[543.125,-14.1581],[544.2387,-14.3104],[545.0587,-14.615],[545.4688,-14.8969],[545.8988,-15.4835],[546.3325,-15.6358],[547.215,-16.2451],[547.6337,-16.6793],[548.0462,-17.0066],[548.4737,-17.1589],[549.695,-18.0728],[550.5213,-18.3774],[551.6063,-18.2251],[552.0075,-18.0728],[552.415,-17.6159],[552.8425,-16.7247],[553.2637,-15.6812],[553.7088,-14.9423],[554.1787,-13.8988],[554.69,-13.0077],[555.1888,-11.9642],[555.6525,-11.2253],[556.0763,-11.073],[564.1325,-10.9207],[564.8788,-10.6161],[565.2812,-10.3114],[565.6913,-9.8545],[566.1187,-9.7022],[566.525,-9.3975],[567.32,-9.2452],[568.595,-8.7883],[575.4575,-8.636],[576.2763,-7.7221],[576.7125,-7.2652],[578.7825,-6.5036],[583.865,-6.8082],[584.2637,-7.1128],[585.1025,-8.0267],[585.525,-8.461],[585.9575,-8.6133],[586.9,-9.2225],[587.3725,-9.5045],[589.0213,-10.1137],[589.4412,-10.4184],[589.85,-10.5707],[592.2825,-10.4184],[592.6788,-9.7864],[593.1175,-8.5452],[593.5525,-7.129],[594.025,-6.1924],[594.4588,-5.7127],[594.8837,-5.5604],[597.9987,-5.4308],[605.515,-5.865],[605.9162,-6.8859],[606.3562,-8.3636],[606.8125,-10.6029],[607.31,-12.1034],[607.7487,-13.1696],[608.185,-13.3446],[608.6012,-13.3673],[610.0387,-13.215],[610.4362,-12.9103],[610.8575,-12.758],[611.2613,-12.4534],[611.6712,-12.3011],[612.085,-11.7145],[612.5138,-11.2803],[612.93,-10.5414],[613.3663,-9.8025],[613.7862,-9.1933],[614.2163,-8.8886],[614.6275,-8.4317],[615.3887,-8.2794],[616.2038,-7.9748],[616.6187,-7.5178],[617.0487,-7.2132],[617.47,-6.6039],[617.9062,-6.147],[618.3763,-5.972],[618.865,-5.8197],[622.0188,-5.6673],[623.4175,-5.515],[624.1625,-5.2104],[624.5825,-4.9058],[625.0213,-4.4488],[625.4425,-3.9919],[626.2763,-3.6873],[628.05,-3.9692],[628.455,-5.8973],[628.9212,-10.0808],[629.375,-15.0032],[629.8537,-16.4129],[634.755,-16.131],[640.2437,-16.2606],[640.6475,-16.5425],[642.7237,-16.6721],[644.1413,-16.954],[644.5525,-17.236],[644.975,-17.7998],[645.39,-18.0818],[646.4987,-18.2341],[649.2463,-18.0818],[649.6537,-17.4952],[650.0888,-16.7563],[650.5775,-15.4536],[651.09,-14.2351],[651.5863,-13.4735],[652.0375,-12.5596],[652.4837,-11.9731],[652.9088,-11.8208],[657.3887,-11.9731],[657.7925,-12.2777],[658.21,-13.1689],[658.6363,-14.1897],[659.0713,-14.6466],[659.4875,-14.7989],[659.9162,-15.3855],[660.3375,-15.6901],[660.7838,-16.1244],[661.215,-16.2767],[661.6388,-16.5586],[662.0525,-16.7109],[667.18,-16.8633],[667.5875,-17.0156],[669.0925,-17.3202],[669.92,-17.6248],[670.6788,-18.0591],[671.0825,-18.3637],[671.5063,-18.8206],[671.9288,-19.1253],[672.3462,-19.2776],[675.1112,-19.5595],[675.9388,-20.1688],[676.3637,-20.6257],[676.8,-20.778],[682.3125,-20.603],[682.8062,-19.9938],[683.3238,-19.0799],[683.8012,-18.2956],[684.2663,-17.6863],[684.6975,-16.9475],[685.5612,-15.1197],[685.9937,-14.6628],[686.425,-14.5104],[695.5475,-14.3581],[695.9475,-14.2058],[696.3625,-13.9012],[696.795,-13.7489],[697.2188,-13.7262],[697.6888,-13.2692],[698.94,-12.7896],[701.1388,-12.6146],[701.5438,-12.4623],[701.9638,-12.1576],[702.3763,-11.7007],[702.7925,-11.3734],[703.2275,-11.221],[703.9813,-11.0687],[705.0487,-10.7641],[705.4675,-10.7414],[705.8837,-10.2845],[706.3125,-10.1094],[706.75,-9.9571],[709.2738,-9.7821],[710.0362,-9.4775],[710.8275,-9.3252],[711.5788,-9.0205],[711.9875,-8.5636],[712.4288,-7.9543],[712.86,-7.0632],[713.2988,-6.4766],[713.735,-6.0197],[714.1812,-5.715],[714.6862,-5.4104],[715.19,-4.9761],[715.6712,-4.5419],[716.11,-4.2373],[716.5338,-3.9326],[716.975,-3.6507],[718.6437,-3.0415],[723.4112,0],[753,0]]
       }
       dyn_wings_metacarpal_02_r_ctrl.Rotation.Z {
        type=MovieSceneScriptingFloatChannel
        default=0
        thinned_from=618
        keys [[0,0],[56.9975,0.0002],[80.1612,0.0001],[81.3025,-0.0011],[81.6875,-0.0012],[90.8175,-0.0008],[92.72,0.0004],[93.5113,0.0012],[93.9062,0.0018],[94.3138,0.0022],[94.7,0.0023],[103.59,0.0025],[105.6575,0.0032],[112.7725,0.0035],[127.245,0.0034],[133.7775,0.0026],[139.6163,0.003],[150.29,0.0028],[150.6788,0.0026],[151.4588,0.0013],[151.8637,0.0011],[156.305,0.0018],[157.1125,0.0039],[157.8963,0.0056],[159.0562,0.0065],[159.7938,0.0067],[160.555,0.0078],[160.9438,0.0086],[161.3313,0.009],[161.7275,0.0092],[162.1,0.0096],[171.3713,0.0099],[171.7513,0.0104],[172.1475,0.0112],[172.9438,0.0123],[174.055,0.0131],[174.4275,0.0136],[177.2812,0.0143],[178.56,0.0159],[179.3487,0.0164],[182.3675,0.0167],[182.7462,0.0173],[183.1313,0.0176],[183.5187,0.0182],[184.6875,0.021],[185.8438,0.0227],[186.6237,0.0232],[189.0188,0.0238],[189.3925,0.0247],[189.7887,0.0252],[190.1825,0.026],[190.9625,0.0271],[191.3525,0.0278],[192.5375,0.0296],[194.12,0.0312],[196.475,0.0327],[197.2425,0.0328],[204.4013,0.0326],[205.5987,0.0318],[209.95,0.0312],[212.3375,0.0302],[213.8888,0.0298],[217.005,0.0287],[218.1575,0.0279],[219.325,0.0266],[221.2825,0.025],[222.0738,0.0247],[223.2337,0.0234],[224.4475,0.0214],[225.2287,0.021],[227.9987,0.0208],[228.765,0.02],[229.9975,0.0177],[230.3963,0.0173],[231.5412,0.0171],[232.2637,0.0167],[232.6475,0.0164],[233.5263,0.0159],[237.4863,0.0158],[237.8712,0.0152],[238.2663,0.0143],[238.6625,0.0132],[239.0675,0.0123],[239.47,0.011],[240.285,0.0104],[242.0362,0.0107],[242.795,0.0115],[243.575,0.012],[243.9713,0.0124],[244.37,0.0137],[244.7613,0.0155],[245.1675,0.0178],[246.345,0.0195],[248.4613,0.0202],[248.83,0.0208],[249.62,0.0232],[250.0125,0.0238],[250.4062,0.0249],[251.1313,0.0253],[251.5087,0.0268],[251.9162,0.0289],[252.3225,0.0319],[252.7475,0.0338],[253.1625,0.0351],[253.5812,0.0358],[254.0088,0.036],[256.8575,0.0361],[258.3575,0.0358],[260.4425,0.0343],[260.8575,0.0342],[270.4987,0.0341],[273.2537,0.0336],[276.0813,0.0338],[277.7062,0.0344],[278.5338,0.0354],[278.95,0.0356],[282.4087,0.0357],[287.1137,0.0353],[289.5438,0.0348],[292.2912,0.0347],[295.8225,0.0348],[296.6575,0.036],[297.11,0.0364],[298.5438,0.037],[300.3587,0.0372],[301.5638,0.0379],[303.3625,0.0381],[311.4688,0.0383],[311.8675,0.0388],[312.2613,0.0398],[312.6788,0.0406],[313.0938,0.0411],[319.37,0.0417],[319.76,0.0425],[320.1637,0.0431],[327.3413,0.0432],[328.925,0.0436],[330.4162,0.0437],[334.085,0.0436],[334.8988,0.0433],[340.4575,0.0431],[341.6788,0.0425],[342.0838,0.0418],[342.9237,0.0413],[347.7387,0.0411],[355.89,0.0413],[357.1137,0.042],[362.2287,0.0422],[362.6888,0.0426],[363.1175,0.0433],[363.5587,0.0442],[363.9813,0.0455],[364.4375,0.0459],[375.0263,0.0464],[375.4137,0.0473],[376.2513,0.0487],[377.5063,0.0499],[378.3112,0.0508],[379.1213,0.0515],[379.5263,0.0516],[380.3388,0.0526],[381.2013,0.0541],[381.6225,0.0546],[382.03,0.0548],[388.6337,0.0545],[389.0325,0.054],[389.9075,0.0511],[390.355,0.0507],[391.1875,0.0508],[393.1475,0.0503],[397.1988,0.0506],[397.5875,0.0512],[398.4288,0.0532],[398.85,0.0539],[399.2763,0.0543],[400.775,0.0549],[401.18,0.0553],[401.99,0.0571],[402.3913,0.0582],[402.7887,0.0577],[403.195,0.0582],[408.6475,0.0586],[409.475,0.0626],[410.3587,0.0671],[410.8038,0.0691],[411.24,0.0702],[411.6513,0.0709],[430.3713,0.071],[431.2237,0.0711],[437.0462,0.0713],[438.2412,0.0715],[439.5425,0.0724],[439.9775,0.0726],[447.945,0.0728],[448.7563,0.0735],[450.325,0.0742],[451.1562,0.0752],[452.02,0.0756],[459.6625,0.0759],[460.52,0.0768],[460.9837,0.0771],[462.2237,0.0772],[470.3188,0.0768],[472.0013,0.0761],[474.8712,0.0754],[476.1338,0.0748],[477.6525,0.0747],[478.8788,0.0742],[480.5675,0.0726],[489.445,0.0723],[501.1225,0.0722],[502.2663,0.0719],[503.1263,0.0711],[503.55,0.0702],[503.9962,0.0688],[504.42,0.0677],[504.835,0.067],[505.645,0.0667],[506.0487,0.0663],[506.8125,0.0659],[508.5612,0.0657],[508.955,0.065],[509.36,0.0635],[509.7787,0.0617],[510.24,0.0599],[510.6625,0.0581],[511.525,0.0548],[512.4075,0.0488],[512.84,0.0445],[513.3,0.0395],[513.76,0.0327],[514.2475,0.0275],[514.69,0.0223],[515.1488,0.0213],[516.2925,0.0209],[517.0662,0.0214],[517.46,0.022],[517.8763,0.0223],[518.28,0.0248],[518.7225,0.0268],[519.155,0.0293],[519.5825,0.0323],[520.0175,0.0358],[520.4737,0.0386],[520.895,0.0399],[521.7313,0.0409],[526.0575,0.0415],[526.4563,0.042],[527.2688,0.0427],[528.0975,0.0439],[529.525,0.0444],[530.2613,0.045],[530.6525,0.0456],[532.4812,0.0463],[533.6413,0.048],[534.04,0.0491],[534.8825,0.0525],[535.3,0.0556],[535.7213,0.0578],[536.1675,0.0593],[536.5925,0.0614],[537.4563,0.0671],[537.9088,0.0694],[538.7462,0.0703],[541.5538,0.0699],[542.2975,0.0694],[542.6975,0.0686],[543.125,0.0682],[544.2387,0.0678],[545.0587,0.067],[545.4688,0.0662],[545.8988,0.0646],[546.3325,0.0641],[547.215,0.0624],[547.6337,0.0612],[548.0462,0.0603],[548.4737,0.0598],[549.695,0.0572],[550.5213,0.0563],[551.6063,0.0567],[552.0075,0.0572],[552.415,0.0585],[552.8425,0.0609],[553.2637,0.0636],[553.7088,0.0655],[554.1787,0.068],[554.69,0.0701],[555.1888,0.0725],[555.6525,0.074],[556.0763,0.0743],[564.1325,0.0746],[564.8788,0.0752],[565.2812,0.0757],[565.6913,0.0765],[566.1187,0.0768],[566.525,0.0773],[568.595,0.0783],[575.4575,0.0785],[576.7125,0.0801],[578.7825,0.0809],[583.865,0.0806],[585.525,0.0789],[589.85,0.0766],[592.2825,0.0767],[592.6788,0.0772],[593.5525,0.0793],[594.025,0.0799],[594.4588,0.0802],[597.9987,0.0804],[605.9162,0.0801],[607.7487,0.0786],[612.085,0.079],[614.2163,0.0795],[625.4425,0.079],[628.455,0.0793],[628.9212,0.0807],[629.375,0.0838],[629.8537,0.085],[634.755,0.0848],[644.1413,0.0852],[645.39,0.0858],[649.6537,0.0855],[651.09,0.084],[652.4837,0.083],[657.7925,0.0832],[658.6363,0.0841],[660.7838,0.0848],[661.6388,0.085],[675.1112,0.0846],[676.3637,0.0844],[682.3125,0.0845],[684.6975,0.0872],[685.5612,0.0884],[685.9937,0.0887],[695.5475,0.089],[698.94,0.0904],[701.5438,0.0907],[702.7925,0.0917],[706.75,0.0926],[711.5788,0.0931],[713.2988,0.0942],[716.975,0.0949],[718.6437,0.0951],[724.5712,0.0949],[725.0063,0.0948],[725.9112,0.094],[727.3525,0.0936],[728.6775,0.0934],[729.9625,0.093],[733.425,0.0928],[736.3613,0.0915],[738.3375,0.0912],[742.105,0.0913],[743.2763,0.0915],[751.5338,0.0917],[753,0.091]]
       }
       dyn_wings_metacarpal_02_r_ctrl.Scale.X {type=MovieSceneScriptingFloatChannel, default=1, const [1,2]}
       dyn_wings_metacarpal_02_r_ctrl.Scale.Y {type=MovieSceneScriptingFloatChannel, default=1, const [1,2]}
       dyn_wings_metacarpal_02_r_ctrl.Scale.Z {type=MovieSceneScriptingFloatChannel, default=1, const [1,2]}
       dyn_wings_phalange_02_r_ctrl.Location.X {type=MovieSceneScriptingFloatChannel, default=0, const [0,2]}
       dyn_wings_phalange_02_r_ctrl.Location.Y {type=MovieSceneScriptingFloatChannel, default=0, const [0,2]}
       dyn_wings_phalange_02_r_ctrl.Location.Z {type=MovieSceneScriptingFloatChannel, default=0, const [0,2]}
       dyn_wings_phalange_02_r_ctrl.Rotation.X {type=MovieSceneScriptingFloatChannel, default=0, const [0,2]}
       dyn_wings_phalange_02_r_ctrl.Rotation.Y {
        type=MovieSceneScriptingFloatChannel
        default=-4.4399
        thinned_from=672
        keys [[1,-4.5754],[12.5238,-4.7225],[12.9,-5.1406],[13.295,-5.7058],[14.5125,-6.5536],[14.9,-6.8362],[15.3012,-6.9717],[16.0738,-7.5369],[22.6812,-7.6724],[24.4362,-7.8196],[24.8188,-7.955],[25.21,-8.2376],[25.965,-8.5203],[28.7162,-8.6557],[42.3825,-8.7912],[42.7425,-8.9384],[43.125,-9.0738],[51.8575,-8.7912],[52.6025,-8.226],[52.9962,-8.0789],[53.7563,-7.5137],[54.1363,-7.3666],[54.5175,-7.084],[55.6237,-6.6659],[58.0163,-6.5304],[58.3825,-6.2478],[59.1225,-6.1007],[59.8363,-5.8181],[61.55,-5.6826],[62.2475,-5.5355],[62.64,-5.4],[63.365,-5.1174],[66.4275,-4.9819],[67.4837,-4.8348],[68.1925,-4.6993],[68.5663,-4.5522],[68.9562,-4.2696],[72.2675,-3.987],[73.0562,-3.7044],[73.4338,-3.4218],[74.1612,-3.2746],[75.8663,-3.1392],[76.24,-2.992],[76.6212,-2.8566],[77.3487,-2.7094],[81.0638,-2.5739],[81.4312,-2.4268],[82.1662,-2.2913],[82.9013,-2.1442],[83.28,-2.0087],[84.0125,-1.8616],[84.3813,-1.7261],[87.76,-1.579],[88.8225,-1.2964],[89.1963,-1.1609],[89.5675,-0.8783],[90.2875,-0.7312],[106.0975,-0.8551],[106.4775,-1.5209],[106.8725,-2.5933],[107.2825,-3.5301],[107.6888,-4.4785],[108.085,-5.0205],[108.47,-5.2915],[109.6,-5.698],[110.6463,-5.8335],[111.4038,-6.1045],[112.13,-6.2516],[112.5,-6.3871],[113.2437,-6.5226],[122.275,-6.3871],[123.0025,-6.1045],[123.3775,-5.8219],[123.7587,-5.4038],[124.1425,-4.7031],[124.5425,-4.1495],[124.9375,-3.5843],[125.3375,-3.3017],[126.0738,-3.0191],[127.7938,-2.8836],[128.5175,-2.7365],[129.2325,-2.601],[129.6037,-2.4538],[129.9875,-2.3184],[130.3713,-2.1712],[130.755,-2.0358],[133.9662,-1.9003],[134.34,-1.7531],[135.125,-1.6177],[135.54,-1.3351],[136.2675,-1.1879],[136.6425,-1.0524],[137.3575,-0.7698],[137.7275,-0.6344],[138.4512,-0.4872],[139.5113,-0.3517],[145.8888,-0.2163],[146.265,0.0664],[146.655,0.2018],[147.0312,0.349],[147.7475,0.4844],[148.4737,0.6199],[149.865,0.7554],[153.6163,0.6199],[169.6488,0.3373],[170.02,0.2018],[170.4013,-0.0691],[171.13,-0.2046],[187.9212,-0.6227],[188.2912,-1.1763],[188.6913,-2.1364],[189.4937,-4.0798],[189.895,-4.7689],[190.29,-5.1869],[190.6775,-5.3224],[192.4125,-5.4579],[192.7775,-5.7289],[193.1687,-6.2825],[193.5688,-6.8245],[193.955,-7.0955],[194.3363,-7.3781],[194.7213,-7.5136],[199.0312,-7.6607],[199.7575,-8.4969],[200.1475,-9.7512],[200.5675,-11.7062],[200.9937,-13.514],[201.4137,-14.6212],[201.8212,-14.9038],[202.9275,-14.7683],[203.3138,-14.4973],[203.7075,-13.9438],[204.1075,-13.2431],[204.5025,-12.825],[208.2862,-12.6895],[211.3625,-12.1359],[211.7663,-10.7461],[212.1862,-9.4918],[212.6125,-8.373],[213.025,-8.2375],[228.5038,-8.102],[231.3837,-7.8194],[231.7613,-7.5368],[232.545,-6.9832],[237.3975,-7.5368],[237.7738,-8.2259],[238.1725,-9.2092],[238.5725,-10.4635],[238.9737,-11.4352],[239.37,-11.9888],[240.4388,-12.1243],[254.2025,-11.9888],[254.9225,-11.5707],[255.7575,-11.2881],[263.8438,-11.1526],[264.2163,-11.0171],[264.6012,-10.7461],[264.9962,-10.6106],[265.3887,-10.328],[265.8012,-9.9215],[266.1975,-9.6389],[266.5963,-9.5034],[266.97,-9.2208],[267.73,-8.9498],[268.46,-8.8143],[270.8537,-8.6788],[271.2275,-8.3962],[271.6187,-8.2607],[271.9962,-7.9781],[272.3775,-7.8426],[274.8863,-7.4245],[275.2663,-7.289],[275.6488,-7.0064],[276.0412,-6.8709],[276.4212,-6.5883],[277.1838,-6.3174],[277.5788,-6.0348],[278.3138,-5.8876],[279.0763,-5.6167],[279.7975,-5.3341],[280.5213,-5.1986],[281.28,-4.916],[281.6762,-4.7805],[282.0688,-4.4979],[284.185,-4.2269],[284.8963,-4.0798],[285.2713,-3.9443],[285.6637,-3.6617],[286.0537,-3.5262],[287.115,-3.3907],[289.1975,-3.2436],[289.5838,-3.1081],[289.9688,-2.8255],[290.355,-2.69],[290.75,-2.5429],[291.5538,-2.4074],[291.9575,-2.2603],[292.7113,-2.1248],[293.5912,-1.8538],[294,-1.7067],[294.7638,-1.4241],[295.8138,-1.2886],[296.8675,-1.006],[297.2575,-0.8705],[297.9837,-0.5879],[299.3675,-0.3053],[306.11,-0.5763],[306.4825,-1.1066],[306.8837,-2.1789],[307.3012,-3.9287],[307.7325,-7.0567],[308.1625,-11.5628],[308.5912,-15.6741],[309.0238,-17.3116],[310.1025,-17.1761],[310.4725,-17.0406],[310.8512,-16.487],[311.265,-15.5037],[311.6762,-14.7914],[312.4538,-14.5088],[326.56,-14.6559],[330.2513,-14.5204],[330.6263,-14.1139],[331.025,-13.843],[331.7638,-13.7075],[335.8375,-13.572],[336.8875,-13.4365],[337.2725,-13.1539],[337.6588,-13.0184],[338.0412,-12.7474],[338.8212,-12.4764],[339.2013,-12.1938],[340.62,-12.0583],[342.7313,-11.7757],[343.4563,-11.6402],[343.8375,-11.3576],[344.2337,-11.2221],[346.3825,-10.9511],[347.11,-10.8156],[348.4987,-10.6801],[350.5525,-10.533],[350.925,-10.3975],[351.6675,-10.262],[352.7313,-10.1266],[353.47,-9.9911],[355.2962,-9.8556],[356.04,-9.7201],[356.4525,-9.5846],[356.8875,-9.302],[357.6788,-9.1665],[359.7375,-9.031],[360.115,-8.7484],[360.505,-8.6129],[361.9037,-8.4774],[362.65,-8.3419],[363.0188,-8.2064],[364.8138,-8.0709],[365.5387,-7.9354],[365.9137,-7.6528],[366.3038,-7.5173],[367.3725,-7.3818],[368.7525,-7.0992],[369.5163,-6.8282],[369.9,-6.5456],[371.4575,-6.0037],[372.1838,-5.8682],[372.5575,-5.5972],[372.9413,-5.4617],[373.325,-5.0436],[374.395,-4.9081],[377.1075,-4.7726],[378.485,-4.6371],[378.8712,-4.5016],[380.94,-4.3661],[381.985,-4.2306],[383.1587,-3.8242],[384.5638,-3.5416],[386.0113,-3.4061],[386.8087,-3.2706],[387.9288,-3.1351],[390.0838,-2.9996],[390.8,-2.8525],[391.1762,-2.717],[391.9288,-2.5815],[394.6588,-2.446],[395.4325,-2.1634],[395.835,-2.0279],[396.2212,-1.8924],[399.2887,-1.7569],[400.0175,-1.6214],[400.7275,-1.4859],[401.1137,-1.3504],[402.1775,-1.2149],[402.9013,-1.0794],[403.9525,-0.9439],[405.3388,-0.8085],[407.4963,-0.5259],[408.2412,-0.3904],[408.6263,-0.1194],[409.035,0.0161],[409.4212,0.1632],[409.8125,0.2987],[410.8813,0.4342],[411.2675,0.5697],[412.0063,0.9878],[412.3875,1.2588],[412.795,1.3943],[413.21,1.8124],[413.61,1.9479],[416.0525,1.8124],[416.43,1.5414],[416.8312,1.1233],[417.2387,0.5697],[417.6862,0.0161],[418.1338,-0.3904],[418.5575,-0.8085],[418.96,-1.0794],[419.4087,-1.4859],[419.8625,-1.7569],[420.295,-2.0279],[420.7013,-2.2989],[421.1112,-2.4344],[421.5163,-2.7054],[422.3075,-2.9763],[422.7162,-3.3828],[423.1162,-3.9364],[423.5187,-4.2074],[423.9175,-4.6139],[424.3138,-5.032],[424.7062,-5.303],[425.105,-5.7211],[425.5063,-5.8566],[426.305,-6.4101],[426.6988,-6.5456],[428.4262,-6.6811],[428.8012,-6.9637],[429.1938,-7.2347],[429.5888,-7.3702],[430.35,-7.6528],[432.43,-7.3702],[432.81,-7.2347],[433.2088,-6.263],[433.62,-4.8732],[434.045,-3.7661],[434.4638,-2.0937],[434.8775,-0.5684],[435.2962,0.1323],[435.7138,0.6859],[436.11,1.104],[437.5063,1.2395],[439.2487,1.6576],[439.6388,1.7931],[440.0362,2.0641],[440.4238,2.1995],[440.81,2.4705],[441.2088,2.8886],[441.9987,3.1596],[442.3887,3.4306],[443.2,3.9726],[443.6187,4.2436],[444.0213,4.65],[444.4288,4.921],[445.64,6.1521],[446.4325,6.6941],[447.2262,6.965],[449.3863,6.8295],[449.8375,6.1288],[450.2713,5.2927],[450.6775,4.6036],[451.1075,3.9145],[451.5612,3.3725],[451.9975,2.695],[452.4162,2.1531],[452.8325,1.7466],[453.2337,1.4756],[453.6275,1.0575],[454.02,0.922],[454.405,0.7865],[454.79,0.38],[455.1888,0.2446],[455.59,-0.038],[455.98,-0.1735],[456.7162,-0.309],[457.455,-0.58],[457.8475,-0.5684],[458.24,-0.8394],[458.645,-1.1104],[459.4187,-1.3814],[459.81,-1.664],[461.9462,-1.935],[462.3188,-2.0704],[467.065,-1.935],[467.4438,-1.7995],[467.8288,-1.6523],[468.2188,-1.5169],[468.6163,-1.2343],[469.0038,-0.9633],[469.3963,-0.8278],[469.7875,-0.5452],[470.175,-0.4097],[470.5663,0.0084],[470.9638,0.1439],[471.3575,0.2794],[473.145,0.4149],[473.885,0.562],[474.2688,0.8446],[474.68,0.9801],[475.0713,1.2511],[475.4712,1.5337],[476.2725,1.8047],[477.025,1.9402],[477.7475,2.0873],[478.1338,2.2228],[478.5312,2.3699],[479.325,2.6409],[480.06,2.7764],[486.3687,2.6409],[488.1075,2.5054],[489.2738,2.0989],[489.675,1.9634],[490.0612,1.8279],[490.79,1.6808],[492.1737,1.5453],[492.5562,1.4098],[492.9425,1.5569],[493.3338,1.4214],[493.7287,1.1388],[494.4712,1.0033],[495.91,0.7324],[496.3212,0.5969],[497.0688,0.4614],[499.8188,0.3259],[500.5562,0.1904],[500.9575,0.0433],[501.35,-0.0922],[502.78,-0.2277],[503.5275,-0.3632],[503.9137,-0.4987],[504.305,-0.6342],[505.7462,-0.7697],[506.8012,-0.9052],[508.9813,-1.3117],[509.3675,-1.4471],[510.1012,-1.5826],[510.8138,-1.7181],[511.2063,-1.8536],[512.3,-1.9891],[519.265,-1.8536],[519.6475,-1.571],[520.0512,-1.2884],[520.4475,-1.1529],[520.8338,-0.8819],[521.58,-0.7464],[523.3075,-0.4638],[523.6925,-0.3283],[524.76,-0.1929],[529.145,-0.0574],[530.855,0.0897],[531.2362,0.2252],[531.6375,0.3607],[532.0288,0.5078],[532.4187,0.6433],[536.1888,0.7788],[538.5825,0.9259],[539.98,1.0614],[540.3512,1.2085],[541.1438,1.4795],[542.6537,1.344],[543.0362,1.4795],[549.4425,1.615],[549.8288,1.8976],[550.2212,2.1686],[550.6237,2.4512],[551.415,2.7222],[551.8062,2.8693],[552.2175,3.0048],[552.6437,3.2874],[554.1187,3.5584],[554.8937,3.8294],[555.2925,3.9765],[555.6888,4.112],[556.0763,4.3946],[557.1513,4.5301],[557.8712,4.6656],[558.6525,4.8011],[559.875,5.2192],[564.6475,5.3547],[565.415,5.6257],[577.7938,5.8966],[578.2212,6.3031],[581.6988,6.5857],[585.4312,6.7212],[585.8163,6.9922],[586.6112,7.2632],[587.015,7.5342],[587.4175,7.9523],[588.2075,8.2233],[598.6837,8.3587],[599.4125,8.4942],[600.6,8.9007],[600.9937,9.1717],[601.39,9.3072],[602.475,9.4427],[604.5625,9.5782],[608.4338,9.7137],[609.2025,9.9847],[609.6388,10.1202],[615.085,10.2557],[616.2513,10.6621],[616.65,10.9331],[617.0537,11.328],[617.4588,11.87],[617.865,12.2764],[627.9887,12.4119],[629.5825,12.9423],[630.33,12.9307],[632.4175,13.0662],[632.8225,13.2017],[633.2212,13.4726],[633.6375,13.6081],[634.0475,13.5965],[634.8512,14.1385],[635.2563,14.1269],[636.4525,14.5333],[637.525,14.5217],[638.2613,14.6572],[638.66,14.5101],[640.5063,14.3746],[640.935,13.8326],[641.4187,12.9307],[641.8975,12.199],[642.3625,11.1848],[642.8087,10.1589],[643.2763,9.6518],[643.7088,9.6634],[644.11,9.5279],[645.54,9.3925],[648.6237,9.5279],[652.34,9.3925],[654.7613,8.8621],[655.1637,8.1846],[655.5838,7.7898],[660.0962,7.5188],[660.4887,7.3833],[660.8925,6.7058],[661.3038,6.04],[661.7262,5.3857],[662.1513,4.8554],[662.5825,4.5844],[663.0013,4.4489],[680.3313,4.5844],[681.3975,4.7199],[681.7838,5.138],[682.195,5.2735],[682.5963,5.5445],[684.045,5.68],[684.7812,5.8155],[686.5387,5.951],[691.3313,6.7639],[691.7462,8.6143],[692.1975,12.3152],[692.6513,15.9928],[693.1125,17.3245],[694.225,17.189],[694.6212,16.7942],[695.4538,15.4625],[695.8712,14.9089],[696.2837,14.6263],[698.6962,14.4908],[711.4425,14.6263],[712.2025,14.7618],[715.6375,14.6263],[716.0163,14.6379],[716.4212,14.2547],[716.8475,13.7592],[717.2988,13.5463],[717.7375,13.163],[718.5675,12.9037],[718.9713,12.9153],[724.095,12.6443],[724.4925,12.261],[724.91,12.0249],[725.3275,11.3707],[725.7375,10.8519],[726.6063,10.0738],[727.0238,9.408],[727.4538,8.4828],[727.8963,6.8685],[728.3575,6.0671],[728.7812,5.3897],[729.2025,5.2542],[729.6037,4.9832],[730.3487,4.9948],[730.7538,4.7238],[731.16,4.3173],[731.9962,3.516],[732.4112,3.3921],[748.4538,3.5276],[753,0]]
       }
       dyn_wings_phalange_02_r_ctrl.Rotation.Z {type=MovieSceneScriptingFloatChannel, default=0, const [0,2]}
       dyn_wings_phalange_02_r_ctrl.Scale.X {type=MovieSceneScriptingFloatChannel, default=1, const [1,2]}
       dyn_wings_phalange_02_r_ctrl.Scale.Y {type=MovieSceneScriptingFloatChannel, default=1, const [1,2]}
       dyn_wings_phalange_02_r_ctrl.Scale.Z {type=MovieSceneScriptingFloatChannel, default=1, const [1,2]}
       dyn_wings_carpals_02_r_ctrl.Location.X {type=MovieSceneScriptingFloatChannel, default=0, const [0,4]}
       dyn_wings_carpals_02_r_ctrl.Location.Y {type=MovieSceneScriptingFloatChannel, default=0, const [0,4]}
       dyn_wings_carpals_02_r_ctrl.Location.Z {type=MovieSceneScriptingFloatChannel, default=0, const [0,4]}
       dyn_wings_carpals_02_r_ctrl.Rotation.X {type=MovieSceneScriptingFloatChannel, default=0, const [0,4]}
       dyn_wings_carpals_02_r_ctrl.Rotation.Y {
        type=MovieSceneScriptingFloatChannel
        default=0
        thinned_from=844
        keys [[-1,0],[1,0],[12.9837,-0.1754],[13.4162,-0.4469],[13.8062,-0.8938],[14.2025,-1.5328],[14.6,-2.2512],[14.9875,-2.5227],[15.3787,-2.6188],[21.2412,-2.7942],[23.015,-2.9696],[24.0875,-3.0657],[24.83,-3.2411],[25.585,-3.5126],[25.975,-3.688],[26.37,-4.1349],[26.7687,-4.231],[27.14,-4.4064],[27.53,-4.5025],[28.63,-4.6779],[29.6938,-4.774],[42.9462,-4.87],[43.685,-4.9661],[44.0587,-5.3504],[44.4512,-5.5593],[44.8612,-6.1357],[45.2887,-6.4407],[45.7188,-6.8249],[46.5263,-7.2092],[46.9175,-7.3053],[47.3088,-7.4975],[47.7625,-7.5142],[48.1963,-7.8985],[48.6225,-8.0906],[49.0125,-8.571],[49.4262,-9.1474],[49.8338,-10.0121],[50.26,-10.9728],[51.1737,-12.4932],[51.6,-13.1657],[52,-13.5668],[52.83,-14.7196],[53.2287,-15.4882],[53.64,-16.5283],[54.055,-17.4723],[54.4675,-18.1448],[54.8888,-18.6251],[55.285,-19.2977],[55.7275,-19.8741],[56.1513,-20.3545],[56.565,-20.5633],[57.3587,-21.5241],[58.1438,-22.1005],[58.5387,-22.0212],[63.0475,-21.7497],[63.43,-21.286],[63.835,-21.0939],[64.23,-20.647],[69.8175,-20.5509],[74.0487,-20.7263],[80.44,-20.6303],[80.8188,-20.0538],[81.2412,-19.8449],[81.6375,-19.5567],[82.0725,-19.3478],[82.9388,-18.3871],[83.3775,-17.5392],[83.7862,-17.0588],[84.18,-16.3237],[84.5925,-15.8601],[84.9875,-15.4925],[85.3938,-15.2043],[85.785,-14.8367],[86.1862,-14.7407],[86.5938,-14.4692],[90.565,-14.3731],[90.9475,-14.0849],[91.345,-13.9888],[93.475,-13.8928],[93.8625,-13.7967],[94.255,-13.4291],[94.6488,-12.7733],[95.0575,-12.3097],[95.4588,-11.9254],[95.8575,-11.1903],[96.2725,-10.4552],[96.6837,-9.8161],[97.1025,-9.7201],[97.495,-9.3525],[98.5888,-9.2564],[100.0138,-9.0643],[100.4125,-8.5213],[100.8163,-7.7862],[101.2225,-6.955],[101.635,-6.2992],[102.05,-5.9317],[102.4575,-5.6602],[102.855,-5.0211],[103.2663,-4.4614],[104.0713,-2.9912],[104.465,-2.7991],[104.8637,-2.5276],[105.2487,-2.2561],[105.6325,-1.8885],[106.0288,-1.6964],[106.4475,-1.6003],[108.2262,-1.6964],[108.6012,-1.9679],[108.9925,-2.4148],[109.3813,-2.5108],[113.36,-2.7823],[114.8963,-3.3253],[115.2875,-3.6929],[115.685,-4.5073],[116.1137,-5.5139],[116.5225,-5.8815],[116.9275,-6.153],[120.1062,-6.249],[126.3575,-6.5205],[127.09,-6.792],[127.475,-6.8881],[128.2275,-7.3517],[128.6187,-8.0075],[129.03,-8.8387],[129.4388,-10.0374],[129.8775,-10.5178],[130.2825,-10.806],[131.0713,-10.9982],[134.265,-11.1903],[134.6575,-11.4618],[135.0412,-11.8461],[135.4475,-12.1343],[135.8512,-12.8861],[136.6925,-15.3338],[137.12,-15.8935],[137.52,-15.9896],[168.895,-16.0857],[169.2738,-16.3572],[170.3713,-16.7247],[170.76,-16.8208],[171.145,-17.1884],[171.55,-17.4599],[171.9737,-17.652],[172.3875,-18.0196],[173.1812,-18.2117],[173.9438,-18.4832],[174.8075,-18.6753],[175.5812,-18.7714],[175.9675,-18.8675],[176.3637,-19.235],[176.7862,-19.5233],[177.2188,-19.9076],[177.6737,-20.3879],[178.5663,-21.7329],[178.9825,-22.3261],[179.4162,-22.7271],[179.8363,-23.4957],[180.2575,-23.9928],[180.6712,-24.5692],[181.0888,-24.8742],[181.8775,-25.2585],[187.1237,-25.5467],[187.5075,-26.2025],[188.7413,-29.1473],[189.165,-30.0913],[189.5738,-30.6677],[189.9813,-31.0352],[191.1112,-31.2274],[201.1338,-31.1313],[201.5375,-30.8431],[201.97,-30.0913],[202.3837,-29.1473],[202.8062,-28.2993],[203.2225,-27.2593],[203.6275,-26.875],[204.0475,-26.7789],[205.5125,-26.6828],[206.015,-26.2985],[206.4775,-25.5467],[206.9112,-24.9703],[208.17,-22.7147],[209.0688,-22.1383],[209.5,-22.0422],[211.66,-21.9461],[212.0362,-21.754],[212.43,-21.1943],[212.8475,-21.0982],[216.02,-20.9061],[216.4013,-20.4424],[216.8175,-20.3463],[217.2063,-20.1542],[217.5963,-20.0581],[218.6875,-19.9621],[219.0625,-19.6906],[219.4613,-19.2269],[219.8612,-19.0348],[220.5938,-18.9387],[225.8062,-19.0348],[226.9688,-19.8493],[228.545,-20.2335],[229.3613,-21.5116],[229.7738,-22.535],[230.6175,-24.7403],[231.0425,-25.3961],[231.4512,-25.7637],[237.495,-25.6676],[237.9225,-25.4922],[238.345,-25.3961],[238.7812,-25.1246],[239.1963,-24.9325],[240.085,-23.6377],[240.5638,-22.7104],[241.0225,-22.0546],[241.44,-21.9585],[244.255,-21.6703],[244.6363,-21.3028],[245.0412,-21.2067],[248.2225,-21.1106],[248.6137,-20.9185],[249.01,-20.8057],[249.4125,-20.9811],[249.805,-20.7889],[252.965,-20.9644],[259.2812,-20.8683],[260.4412,-20.5801],[264.3462,-20.2918],[264.7287,-19.9076],[265.5562,-18.0822],[265.9825,-17.2342],[266.4062,-16.7539],[266.805,-16.8332],[280.585,-16.5617],[280.9813,-16.1148],[281.3813,-15.5719],[281.7838,-15.4758],[282.1687,-15.2043],[283.9562,-15.0289],[284.7013,-14.9328],[285.46,-14.7574],[289.0362,-14.582],[289.42,-14.4859],[292.2812,-14.3105],[292.665,-14.039],[293.43,-13.7675],[294.5238,-13.5921],[295.605,-13.3206],[296.01,-12.8737],[296.4087,-12.6022],[296.8087,-12.1553],[297.205,-11.9799],[297.5888,-11.7084],[297.9813,-11.533],[299.4712,-10.9901],[299.8525,-10.5432],[300.2738,-10.351],[300.7025,-10.0795],[301.1575,-9.9041],[307.2063,-9.808],[307.58,-9.6326],[308.3613,-9.4405],[308.75,-9.169],[309.145,-9.0729],[309.53,-8.8014],[309.9362,-8.7054],[310.6675,-8.7847],[311.4187,-9.0395],[312.2075,-9.3903],[315.3863,-9.6618],[315.7738,-10.1087],[316.2,-10.2047],[316.5963,-10.5556],[317,-10.731],[318.1575,-11.0025],[318.9237,-11.0985],[319.305,-11.274],[319.7075,-11.5454],[320.0975,-11.6415],[320.485,-11.913],[320.8775,-12.0091],[321.2775,-12.2806],[322.0337,-12.3766],[322.7613,-12.8235],[323.1575,-13.1911],[323.5525,-13.3832],[323.9538,-13.6547],[324.7088,-13.7508],[325.0962,-13.9262],[325.8925,-14.1184],[326.2812,-14.2938],[327.7162,-14.3898],[328.845,-15.2043],[329.2437,-15.3004],[329.6275,-15.5719],[330.7262,-15.6679],[331.4688,-15.9394],[331.875,-16.2109],[332.3175,-16.307],[332.7713,-16.5785],[334.3175,-16.946],[335.1,-17.0421],[335.9513,-17.9359],[336.7513,-18.4789],[343.0188,-18.2867],[343.3975,-17.631],[343.8125,-17.4388],[344.2025,-17.3427],[348.085,-17.2467],[348.4737,-17.0713],[348.8637,-16.7998],[349.6662,-16.6076],[350.0587,-16.1607],[350.4725,-15.9686],[350.875,-15.6971],[351.27,-15.3295],[351.6788,-15.058],[352.0688,-14.962],[356.305,-14.8659],[356.69,-14.6905],[357.0825,-14.5944],[358.32,-13.78],[359.11,-13.5878],[359.8775,-13.3163],[365.2862,-13.2202],[365.6637,-13.0448],[366.06,-12.5019],[366.5213,-11.9422],[367.4288,-11.3992],[369.93,-11.3031],[370.3288,-11.207],[370.7162,-11.0316],[371.0987,-10.9356],[373.5863,-10.8395],[373.9662,-10.4719],[374.785,-9.9289],[375.1875,-9.7368],[375.615,-9.1938],[376.0312,-9.0017],[376.4588,-8.7302],[376.8537,-8.4587],[377.26,-8.3626],[378.0063,-8.0912],[382.8913,-7.9951],[384.3737,-7.899],[384.7575,-7.7236],[388.03,-7.4521],[390.0038,-6.9885],[390.83,-6.717],[392.0512,-6.4288],[392.4512,-6.2534],[392.85,-6.1573],[393.975,-6.0612],[394.3562,-5.9651],[394.7487,-5.7897],[398.4013,-5.6937],[402.3212,-5.8691],[403.4038,-6.0445],[403.7925,-7.097],[404.2063,-8.8846],[404.635,-11.1358],[405.0925,-12.4932],[405.545,-12.9569],[405.945,-13.0529],[406.7188,-13.149],[407.11,-13.2451],[407.5013,-13.4372],[407.9062,-13.9176],[408.3288,-14.8616],[408.74,-15.7262],[409.16,-16.3194],[409.5888,-16.7204],[410.0088,-17.2969],[410.4325,-17.9694],[410.8675,-19.0262],[411.3162,-20.6594],[411.7825,-23.0612],[412.23,-25.4798],[412.6962,-27.7062],[413.16,-29.0847],[413.6175,-29.9995],[414.0625,-30.6261],[414.5175,-31.1567],[414.97,-31.5912],[415.4087,-32.4267],[415.9262,-33.7593],[416.365,-35.0919],[416.8525,-36.1028],[417.2937,-36.4078],[418.1237,-36.5999],[418.5375,-36.4245],[418.955,-36.7127],[419.3613,-36.6334],[419.7812,-36.8422],[420.1825,-36.9551],[420.5888,-37.1472],[421.0025,-36.9718],[421.7613,-36.8757],[422.155,-36.5082],[422.5538,-35.1966],[423.0063,-33.3879],[423.4462,-30.9527],[423.9275,-28.8515],[424.3887,-27.1179],[424.8475,-25.2256],[425.79,-19.6825],[426.2525,-16.5121],[426.71,-13.421],[427.19,-9.7535],[427.7,-6.6625],[428.1612,-4.4361],[428.6175,-1.9549],[429.09,0.5721],[429.5487,3.4667],[430.0713,5.6262],[430.5725,7.4974],[431.47,9.448],[431.9013,10.2792],[432.325,11.1898],[432.7725,12.5639],[433.2038,14.2096],[433.66,16.4149],[434.12,19.2426],[434.5812,21.5273],[435.0125,21.9909],[438.22,21.8949],[439.3175,21.5106],[439.7362,19.4639],[440.195,15.1158],[440.7025,10.7844],[441.135,5.7179],[441.6125,2.7607],[442.0675,0.8895],[442.53,0.1376],[442.9588,-0.0712],[443.3863,-0.631],[443.8125,-1.1113],[444.215,-1.6878],[444.625,-2.615],[445.055,-3.4629],[445.4787,-3.8472],[445.91,-4.3276],[446.325,-5.0001],[446.7525,-5.848],[447.1912,-7.0802],[447.6413,-8.5046],[448.105,-10.1211],[448.58,-11.6248],[449.9037,-14.4902],[450.3475,-14.9706],[450.7638,-15.3549],[451.1787,-15.9146],[451.5963,-16.6665],[452.0375,-17.5144],[452.4488,-18.4416],[452.8675,-19.2896],[453.2875,-19.9453],[453.7075,-20.2336],[454.1075,-20.4257],[454.9425,-21.1776],[455.8012,-22.1383],[456.6175,-23.2744],[457.0275,-23.7381],[457.445,-23.9302],[457.8537,-24.0263],[464.7525,-23.9302],[466.5638,-23.8341],[466.95,-23.5459],[467.785,-22.5058],[468.1988,-22.1215],[468.6225,-21.9294],[469.0375,-21.8333],[471.89,-21.7373],[472.2925,-21.5451],[473.1087,-20.9687],[473.5138,-20.8726],[475.1175,-20.104],[475.54,-19.8158],[476.4225,-18.8551],[476.84,-18.4708],[477.2625,-18.1826],[477.6637,-18.0865],[480.16,-17.9904],[480.5475,-17.8944],[480.9525,-17.7022],[481.7113,-17.6061],[491.6688,-16.9838],[492.0562,-16.5369],[492.4663,-16.1694],[492.8913,-15.9939],[493.3325,-15.6264],[493.795,-15.451],[494.905,-15.3549],[495.29,-15.2588],[495.6825,-15.0667],[496.1112,-14.4276],[496.5412,-14.1394],[496.9488,-14.0434],[500.54,-14.1394],[501.295,-14.6824],[501.6925,-15.4175],[502.1463,-15.8811],[502.5812,-16.6163],[503.0063,-16.8878],[503.4137,-16.9838],[504.215,-17.5435],[504.9725,-17.6396],[505.3687,-17.9111],[505.77,-18.2954],[506.185,-18.5836],[506.5863,-19.1433],[507.4262,-20.0873],[507.8438,-20.3755],[508.25,-20.5677],[509.0525,-20.7598],[509.8612,-21.3195],[510.2587,-21.7831],[510.6712,-21.9753],[511.0825,-22.0714],[511.485,-22.3428],[514.76,-23.078],[515.515,-23.3494],[515.9175,-23.6209],[522.6637,-23.5249],[525.245,-23.078],[531.9013,-23.2534],[532.655,-23.3494],[533.0425,-23.5249],[533.8138,-23.7963],[534.5587,-23.9718],[535.3375,-24.5147],[535.74,-24.9616],[536.5812,-26.0476],[536.9962,-26.1437],[537.7587,-26.2397],[538.4875,-26.4151],[539.61,-26.5112],[539.9925,-26.6866],[540.4013,-26.7827],[540.8062,-26.9581],[546.2212,-26.6866],[546.6263,-25.9348],[547.0612,-24.8321],[547.4937,-23.9048],[547.9125,-23.2491],[548.3212,-22.9608],[548.7375,-22.4972],[549.1438,-22.3051],[551.0913,-21.8414],[551.49,-21.6493],[552.2937,-21.4571],[552.7075,-20.9768],[553.1087,-20.5925],[554.0375,-19.2642],[554.49,-18.4163],[554.9262,-17.7438],[555.3625,-17.1673],[555.79,-16.687],[556.2375,-16.3988],[556.7075,-16.3027],[566.565,-16.2066],[566.9488,-16.1105],[567.3587,-15.8223],[567.7787,-15.342],[568.1963,-14.6527],[569.0625,-12.3637],[569.5087,-12.1716],[569.9175,-12.0755],[578.365,-11.7873],[579.175,-11.2108],[584.765,-11.7705],[585.1975,-12.6017],[586.135,-14.2808],[586.5863,-15.2081],[587.4362,-16.3442],[587.8813,-16.5364],[588.335,-16.6325],[588.7713,-16.8246],[589.1787,-17.1128],[589.5963,-17.7686],[590.0175,-18.5204],[590.4525,-19.0802],[590.8675,-19.832],[592.115,-20.9681],[592.5213,-21.2564],[592.9275,-21.6406],[593.3512,-22.121],[593.76,-22.8729],[594.185,-23.5454],[594.5963,-24.314],[595.0312,-25.0199],[595.49,-25.6131],[595.9262,-26.0141],[596.3487,-26.3191],[596.7563,-26.7034],[597.1825,-26.8955],[597.5863,-27.0083],[597.9937,-27.025],[598.4137,-27.3133],[598.83,-27.4261],[599.2513,-27.3467],[599.66,-27.6517],[600.0863,-28.0527],[600.94,-28.1363],[601.365,-28.0905],[602.2613,-28.525],[602.68,-28.5417],[603.0938,-28.3663],[603.855,-28.4623],[606.65,-28.5751],[607.4625,-28.8008],[607.8813,-29.0096],[608.3138,-28.851],[609.1325,-29.4274],[609.5425,-29.6363],[609.9638,-30.1166],[610.3737,-30.5176],[610.8,-30.7265],[611.2038,-30.8393],[611.6012,-31.0315],[612.0088,-30.8561],[612.7675,-30.9521],[619.5825,-31.1275],[620.8825,-30.9354],[621.2875,-30.7433],[621.6975,-30.455],[622.1237,-30.2462],[622.5512,-29.8619],[623.425,-28.9011],[623.875,-28.6129],[624.31,-28.2286],[624.7437,-28.003],[625.1775,-27.5227],[625.6612,-27.2177],[626.2887,-26.4491],[626.77,-26.2403],[627.2188,-26.1275],[628.0775,-26.0314],[629.7113,-25.9353],[630.1587,-25.6471],[630.62,-25.6304],[631.0675,-25.5343],[633.56,-25.4047],[633.99,-25.4841],[634.4425,-25.6595],[634.8813,-25.6428],[635.7313,-25.9936],[636.5487,-26.169],[637.4375,-27.0628],[637.8937,-27.6058],[638.33,-28.3242],[639.2125,-29.218],[639.6437,-29.3934],[640.4937,-29.9364],[640.93,-30.2872],[641.4087,-31.1016],[641.8563,-32.1709],[642.3212,-33.687],[642.83,-34.7562],[643.2925,-35.4746],[643.7387,-35.65],[644.8875,-35.9215],[645.3125,-36.4477],[645.7462,-37.07],[646.1988,-37.9638],[646.6875,-38.7616],[647.1125,-40.0062],[647.5788,-41.9525],[648.0362,-44.0742],[648.5475,-45.0641],[649.4887,-45.6864],[649.9112,-46.3087],[650.3487,-47.474],[650.8062,-48.5599],[651.295,-49.5498],[651.7862,-49.9967],[652.24,-50.1721],[654.5263,-49.9967],[656.5938,-49.4537],[657.005,-49.1822],[657.8525,-48.7353],[658.2738,-48.3678],[659.125,-47.8248],[659.5562,-47.4572],[659.98,-46.9143],[661.2537,-45.9077],[661.6788,-45.2686],[662.5087,-44.9971],[664.3238,-44.5502],[665.1325,-43.9112],[668.4362,-43.7358],[669.6075,-43.4643],[670.0125,-43.2889],[670.4225,-43.0174],[670.83,-42.5705],[671.2663,-42.2029],[671.6675,-41.6599],[672.0962,-41.2924],[672.505,-40.7494],[672.9237,-40.3819],[673.335,-39.935],[674.165,-38.7529],[674.5925,-37.9385],[675.4475,-35.9253],[675.88,-34.5511],[676.7862,-33.1769],[677.6263,-31.7067],[678.0537,-30.4286],[678.5,-28.879],[678.93,-27.2167],[679.3813,-25.9386],[679.8375,-25.1074],[680.3075,-24.6605],[680.7613,-24.2929],[681.1888,-24.1968],[681.6175,-23.9253],[682.0312,-23.5578],[682.4613,-23.0942],[682.9337,-22.8227],[687.7587,-22.5512],[688.1737,-22.2797],[688.5738,-21.7367],[688.995,-21.3692],[689.4112,-20.8262],[689.8238,-20.5547],[690.2362,-19.9157],[690.6425,-19.5481],[691.0537,-19.1012],[691.4825,-19.0051],[691.9062,-18.7336],[696,-11.7398],[702,-11.7005],[711,0],[753,0]]
       }
       dyn_wings_carpals_02_r_ctrl.Rotation.Z {type=MovieSceneScriptingFloatChannel, default=0, const [0,4]}
       dyn_wings_carpals_02_r_ctrl.Scale.X {type=MovieSceneScriptingFloatChannel, default=1, const [1,4]}
       dyn_wings_carpals_02_r_ctrl.Scale.Y {type=MovieSceneScriptingFloatChannel, default=1, const [1,4]}
       dyn_wings_carpals_02_r_ctrl.Scale.Z {type=MovieSceneScriptingFloatChannel, default=1, const [1,4]}
       dyn_wings_metacarpal_03_r_ctrl.Location.X {type=MovieSceneScriptingFloatChannel, default=0, const [0,1]}
       dyn_wings_metacarpal_03_r_ctrl.Location.Y {type=MovieSceneScriptingFloatChannel, default=0, const [0,1]}
       dyn_wings_metacarpal_03_r_ctrl.Location.Z {type=MovieSceneScriptingFloatChannel, default=0, const [0,1]}
       dyn_wings_metacarpal_03_r_ctrl.Rotation.X {type=MovieSceneScriptingFloatChannel, default=0, const [0,1]}
       dyn_wings_metacarpal_03_r_ctrl.Rotation.Y {
        type=MovieSceneScriptingFloatChannel
        default=0
        thinned_from=1129
        keys [[5.4325,-0.1768],[6.2063,-1.9755],[6.6175,-4.1278],[7.0525,-5.5627],[7.4512,-6.0098],[25.6513,-6.0098],[27.535,-5.833],[27.9325,-5.833],[28.7113,-5.7394],[32.7862,-5.7394],[34.2775,-5.4691],[34.6737,-5.1988],[35.0875,-4.6581],[35.5025,-4.4813],[35.915,-4.3877],[36.3212,-4.211],[36.7275,-4.211],[37.8425,-4.0342],[38.6325,-3.8471],[39.045,-3.6703],[39.4412,-3.6703],[39.8438,-3.5767],[40.2337,-3.5767],[40.6325,-3.4],[41.0412,-3.4],[42.2475,-3.2232],[42.685,-3.1296],[43.1313,-2.9529],[43.5487,-2.9529],[43.9288,-2.8593],[44.3313,-2.6826],[44.7275,-2.4122],[45.1388,-2.1419],[45.575,-1.9651],[46.0375,-1.518],[46.465,-1.3413],[46.87,-1.2477],[47.2625,-1.0709],[47.6625,-1.0709],[50.1975,-0.9774],[50.585,-0.8006],[50.9688,-0.5303],[51.39,-0.3535],[51.7962,-0.3535],[57.4375,-0.2599],[57.8075,-0.0832],[58.1875,0.1872],[58.57,0.811],[58.9588,1.2581],[59.34,1.4349],[60.4013,1.5284],[70.8212,1.5284],[72.0138,1.7988],[72.4275,2.0691],[72.85,2.693],[73.245,3.14],[73.6337,3.5871],[74.0425,4.4813],[74.485,4.7517],[74.9125,4.9284],[76.6862,4.9284],[78.52,4.6581],[78.8925,4.107],[79.6762,3.192],[80.0625,2.8281],[80.4525,2.8281],[80.8263,2.6514],[81.2038,2.5578],[81.5825,2.1939],[81.9775,1.0189],[82.445,0.2911],[82.8625,-0.1664],[83.255,-0.6239],[83.6375,-0.7174],[84.7262,-0.8942],[87.835,-0.9878],[88.2,-1.1645],[89.6337,-1.2581],[91.4187,-1.1645],[91.7925,-1.1645],[92.1725,-0.8006],[92.5675,-0.0728],[92.9638,0.5615],[93.3675,0.9254],[93.7687,1.3829],[94.175,1.9235],[94.5938,2.1939],[94.9825,2.4642],[95.39,2.5578],[95.7775,2.7345],[96.1525,2.8281],[96.5325,3.0049],[97.9575,3.0985],[98.3262,3.3688],[98.715,3.6391],[99.1,3.8159],[100.52,3.9095],[100.8988,4.0862],[101.285,4.3566],[101.6625,4.5333],[102.0462,4.6269],[103.845,4.6269],[104.2038,4.5333],[104.59,4.5333],[104.9638,4.3462],[105.35,3.6183],[105.785,2.6202],[106.235,1.622],[106.675,0.7174],[107.0913,0.3535],[107.49,-0.0104],[107.8763,-0.104],[108.2613,-0.2807],[110.4412,-0.3743],[110.8075,-0.7382],[111.19,-1.2789],[111.5875,-1.5492],[114.7038,-1.3725],[115.0675,-1.2789],[115.4538,-1.2789],[115.8363,-1.0086],[116.22,-0.5615],[116.6237,0.5095],[117.055,1.5908],[117.4638,2.485],[117.865,3.2024],[118.2537,3.6495],[118.6437,3.8263],[119.0312,4.0966],[119.42,4.2734],[120.15,4.4501],[121.22,4.5437],[121.9337,4.7205],[122.655,4.814],[140.24,4.814],[142.0587,4.7205],[143.2225,3.6287],[143.6225,3.2648],[144.0125,3.0776],[144.395,2.8073],[144.7875,2.7137],[145.17,2.4434],[145.9413,2.2562],[146.3188,2.0795],[147.0788,1.8923],[147.48,1.622],[147.8637,1.5284],[148.245,1.1645],[148.6313,1.1645],[148.9987,0.8942],[149.3863,0.8006],[149.7675,0.5302],[150.1575,0.2599],[150.5425,-0.0104],[150.9237,-0.3743],[151.3112,-0.6447],[151.6913,-0.7382],[155.1525,-0.915],[158.5912,-1.2789],[158.9737,-2.0067],[159.37,-3.0985],[159.775,-4.0031],[160.18,-4.6373],[160.5788,-5.178],[160.9725,-5.6251],[161.3613,-5.8954],[162.7975,-5.8954],[164.5275,-6.1658],[164.8988,-6.4361],[166.305,-6.4361],[166.6637,-6.5297],[167.0512,-6.7064],[167.4375,-6.9768],[167.83,-7.3407],[168.2163,-7.7878],[168.9775,-8.2349],[171.92,-8.2349],[173.0675,-8.0581],[173.4538,-7.9645],[173.845,-7.4239],[174.24,-7.0703],[174.6288,-6.6232],[175.035,-6.4465],[175.4225,-6.4465],[176.1513,-6.1762],[177.9538,-6.1762],[178.7287,-5.8226],[179.1338,-5.5523],[179.5287,-5.5523],[179.9062,-5.3755],[185.48,-5.3755],[185.8425,-5.1988],[186.6137,-4.6581],[187.0088,-4.211],[187.405,-3.8575],[187.7912,-3.8575],[189.1812,-3.7639],[189.5538,-3.5872],[190.325,-3.4],[192.8363,-3.4],[193.585,-3.0361],[194.36,-2.4954],[194.7513,-2.1315],[195.1375,-1.9548],[195.5213,-1.6844],[195.915,-1.6844],[196.9775,-1.5909],[197.3562,-1.2269],[197.7563,-0.9566],[198.145,-0.6863],[198.5287,-0.6863],[201.4438,-0.5095],[201.8412,-0.5095],[204.0487,-0.6031],[204.8463,-1.3309],[205.2425,-1.6948],[205.6325,-1.882],[206.025,-2.3395],[206.43,-2.8801],[207.2475,-3.2441],[207.6275,-3.3376],[211.12,-3.3376],[212.8725,-3.608],[215.1325,-4.159],[216.8925,-4.159],[217.9425,-4.2526],[219.045,-4.523],[219.4325,-4.7933],[219.825,-5.1572],[220.2225,-5.4275],[220.6112,-5.6147],[221.0025,-5.885],[221.3925,-6.2489],[221.7975,-6.5193],[222.1963,-6.7064],[222.6313,-6.9768],[223.065,-7.1535],[223.4762,-7.2471],[226.2738,-7.2471],[226.6375,-7.5174],[227.3837,-7.7878],[228.1625,-8.3285],[231.3913,-8.3285],[232.1225,-8.1517],[232.5525,-8.1517],[232.9713,-7.7046],[233.3963,-6.904],[233.8075,-6.3633],[234.1963,-6.0098],[234.5912,-5.8331],[235.0263,-5.7395],[235.9262,-5.1988],[236.7175,-4.6581],[237.1087,-4.5646],[237.4937,-4.5646],[237.865,-4.471],[238.2487,-4.471],[239.315,-4.2006],[240.4875,-3.3896],[240.8813,-3.0361],[241.6525,-2.4954],[244.1162,-2.4954],[244.4837,-2.589],[244.87,-2.9529],[245.2625,-3.7743],[246.0875,-5.9578],[246.505,-6.956],[246.91,-7.4967],[247.9962,-7.4967],[249.4038,-7.5902],[257.95,-7.5902],[258.6712,-7.4967],[259.0462,-7.3199],[259.4338,-7.0496],[259.8263,-6.956],[260.215,-6.7792],[260.6037,-6.6857],[265.0337,-6.6857],[266.4425,-6.7792],[266.8525,-6.8728],[267.9925,-6.8728],[268.3587,-7.0496],[270.3038,-7.5175],[273.4913,-7.5175],[274.2225,-7.611],[276.3537,-7.611],[276.7213,-7.7046],[277.105,-7.9749],[277.4937,-8.0685],[277.8863,-8.2453],[280.0188,-8.2453],[280.7487,-8.3389],[281.4825,-8.1621],[285.135,-8.1621],[285.8763,-8.0685],[286.2563,-8.0685],[286.645,-7.8918],[287.0575,-7.8918],[287.7962,-7.7982],[288.1825,-7.2575],[288.5888,-6.7169],[288.9925,-5.9994],[289.4025,-5.4588],[289.8138,-5.0117],[290.215,-4.8349],[291.0013,-4.2942],[291.3938,-4.1175],[294.6475,-4.1175],[296.1575,-4.2942],[296.5575,-4.4814],[296.9613,-4.7517],[297.7025,-4.8453],[299.2637,-5.2092],[300.0612,-5.937],[300.4588,-6.3841],[300.8737,-6.748],[301.6612,-7.2887],[302.0625,-7.6526],[302.4675,-7.8294],[302.8625,-7.923],[309.2437,-7.923],[309.6237,-8.1933],[310.755,-8.4636],[313.96,-8.4636],[314.3325,-8.5572],[315.0675,-8.8276],[316.16,-8.8276],[317.2262,-8.9211],[317.9762,-9.0979],[320.1662,-9.0979],[320.8988,-8.9211],[321.28,-8.9211],[321.66,-8.7444],[322.045,-8.7444],[322.4425,-8.474],[323.9075,-8.474],[324.6263,-8.7444],[325.0125,-9.0147],[325.4062,-9.2019],[326.19,-9.7425],[326.585,-9.8361],[334.6637,-9.8361],[335.0425,-10.1064],[335.4438,-10.2],[335.84,-10.4704],[337.97,-10.4704],[338.345,-10.5639],[339.08,-10.6575],[341.2075,-10.6575],[341.9275,-10.7511],[342.675,-10.7511],[343.0587,-11.0214],[345.9725,-11.0214],[346.3462,-11.115],[346.7413,-11.2918],[347.1225,-11.3853],[347.5238,-11.7493],[347.9225,-11.8428],[348.7075,-12.3835],[358.1475,-12.3835],[358.54,-12.2068],[359.8125,-12.2068],[360.5338,-12.03],[360.9137,-11.7597],[361.33,-10.7823],[361.7975,-9.8049],[362.2775,-9.0043],[362.735,-8.2973],[363.1488,-7.767],[363.5525,-7.3199],[364.3338,-6.9664],[389.1475,-6.5648],[389.9375,-6.3984],[390.37,-6.0449],[390.8275,-5.6914],[391.2713,-5.2547],[391.675,-4.9948],[392.065,-4.818],[392.4563,-4.818],[393.2563,-4.6413],[395.8412,-4.6413],[396.2138,-4.7348],[396.6063,-4.5581],[399.1275,-4.5581],[399.9,-4.2046],[402.4375,-4.2046],[403.185,-4.0278],[403.57,-3.7575],[403.965,-3.5807],[404.355,-3.5807],[404.7337,-3.404],[405.115,-3.404],[405.4963,-3.3104],[410.115,-3.3104],[410.8325,-3.404],[411.2163,-3.4975],[412.2775,-3.4975],[413.3525,-3.5911],[413.7475,-3.8614],[414.1463,-3.955],[414.8875,-4.0486],[415.9662,-4.1422],[416.3537,-4.3189],[419.6075,-4.3189],[420.7138,-4.4125],[421.0888,-4.5893],[421.4812,-4.6829],[422.285,-4.7764],[422.7125,-4.9532],[423.5275,-5.4939],[423.9425,-5.681],[424.345,-6.0449],[425.2175,-6.7728],[425.6575,-7.3134],[426.0713,-8.218],[426.4913,-10.4743],[426.9438,-13.8015],[427.39,-17.1183],[427.8412,-19.2706],[428.285,-19.7177],[428.6812,-20.0712],[430.885,-20.0712],[431.2575,-20.2479],[434.935,-20.2479],[435.3125,-20.0712],[435.7025,-19.8008],[436.105,-18.6363],[436.5163,-17.3886],[436.9425,-16.0577],[437.3562,-14.8919],[437.7638,-13.9146],[438.1687,-13.1971],[438.5725,-12.3965],[438.9762,-11.6687],[439.8025,-10.0467],[440.6112,-8.7782],[441.0213,-8.0608],[441.4288,-7.5201],[441.8487,-6.9794],[442.2637,-6.262],[442.6762,-6.0852],[443.0775,-5.9917],[443.4762,-5.8149],[443.8663,-5.7213],[445.35,-5.7213],[446.065,-5.451],[446.4638,-5.0039],[446.875,-4.4632],[447.2887,-3.7458],[447.7025,-3.1116],[448.1137,-2.3941],[448.535,-1.7703],[448.9525,-1.4999],[449.3525,-1.2296],[449.7475,-1.2296],[450.49,-1.136],[450.8825,-0.8657],[451.2738,-0.7721],[451.6725,-0.4082],[452.07,-0.0443],[452.4762,0.4028],[452.8813,0.6731],[454.4375,0.7667],[456.5875,0.7667],[457.0038,0.9435],[458.7887,1.0371],[459.5163,1.3074],[462.2375,1.2138],[462.6375,1.1202],[463.045,0.7563],[463.4462,0.0285],[463.8587,-0.8865],[464.2788,-2.1654],[464.7013,-2.9764],[465.1137,-3.3403],[466.225,-3.3403],[466.9562,-3.4339],[470.1938,-3.4339],[470.5812,-3.7978],[470.9913,-3.985],[471.395,-4.2657],[471.7925,-4.3593],[472.9062,-4.4528],[475.045,-4.4528],[475.4275,-4.5464],[476.1762,-4.4632],[477.6637,-4.4632],[478.4087,-4.2865],[478.7887,-4.2865],[479.8563,-4.1097],[481.3138,-4.1097],[481.685,-3.933],[482.0738,-3.933],[482.8062,-3.7562],[483.19,-3.5795],[483.5838,-3.2259],[486.2575,-3.2259],[486.6337,-3.0492],[487.0325,-3.0492],[487.415,-2.8724],[487.8188,-2.7789],[488.2513,-2.6021],[490.8475,-2.6021],[491.225,-2.3318],[491.63,-2.3318],[492.0238,-2.0614],[492.4262,-1.7911],[492.855,-1.6143],[493.2437,-1.6143],[493.62,-1.344],[494.0225,-1.2504],[494.415,-0.9801],[494.8087,-0.533],[495.215,-0.4394],[495.61,-0.2627],[496,-0.2627],[497.09,-0.1691],[497.4787,-0.1691],[498.2212,-0.0755],[498.6063,0.0181],[499,0.1948],[499.4062,0.2884],[499.8087,0.5588],[501.9413,0.5588],[503.0475,0.6523],[503.445,0.8291],[503.8525,0.9227],[504.2513,1.0994],[512.6975,1.0994],[513.4312,1.0058],[513.8175,0.6419],[514.2138,0.278],[514.6137,0.1844],[515.0038,0.0077],[515.7713,0.0077],[518.65,-0.0859],[519.4087,-0.1795],[519.835,-0.4498],[520.6537,-0.5434],[521.3875,-0.8137],[521.7825,-0.9073],[522.18,-1.1776],[522.9662,-1.2712],[523.7013,-1.5416],[524.79,-1.6351],[525.1875,-1.9055],[526.7787,-2.0926],[527.1612,-2.1862],[527.5525,-2.363],[527.9488,-2.4565],[536.5187,-2.4565],[536.8913,-2.5501],[537.2862,-2.8204],[538.8188,-3.1844],[539.2088,-3.3715],[539.605,-3.7354],[540.01,-4.0058],[540.4212,-4.3697],[540.8363,-4.4633],[550.5163,-4.4633],[550.9013,-4.5568],[551.3313,-4.9207],[551.7775,-5.0143],[552.21,-5.4718],[552.655,-6.3764],[553.0863,-6.5636],[553.4887,-6.8339],[554.25,-7.1042],[554.995,-7.1042],[555.37,-7.3746],[555.77,-7.6449],[556.165,-8.092],[556.5663,-8.2688],[556.9688,-8.3623],[557.3637,-8.8094],[557.7663,-9.3501],[558.17,-9.9843],[558.5763,-10.8785],[558.9862,-11.4192],[559.395,-12.2198],[559.8125,-12.4901],[564.0913,-12.4901],[564.4787,-12.2198],[564.88,-11.8559],[565.2825,-11.6791],[566.0638,-11.6791],[566.4563,-12.3134],[566.8737,-13.3115],[567.3012,-14.5696],[567.7313,-15.7342],[568.155,-16.8051],[568.5863,-17.9696],[569.01,-19.6748],[569.4563,-21.8375],[569.9037,-24.0938],[570.35,-24.9879],[570.7738,-25.2583],[574.3162,-25.2583],[575.1025,-25.0711],[575.5,-24.624],[575.9,-24.2601],[576.305,-24.0834],[583.12,-24.0834],[583.5638,-23.5427],[583.9937,-20.2051],[584.4425,-14.1154],[584.9,-5.4307],[585.345,1.968],[585.7912,6.179],[586.235,6.6261],[587.42,6.6261],[587.8138,6.179],[588.2138,5.2848],[588.6275,4.1053],[589.0462,3.2808],[589.4588,2.5634],[589.8725,2.0227],[590.285,1.7523],[590.6837,1.7523],[591.0725,1.6588],[591.4725,1.3884],[592.28,1.2013],[592.69,0.9309],[593.0825,0.8374],[593.4875,0.567],[593.89,0.4734],[594.28,0.4734],[594.67,0.2031],[595.0688,0.1095],[595.46,0.1095],[597.6,0.2031],[598.6875,0.2967],[599.0738,0.4734],[599.4787,0.9309],[599.8988,1.2013],[600.305,1.7419],[600.7125,1.8355],[601.1087,2.1059],[605.5287,2.1059],[606.2625,2.1994],[606.65,2.6569],[607.0575,3.3848],[607.4737,4.2062],[607.9088,4.7572],[608.72,5.6722],[609.1463,5.849],[611.505,5.849],[611.935,5.9426],[616.46,5.9426],[616.8475,5.849],[617.6263,5.849],[618.0125,5.5786],[618.8238,5.3915],[619.2237,5.0276],[619.6288,4.934],[620.0213,4.5701],[620.4338,4.2062],[621.2387,4.019],[621.6413,4.019],[622.0312,3.7487],[622.8388,3.5615],[623.2337,3.104],[623.6463,2.8337],[624.0525,2.7401],[627.5287,2.7401],[627.925,2.4698],[628.325,2.3762],[628.7362,2.0123],[629.145,1.9187],[629.555,1.378],[629.9775,1.1909],[630.3875,1.0141],[633.8487,1.0141],[634.5675,0.9205],[634.9525,0.7438],[635.35,0.6502],[635.7462,0.3799],[636.145,0.2863],[642.1775,0.2863],[643.2975,0.1927],[643.7262,-0.2648],[644.165,-0.7223],[647.5587,-0.7223],[647.9425,-1.0862],[648.35,-1.1798],[648.7475,-1.5437],[649.5575,-2.0843],[649.955,-2.1779],[650.7,-2.2715],[652.5725,-2.2715],[652.9713,-2.3651],[653.3713,-2.9993],[653.7925,-3.2697],[654.205,-3.4568],[655.78,-3.4568],[656.1712,-3.5504],[657.3687,-3.5504],[657.7687,-3.644],[658.1662,-3.8207],[658.9688,-3.8207],[659.355,-3.7272],[659.76,-3.54],[660.165,-2.8954],[660.5925,-1.9641],[661.0263,-0.682],[661.45,0.3057],[661.8712,0.9296],[663.3587,0.9296],[664.075,0.4825],[664.47,0.2122],[664.875,0.2122],[665.605,0.1186],[665.9875,0.1186],[667.4013,0.025],[667.785,0.025],[668.8563,-0.0686],[670.3,-0.2453],[670.6862,-0.3389],[683.3713,-0.3389],[683.76,-0.5261],[684.1662,-0.89],[684.5788,-0.9836],[684.9887,-0.9836],[685.7475,-1.0771],[686.1475,-1.2539],[686.5712,-1.3475],[686.9813,-1.7114],[687.4062,-1.9817],[687.8225,-2.1689],[688.23,-2.2625],[688.6325,-2.6264],[689.0425,-2.8967],[689.4438,-3.1774],[689.85,-3.4478],[690.2487,-3.8117],[691.4325,-3.8117],[691.8225,-4.082],[692.2225,-4.4459],[692.6288,-4.9034],[693.0375,-5.5377],[694.2675,-6.9102],[694.68,-8.0955],[695.1112,-9.4576],[695.5475,-10.9132],[695.9688,-11.5475],[697.155,-11.5475],[698.3537,-12.0777],[701.9025,-12.0777],[702.3088,-11.8074],[702.73,-11.4264],[703.1475,-10.686],[703.56,-9.9265],[703.98,-9.5184],[704.3863,-9.2241],[705.1413,-8.6475],[705.5462,-8.5206],[705.945,-8.3438],[706.7462,-8.2238],[707.5812,-8.2238],[707.9762,-7.9535],[708.3813,-7.8422],[708.7862,-7.5719],[709.2063,-7.3764],[712.7912,-7.147],[716.235,-6.898],[719.3575,-6.6491],[720.765,-5.9226],[721.1625,-5.1465],[721.57,-4.2692],[721.9837,-3.6612],[722.3875,-3.4317],[722.79,-3.317],[724.3725,-3.317],[724.7763,-3.1402],[725.5287,-3.0107],[729.0138,-2.8899],[729.4125,-3.6977],[729.8375,-5.1533],[730.2887,-6.1515],[730.7375,-6.4218],[733.1162,-6.4218],[733.5063,-6.5154],[735.0838,-6.5154],[735.475,-6.6921],[735.8775,-6.9625],[736.28,-7.0561],[736.68,-7.3144],[737.44,-7.2682],[739.395,-7.0938],[740.4863,-6.9507],[741.6425,-6.7166],[742.3913,-6.551],[742.7725,-6.4916],[743.1575,-6.4513],[744.2825,-6.2673],[744.6712,-6.1215],[745.0725,-5.7694],[745.4762,-4.8121],[745.8888,-3.42],[746.3162,-2.1544],[746.7375,-1.1406],[747.1475,-1.0108],[747.5487,-0.6787],[748.3075,-0.452],[748.6962,-0.4023],[749.4525,-0.333],[750.5163,-0.2122],[750.9037,0.0044],[753,0.0044]]
       }
       dyn_wings_metacarpal_03_r_ctrl.Rotation.Z {type=MovieSceneScriptingFloatChannel, default=0, const [0,1]}
       dyn_wings_metacarpal_03_r_ctrl.Scale.X {type=MovieSceneScriptingFloatChannel, default=1, const [1,1]}
       dyn_wings_metacarpal_03_r_ctrl.Scale.Y {type=MovieSceneScriptingFloatChannel, default=1, const [1,1]}
       dyn_wings_metacarpal_03_r_ctrl.Scale.Z {type=MovieSceneScriptingFloatChannel, default=1, const [1,1]}
       dyn_wings_phalange_03_r_ctrl.Location.X {type=MovieSceneScriptingFloatChannel, default=0, const [0,2]}
       dyn_wings_phalange_03_r_ctrl.Location.Y {type=MovieSceneScriptingFloatChannel, default=0, const [0,2]}
       dyn_wings_phalange_03_r_ctrl.Location.Z {type=MovieSceneScriptingFloatChannel, default=0, const [0,2]}
       dyn_wings_phalange_03_r_ctrl.Rotation.X {type=MovieSceneScriptingFloatChannel, default=0, const [0,2]}
       dyn_wings_phalange_03_r_ctrl.Rotation.Y {
        type=MovieSceneScriptingFloatChannel
        default=0
        thinned_from=701
        keys [[0,-0.076],[8.7337,0.0982],[13.5512,0.2724],[13.9125,0.7191],[14.3025,1.0676],[14.6837,1.5903],[15.0662,2.756],[15.4663,3.8997],[15.8788,5.0433],[16.2812,5.4899],[36.1637,6.1108],[36.5462,7.0042],[36.9375,7.3526],[37.325,7.5269],[45.9737,7.3526],[46.34,6.8078],[46.7325,6.3611],[47.105,5.8163],[47.5175,5.5438],[48.6837,4.7266],[49.0575,4.4541],[49.4475,4.2799],[49.8163,4.1817],[60.64,5.1511],[61.025,6.8174],[61.4375,8.9304],[61.8663,10.8692],[62.275,12.0888],[62.6775,12.3391],[63.0612,12.5133],[84.1225,12.4152],[85.9538,12.1427],[86.3275,11.8703],[90.3388,11.4997],[90.7113,11.1291],[91.1,10.5082],[91.4812,9.9633],[91.8687,9.5166],[92.2587,9.3424],[92.635,9.07],[100.9488,8.8957],[104.445,9.07],[104.82,9.3424],[105.2,9.9633],[105.5925,10.3118],[105.9862,10.486],[107.7525,10.3878],[108.4725,9.7669],[108.8512,8.7754],[109.2525,7.2612],[109.6625,5.9212],[110.0675,4.7776],[110.4563,3.71],[110.8575,3.2633],[111.2375,2.9149],[111.6112,2.7406],[111.9962,2.4682],[112.3813,2.0215],[112.7775,1.5749],[113.1538,0.7797],[113.5663,0.3331],[113.95,-0.4621],[114.3338,-0.8106],[114.715,-1.083],[115.105,-1.4315],[115.4837,-1.8781],[115.9112,-2.6733],[116.3475,-3.0217],[116.7613,-3.196],[117.1562,-3.2942],[121.4238,-3.1199],[121.7912,-2.9457],[122.165,-2.8475],[122.9013,-2.6733],[123.2738,-2.499],[123.6625,-2.4008],[124.4275,-2.0524],[125.1675,-1.8781],[125.5338,-1.7039],[126.96,-1.3554],[127.3375,-1.0069],[127.7313,-0.7345],[128.1137,-0.2118],[128.5063,0.4852],[128.8863,0.8336],[129.27,1.1821],[130.0337,1.3563],[132.155,1.5306],[132.5263,1.803],[132.915,1.9772],[134.3212,2.1515],[135.835,2.8484],[144.0863,3.7196],[144.4787,4.5908],[144.8637,5.9087],[145.2663,6.9541],[145.675,7.3026],[146.3988,7.4768],[170.3225,7.3786],[177.3188,7.5529],[177.6888,7.9013],[178.075,8.424],[178.465,8.7725],[178.865,8.8486],[179.295,9.0228],[180.0888,9.197],[180.46,9.0988],[181.2,9.2731],[192.2675,9.0988],[192.6263,8.8264],[193.01,8.1073],[193.4162,7.0398],[193.8125,5.3513],[194.235,3.3904],[194.6413,0.7547],[195.0738,-1.1841],[195.4825,-2.6762],[195.895,-3.2971],[196.2912,-3.4713],[196.6625,-3.6455],[199.1438,-3.8198],[199.5213,-4.3425],[199.9062,-4.8652],[200.3088,-5.0394],[203.13,-4.8652],[204.8875,-4.691],[205.2788,-4.2443],[205.6875,-3.5474],[206.0725,-2.3277],[206.4775,-0.5853],[206.89,1.0589],[207.32,2.4528],[207.72,3.4982],[208.1263,3.8467],[208.505,4.0209],[211.3925,4.1952],[211.7975,4.7179],[212.1812,5.4148],[212.5838,5.9375],[212.97,6.286],[221.8388,6.1118],[223.2312,5.9375],[223.6087,5.8393],[224.7038,5.6651],[226.8025,5.5669],[227.9275,5.3927],[228.6588,5.2184],[233.165,5.0442],[233.9,4.946],[234.295,4.7718],[235.0425,4.5975],[235.77,4.1509],[236.1425,3.9766],[236.5275,3.7042],[237.2613,3.53],[238.0263,3.0833],[238.4212,2.4624],[238.8312,2.19],[239.21,1.493],[239.605,1.0464],[239.9913,0.077],[240.395,-1.0666],[240.8038,-2.1342],[241.2025,-2.9293],[241.6137,-3.1036],[242,-3.4521],[242.83,-3.6263],[243.9513,-3.9748],[244.3238,-4.4975],[244.7088,-5.3687],[245.1263,-6.8386],[245.57,-8.1343],[246.0412,-9.1798],[246.4525,-9.7785],[247.64,-10.3012],[250.7625,-10.127],[251.8363,-9.8546],[252.9162,-9.6803],[253.2962,-9.5821],[254.7162,-9.4079],[255.4375,-9.2337],[256.8675,-9.0594],[257.2437,-8.787],[257.6513,-8.6128],[258.0425,-8.1661],[258.4538,-7.4692],[258.8525,-6.7722],[259.2738,-5.901],[259.675,-4.7574],[260.0713,-3.788],[260.485,-3.0911],[260.88,-2.7426],[261.9725,-2.5684],[265.1612,-2.3941],[273.785,-2.5684],[277.075,-2.6665],[278.2113,-2.8408],[284.0713,-2.939],[284.8075,-2.4163],[285.1975,-1.469],[285.6137,-0.2494],[286.0263,1.2427],[286.4525,2.1139],[287.2625,3.1593],[287.6862,3.6821],[288.095,4.379],[288.4837,5.2502],[288.8963,6.2196],[289.3075,6.7423],[300.065,6.3717],[300.4412,5.2059],[300.845,4.2144],[301.245,3.3211],[301.645,2.1553],[302.06,1.4362],[302.4663,0.3686],[302.8737,-0.4265],[303.2587,-0.6989],[303.6437,-0.8732],[305.0813,-1.0474],[305.48,-1.4941],[305.9225,-1.6683],[306.35,-2.115],[307.1225,-2.2892],[307.4937,-2.3874],[307.8775,-2.9101],[308.295,-3.1825],[308.7413,-3.455],[309.1762,-3.8034],[309.5738,-4.2501],[313.075,-4.4243],[313.445,-4.5225],[313.86,-4.9692],[314.2575,-5.3177],[314.6637,-5.5901],[315.0688,-5.7643],[316.1562,-5.9386],[316.53,-6.1128],[316.9275,-6.211],[317.3138,-6.5595],[317.725,-6.908],[318.1175,-7.0822],[318.5063,-7.3546],[318.9225,-7.7031],[319.3225,-8.1498],[320.4325,-8.324],[321.5238,-8.5964],[322.6012,-8.7707],[330.8587,-8.4222],[331.24,-7.8234],[331.6375,-7.4749],[333.8062,-7.1265],[334.1825,-6.778],[335.6087,-6.8762],[344.1175,-7.0504],[344.5187,-7.2247],[344.9037,-7.4971],[345.3062,-7.8456],[345.705,-8.2922],[346.0888,-8.6407],[346.4863,-9.1634],[346.8888,-9.7843],[347.2912,-10.231],[347.7088,-10.5795],[348.4688,-10.8519],[349.23,-11.5488],[349.62,-11.8213],[350.0125,-11.9955],[350.4125,-12.2679],[351.1575,-12.4422],[351.8725,-12.7146],[352.6513,-13.0631],[354.7538,-13.1613],[363.1562,-12.8128],[363.5412,-12.1158],[363.9425,-10.9722],[364.3625,-9.6544],[364.7587,-8.0641],[365.1725,-7.0187],[365.565,-6.3978],[365.9638,-6.2235],[373.5287,-6.0493],[382.5187,-5.5266],[382.9075,-5.0039],[383.3138,-4.3829],[383.71,-3.5878],[384.1213,-3.4136],[400.1112,-3.2393],[400.5163,-3.1633],[400.9575,-2.8148],[401.3963,-2.2921],[401.8025,-1.1485],[402.2113,0.1694],[402.64,2.2063],[403.5863,8.0447],[404.0863,9.9835],[404.525,10.7786],[404.935,11.3013],[407.0987,11.4756],[411.615,11.5738],[414.7675,11.748],[415.5063,11.9222],[415.8963,12.0965],[416.285,12.445],[416.6788,12.8916],[417.0813,13.7849],[417.505,15.201],[417.9188,15.7237],[418.315,15.8219],[418.7188,15.9961],[422.555,16.1704],[422.9362,16.5188],[423.3212,17.1397],[423.7213,18.3594],[424.1275,19.7755],[424.545,21.0933],[424.9625,22.0627],[431.605,22.4112],[431.9987,22.4872],[432.4325,22.6614],[432.8763,22.9117],[433.2788,23.086],[433.6675,23.162],[434.0537,23.3363],[434.4462,23.4123],[434.8637,23.5644],[435.3138,23.9889],[436.1525,24.141],[436.96,23.9446],[437.3525,24.0207],[437.7775,23.8243],[438.5812,23.0609],[438.9775,22.4179],[439.7987,20.9354],[440.1963,20.2924],[440.6075,19.8236],[441.0063,19.0063],[441.425,18.3632],[442.6375,16.9568],[443.0325,16.0413],[443.4475,15.2241],[443.8563,14.1122],[444.2663,12.6519],[444.6988,11.54],[445.1037,10.3521],[445.5163,8.99],[445.935,7.9003],[446.3462,6.7124],[446.765,5.2521],[447.1975,3.9881],[447.6475,2.5499],[448.0813,1.1878],[448.5113,0.3705],[448.9075,-0.3486],[449.3112,-0.9916],[449.7138,-1.9832],[450.1263,-2.7022],[450.5362,-3.6177],[450.935,-4.7074],[451.3537,-5.8732],[451.7663,-6.8647],[452.17,-8.1287],[452.595,-8.6735],[452.9887,-9.4908],[453.3938,-10.2099],[453.7975,-11.0272],[454.1938,-11.572],[454.6025,-12.3893],[455.0125,-13.2826],[455.4275,-13.8275],[455.8175,-14.5466],[456.215,-15.0914],[456.615,-15.9848],[457.44,-17.0745],[457.84,-17.8918],[458.2513,-18.2624],[459.0387,-19.1557],[459.4338,-19.4281],[461.03,-21.6076],[461.4375,-22.4248],[461.835,-22.9697],[462.24,-23.6888],[462.6413,-24.1354],[463.4613,-24.6803],[463.8888,-24.7785],[467.6275,-24.8767],[468.0225,-24.9749],[468.4225,-24.8988],[468.815,-24.7246],[469.2138,-24.3761],[469.6275,-23.8534],[470.0625,-22.9822],[470.4913,-22.2853],[470.8975,-21.7625],[471.3537,-21.2398],[471.7725,-20.7171],[472.1762,-20.3686],[472.5638,-20.1944],[472.9575,-19.7477],[474.1112,-19.3011],[474.49,-18.9526],[475.23,-18.7784],[493.2487,-18.1575],[494.0525,-16.1205],[494.4613,-14.8027],[494.8788,-14.1818],[495.31,-13.3866],[495.7563,-13.0381],[496.1988,-12.4172],[496.6212,-11.9706],[497.0575,-11.6221],[501.63,-11.4479],[502.0312,-10.7509],[502.4362,-10.13],[502.86,-8.8122],[503.2825,-6.9716],[503.7237,-4.5862],[504.1575,-0.6105],[504.6163,2.6461],[505.0525,4.4867],[505.4812,5.3579],[505.8788,5.5321],[513.1413,5.4339],[513.8737,5.2597],[518.3763,5.1615],[519.1025,4.9873],[519.52,4.4424],[519.9513,3.9958],[520.7787,3.4509],[521.8937,3.1785],[522.6288,3.0042],[523.0188,2.7318],[523.4212,2.5576],[524.1587,2.3833],[524.5362,2.2851],[525.6275,2.1109],[526,2.0127],[527.5338,1.8385],[527.9662,1.5661],[528.37,1.2936],[529.1513,0.7488],[529.8988,0.5745],[530.3062,0.4003],[530.7462,0.3021],[531.1575,0.1279],[531.5487,0.0297],[531.94,-0.1446],[535.0788,-0.3188],[535.4762,-0.417],[536.25,-0.7655],[536.655,-0.8636],[537.465,-1.2121],[537.8675,-1.4846],[538.6237,-1.6588],[539.0113,-1.757],[540.4625,-1.9312],[553.8913,-1.757],[554.9613,-1.5827],[555.3475,-1.6809],[557.52,-1.3325],[557.9125,-0.984],[558.3225,-0.6355],[558.77,-0.1128],[559.2287,0.2357],[559.635,0.4099],[560.0225,0.8566],[560.4212,1.2051],[561.1675,1.3793],[561.5625,1.4775],[561.9862,1.6517],[562.4338,1.826],[562.84,2.3487],[563.245,2.7953],[563.6413,3.4923],[564.0475,4.1132],[564.46,5.0065],[564.865,6.5968],[565.3125,8.7319],[565.7437,11.2156],[566.1888,13.5249],[566.6263,14.6685],[567.0425,14.8428],[575.8038,14.7446],[579.6837,14.4722],[580.75,14.374],[581.1388,13.9273],[581.5387,13.3825],[582.3462,12.8376],[582.735,12.7394],[583.4837,12.2927],[583.8763,11.9221],[584.7,11.0067],[585.1012,10.5379],[585.5075,10.3636],[585.8937,10.2654],[586.3,10.0912],[586.6862,9.993],[587.9125,9.1757],[588.3,8.8051],[588.6975,8.2602],[589.11,7.8896],[589.51,7.0723],[589.9388,6.2551],[590.8238,5.7102],[591.2475,5.612],[592,5.3396],[592.385,5.1654],[592.795,4.6205],[593.2163,4.2499],[593.6575,4.0756],[594.0987,3.8032],[594.5187,3.5308],[594.9162,3.2584],[595.315,2.8877],[596.1137,2.3429],[596.5087,1.9723],[596.9288,1.2532],[597.3462,0.8065],[597.76,0.4359],[598.1712,-0.0108],[598.5738,-0.4574],[598.9813,-0.8281],[599.3887,-1.2747],[599.7962,-1.7974],[600.2038,-2.0699],[600.595,-2.3423],[600.9987,-2.9632],[601.4137,-3.3338],[601.8263,-3.7805],[602.2275,-4.3253],[602.6288,-4.6738],[603.0337,-4.848],[605.885,-4.9462],[607.01,-5.2187],[607.4025,-5.3929],[610.2225,-5.4911],[611.0063,-5.8396],[611.4013,-6.112],[612.51,-6.2102],[619.0725,-6.3844],[619.8175,-6.6568],[620.1988,-6.9293],[620.9937,-7.2777],[621.375,-7.5502],[621.8263,-7.7244],[624.0525,-7.9968],[624.4425,-8.3453],[624.8675,-8.6177],[625.7925,-9.5111],[626.2163,-9.8595],[626.6313,-9.9577],[627.03,-10.132],[628.1375,-10.3062],[628.5238,-10.8511],[628.95,-11.2977],[629.3512,-11.9186],[629.7687,-12.6377],[630.1712,-13.0844],[630.5688,-13.2586],[630.9725,-13.7053],[631.3813,-13.8795],[631.7962,-14.152],[632.6388,-15.6662],[633.0487,-16.4835],[633.4613,-17.1044],[633.875,-17.2786],[634.2675,-17.551],[641.2163,-17.3768],[641.9538,-16.9301],[642.3525,-15.6883],[642.7638,-13.6514],[643.2138,-12.4096],[643.6225,-11.6144],[644.0412,-11.4402],[655.245,-11.266],[656.0275,-10.9175],[659.0375,-10.7433],[660.6187,-11.136],[661.3913,-11.2342],[661.7862,-11.5066],[662.5562,-11.6048],[665.7088,-11.4306],[666.0962,-11.2563],[666.485,-10.5594],[666.8925,-10.2109],[667.3125,-8.9691],[667.745,-7.9997],[668.1575,-6.7579],[668.5863,-5.3419],[668.9975,-4.1761],[669.4175,-3.2067],[669.83,-2.1392],[670.2525,-1.2458],[670.6562,-0.0801],[671.1237,0.0941],[671.5538,0.2684],[679.895,0.3666],[680.6275,0.5408],[681.0275,0.8132],[681.7812,0.9875],[690.4087,1.1617],[690.7912,1.6084],[691.585,1.9569],[691.99,2.055],[693.4462,2.2293],[695.2075,2.5017],[695.5938,2.6759],[696.4,3.5693],[696.8112,3.8417],[697.205,4.0159],[700,4.1902],[700.7912,4.735],[701.8913,4.9093],[702.2713,5.1817],[702.6688,5.3559],[703.8737,6.1732],[704.6775,6.5217],[705.07,6.7941],[705.4675,6.8923],[705.8725,7.2408],[706.2812,7.339],[706.7,7.6114],[707.12,7.7856],[707.5387,8.0581],[708.65,8.2323],[712.14,8.0581],[712.5312,7.9599],[712.9275,7.7635],[713.3262,7.2186],[713.745,6.4774],[714.175,5.7362],[714.585,4.8207],[715.0125,4.1777],[715.4238,3.731],[716.5525,3.6328],[719.8725,3.4586],[722.7638,3.3604],[723.1425,2.9137],[723.555,2.467],[723.9688,2.0964],[724.3988,1.6498],[724.8438,1.4755],[753,0]]
       }
       dyn_wings_phalange_03_r_ctrl.Rotation.Z {type=MovieSceneScriptingFloatChannel, default=0, const [0,2]}
       dyn_wings_phalange_03_r_ctrl.Scale.X {type=MovieSceneScriptingFloatChannel, default=1, const [1,2]}
       dyn_wings_phalange_03_r_ctrl.Scale.Y {type=MovieSceneScriptingFloatChannel, default=1, const [1,2]}
       dyn_wings_phalange_03_r_ctrl.Scale.Z {type=MovieSceneScriptingFloatChannel, default=1, const [1,2]}
       dyn_wings_metacarpal_04_r_ctrl.Location.X {type=MovieSceneScriptingFloatChannel, default=0, const [0,8]}
       dyn_wings_metacarpal_04_r_ctrl.Location.Y {type=MovieSceneScriptingFloatChannel, default=0, const [0,8]}
       dyn_wings_metacarpal_04_r_ctrl.Location.Z {type=MovieSceneScriptingFloatChannel, default=0, const [0,8]}
       dyn_wings_metacarpal_04_r_ctrl.Rotation.X {type=MovieSceneScriptingFloatChannel, default=0, const [0,8]}
       dyn_wings_metacarpal_04_r_ctrl.Rotation.Y {
        type=MovieSceneScriptingFloatChannel
        default=0
        thinned_from=958
        keys [[1,0.0306],[6.3062,-0.4707],[6.7175,-0.8049],[7.1075,-0.972],[7.485,-0.9415],[7.8575,-1.1391],[8.59,-1.1086],[9.3338,-1.3062],[9.715,-1.4428],[10.1025,-1.6099],[11.5512,-1.8075],[11.9062,-1.777],[12.3012,-2.1723],[12.7038,-2.3394],[13.0775,-2.4759],[13.4663,-2.8712],[13.8388,-3.0383],[14.2125,-3.0078],[14.5788,-3.2054],[14.95,-3.1749],[15.6812,-3.3725],[16.0325,-3.342],[16.4075,-3.5396],[18.1812,-3.7373],[18.9362,-4.0715],[19.6612,-4.2691],[20.0263,-4.4362],[20.4075,-4.6339],[20.7912,-4.6033],[21.1612,-4.9681],[21.595,-5.3023],[22.025,-5.8036],[22.485,-6.0767],[22.9013,-6.4414],[23.285,-6.4108],[44.22,-6.5779],[44.61,-7.0487],[45.0113,-7.3523],[45.4087,-7.7925],[45.8087,-8.2633],[46.6025,-8.5363],[46.98,-8.8094],[47.3787,-9.1436],[47.77,-9.6144],[48.1675,-10.2828],[48.6087,-10.7535],[49.0025,-11.0266],[49.4062,-11.1937],[53.6488,-11.3608],[54.0525,-11.3302],[54.4362,-11.5279],[54.81,-11.4973],[55.5712,-11.695],[55.975,-11.8315],[56.395,-12.5305],[56.815,-12.9401],[57.2138,-14.0038],[57.625,-14.4746],[58.0138,-14.6111],[58.4038,-14.5805],[60.4812,-14.9147],[60.8612,-15.0207],[61.6413,-15.3549],[72.415,-15.522],[73.2188,-15.8562],[73.6263,-16.3881],[74.0163,-16.7223],[74.41,-16.8588],[74.8025,-17.1625],[75.2075,-17.7249],[75.5912,-17.892],[75.9737,-17.8614],[82.5138,-17.7554],[82.895,-17.3458],[83.3088,-16.6325],[83.725,-15.1592],[84.1712,-13.3965],[84.6187,-12.1963],[85.0975,-11.2997],[85.5312,-10.9349],[85.9288,-10.7678],[86.69,-10.6007],[95.8725,-10.7678],[96.2513,-11.1326],[96.6425,-11.695],[97.0462,-12.561],[97.4412,-13.4576],[97.87,-14.1566],[98.2613,-14.4602],[98.6612,-14.5968],[99.0625,-15.1286],[99.4462,-15.2346],[100.2525,-16.2983],[100.6363,-16.8302],[101.025,-17.6962],[101.46,-18.1669],[101.86,-18.6377],[102.6938,-19.245],[103.1112,-19.2144],[103.9112,-19.5181],[104.32,-19.6241],[104.7075,-20.0948],[105.1213,-20.2762],[105.5163,-20.9141],[105.9362,-21.1872],[106.3338,-21.2626],[107.1012,-21.2015],[107.8288,-21.3686],[108.1963,-21.5357],[108.9513,-21.7477],[109.7275,-21.5643],[110.1287,-21.7314],[111.5712,-21.6703],[111.95,-21.8069],[112.73,-21.7152],[113.1012,-21.8823],[113.86,-21.8212],[131.2962,-22.1554],[132.0738,-22.7627],[132.4837,-23.37],[132.925,-23.9161],[133.3512,-24.9331],[133.8112,-25.7523],[134.2225,-26.4351],[134.6463,-26.6165],[135.055,-26.6919],[136.1862,-26.6308],[136.5612,-26.8285],[136.9262,-26.7979],[137.3012,-26.7063],[137.6988,-26.6757],[138.0813,-26.584],[138.4775,-26.5535],[138.85,-26.69],[142.3338,-26.6595],[151.4913,-26.69],[152.245,-26.7511],[152.63,-26.9793],[154.11,-27.071],[158.6925,-27.0404],[159.0612,-27.5866],[159.4825,-28.0717],[159.9062,-28.9521],[160.3637,-29.1029],[161.165,-28.8891],[161.56,-28.8585],[161.9362,-29.0562],[170.58,-28.8891],[170.9425,-28.5243],[171.33,-28.5549],[171.7138,-28.3878],[172.5587,-28.4794],[176.0487,-28.2818],[181.0512,-28.3429],[181.4187,-28.404],[181.805,-28.2675],[182.24,-28.298],[182.67,-28.1309],[184.4887,-28.192],[192.0487,-28.0249],[192.435,-27.4931],[192.8425,-27.0529],[193.2537,-26.8409],[193.6513,-26.2947],[194.075,-25.5814],[194.5025,-24.7927],[194.925,-23.7309],[195.3637,-22.8506],[195.7887,-22.5326],[196.23,-22.3512],[197.0763,-21.5013],[197.8825,-20.9246],[198.2675,-20.9552],[198.6588,-20.7881],[202.215,-20.621],[202.5938,-20.6515],[203.3238,-20.4539],[205.075,-20.4844],[205.8212,-20.3479],[206.2088,-20.2419],[206.6112,-19.9688],[207.01,-19.8628],[207.3938,-19.5592],[207.7925,-19.4837],[208.185,-19.0435],[208.585,-18.4668],[209.0163,-18.1488],[209.4238,-17.9674],[209.83,-17.9225],[210.2325,-17.9531],[218.39,-17.9836],[219.1475,-17.7716],[219.5425,-17.8328],[219.9275,-17.6657],[222.4175,-17.7573],[222.8288,-17.6513],[223.2225,-17.743],[223.6037,-17.6981],[224.0113,-17.5616],[224.3975,-17.5921],[224.795,-17.425],[225.5525,-17.4861],[225.9413,-17.5472],[226.3238,-17.4107],[226.715,-17.5023],[227.12,-17.3352],[228.2163,-17.5329],[228.6012,-17.8671],[228.9825,-19.265],[229.4388,-20.6323],[230.3012,-23.6239],[230.735,-25.3846],[231.17,-27.7527],[231.615,-29.5888],[232.0437,-30.6812],[232.9075,-30.5122],[233.9913,-30.7099],[234.7275,-30.7404],[235.4762,-30.5428],[236.5812,-30.6039],[246.42,-30.6344],[246.7962,-30.7566],[247.6012,-30.7585],[248,-31.003],[248.805,-31.0354],[249.195,-30.7929],[249.605,-30.9762],[249.9925,-30.8702],[252.845,-30.9313],[253.2212,-30.7642],[253.9887,-30.6134],[254.8,-30.2199],[255.655,-29.7368],[256.5075,-29.1314],[256.945,-28.6463],[257.4187,-28.1918],[257.8763,-28.2386],[258.3038,-28.102],[260.81,-28.1631],[261.8788,-27.9655],[263.3162,-27.996],[263.7175,-27.8595],[264.1137,-27.617],[264.5425,-27.1768],[265.38,-26.115],[266.2113,-25.3568],[266.6237,-25.1754],[267.8338,-23.977],[268.2513,-23.4757],[268.6437,-23.3698],[272.1087,-23.4003],[272.4812,-23.2332],[272.8675,-23.2638],[273.255,-23.0661],[278.0738,-23.0967],[278.8475,-22.8236],[279.25,-22.7481],[279.655,-22.3079],[280.0713,-21.5641],[280.5087,-20.8508],[280.925,-20.107],[281.3625,-19.7279],[281.785,-19.5914],[282.6488,-19.6525],[289.0075,-19.8196],[289.7475,-19.9867],[290.1425,-20.3514],[291.3162,-20.8527],[291.7088,-21.0504],[292.1162,-21.1869],[292.545,-21.5517],[292.9312,-21.6577],[293.7225,-22.3261],[294.505,-22.5992],[294.8875,-22.9334],[295.2887,-23.1005],[295.6812,-23.0699],[296.0725,-23.237],[297.145,-23.2064],[297.8712,-23.343],[298.2437,-23.7077],[298.6288,-23.6466],[299.0213,-23.8443],[299.8075,-23.7832],[300.9462,-24.2234],[301.3363,-24.1928],[302.1187,-24.4659],[302.5075,-24.4048],[302.9112,-24.7084],[304.47,-25.1935],[306.9813,-25.3606],[307.3988,-25.4972],[307.8138,-25.4361],[309.0962,-25.8762],[310.3338,-25.754],[311.1175,-26.0271],[311.5,-26.3002],[312.3,-26.6038],[312.7125,-26.5733],[315.1888,-26.938],[316.3663,-28.6253],[316.7525,-29.0206],[317.1488,-29.2182],[317.5338,-29.6135],[318.2837,-29.6441],[319.76,-29.8417],[321.18,-29.6441],[322.3062,-29.7052],[325.7425,-29.6746],[326.1112,-29.7806],[326.4963,-30.0537],[326.9075,-30.1597],[327.2975,-30.0986],[338.6337,-29.9315],[340.1162,-29.962],[340.9212,-29.8255],[341.7025,-29.9783],[342.5038,-29.7052],[344.385,-29.8274],[345.14,-29.5849],[345.5338,-29.7071],[345.9262,-29.6316],[346.3162,-29.6927],[350.1875,-29.5562],[350.5763,-29.3891],[351.345,-29.4502],[352.9025,-28.8735],[353.3088,-28.798],[353.6988,-28.6615],[354.0888,-28.692],[354.4688,-28.5555],[355.2225,-28.5861],[355.6163,-28.3884],[356.0025,-28.419],[357.1838,-28.0093],[357.5788,-28.0399],[357.9613,-27.8728],[358.36,-27.9339],[358.7462,-27.7362],[359.5,-27.7668],[359.8788,-27.5997],[360.2788,-27.4937],[360.6712,-27.19],[361.07,-26.7499],[361.87,-26.112],[362.2775,-25.9755],[363.7213,-25.7778],[364.095,-25.6107],[364.4812,-25.2765],[365.6525,-24.8974],[366.0325,-24.7303],[367.4638,-24.5632],[367.8388,-24.5938],[368.235,-24.2596],[369.8138,-23.6829],[378.6888,-23.9111],[379.055,-23.8805],[379.7887,-24.0782],[380.1687,-24.0476],[380.5712,-24.2453],[383.705,-24.2147],[384.49,-24.5184],[385.2325,-24.4878],[385.9937,-24.8831],[386.3863,-24.8526],[386.77,-24.9891],[387.1762,-24.928],[387.56,-25.0645],[387.9538,-25.034],[388.3388,-25.2316],[392.8388,-25.2011],[393.5688,-25.3987],[399.0738,-25.4293],[399.4375,-25.2316],[400.2075,-25.2928],[402.0525,-25.1257],[406.345,-25.1562],[407.0625,-24.9586],[410.1687,-24.9891],[415.7325,-24.9586],[416.815,-25.1257],[417.1888,-25.3987],[417.5912,-25.9755],[418.02,-26.3546],[418.44,-26.5971],[418.8525,-26.4749],[419.2375,-26.4443],[419.6313,-26.6114],[420.3913,-26.5198],[420.7838,-26.367],[421.1812,-26.3508],[421.5875,-26.3956],[421.98,-26.5016],[423.1438,-26.3489],[423.5287,-26.4243],[423.9337,-26.4081],[424.3438,-26.4978],[424.7763,-26.6936],[425.2025,-26.5389],[425.6275,-26.749],[426.0638,-26.7471],[426.4825,-26.8999],[426.9075,-26.686],[427.7225,-26.8063],[428.1213,-26.6536],[428.52,-26.729],[428.9237,-26.7433],[429.3275,-26.4378],[429.7475,-26.5887],[430.1625,-26.2526],[430.5875,-26.2669],[430.9887,-26.2058],[431.3737,-26.3729],[433.5938,-26.4034],[434.0013,-26.921],[434.4187,-27.9111],[434.88,-29.1763],[435.345,-30.6247],[435.8312,-31.1604],[436.2825,-31.209],[436.72,-30.9522],[437.13,-30.4671],[437.5812,-30.3162],[437.9925,-29.6641],[438.8763,-27.677],[439.2925,-26.5703],[439.7625,-24.9318],[440.2163,-22.9591],[440.6762,-21.396],[441.1075,-20.3342],[442.01,-17.9661],[442.4863,-16.5234],[442.9312,-15.4759],[443.3863,-14.5344],[443.8,-13.8966],[444.6263,-12.9245],[445.0263,-12.4232],[445.4462,-12.5454],[445.85,-12.0442],[446.2475,-11.9382],[447.0462,-11.604],[448.8338,-11.4063],[449.2075,-11.0721],[450.0088,-10.4037],[450.4062,-10.4343],[451.4787,-10.4037],[451.855,-10.4486],[452.2688,-11.3739],[452.7,-12.2991],[453.1475,-12.9513],[453.5838,-13.9988],[454.0013,-14.9402],[454.425,-15.5475],[455.2575,-16.3668],[455.685,-16.6704],[456.0913,-16.7459],[456.5,-16.9884],[457.2962,-17.2921],[457.6788,-17.3675],[458.085,-17.3064],[458.4762,-17.9748],[458.8837,-18.3539],[459.3162,-18.9917],[459.7287,-19.3708],[460.145,-19.6439],[460.5838,-19.7193],[461.0125,-19.8558],[461.4375,-19.8253],[461.83,-19.9924],[462.2175,-20.0984],[462.6437,-20.3409],[463.0525,-20.72],[463.4775,-21.3273],[463.8788,-21.5393],[464.2975,-21.6147],[464.7188,-21.5841],[465.1413,-21.7818],[468.4712,-21.7512],[471.6475,-21.7818],[472.0238,-22.01],[472.4462,-22.0406],[472.8425,-22.1322],[473.2487,-22.0262],[474.05,-21.48],[474.45,-21.3435],[477.65,-21.3741],[478.02,-21.4352],[478.4162,-21.2681],[478.805,-21.2986],[479.195,-21.1315],[480.3487,-21.2232],[480.7362,-21.0561],[481.1375,-20.783],[481.5338,-20.3734],[481.9625,-20.0248],[482.405,-19.4175],[483.2437,-18.5677],[484.0675,-17.991],[484.8663,-17.0801],[485.2788,-16.7459],[485.6737,-16.4728],[486.495,-15.9572],[486.8875,-15.4253],[487.2975,-15.3193],[488.11,-14.6509],[488.4937,-14.5144],[488.8825,-14.5449],[489.625,-14.3473],[490.3963,-14.4084],[490.7975,-14.2107],[495.0575,-14.4084],[495.4338,-15.2439],[495.845,-15.8817],[496.7387,-16.9149],[497.1938,-17.0514],[497.6087,-17.2185],[498.4675,-17.9175],[499.3625,-18.92],[500.2188,-19.5273],[501.0263,-19.831],[504.23,-19.9981],[504.6137,-19.9675],[505.0088,-20.1346],[505.7663,-20.1041],[506.8312,-20.3017],[507.21,-20.2712],[508.3837,-20.7114],[510.165,-20.6808],[511.9775,-20.9844],[512.365,-20.9233],[513.1675,-21.227],[513.565,-21.1659],[514.3038,-21.3635],[521.5912,-21.333],[521.965,-21.439],[522.36,-21.3779],[523.1762,-21.6509],[523.98,-21.5287],[525.175,-21.8162],[525.5725,-22.0892],[525.975,-22.4683],[526.3887,-22.7252],[526.835,-23.3774],[527.2738,-23.9684],[527.7387,-24.2702],[528.2262,-24.4802],[528.7287,-24.3399],[529.1838,-24.3972],[529.6562,-24.3629],[530.1275,-24.2225],[531.0738,-24.4021],[531.4737,-24.3715],[535.7013,-24.5386],[536.0775,-24.614],[536.4863,-24.4918],[536.8913,-24.7649],[537.3012,-24.8098],[537.7075,-24.6265],[538.115,-24.6714],[538.5187,-24.9139],[538.9375,-25.0199],[539.3325,-24.9282],[540.56,-25.2768],[541.375,-26.202],[541.8225,-27.0213],[542.2563,-28.0382],[542.7125,-28.9186],[543.145,-30.2554],[543.5963,-31.3945],[544.0362,-33.3548],[544.4963,-36.1811],[544.9375,-41.0289],[545.3887,-47.5027],[545.8288,-53.3082],[546.28,-58.6878],[546.7475,-62.4107],[547.2113,-64.9783],[547.6612,-66.6044],[548.09,-67.2728],[548.5087,-67.6376],[548.8975,-67.8047],[551.02,-67.6376],[552.6437,-66.2397],[553.0412,-66.0726],[553.4238,-66.1031],[553.8212,-65.9055],[554.62,-64.8112],[555.0263,-64.0817],[556.2475,-62.4862],[557.8425,-61.8789],[561.475,-62.0154],[561.9,-62.5922],[562.3763,-63.5031],[562.845,-64.1104],[563.2675,-64.3529],[563.6988,-64.4589],[567.9775,-64.4284],[568.3562,-64.1858],[568.775,-63.7313],[569.2225,-62.8671],[569.6875,-62.6714],[570.1413,-62.9464],[571.3925,-63.0543],[571.7938,-62.8872],[572.1963,-62.9483],[573.0025,-62.5243],[573.425,-61.8721],[574.3288,-60.0379],[575.2337,-58.7803],[575.6875,-58.3258],[576.14,-57.5676],[576.5888,-57.0071],[577.03,-55.7477],[577.9088,-53.8971],[578.3663,-53.0473],[578.8025,-52.5623],[579.2463,-51.5902],[579.6662,-50.771],[580.1112,-49.4198],[580.5487,-47.7956],[581.0088,-46.4751],[581.4538,-45.6252],[582.33,-45.124],[582.7375,-45.1545],[583.1413,-44.9874],[583.53,-44.6838],[583.9413,-44.0459],[584.7962,-41.676],[585.2312,-40.2638],[585.6662,-38.2586],[586.1237,-36.7547],[587.0138,-34.02],[587.4488,-32.0597],[587.8837,-29.1723],[588.3438,-25.859],[588.795,-23.4423],[589.2487,-22.8798],[591.795,-22.6822],[592.225,-22.7127],[596.88,-22.5151],[597.2637,-21.5879],[597.695,-19.7336],[598.1637,-16.9827],[598.615,-12.2104],[599.0838,-8.3347],[599.5175,-5.3861],[599.9625,-4.4895],[600.3813,-4.5201],[600.7738,-4.3224],[608.03,-4.2919],[608.4087,-4.3673],[608.8138,-4.8686],[609.22,-4.777],[609.63,-4.9135],[610.035,-4.8829],[610.775,-5.05],[611.1637,-5.4148],[611.5575,-5.5513],[611.9813,-6.0526],[612.4262,-7.1163],[612.9125,-7.7847],[613.335,-8.1495],[613.7475,-8.1189],[616.8788,-8.286],[617.2525,-8.2555],[618.0075,-8.4226],[618.39,-8.5897],[618.8012,-9.0604],[619.2138,-9.5006],[619.6313,-10.1079],[620.0412,-10.7457],[620.455,-11.1859],[620.8737,-11.3224],[621.2663,-11.4895],[626.795,-11.6872],[627.1775,-11.8237],[627.585,-12.0968],[627.98,-12.6287],[628.3863,-13.0383],[628.8087,-13.8738],[629.2312,-14.2834],[629.6525,-14.7236],[630.0525,-15.3309],[630.4787,-16.1359],[630.9125,-16.7126],[631.33,-17.2282],[631.78,-17.9109],[632.2113,-18.396],[632.6737,-18.5774],[633.1,-18.82],[633.54,-19.1236],[633.9325,-19.0931],[634.3188,-19.2296],[635.4625,-19.1685],[635.845,-19.3356],[636.2387,-19.2745],[637.045,-19.7901],[637.8737,-20.6094],[638.2812,-20.8519],[638.6962,-21.2921],[639.1137,-21.5041],[639.5175,-21.549],[639.9312,-22.1257],[640.8212,-22.7922],[641.2387,-23.2018],[642.0775,-23.6258],[642.5087,-24.0049],[642.955,-24.1557],[643.3887,-24.5043],[643.8025,-24.7162],[644.2287,-24.6551],[644.9588,-24.8528],[647.4238,-24.6551],[648.83,-24.488],[649.2125,-24.382],[649.6225,-24.109],[650.0238,-23.5017],[650.4775,-22.8944],[650.9025,-22.5153],[651.3413,-22.2728],[651.7437,-22.3644],[652.9562,-22.077],[653.7725,-21.8039],[659.2925,-21.6368],[659.6775,-21.3943],[660.0975,-20.8481],[660.5213,-20.1348],[660.9562,-19.1485],[661.4262,-17.889],[661.8675,-16.4319],[662.3225,-14.8077],[662.785,-13.6542],[663.22,-13.4117],[667.3863,-13.4423],[676.6625,-13.6094],[677.04,-13.9436],[677.4488,-14.6731],[677.8537,-15.0073],[678.2688,-15.5697],[678.665,-15.7062],[679.0638,-15.9039],[679.4563,-16.4357],[679.8563,-16.937],[680.2862,-17.6359],[680.6938,-18.3349],[681.1275,-18.6691],[681.5238,-18.8362],[683.6488,-19.2315],[684.0275,-19.474],[684.4512,-19.9142],[684.89,-20.4909],[685.3262,-21.2653],[685.78,-22.3739],[686.7838,-23.6782],[687.28,-24.2244],[687.735,-24.2082],[688.21,-24.4813],[688.6788,-24.3591],[689.1313,-24.6016],[689.5612,-24.5854],[689.99,-24.7974],[690.41,-24.8422],[690.8263,-24.8117],[692.5825,-24.8422],[692.96,-24.7668],[693.3687,-24.3877],[693.795,-24.1003],[694.25,-23.5093],[695.1725,-22.753],[695.6037,-21.9948],[696.0612,-21.4038],[696.4787,-21.2529],[696.91,-21.3751],[697.3125,-21.2386],[700.4362,-21.102],[700.83,-21.2242],[701.245,-21.2548],[701.6437,-21.0877],[703.455,-21.1488],[703.8512,-21.0734],[704.2537,-20.9225],[704.7062,-20.6045],[705.5538,-20.3333],[706.3775,-20.1825],[706.7862,-20.0459],[707.1888,-20.1376],[707.6,-20.001],[708,-19.7279],[708.4038,-19.6831],[708.8188,-19.3345],[709.2525,-19.0471],[709.6788,-18.0607],[710.1237,-17.6979],[710.5512,-17.4716],[710.995,-17.4267],[711.4025,-17.2902],[711.8163,-17.3207],[714.6063,-17.3513],[723,-16.0652],[727,-5.8234],[730,-5.8556],[739,-6.0296],[744,2.921],[748,2.8694],[751,0]]
       }
       dyn_wings_metacarpal_04_r_ctrl.Rotation.Z {type=MovieSceneScriptingFloatChannel, default=0, const [0,8]}
       dyn_wings_metacarpal_04_r_ctrl.Scale.X {type=MovieSceneScriptingFloatChannel, default=1, const [1,8]}
       dyn_wings_metacarpal_04_r_ctrl.Scale.Y {type=MovieSceneScriptingFloatChannel, default=1, const [1,8]}
       dyn_wings_metacarpal_04_r_ctrl.Scale.Z {type=MovieSceneScriptingFloatChannel, default=1, const [1,8]}
       dyn_wings_phalange_04_r_ctrl.Location.X {type=MovieSceneScriptingFloatChannel, default=0, const [0,8]}
       dyn_wings_phalange_04_r_ctrl.Location.Y {type=MovieSceneScriptingFloatChannel, default=0, const [0,8]}
       dyn_wings_phalange_04_r_ctrl.Location.Z {type=MovieSceneScriptingFloatChannel, default=0, const [0,8]}
       dyn_wings_phalange_04_r_ctrl.Rotation.X {type=MovieSceneScriptingFloatChannel, default=0, const [0,8]}
       dyn_wings_phalange_04_r_ctrl.Rotation.Y {
        type=MovieSceneScriptingFloatChannel
        default=-2.4672
        thinned_from=652
        keys [[0,-2.4672],[8.6712,-2.2672],[9.4338,-1.2678],[9.8125,-0.6678],[10.2138,-0.0684],[10.5963,0.131],[12.695,0.531],[13.0738,1.1316],[13.4575,1.7316],[13.83,2.1316],[14.215,2.5316],[14.5975,2.7316],[15.66,2.9316],[23.8,3.1316],[24.1562,3.1322],[24.9188,3.5316],[25.6488,3.7316],[28.0738,3.5316],[34.6388,3.9322],[35.0063,4.7322],[35.3925,5.3322],[35.7675,6.1328],[36.1513,7.1328],[36.5562,8.5333],[36.95,11.1351],[37.3562,14.7362],[37.7962,18.138],[38.2163,21.5397],[38.63,24.7408],[39.0575,27.7414],[39.4863,30.5414],[39.9275,31.942],[40.3313,33.542],[40.7475,34.942],[41.1475,35.942],[41.5475,36.542],[41.9275,36.742],[46.8463,36.542],[50.9925,36.1414],[51.3713,35.7414],[52.2287,35.3414],[53.0525,35.3414],[54.9638,34.9408],[55.3475,34.7402],[55.745,34.1396],[56.5087,33.9396],[70,33.739],[70.3863,33.5385],[70.7675,32.9379],[71.1912,32.1373],[71.5963,31.1356],[71.9825,29.7344],[72.405,28.1327],[72.83,26.3321],[73.2688,24.7309],[73.6837,22.1304],[74.5013,19.3286],[74.8975,18.3281],[75.3,17.5269],[75.7062,16.5263],[76.09,15.3257],[76.505,14.3252],[76.8863,13.9252],[82.0225,14.1252],[84.2175,14.5257],[84.5987,15.3275],[85.0175,17.9298],[85.4675,21.5339],[85.9225,26.1379],[86.415,30.1426],[86.855,32.1443],[87.2788,32.3449],[99.6,32.1449],[100.355,31.3431],[101.14,30.9426],[115.5688,31.1426],[116.31,31.3426],[116.6712,31.5426],[117.4438,31.7425],[117.85,31.7431],[118.2425,32.1425],[118.6313,32.3425],[121.77,32.5425],[122.8988,33.1437],[123.2675,33.5437],[124.0038,33.7437],[124.3637,33.9437],[125.0875,34.1443],[125.465,34.5443],[125.8563,34.7443],[126.5888,34.9443],[127.98,35.1443],[129.4112,35.5443],[129.7725,35.5449],[130.1513,35.7443],[130.8888,35.9443],[144.24,36.3443],[144.61,36.7443],[144.9887,37.3443],[145.77,38.9443],[146.1475,39.5454],[146.5663,40.7466],[147.4475,42.3483],[147.87,42.7483],[148.2425,43.3489],[148.65,43.7489],[149.07,43.9495],[149.4963,44.5495],[150.6075,44.5495],[152.0238,44.7495],[152.3863,45.3495],[153.9225,46.9501],[154.315,46.9506],[154.7188,47.1501],[155.5,47.3501],[156.1988,47.7501],[156.58,47.7506],[156.9562,47.9501],[172.6137,47.9495],[172.9787,47.7495],[173.7525,46.9483],[174.5263,46.5472],[174.9125,46.5466],[175.295,46.3472],[176.7062,46.1466],[177.44,45.9466],[183.3338,45.7466],[183.7088,45.3454],[184.4937,43.7431],[185.2912,40.9396],[185.6875,39.939],[186.4837,38.3373],[187.2887,36.3356],[187.675,34.5338],[188.1025,31.5315],[188.5438,27.128],[188.965,22.3234],[189.365,15.117],[189.7787,7.9112],[190.2237,3.1089],[190.63,-0.0928],[191.0425,-1.4928],[191.4588,-1.6928],[194.225,-1.6934],[194.595,-2.2928],[194.9875,-3.2934],[195.3713,-4.4939],[195.7613,-6.4939],[196.1662,-9.4945],[196.5812,-13.2951],[197.0138,-18.495],[197.44,-25.0956],[197.8788,-30.4962],[198.2775,-34.895],[198.71,-37.4944],[199.135,-37.6938],[203.6675,-37.4938],[204.0437,-36.8938],[204.4325,-35.4932],[204.84,-34.6926],[205.22,-33.6926],[205.635,-33.2921],[206.01,-33.0921],[206.38,-32.6921],[206.7713,-32.4921],[207.5538,-31.6915],[207.9525,-31.4915],[208.675,-31.2915],[209.3925,-30.8909],[209.7975,-30.4909],[210.25,-29.8903],[210.6825,-29.0897],[211.0913,-28.6898],[211.49,-28.4898],[213.3212,-28.0892],[214.4837,-26.288],[214.8663,-25.8874],[215.6475,-25.4874],[216.7513,-25.4874],[217.1112,-24.8874],[217.4937,-23.6857],[217.9175,-22.084],[218.745,-17.0799],[219.1587,-14.2776],[219.5987,-11.6747],[220.0138,-8.8718],[220.4275,-5.066],[220.8725,-1.2596],[221.2875,2.9456],[221.7387,5.7497],[222.1862,6.9502],[222.5838,7.1502],[232.985,6.7497],[233.3625,6.5497],[234.5338,5.3479],[234.9212,4.5473],[235.3112,3.9468],[235.7113,2.9456],[236.125,1.945],[236.5275,1.1444],[236.9312,0.5439],[237.3062,-0.2567],[237.69,-0.4573],[238.065,-0.6573],[245.885,-0.4573],[248.315,-0.2573],[248.695,-0.0567],[249.0738,0.5433],[249.4688,0.7433],[249.8363,0.9433],[250.2113,1.1439],[250.945,1.3439],[254.7362,1.5439],[255.1087,1.7444],[255.5,2.1444],[255.875,2.545],[256.2525,2.745],[256.9837,2.945],[271.2525,2.9456],[271.6263,3.345],[272.0025,3.9456],[272.3963,4.5462],[272.7725,5.3468],[273.18,6.3479],[273.615,8.5491],[274.0863,10.7508],[274.5275,12.952],[274.9413,15.1525],[275.37,17.3537],[275.8138,19.3548],[276.7162,21.7554],[277.1237,22.5554],[277.5225,23.156],[277.9088,23.956],[278.3088,24.156],[286.23,24.356],[295.6063,24.356],[297.6762,23.956],[298.0487,23.156],[298.46,22.556],[298.8388,21.7554],[299.2375,21.3548],[299.6187,21.1548],[299.995,20.7548],[300.7413,20.7548],[301.4887,20.5548],[305.3,20.3548],[305.7025,20.1543],[313.8475,19.9537],[314.2225,19.7537],[314.61,19.3537],[319.4712,19.5543],[320.225,19.9537],[320.6263,19.9543],[322.1538,20.7543],[322.53,21.1543],[323.2912,21.5548],[336.9837,21.7554],[337.3887,21.9548],[338.1313,22.1548],[354.8937,22.1548],[359.0225,21.9548],[360.4512,21.9548],[368.8475,21.7548],[369.615,21.5548],[380.29,21.7548],[383.7663,21.7548],[384.1275,21.9548],[386.5963,21.9548],[386.975,22.1548],[400.4762,22.3548],[415.3138,22.1548],[418.4438,21.9543],[418.8163,21.7543],[419.1975,20.9531],[419.6,19.9519],[419.99,19.1514],[420.3988,17.3502],[420.8163,15.749],[421.265,13.1473],[421.7325,9.9456],[422.1812,5.3433],[422.5987,0.1398],[423.0375,-4.8631],[423.4613,-10.0666],[423.89,-14.0683],[424.3138,-17.2694],[424.7387,-19.67],[425.15,-20.67],[425.5688,-20.87],[431.3975,-20.67],[431.8012,-20.0688],[432.2412,-18.4683],[432.6825,-17.2671],[433.085,-16.2659],[433.4913,-15.4654],[433.9025,-14.2648],[434.3313,-12.2636],[434.7975,-9.8619],[435.2325,-7.259],[435.67,-5.0573],[436.085,-3.6567],[436.4825,-2.4555],[436.8825,-1.8555],[437.2563,-1.4555],[437.65,-0.8555],[438.0288,-0.055],[438.42,0.5462],[438.815,1.9468],[439.215,2.7474],[439.6087,3.9479],[440.3787,5.5497],[440.7675,6.7508],[441.1612,8.152],[441.5775,9.5531],[441.9862,11.7549],[442.8225,16.1584],[443.2575,17.5601],[444.0987,19.1624],[444.4925,19.763],[445.2763,20.5642],[445.6725,21.3653],[446.0675,21.7659],[446.4737,22.5676],[446.86,23.1688],[447.2437,23.9694],[448.0337,24.7711],[448.4262,24.9717],[449.1612,24.9723],[449.5612,25.1723],[449.955,25.5734],[450.7225,25.5734],[453.8388,25.3734],[454.2,25.1734],[454.5838,24.5734],[455.355,23.7734],[455.7313,23.5735],[456.1087,22.9735],[457.2675,21.7723],[458.085,21.3723],[461.2637,21.1723],[461.9887,20.9723],[468.0825,21.1723],[469.8012,21.5723],[471.6837,22.5723],[472.8112,22.5723],[473.1912,22.7723],[473.9275,22.9723],[474.2937,23.3723],[474.6988,23.5729],[475.4525,23.7729],[490.9275,23.9729],[492.3487,24.3734],[492.7287,24.9734],[493.1175,25.3734],[493.505,25.574],[495.0475,27.174],[495.4638,27.774],[495.8888,28.174],[497.0938,28.774],[498.2575,28.974],[498.985,29.174],[499.3438,29.3746],[499.7225,29.7746],[500.8775,30.3757],[501.2575,30.7757],[502.0263,31.1757],[503.1062,31.3763],[503.8663,31.7763],[504.245,32.1769],[505.0125,32.5769],[505.8212,32.7769],[507.6063,32.7769],[511.4137,32.5769],[512.4762,32.3763],[512.8425,32.1763],[513.9212,31.9763],[515.4325,31.1752],[516.91,30.7752],[517.2688,30.3752],[518.4025,29.7746],[518.785,29.1746],[519.9162,28.5746],[531.4812,28.1746],[531.855,27.374],[532.2412,26.7734],[532.6388,25.9734],[533.4137,24.7729],[533.7987,24.3729],[534.9775,21.9717],[535.7587,20.7711],[536.5438,19.9705],[536.9262,19.7705],[537.65,19.5706],[538.0238,19.17],[538.8012,17.1694],[539.1963,16.7688],[539.5925,16.1688],[579.6737,16.3688],[586.5362,16.1688],[586.9075,14.7677],[587.2975,12.9665],[587.6938,11.3659],[588.095,9.9654],[588.4987,9.3654],[589.2537,8.9654],[590.3887,8.5654],[590.785,8.1654],[591.2212,7.1654],[591.62,6.1654],[592.0475,5.3648],[592.4837,4.5648],[592.9075,3.7648],[593.3162,2.9648],[593.7075,1.7648],[594.1087,0.9642],[594.4975,-0.2358],[594.9013,-1.0358],[595.2925,-1.8363],[595.6812,-2.4363],[596.07,-2.6363],[596.44,-3.0363],[596.815,-3.2363],[597.1988,-3.8363],[597.6012,-4.4363],[597.9875,-5.2363],[598.7625,-5.6369],[606.0888,-5.4369],[607.835,-5.2363],[608.22,-5.0363],[608.9787,-4.8363],[609.7812,-4.4357],[610.8825,-4.2357],[614.6675,-4.0357],[615.4288,-4.0357],[616.1513,-3.8357],[616.5138,-3.6352],[616.9062,-3.0352],[617.2862,-2.4334],[617.695,-1.8328],[618.4575,-1.4323],[618.845,-1.2317],[619.2212,-1.0317],[619.5912,-0.6311],[619.9813,-0.2305],[620.3613,0.5701],[620.74,0.9706],[621.525,0.9706],[628.6187,1.1706],[629.0013,2.3712],[629.3925,3.5724],[629.8038,4.9741],[630.2025,6.1753],[630.6037,7.577],[631.01,10.1799],[631.4625,13.3834],[631.9162,15.3851],[632.3175,16.3857],[632.7213,16.7863],[633.1087,16.9863],[638.6187,16.7863],[639.3313,16.5863],[642.815,16.5869],[643.175,17.1869],[643.565,18.588],[643.9762,19.5892],[644.365,20.7903],[644.7738,21.1903],[645.1513,21.3909],[652.3275,21.7909],[652.7088,23.1932],[653.115,26.1967],[653.605,30.0019],[654.09,33.206],[654.5387,35.6083],[654.9675,37.01],[655.4212,37.4106],[670.6962,37.2106],[671.4262,37.01],[672.1825,37.01],[672.9362,36.2089],[674.8175,35.206],[675.555,35.006],[676.3025,35.006],[676.6688,34.806],[677.7425,34.6054],[678.1137,34.4054],[678.4863,34.4048],[678.8575,34.2054],[679.2362,34.2048],[679.615,34.0054],[680.3687,34.0054],[681.0738,33.8054],[681.4475,33.6048],[682.2225,32.8025],[683.4013,30.3996],[683.7787,30.1996],[684.5325,30.1996],[691.0537,29.799],[691.4475,28.5978],[691.8712,26.9955],[692.3138,25.1926],[692.7437,22.5891],[693.19,19.9862],[693.6125,17.5833],[694.0487,15.3799],[694.4688,13.5775],[694.8888,12.977],[695.2625,12.777],[696.0138,12.777],[701.4613,12.577],[701.8338,11.9764],[702.2163,11.5764],[702.5938,11.3764],[713.2163,11.377],[714.345,11.9776],[714.7387,12.5787],[715.1162,12.7793],[723,12.7648],[727,9.3059],[730,9.2772],[734,-2.0704],[744,-2.2466],[746,-0.0372],[751,0]]
       }
       dyn_wings_phalange_04_r_ctrl.Rotation.Z {type=MovieSceneScriptingFloatChannel, default=0, const [0,8]}
       dyn_wings_phalange_04_r_ctrl.Scale.X {type=MovieSceneScriptingFloatChannel, default=1, const [1,8]}
       dyn_wings_phalange_04_r_ctrl.Scale.Y {type=MovieSceneScriptingFloatChannel, default=1, const [1,8]}
       dyn_wings_phalange_04_r_ctrl.Scale.Z {type=MovieSceneScriptingFloatChannel, default=1, const [1,8]}
      ]
     }
     {
      type=MovieSceneControlRigParameterSection
      end_frame=753
      start_frame=0
      channels [
       pack_ctrl.Location.X {type=MovieSceneScriptingFloatChannel, default=0}
       pack_ctrl.Location.Y {type=MovieSceneScriptingFloatChannel, default=0}
       pack_ctrl.Location.Z {type=MovieSceneScriptingFloatChannel, default=0}
       pack_ctrl.Rotation.X {type=MovieSceneScriptingFloatChannel, default=0}
       pack_ctrl.Rotation.Y {type=MovieSceneScriptingFloatChannel, default=0}
       pack_ctrl.Rotation.Z {type=MovieSceneScriptingFloatChannel, default=0}
       pack_ctrl.Scale.X {type=MovieSceneScriptingFloatChannel, default=0}
       pack_ctrl.Scale.Y {type=MovieSceneScriptingFloatChannel, default=0}
       pack_ctrl.Scale.Z {type=MovieSceneScriptingFloatChannel, default=0}
       dyn_wings_humelus_l_ctrl.Location.X {type=MovieSceneScriptingFloatChannel, default=0}
       dyn_wings_humelus_l_ctrl.Location.Y {type=MovieSceneScriptingFloatChannel, default=0}
       dyn_wings_humelus_l_ctrl.Location.Z {type=MovieSceneScriptingFloatChannel, default=0}
       dyn_wings_humelus_l_ctrl.Rotation.X {type=MovieSceneScriptingFloatChannel, default=0}
       dyn_wings_humelus_l_ctrl.Rotation.Y {type=MovieSceneScriptingFloatChannel, default=0}
       dyn_wings_humelus_l_ctrl.Rotation.Z {type=MovieSceneScriptingFloatChannel, default=0}
       dyn_wings_humelus_l_ctrl.Scale.X {type=MovieSceneScriptingFloatChannel, default=0}
       dyn_wings_humelus_l_ctrl.Scale.Y {type=MovieSceneScriptingFloatChannel, default=0}
       dyn_wings_humelus_l_ctrl.Scale.Z {type=MovieSceneScriptingFloatChannel, default=0}
       dyn_wings_radius_l_ctrl.Location.X {type=MovieSceneScriptingFloatChannel, default=0}
       dyn_wings_radius_l_ctrl.Location.Y {type=MovieSceneScriptingFloatChannel, default=0}
       dyn_wings_radius_l_ctrl.Location.Z {type=MovieSceneScriptingFloatChannel, default=0}
       dyn_wings_radius_l_ctrl.Rotation.X {type=MovieSceneScriptingFloatChannel, default=0}
       dyn_wings_radius_l_ctrl.Rotation.Y {type=MovieSceneScriptingFloatChannel, default=0}
       dyn_wings_radius_l_ctrl.Rotation.Z {type=MovieSceneScriptingFloatChannel, default=0}
       dyn_wings_radius_l_ctrl.Scale.X {type=MovieSceneScriptingFloatChannel, default=0}
       dyn_wings_radius_l_ctrl.Scale.Y {type=MovieSceneScriptingFloatChannel, default=0}
       dyn_wings_radius_l_ctrl.Scale.Z {type=MovieSceneScriptingFloatChannel, default=0}
       dyn_wings_metacarpal_01_l_ctrl.Location.X {type=MovieSceneScriptingFloatChannel, default=0}
       dyn_wings_metacarpal_01_l_ctrl.Location.Y {type=MovieSceneScriptingFloatChannel, default=0}
       dyn_wings_metacarpal_01_l_ctrl.Location.Z {type=MovieSceneScriptingFloatChannel, default=0}
       dyn_wings_metacarpal_01_l_ctrl.Rotation.X {type=MovieSceneScriptingFloatChannel, default=0}
       dyn_wings_metacarpal_01_l_ctrl.Rotation.Y {type=MovieSceneScriptingFloatChannel, default=0}
       dyn_wings_metacarpal_01_l_ctrl.Rotation.Z {type=MovieSceneScriptingFloatChannel, default=0}
       dyn_wings_metacarpal_01_l_ctrl.Scale.X {type=MovieSceneScriptingFloatChannel, default=0}
       dyn_wings_metacarpal_01_l_ctrl.Scale.Y {type=MovieSceneScriptingFloatChannel, default=0}
       dyn_wings_metacarpal_01_l_ctrl.Scale.Z {type=MovieSceneScriptingFloatChannel, default=0}
       dyn_wings_phalange_01_l_ctrl.Location.X {type=MovieSceneScriptingFloatChannel, default=0}
       dyn_wings_phalange_01_l_ctrl.Location.Y {type=MovieSceneScriptingFloatChannel, default=0}
       dyn_wings_phalange_01_l_ctrl.Location.Z {type=MovieSceneScriptingFloatChannel, default=0}
       dyn_wings_phalange_01_l_ctrl.Rotation.X {type=MovieSceneScriptingFloatChannel, default=0}
       dyn_wings_phalange_01_l_ctrl.Rotation.Y {type=MovieSceneScriptingFloatChannel, default=0}
       dyn_wings_phalange_01_l_ctrl.Rotation.Z {type=MovieSceneScriptingFloatChannel, default=0}
       dyn_wings_phalange_01_l_ctrl.Scale.X {type=MovieSceneScriptingFloatChannel, default=0}
       dyn_wings_phalange_01_l_ctrl.Scale.Y {type=MovieSceneScriptingFloatChannel, default=0}
       dyn_wings_phalange_01_l_ctrl.Scale.Z {type=MovieSceneScriptingFloatChannel, default=0}
       dyn_wings_carpals_01_l_ctrl.Location.X {type=MovieSceneScriptingFloatChannel, default=0, const [0,5]}
       dyn_wings_carpals_01_l_ctrl.Location.Y {type=MovieSceneScriptingFloatChannel, default=0, const [0,5]}
       dyn_wings_carpals_01_l_ctrl.Location.Z {type=MovieSceneScriptingFloatChannel, default=0, const [0,5]}
       dyn_wings_carpals_01_l_ctrl.Rotation.X {type=MovieSceneScriptingFloatChannel, default=0, const [0,5]}
       dyn_wings_carpals_01_l_ctrl.Rotation.Y {type=MovieSceneScriptingFloatChannel, default=0, keys [[233,0],[244,34.3408],[265,34.3408],[273,16.23],[315,0]]}
       dyn_wings_carpals_01_l_ctrl.Rotation.Z {type=MovieSceneScriptingFloatChannel, default=0, const [0,5]}
       dyn_wings_carpals_01_l_ctrl.Scale.X {type=MovieSceneScriptingFloatChannel, default=0, const [0,5]}
       dyn_wings_carpals_01_l_ctrl.Scale.Y {type=MovieSceneScriptingFloatChannel, default=0, const [0,5]}
       dyn_wings_carpals_01_l_ctrl.Scale.Z {type=MovieSceneScriptingFloatChannel, default=0, const [0,5]}
       dyn_wings_metacarpal_02_l_ctrl.Location.X {type=MovieSceneScriptingFloatChannel, default=0}
       dyn_wings_metacarpal_02_l_ctrl.Location.Y {type=MovieSceneScriptingFloatChannel, default=0}
       dyn_wings_metacarpal_02_l_ctrl.Location.Z {type=MovieSceneScriptingFloatChannel, default=0}
       dyn_wings_metacarpal_02_l_ctrl.Rotation.X {type=MovieSceneScriptingFloatChannel, default=0}
       dyn_wings_metacarpal_02_l_ctrl.Rotation.Y {type=MovieSceneScriptingFloatChannel, default=0}
       dyn_wings_metacarpal_02_l_ctrl.Rotation.Z {type=MovieSceneScriptingFloatChannel, default=0}
       dyn_wings_metacarpal_02_l_ctrl.Scale.X {type=MovieSceneScriptingFloatChannel, default=0}
       dyn_wings_metacarpal_02_l_ctrl.Scale.Y {type=MovieSceneScriptingFloatChannel, default=0}
       dyn_wings_metacarpal_02_l_ctrl.Scale.Z {type=MovieSceneScriptingFloatChannel, default=0}
       dyn_wings_phalange_02_l_ctrl.Location.X {type=MovieSceneScriptingFloatChannel, default=0}
       dyn_wings_phalange_02_l_ctrl.Location.Y {type=MovieSceneScriptingFloatChannel, default=0}
       dyn_wings_phalange_02_l_ctrl.Location.Z {type=MovieSceneScriptingFloatChannel, default=0}
       dyn_wings_phalange_02_l_ctrl.Rotation.X {type=MovieSceneScriptingFloatChannel, default=0}
       dyn_wings_phalange_02_l_ctrl.Rotation.Y {type=MovieSceneScriptingFloatChannel, default=0}
       dyn_wings_phalange_02_l_ctrl.Rotation.Z {type=MovieSceneScriptingFloatChannel, default=0}
       dyn_wings_phalange_02_l_ctrl.Scale.X {type=MovieSceneScriptingFloatChannel, default=0}
       dyn_wings_phalange_02_l_ctrl.Scale.Y {type=MovieSceneScriptingFloatChannel, default=0}
       dyn_wings_phalange_02_l_ctrl.Scale.Z {type=MovieSceneScriptingFloatChannel, default=0}
       dyn_wings_carpals_02_l_ctrl.Location.X {type=MovieSceneScriptingFloatChannel, default=0, const [0,5]}
       dyn_wings_carpals_02_l_ctrl.Location.Y {type=MovieSceneScriptingFloatChannel, default=0, const [0,5]}
       dyn_wings_carpals_02_l_ctrl.Location.Z {type=MovieSceneScriptingFloatChannel, default=0, const [0,5]}
       dyn_wings_carpals_02_l_ctrl.Rotation.X {type=MovieSceneScriptingFloatChannel, default=0, const [0,5]}
       dyn_wings_carpals_02_l_ctrl.Rotation.Y {type=MovieSceneScriptingFloatChannel, default=0, keys [[233,0],[244,-30.0819],[265,-30.0819],[273,-28.018],[315,0]]}
       dyn_wings_carpals_02_l_ctrl.Rotation.Z {type=MovieSceneScriptingFloatChannel, default=0, const [0,5]}
       dyn_wings_carpals_02_l_ctrl.Scale.X {type=MovieSceneScriptingFloatChannel, default=0, const [0,5]}
       dyn_wings_carpals_02_l_ctrl.Scale.Y {type=MovieSceneScriptingFloatChannel, default=0, const [0,5]}
       dyn_wings_carpals_02_l_ctrl.Scale.Z {type=MovieSceneScriptingFloatChannel, default=0, const [0,5]}
       dyn_wings_metacarpal_03_l_ctrl.Location.X {type=MovieSceneScriptingFloatChannel, default=0}
       dyn_wings_metacarpal_03_l_ctrl.Location.Y {type=MovieSceneScriptingFloatChannel, default=0}
       dyn_wings_metacarpal_03_l_ctrl.Location.Z {type=MovieSceneScriptingFloatChannel, default=0}
       dyn_wings_metacarpal_03_l_ctrl.Rotation.X {type=MovieSceneScriptingFloatChannel, default=0}
       dyn_wings_metacarpal_03_l_ctrl.Rotation.Y {type=MovieSceneScriptingFloatChannel, default=0}
       dyn_wings_metacarpal_03_l_ctrl.Rotation.Z {type=MovieSceneScriptingFloatChannel, default=0}
       dyn_wings_metacarpal_03_l_ctrl.Scale.X {type=MovieSceneScriptingFloatChannel, default=0}
       dyn_wings_metacarpal_03_l_ctrl.Scale.Y {type=MovieSceneScriptingFloatChannel, default=0}
       dyn_wings_metacarpal_03_l_ctrl.Scale.Z {type=MovieSceneScriptingFloatChannel, default=0}
       dyn_wings_phalange_03_l_ctrl.Location.X {type=MovieSceneScriptingFloatChannel, default=0}
       dyn_wings_phalange_03_l_ctrl.Location.Y {type=MovieSceneScriptingFloatChannel, default=0}
       dyn_wings_phalange_03_l_ctrl.Location.Z {type=MovieSceneScriptingFloatChannel, default=0}
       dyn_wings_phalange_03_l_ctrl.Rotation.X {type=MovieSceneScriptingFloatChannel, default=0}
       dyn_wings_phalange_03_l_ctrl.Rotation.Y {type=MovieSceneScriptingFloatChannel, default=0}
       dyn_wings_phalange_03_l_ctrl.Rotation.Z {type=MovieSceneScriptingFloatChannel, default=0}
       dyn_wings_phalange_03_l_ctrl.Scale.X {type=MovieSceneScriptingFloatChannel, default=0}
       dyn_wings_phalange_03_l_ctrl.Scale.Y {type=MovieSceneScriptingFloatChannel, default=0}
       dyn_wings_phalange_03_l_ctrl.Scale.Z {type=MovieSceneScriptingFloatChannel, default=0}
       dyn_wings_metacarpal_04_l_ctrl.Location.X {type=MovieSceneScriptingFloatChannel, default=0}
       dyn_wings_metacarpal_04_l_ctrl.Location.Y {type=MovieSceneScriptingFloatChannel, default=0}
       dyn_wings_metacarpal_04_l_ctrl.Location.Z {type=MovieSceneScriptingFloatChannel, default=0}
       dyn_wings_metacarpal_04_l_ctrl.Rotation.X {type=MovieSceneScriptingFloatChannel, default=0}
       dyn_wings_metacarpal_04_l_ctrl.Rotation.Y {type=MovieSceneScriptingFloatChannel, default=0}
       dyn_wings_metacarpal_04_l_ctrl.Rotation.Z {type=MovieSceneScriptingFloatChannel, default=0}
       dyn_wings_metacarpal_04_l_ctrl.Scale.X {type=MovieSceneScriptingFloatChannel, default=0}
       dyn_wings_metacarpal_04_l_ctrl.Scale.Y {type=MovieSceneScriptingFloatChannel, default=0}
       dyn_wings_metacarpal_04_l_ctrl.Scale.Z {type=MovieSceneScriptingFloatChannel, default=0}
       dyn_wings_phalange_04_l_ctrl.Location.X {type=MovieSceneScriptingFloatChannel, default=0}
       dyn_wings_phalange_04_l_ctrl.Location.Y {type=MovieSceneScriptingFloatChannel, default=0}
       dyn_wings_phalange_04_l_ctrl.Location.Z {type=MovieSceneScriptingFloatChannel, default=0}
       dyn_wings_phalange_04_l_ctrl.Rotation.X {type=MovieSceneScriptingFloatChannel, default=5.5394}
       dyn_wings_phalange_04_l_ctrl.Rotation.Y {type=MovieSceneScriptingFloatChannel, default=11.0403}
       dyn_wings_phalange_04_l_ctrl.Rotation.Z {type=MovieSceneScriptingFloatChannel, default=-2.9973}
       dyn_wings_phalange_04_l_ctrl.Scale.X {type=MovieSceneScriptingFloatChannel, default=0}
       dyn_wings_phalange_04_l_ctrl.Scale.Y {type=MovieSceneScriptingFloatChannel, default=0}
       dyn_wings_phalange_04_l_ctrl.Scale.Z {type=MovieSceneScriptingFloatChannel, default=0}
       dyn_wings_humelus_r_ctrl.Location.X {type=MovieSceneScriptingFloatChannel, default=0}
       dyn_wings_humelus_r_ctrl.Location.Y {type=MovieSceneScriptingFloatChannel, default=0}
       dyn_wings_humelus_r_ctrl.Location.Z {type=MovieSceneScriptingFloatChannel, default=0}
       dyn_wings_humelus_r_ctrl.Rotation.X {type=MovieSceneScriptingFloatChannel, default=0}
       dyn_wings_humelus_r_ctrl.Rotation.Y {type=MovieSceneScriptingFloatChannel, default=0}
       dyn_wings_humelus_r_ctrl.Rotation.Z {type=MovieSceneScriptingFloatChannel, default=0}
       dyn_wings_humelus_r_ctrl.Scale.X {type=MovieSceneScriptingFloatChannel, default=0}
       dyn_wings_humelus_r_ctrl.Scale.Y {type=MovieSceneScriptingFloatChannel, default=0}
       dyn_wings_humelus_r_ctrl.Scale.Z {type=MovieSceneScriptingFloatChannel, default=0}
       dyn_wings_radius_r_ctrl.Location.X {type=MovieSceneScriptingFloatChannel, default=0}
       dyn_wings_radius_r_ctrl.Location.Y {type=MovieSceneScriptingFloatChannel, default=0}
       dyn_wings_radius_r_ctrl.Location.Z {type=MovieSceneScriptingFloatChannel, default=0}
       dyn_wings_radius_r_ctrl.Rotation.X {type=MovieSceneScriptingFloatChannel, default=0}
       dyn_wings_radius_r_ctrl.Rotation.Y {type=MovieSceneScriptingFloatChannel, default=0}
       dyn_wings_radius_r_ctrl.Rotation.Z {type=MovieSceneScriptingFloatChannel, default=0}
       dyn_wings_radius_r_ctrl.Scale.X {type=MovieSceneScriptingFloatChannel, default=0}
       dyn_wings_radius_r_ctrl.Scale.Y {type=MovieSceneScriptingFloatChannel, default=0}
       dyn_wings_radius_r_ctrl.Scale.Z {type=MovieSceneScriptingFloatChannel, default=0}
       dyn_wings_metacarpal_01_r_ctrl.Location.X {type=MovieSceneScriptingFloatChannel, default=0}
       dyn_wings_metacarpal_01_r_ctrl.Location.Y {type=MovieSceneScriptingFloatChannel, default=0}
       dyn_wings_metacarpal_01_r_ctrl.Location.Z {type=MovieSceneScriptingFloatChannel, default=0}
       dyn_wings_metacarpal_01_r_ctrl.Rotation.X {type=MovieSceneScriptingFloatChannel, default=0}
       dyn_wings_metacarpal_01_r_ctrl.Rotation.Y {type=MovieSceneScriptingFloatChannel, default=0}
       dyn_wings_metacarpal_01_r_ctrl.Rotation.Z {type=MovieSceneScriptingFloatChannel, default=0}
       dyn_wings_metacarpal_01_r_ctrl.Scale.X {type=MovieSceneScriptingFloatChannel, default=0}
       dyn_wings_metacarpal_01_r_ctrl.Scale.Y {type=MovieSceneScriptingFloatChannel, default=0}
       dyn_wings_metacarpal_01_r_ctrl.Scale.Z {type=MovieSceneScriptingFloatChannel, default=0}
       dyn_wings_phalange_01_r_ctrl.Location.X {type=MovieSceneScriptingFloatChannel, default=0}
       dyn_wings_phalange_01_r_ctrl.Location.Y {type=MovieSceneScriptingFloatChannel, default=0}
       dyn_wings_phalange_01_r_ctrl.Location.Z {type=MovieSceneScriptingFloatChannel, default=0}
       dyn_wings_phalange_01_r_ctrl.Rotation.X {type=MovieSceneScriptingFloatChannel, default=0}
       dyn_wings_phalange_01_r_ctrl.Rotation.Y {type=MovieSceneScriptingFloatChannel, default=0}
       dyn_wings_phalange_01_r_ctrl.Rotation.Z {type=MovieSceneScriptingFloatChannel, default=0}
       dyn_wings_phalange_01_r_ctrl.Scale.X {type=MovieSceneScriptingFloatChannel, default=0}
       dyn_wings_phalange_01_r_ctrl.Scale.Y {type=MovieSceneScriptingFloatChannel, default=0}
       dyn_wings_phalange_01_r_ctrl.Scale.Z {type=MovieSceneScriptingFloatChannel, default=0}
       dyn_wings_carpals_01_r_ctrl.Location.X {type=MovieSceneScriptingFloatChannel, default=0}
       dyn_wings_carpals_01_r_ctrl.Location.Y {type=MovieSceneScriptingFloatChannel, default=0}
       dyn_wings_carpals_01_r_ctrl.Location.Z {type=MovieSceneScriptingFloatChannel, default=0}
       dyn_wings_carpals_01_r_ctrl.Rotation.X {type=MovieSceneScriptingFloatChannel, default=0}
       dyn_wings_carpals_01_r_ctrl.Rotation.Y {type=MovieSceneScriptingFloatChannel, default=0}
       dyn_wings_carpals_01_r_ctrl.Rotation.Z {type=MovieSceneScriptingFloatChannel, default=0}
       dyn_wings_carpals_01_r_ctrl.Scale.X {type=MovieSceneScriptingFloatChannel, default=0}
       dyn_wings_carpals_01_r_ctrl.Scale.Y {type=MovieSceneScriptingFloatChannel, default=0}
       dyn_wings_carpals_01_r_ctrl.Scale.Z {type=MovieSceneScriptingFloatChannel, default=0}
       dyn_wings_metacarpal_02_r_ctrl.Location.X {type=MovieSceneScriptingFloatChannel, default=0}
       dyn_wings_metacarpal_02_r_ctrl.Location.Y {type=MovieSceneScriptingFloatChannel, default=0}
       dyn_wings_metacarpal_02_r_ctrl.Location.Z {type=MovieSceneScriptingFloatChannel, default=0}
       dyn_wings_metacarpal_02_r_ctrl.Rotation.X {type=MovieSceneScriptingFloatChannel, default=0}
       dyn_wings_metacarpal_02_r_ctrl.Rotation.Y {type=MovieSceneScriptingFloatChannel, default=0}
       dyn_wings_metacarpal_02_r_ctrl.Rotation.Z {type=MovieSceneScriptingFloatChannel, default=0}
       dyn_wings_metacarpal_02_r_ctrl.Scale.X {type=MovieSceneScriptingFloatChannel, default=0}
       dyn_wings_metacarpal_02_r_ctrl.Scale.Y {type=MovieSceneScriptingFloatChannel, default=0}
       dyn_wings_metacarpal_02_r_ctrl.Scale.Z {type=MovieSceneScriptingFloatChannel, default=0}
       dyn_wings_phalange_02_r_ctrl.Location.X {type=MovieSceneScriptingFloatChannel, default=0}
       dyn_wings_phalange_02_r_ctrl.Location.Y {type=MovieSceneScriptingFloatChannel, default=0}
       dyn_wings_phalange_02_r_ctrl.Location.Z {type=MovieSceneScriptingFloatChannel, default=0}
       dyn_wings_phalange_02_r_ctrl.Rotation.X {type=MovieSceneScriptingFloatChannel, default=0}
       dyn_wings_phalange_02_r_ctrl.Rotation.Y {type=MovieSceneScriptingFloatChannel, default=0}
       dyn_wings_phalange_02_r_ctrl.Rotation.Z {type=MovieSceneScriptingFloatChannel, default=0}
       dyn_wings_phalange_02_r_ctrl.Scale.X {type=MovieSceneScriptingFloatChannel, default=0}
       dyn_wings_phalange_02_r_ctrl.Scale.Y {type=MovieSceneScriptingFloatChannel, default=0}
       dyn_wings_phalange_02_r_ctrl.Scale.Z {type=MovieSceneScriptingFloatChannel, default=0}
       dyn_wings_carpals_02_r_ctrl.Location.X {type=MovieSceneScriptingFloatChannel, default=0}
       dyn_wings_carpals_02_r_ctrl.Location.Y {type=MovieSceneScriptingFloatChannel, default=0}
       dyn_wings_carpals_02_r_ctrl.Location.Z {type=MovieSceneScriptingFloatChannel, default=0}
       dyn_wings_carpals_02_r_ctrl.Rotation.X {type=MovieSceneScriptingFloatChannel, default=0}
       dyn_wings_carpals_02_r_ctrl.Rotation.Y {type=MovieSceneScriptingFloatChannel, default=0}
       dyn_wings_carpals_02_r_ctrl.Rotation.Z {type=MovieSceneScriptingFloatChannel, default=0}
       dyn_wings_carpals_02_r_ctrl.Scale.X {type=MovieSceneScriptingFloatChannel, default=0}
       dyn_wings_carpals_02_r_ctrl.Scale.Y {type=MovieSceneScriptingFloatChannel, default=0}
       dyn_wings_carpals_02_r_ctrl.Scale.Z {type=MovieSceneScriptingFloatChannel, default=0}
       dyn_wings_metacarpal_03_r_ctrl.Location.X {type=MovieSceneScriptingFloatChannel, default=0}
       dyn_wings_metacarpal_03_r_ctrl.Location.Y {type=MovieSceneScriptingFloatChannel, default=0}
       dyn_wings_metacarpal_03_r_ctrl.Location.Z {type=MovieSceneScriptingFloatChannel, default=0}
       dyn_wings_metacarpal_03_r_ctrl.Rotation.X {type=MovieSceneScriptingFloatChannel, default=0}
       dyn_wings_metacarpal_03_r_ctrl.Rotation.Y {type=MovieSceneScriptingFloatChannel, default=0}
       dyn_wings_metacarpal_03_r_ctrl.Rotation.Z {type=MovieSceneScriptingFloatChannel, default=0}
       dyn_wings_metacarpal_03_r_ctrl.Scale.X {type=MovieSceneScriptingFloatChannel, default=0}
       dyn_wings_metacarpal_03_r_ctrl.Scale.Y {type=MovieSceneScriptingFloatChannel, default=0}
       dyn_wings_metacarpal_03_r_ctrl.Scale.Z {type=MovieSceneScriptingFloatChannel, default=0}
       dyn_wings_phalange_03_r_ctrl.Location.X {type=MovieSceneScriptingFloatChannel, default=0}
       dyn_wings_phalange_03_r_ctrl.Location.Y {type=MovieSceneScriptingFloatChannel, default=0}
       dyn_wings_phalange_03_r_ctrl.Location.Z {type=MovieSceneScriptingFloatChannel, default=0}
       dyn_wings_phalange_03_r_ctrl.Rotation.X {type=MovieSceneScriptingFloatChannel, default=0}
       dyn_wings_phalange_03_r_ctrl.Rotation.Y {type=MovieSceneScriptingFloatChannel, default=0}
       dyn_wings_phalange_03_r_ctrl.Rotation.Z {type=MovieSceneScriptingFloatChannel, default=0}
       dyn_wings_phalange_03_r_ctrl.Scale.X {type=MovieSceneScriptingFloatChannel, default=0}
       dyn_wings_phalange_03_r_ctrl.Scale.Y {type=MovieSceneScriptingFloatChannel, default=0}
       dyn_wings_phalange_03_r_ctrl.Scale.Z {type=MovieSceneScriptingFloatChannel, default=0}
       dyn_wings_metacarpal_04_r_ctrl.Location.X {type=MovieSceneScriptingFloatChannel, default=0}
       dyn_wings_metacarpal_04_r_ctrl.Location.Y {type=MovieSceneScriptingFloatChannel, default=0}
       dyn_wings_metacarpal_04_r_ctrl.Location.Z {type=MovieSceneScriptingFloatChannel, default=0}
       dyn_wings_metacarpal_04_r_ctrl.Rotation.X {type=MovieSceneScriptingFloatChannel, default=0}
       dyn_wings_metacarpal_04_r_ctrl.Rotation.Y {type=MovieSceneScriptingFloatChannel, default=0}
       dyn_wings_metacarpal_04_r_ctrl.Rotation.Z {type=MovieSceneScriptingFloatChannel, default=0}
       dyn_wings_metacarpal_04_r_ctrl.Scale.X {type=MovieSceneScriptingFloatChannel, default=0}
       dyn_wings_metacarpal_04_r_ctrl.Scale.Y {type=MovieSceneScriptingFloatChannel, default=0}
       dyn_wings_metacarpal_04_r_ctrl.Scale.Z {type=MovieSceneScriptingFloatChannel, default=0}
       dyn_wings_phalange_04_r_ctrl.Location.X {type=MovieSceneScriptingFloatChannel, default=0}
       dyn_wings_phalange_04_r_ctrl.Location.Y {type=MovieSceneScriptingFloatChannel, default=0}
       dyn_wings_phalange_04_r_ctrl.Location.Z {type=MovieSceneScriptingFloatChannel, default=0}
       dyn_wings_phalange_04_r_ctrl.Rotation.X {type=MovieSceneScriptingFloatChannel, default=0}
       dyn_wings_phalange_04_r_ctrl.Rotation.Y {type=MovieSceneScriptingFloatChannel, default=0}
       dyn_wings_phalange_04_r_ctrl.Rotation.Z {type=MovieSceneScriptingFloatChannel, default=0}
       dyn_wings_phalange_04_r_ctrl.Scale.X {type=MovieSceneScriptingFloatChannel, default=0}
       dyn_wings_phalange_04_r_ctrl.Scale.Y {type=MovieSceneScriptingFloatChannel, default=0}
       dyn_wings_phalange_04_r_ctrl.Scale.Z {type=MovieSceneScriptingFloatChannel, default=0}
      ]
     }
     {
      type=MovieSceneControlRigParameterSection
      end_frame=753
      start_frame=-282
      channels [
       pack_ctrl.Location.X {type=MovieSceneScriptingFloatChannel, default=0}
       pack_ctrl.Location.Y {type=MovieSceneScriptingFloatChannel, default=0}
       pack_ctrl.Location.Z {type=MovieSceneScriptingFloatChannel, default=0}
       pack_ctrl.Rotation.X {type=MovieSceneScriptingFloatChannel, default=0}
       pack_ctrl.Rotation.Y {type=MovieSceneScriptingFloatChannel, default=0}
       pack_ctrl.Rotation.Z {type=MovieSceneScriptingFloatChannel, default=0}
       pack_ctrl.Scale.X {type=MovieSceneScriptingFloatChannel, default=0}
       pack_ctrl.Scale.Y {type=MovieSceneScriptingFloatChannel, default=0}
       pack_ctrl.Scale.Z {type=MovieSceneScriptingFloatChannel, default=0}
       dyn_wings_humelus_l_ctrl.Location.X {type=MovieSceneScriptingFloatChannel, default=0, thinned_from=39, keys [[-253,0],[735,0]]}
       dyn_wings_humelus_l_ctrl.Location.Y {type=MovieSceneScriptingFloatChannel, default=0, thinned_from=39, keys [[-253,0],[735,0]]}
       dyn_wings_humelus_l_ctrl.Location.Z {type=MovieSceneScriptingFloatChannel, default=0, thinned_from=39, keys [[-253,0],[735,0]]}
       dyn_wings_humelus_l_ctrl.Rotation.X {
        type=MovieSceneScriptingFloatChannel
        default=0
        thinned_from=45
        keys [[-273,-6.5438],[-253,18.8148],[-224,0],[1,0],[17,10.4816],[24,10.4816],[27,20.1084],[40,28.5735],[56,28.5735],[69,14.2868],[95,14.2868],[98,0],[125,0],[145,16.3464],[199,16.3464],[201,9.9127],[205,9.9127],[213,0],[242,-21.3091],[269,-20.614],[288,0],[314,0],[320,-0.018],[324,-0.0081],[330,0],[343,-0.0415],[346,-0.0457],[351,0.0165],[359,0],[397,-0.6811],[425,-8.2658],[455,-8.2658],[465,10.6284],[496,10.6284],[502,-0.6811],[551,0],[566,4.3738],[585,4.3738],[591,10.7394],[616,13.115],[626,21.083],[658,21.9757],[735,0]]
       }
       dyn_wings_humelus_l_ctrl.Rotation.Y {
        type=MovieSceneScriptingFloatChannel
        default=0
        thinned_from=46
        keys [[-273,1.7257],[-253,-7.5236],[-224,0],[1,0],[17,-4.7804],[24,-4.7804],[27,-9.1709],[40,-13.0317],[56,-13.0317],[69,-6.5158],[95,-6.5158],[98,0],[125,0],[145,-2.2635],[199,-2.2635],[201,-1.3726],[205,-1.3726],[213,0],[242,14.9442],[269,14.919],[288,0],[314,0],[320,0.0011],[324,0.0005],[330,0],[341,0.05],[343,0.0469],[346,0.0346],[351,-0.0124],[359,0],[397,-0.6811],[425,-2.1725],[455,-2.1725],[465,-7.76],[496,-7.76],[502,-0.6811],[551,0],[566,1.0112],[585,1.0112],[591,2.4827],[616,3.0307],[626,4.8713],[658,5.0761],[735,0]]
       }
       dyn_wings_humelus_l_ctrl.Rotation.Z {
        type=MovieSceneScriptingFloatChannel
        default=0
        thinned_from=48
        keys [[-273,-1.3744],[-253,2.3502],[-224,0],[1,0],[17,2.6858],[24,2.6858],[27,5.1526],[40,7.3217],[56,7.3217],[69,3.6609],[95,3.6609],[98,0],[125,0],[145,8.1475],[199,8.1475],[201,4.9408],[205,4.9408],[213,0],[242,-8.6945],[269,-8.3646],[288,0],[314,0],[320,7.0962],[324,15.304],[330,0],[333,-29.3964],[335,-41.6508],[338,-33.4514],[341,-5.9469],[343,1.2477],[346,1.7179],[353,-3.3663],[363,0],[397,-0.6811],[425,-8.8366],[455,-8.8366],[465,6.9009],[496,6.9009],[502,-0.6811],[551,0],[566,-2.9277],[585,-2.9277],[591,-7.195],[616,-8.8198],[626,-14.1954],[658,-14.837],[735,0]]
       }
       dyn_wings_humelus_l_ctrl.Scale.X {type=MovieSceneScriptingFloatChannel, default=0, const [0,1]}
       dyn_wings_humelus_l_ctrl.Scale.Y {type=MovieSceneScriptingFloatChannel, default=0, const [0,1]}
       dyn_wings_humelus_l_ctrl.Scale.Z {type=MovieSceneScriptingFloatChannel, default=0, const [0,1]}
       dyn_wings_radius_l_ctrl.Location.X {
        type=MovieSceneScriptingFloatChannel
        default=0
        thinned_from=32
        keys [[-273,-0.5278],[-253,0],[330,0],[332,-0.2673],[333,-0.3608],[335,-0.2932],[337,-0.1804],[341,0],[658,0]]
       }
       dyn_wings_radius_l_ctrl.Location.Y {
        type=MovieSceneScriptingFloatChannel
        default=0
        thinned_from=32
        keys [[-273,-0.5278],[-253,0],[330,0],[332,-0.2673],[333,-0.3608],[335,-0.2932],[337,-0.1804],[341,0],[658,0]]
       }
       dyn_wings_radius_l_ctrl.Location.Z {
        type=MovieSceneScriptingFloatChannel
        default=0
        thinned_from=32
        keys [[-273,-0.5278],[-253,0],[330,0],[332,-0.2673],[333,-0.3608],[335,-0.2932],[337,-0.1804],[341,0],[658,0]]
       }
       dyn_wings_radius_l_ctrl.Rotation.X {
        type=MovieSceneScriptingFloatChannel
        default=0
        thinned_from=32
        keys [[-273,-0.5278],[-253,0],[125,0],[145,6.5491],[199,6.5491],[201,3.9715],[205,3.9715],[213,0],[219,0],[222,-1.0382],[224,-1.0382],[226,-1.2155],[237,-1.2155],[239,-1.4778],[281,-1.2601],[288,0],[314,0],[323,-2.8354],[324,-9.4118],[329,0.0104],[331,11.304],[334,23.1494],[335,5.7361],[337,-2.4162],[341,-7.1539],[342,-6.5389],[346,-0.589],[359,0],[658,0]]
       }
       dyn_wings_radius_l_ctrl.Rotation.Y {
        type=MovieSceneScriptingFloatChannel
        default=0
        thinned_from=34
        keys [[-273,-0.5278],[-253,0],[1,0],[40,0.0097],[125,0.0105],[145,-2.6913],[199,-2.6913],[201,-1.632],[205,-1.632],[213,0],[219,0],[222,9.5112],[224,9.5112],[226,14.4377],[237,14.4377],[239,30.0275],[281,29.9929],[288,0],[314,-0.0673],[320,-3.1644],[324,-3.72],[329,-2.8874],[332,0.4263],[333,2.438],[335,4.2117],[337,-2.5437],[341,-11.6524],[343,-10.3119],[346,-1.0878],[351,0.3916],[359,0],[658,0]]
       }
       dyn_wings_radius_l_ctrl.Rotation.Z {
        type=MovieSceneScriptingFloatChannel
        default=0
        thinned_from=35
        keys [[-273,-0.5278],[-253,0],[125,0],[145,-16.1198],[199,-16.1198],[201,-9.7753],[205,-9.7753],[213,0],[219,0],[222,3.1828],[227,3.1828],[229,3.8691],[233,3.8691],[235,5.2434],[281,4.6936],[288,0],[314,0],[320,-4.7978],[324,-9.0868],[329,-2.5223],[331,6.7761],[332,11.7319],[333,1.5063],[334,-6.0216],[336,-7.4265],[340,4.3707],[341,6.7069],[344,9.5796],[351,-2.9888],[359,0],[658,0]]
       }
       dyn_wings_radius_l_ctrl.Scale.X {type=MovieSceneScriptingFloatChannel, default=0, const [0,6]}
       dyn_wings_radius_l_ctrl.Scale.Y {type=MovieSceneScriptingFloatChannel, default=0, const [0,6]}
       dyn_wings_radius_l_ctrl.Scale.Z {type=MovieSceneScriptingFloatChannel, default=0, const [0,6]}
       dyn_wings_metacarpal_01_l_ctrl.Location.X {
        type=MovieSceneScriptingFloatChannel
        default=0
        thinned_from=27
        keys [[-273,-0.5278],[-253,0],[330,0],[333,-0.3608],[335,-0.2932],[337,-0.1804],[341,0],[658,0]]
       }
       dyn_wings_metacarpal_01_l_ctrl.Location.Y {
        type=MovieSceneScriptingFloatChannel
        default=0
        thinned_from=27
        keys [[-273,-0.5278],[-253,0],[330,0],[333,-0.3608],[335,-0.2932],[337,-0.1804],[341,0],[658,0]]
       }
       dyn_wings_metacarpal_01_l_ctrl.Location.Z {
        type=MovieSceneScriptingFloatChannel
        default=0
        thinned_from=27
        keys [[-273,-0.5278],[-253,0],[330,0],[333,-0.3608],[335,-0.2932],[337,-0.1804],[341,0],[658,0]]
       }
       dyn_wings_metacarpal_01_l_ctrl.Rotation.X {
        type=MovieSceneScriptingFloatChannel
        default=0
        thinned_from=27
        keys [[-273,-0.5571],[-253,0.0322],[-224,0],[125,0],[145,-0.8177],[199,-0.8177],[201,-0.4959],[205,-0.4959],[213,0],[330,0],[333,-0.3808],[335,-0.3094],[337,-0.1904],[341,0],[658,0]]
       }
       dyn_wings_metacarpal_01_l_ctrl.Rotation.Y {
        type=MovieSceneScriptingFloatChannel
        default=0
        thinned_from=27
        keys [[-273,-0.528],[-253,0.0012],[-224,0],[125,0],[145,-0.0464],[199,-0.0464],[201,-0.0282],[205,-0.0282],[213,0],[330,0],[333,-0.361],[335,-0.2933],[337,-0.1805],[341,0],[658,0]]
       }
       dyn_wings_metacarpal_01_l_ctrl.Rotation.Z {
        type=MovieSceneScriptingFloatChannel
        default=0
        thinned_from=27
        keys [[-273,6.1416],[-253,-10.8313],[-224,0],[125,0],[145,-6.5526],[199,-6.5526],[201,-3.9736],[205,-3.9736],[213,0],[330,0],[333,4.1984],[335,3.4112],[337,2.0992],[341,0],[658,0]]
       }
       dyn_wings_metacarpal_01_l_ctrl.Scale.X {type=MovieSceneScriptingFloatChannel, default=0, const [0,1]}
       dyn_wings_metacarpal_01_l_ctrl.Scale.Y {type=MovieSceneScriptingFloatChannel, default=0, const [0,1]}
       dyn_wings_metacarpal_01_l_ctrl.Scale.Z {type=MovieSceneScriptingFloatChannel, default=0, const [0,1]}
       dyn_wings_phalange_01_l_ctrl.Location.X {
        type=MovieSceneScriptingFloatChannel
        default=0
        thinned_from=27
        keys [[-273,-0.5278],[-253,0],[330,0],[333,-0.3608],[335,-0.2932],[337,-0.1804],[341,0],[658,0]]
       }
       dyn_wings_phalange_01_l_ctrl.Location.Y {
        type=MovieSceneScriptingFloatChannel
        default=0
        thinned_from=27
        keys [[-273,-0.5278],[-253,0],[330,0],[333,-0.3608],[335,-0.2932],[337,-0.1804],[341,0],[658,0]]
       }
       dyn_wings_phalange_01_l_ctrl.Location.Z {
        type=MovieSceneScriptingFloatChannel
        default=0
        thinned_from=27
        keys [[-273,-0.5278],[-253,0],[330,0],[333,-0.3608],[335,-0.2932],[337,-0.1804],[341,0],[658,0]]
       }
       dyn_wings_phalange_01_l_ctrl.Rotation.X {
        type=MovieSceneScriptingFloatChannel
        default=0
        thinned_from=27
        keys [[-273,-0.1797],[-253,-0.6414],[-224,0],[125,0],[145,-1.8354],[199,-1.8354],[201,-1.113],[205,-1.113],[213,0],[330,0],[333,-0.1229],[335,-0.0998],[337,-0.0614],[341,0],[658,0]]
       }
       dyn_wings_phalange_01_l_ctrl.Rotation.Y {
        type=MovieSceneScriptingFloatChannel
        default=0
        thinned_from=27
        keys [[-273,-0.5136],[-253,-0.096],[-224,0],[125,0],[145,-0.5301],[199,-0.5301],[201,-0.3215],[205,-0.3215],[213,0],[330,0],[333,-0.3511],[335,-0.2852],[337,-0.1755],[341,0],[658,0]]
       }
       dyn_wings_phalange_01_l_ctrl.Rotation.Z {
        type=MovieSceneScriptingFloatChannel
        default=0
        thinned_from=27
        keys [[-273,6.1534],[-253,-10.8497],[-224,0],[125,0],[145,-32.2632],[199,-32.2632],[201,-19.5649],[205,-19.5649],[213,0],[330,0],[333,4.2065],[335,3.4177],[337,2.1032],[341,0],[658,0]]
       }
       dyn_wings_phalange_01_l_ctrl.Scale.X {type=MovieSceneScriptingFloatChannel, default=0, const [0,1]}
       dyn_wings_phalange_01_l_ctrl.Scale.Y {type=MovieSceneScriptingFloatChannel, default=0, const [0,1]}
       dyn_wings_phalange_01_l_ctrl.Scale.Z {type=MovieSceneScriptingFloatChannel, default=0, const [0,1]}
       dyn_wings_carpals_01_l_ctrl.Location.X {type=MovieSceneScriptingFloatChannel, default=0, thinned_from=25, keys [[-273,-0.5278],[-253,0],[658,0]]}
       dyn_wings_carpals_01_l_ctrl.Location.Y {type=MovieSceneScriptingFloatChannel, default=0, thinned_from=25, keys [[-273,-0.5278],[-253,0],[658,0]]}
       dyn_wings_carpals_01_l_ctrl.Location.Z {type=MovieSceneScriptingFloatChannel, default=0, thinned_from=25, keys [[-273,-0.5278],[-253,0],[658,0]]}
       dyn_wings_carpals_01_l_ctrl.Rotation.X {
        type=MovieSceneScriptingFloatChannel
        default=0
        thinned_from=27
        keys [[-273,2.3792],[-253,0],[125,0],[145,-1.856],[199,-1.856],[201,-1.1255],[205,-1.1255],[213,0],[242,-3.859],[269,-3.8577],[288,0],[314,0],[320,1.7875],[324,0.8045],[330,0],[333,19.8853],[335,10.6677],[337,0.7009],[341,-0.1983],[346,-0.2246],[351,0.0809],[359,0],[658,0]]
       }
       dyn_wings_carpals_01_l_ctrl.Rotation.Y {
        type=MovieSceneScriptingFloatChannel
        default=0
        thinned_from=27
        keys [[-273,2.8802],[-253,0],[125,0],[145,-15.6581],[199,-15.6581],[201,-9.4953],[205,-9.4953],[213,0],[242,1.2962],[269,0.9012],[288,0],[314,0],[320,0.4989],[324,0.2245],[330,0],[333,1.2491],[335,0.9753],[337,0.4707],[341,-0.7805],[346,-0.5071],[351,0.1826],[359,0],[658,0]]
       }
       dyn_wings_carpals_01_l_ctrl.Rotation.Z {
        type=MovieSceneScriptingFloatChannel
        default=0
        thinned_from=27
        keys [[-273,29.6212],[-253,0],[125,0],[145,-22.4039],[199,-22.4039],[201,-13.586],[205,-13.586],[213,0],[242,10.7575],[269,9.9636],[288,0],[314,0],[320,2.5322],[324,-0.9962],[330,20.4307],[333,40.2656],[335,28.7091],[337,12.1815],[341,0.4967],[346,-1.604],[351,0.5774],[359,0],[658,0]]
       }
       dyn_wings_carpals_01_l_ctrl.Scale.X {type=MovieSceneScriptingFloatChannel, default=0, const [0,1]}
       dyn_wings_carpals_01_l_ctrl.Scale.Y {type=MovieSceneScriptingFloatChannel, default=0, const [0,1]}
       dyn_wings_carpals_01_l_ctrl.Scale.Z {type=MovieSceneScriptingFloatChannel, default=0, const [0,1]}
       dyn_wings_metacarpal_02_l_ctrl.Location.X {
        type=MovieSceneScriptingFloatChannel
        default=0
        thinned_from=39
        keys [[-273,-0.5278],[-253,0],[330,0],[333,-0.3608],[335,-0.2932],[337,-0.1804],[341,0],[658,0]]
       }
       dyn_wings_metacarpal_02_l_ctrl.Location.Y {
        type=MovieSceneScriptingFloatChannel
        default=0
        thinned_from=39
        keys [[-273,-0.5278],[-253,0],[330,0],[333,-0.3608],[335,-0.2932],[337,-0.1804],[341,0],[658,0]]
       }
       dyn_wings_metacarpal_02_l_ctrl.Location.Z {
        type=MovieSceneScriptingFloatChannel
        default=0
        thinned_from=39
        keys [[-273,-0.5278],[-253,0],[330,0],[333,-0.3608],[335,-0.2932],[337,-0.1804],[341,0],[658,0]]
       }
       dyn_wings_metacarpal_02_l_ctrl.Rotation.X {
        type=MovieSceneScriptingFloatChannel
        default=0
        thinned_from=39
        keys [[-273,-2.8233],[-253,6.9287],[-224,0],[19,0],[21,0.6432],[29,0.6432],[31,1.0159],[37,1.0159],[39,0.8145],[43,0.8145],[44,0.5072],[45,0.5072],[46,0.2958],[51,0.2958],[53,0],[125,0],[145,1.1156],[199,1.1156],[201,0.6765],[205,0.6765],[213,0],[330,0],[333,-0.7242],[335,0.1262],[337,1.5434],[341,3.811],[346,0],[658,0]]
       }
       dyn_wings_metacarpal_02_l_ctrl.Rotation.Y {
        type=MovieSceneScriptingFloatChannel
        default=0
        thinned_from=39
        keys [[-273,0.5689],[-253,-3.8375],[-224,0],[19,0],[21,0.0492],[29,0.0492],[31,0.0777],[37,0.0777],[39,0.0623],[43,0.0623],[44,0.0388],[45,0.0388],[46,0.0226],[51,0.0226],[53,0],[125,0],[145,0.0634],[199,0.0634],[201,0.0384],[205,0.0384],[213,0],[330,0],[333,0.4855],[335,0.4521],[337,0.3841],[341,0.208],[346,0],[658,0]]
       }
       dyn_wings_metacarpal_02_l_ctrl.Rotation.Z {
        type=MovieSceneScriptingFloatChannel
        default=0
        thinned_from=39
        keys [[-273,5.8937],[-253,-7.9407],[-224,0],[19,0],[21,-3.2413],[29,-3.2413],[31,-5.1192],[37,-5.1192],[39,-4.1043],[43,-4.1043],[44,-2.5558],[45,-2.5558],[46,-1.4907],[51,-1.4907],[53,0],[125,0],[145,-6.5962],[199,-6.5962],[201,-4],[205,-4],[213,0],[330,0],[333,3.4828],[335,2.5062],[337,0.8785],[341,-1.7258],[346,0],[658,0]]
       }
       dyn_wings_metacarpal_02_l_ctrl.Scale.X {type=MovieSceneScriptingFloatChannel, default=0, const [0,1]}
       dyn_wings_metacarpal_02_l_ctrl.Scale.Y {type=MovieSceneScriptingFloatChannel, default=0, const [0,1]}
       dyn_wings_metacarpal_02_l_ctrl.Scale.Z {type=MovieSceneScriptingFloatChannel, default=0, const [0,1]}
       dyn_wings_phalange_02_l_ctrl.Location.X {
        type=MovieSceneScriptingFloatChannel
        default=0
        thinned_from=39
        keys [[-273,-0.5278],[-253,0],[330,0],[333,-0.3608],[335,-0.2932],[337,-0.1804],[341,0],[658,0]]
       }
       dyn_wings_phalange_02_l_ctrl.Location.Y {
        type=MovieSceneScriptingFloatChannel
        default=0
        thinned_from=39
        keys [[-273,-0.5278],[-253,0],[330,0],[333,-0.3608],[335,-0.2932],[337,-0.1804],[341,0],[658,0]]
       }
       dyn_wings_phalange_02_l_ctrl.Location.Z {
        type=MovieSceneScriptingFloatChannel
        default=0
        thinned_from=39
        keys [[-273,-0.5278],[-253,0],[330,0],[333,-0.3608],[335,-0.2932],[337,-0.1804],[341,0],[658,0]]
       }
       dyn_wings_phalange_02_l_ctrl.Rotation.X {
        type=MovieSceneScriptingFloatChannel
        default=0
        thinned_from=39
        keys [[-273,-0.3635],[-253,-0.2209],[-224,0],[19,0],[21,-1.9481],[29,-1.9481],[31,-3.0767],[37,-3.0767],[39,-2.4668],[43,-2.4668],[44,-1.5361],[45,-1.5361],[46,-0.8959],[51,-0.8959],[53,0],[125,0],[145,3.5533],[199,3.5533],[201,2.1548],[205,2.1548],[213,0],[330,0],[333,1.8614],[335,3.1594],[337,4.3357],[341,6.6681],[346,-0.0193],[351,0.0069],[359,0],[658,0]]
       }
       dyn_wings_phalange_02_l_ctrl.Rotation.Y {
        type=MovieSceneScriptingFloatChannel
        default=0
        thinned_from=39
        keys [[-273,-0.5255],[-253,-0.0733],[-224,0],[19,0],[21,-0.8277],[29,-0.8277],[31,-1.3072],[37,-1.3072],[39,-1.0481],[43,-1.0481],[44,-0.6527],[45,-0.6527],[46,-0.3807],[51,-0.3807],[53,0],[125,0],[145,1.0265],[199,1.0265],[201,0.6225],[205,0.6225],[213,0],[330,0],[333,0.5568],[337,1.7567],[341,2.8951],[346,-0.0005],[351,0.0002],[359,0],[658,0]]
       }
       dyn_wings_phalange_02_l_ctrl.Rotation.Z {
        type=MovieSceneScriptingFloatChannel
        default=0
        thinned_from=39
        keys [[-273,18.0821],[-253,-23.3445],[-224,0],[19,0],[21,-16.8284],[29,-16.8284],[31,-26.5781],[37,-26.5781],[39,-21.309],[43,-21.309],[44,-13.2694],[45,-13.2694],[46,-7.7396],[51,-7.7396],[53,0],[125,0],[145,-32.3911],[199,-32.3911],[201,-19.6424],[205,-19.6424],[213,0],[330,0],[333,0.5484],[335,-6.5543],[337,-18.3923],[341,-37.333],[346,-1.9088],[351,0.6872],[359,0],[658,0]]
       }
       dyn_wings_phalange_02_l_ctrl.Scale.X {type=MovieSceneScriptingFloatChannel, default=0, const [0,1]}
       dyn_wings_phalange_02_l_ctrl.Scale.Y {type=MovieSceneScriptingFloatChannel, default=0, const [0,1]}
       dyn_wings_phalange_02_l_ctrl.Scale.Z {type=MovieSceneScriptingFloatChannel, default=0, const [0,1]}
       dyn_wings_carpals_02_l_ctrl.Location.X {type=MovieSceneScriptingFloatChannel, default=0, thinned_from=19, keys [[-273,-0.5278],[-253,0],[658,0]]}
       dyn_wings_carpals_02_l_ctrl.Location.Y {type=MovieSceneScriptingFloatChannel, default=0, thinned_from=19, keys [[-273,-0.5278],[-253,0],[658,0]]}
       dyn_wings_carpals_02_l_ctrl.Location.Z {type=MovieSceneScriptingFloatChannel, default=0, thinned_from=19, keys [[-273,-0.5278],[-253,0],[658,0]]}
       dyn_wings_carpals_02_l_ctrl.Rotation.X {
        type=MovieSceneScriptingFloatChannel
        default=0
        thinned_from=19
        keys [[-273,-2.4063],[-253,16.8726],[-224,0],[125,0],[145,23.1449],[199,23.1449],[201,14.0354],[205,14.0354],[213,0],[242,2.0085],[269,2.3919],[288,0],[658,0]]
       }
       dyn_wings_carpals_02_l_ctrl.Rotation.Y {
        type=MovieSceneScriptingFloatChannel
        default=0
        thinned_from=19
        keys [[-273,1.3405],[-253,-3.518],[-224,0],[125,0],[145,-0.8215],[199,-0.8215],[201,-0.4982],[205,-0.4982],[213,0],[242,-0.2283],[269,-0.2718],[288,0],[658,0]]
       }
       dyn_wings_carpals_02_l_ctrl.Rotation.Z {
        type=MovieSceneScriptingFloatChannel
        default=0
        thinned_from=19
        keys [[-273,15.203],[-253,-24.9003],[-224,0],[125,0],[145,-25.6609],[199,-25.6609],[201,-15.5612],[205,-15.5612],[213,0],[242,8.8709],[269,8.6904],[288,0],[658,0]]
       }
       dyn_wings_carpals_02_l_ctrl.Scale.X {type=MovieSceneScriptingFloatChannel, default=0, const [0,1]}
       dyn_wings_carpals_02_l_ctrl.Scale.Y {type=MovieSceneScriptingFloatChannel, default=0, const [0,1]}
       dyn_wings_carpals_02_l_ctrl.Scale.Z {type=MovieSceneScriptingFloatChannel, default=0, const [0,1]}
       dyn_wings_metacarpal_03_l_ctrl.Location.X {
        type=MovieSceneScriptingFloatChannel
        default=0
        thinned_from=40
        keys [[-273,-0.5278],[-253,0],[330,0],[333,-0.4266],[335,-0.3766],[337,-0.2733],[341,0],[658,0]]
       }
       dyn_wings_metacarpal_03_l_ctrl.Location.Y {
        type=MovieSceneScriptingFloatChannel
        default=0
        thinned_from=40
        keys [[-273,-0.5278],[-253,0],[330,0],[333,-0.4266],[335,-0.3766],[337,-0.2733],[341,0],[658,0]]
       }
       dyn_wings_metacarpal_03_l_ctrl.Location.Z {
        type=MovieSceneScriptingFloatChannel
        default=0
        thinned_from=40
        keys [[-273,-0.5278],[-253,0],[330,0],[333,-0.4266],[335,-0.3766],[337,-0.2733],[341,0],[658,0]]
       }
       dyn_wings_metacarpal_03_l_ctrl.Rotation.X {
        type=MovieSceneScriptingFloatChannel
        default=0
        thinned_from=40
        keys [[-273,-17.2531],[-253,16.1815],[-224,0],[19,0],[21,-0.3925],[29,-0.3925],[31,-0.6199],[37,-0.6199],[39,-0.497],[43,-0.497],[44,-0.3095],[45,-0.3095],[46,-0.1805],[51,-0.1805],[53,0],[125,0],[145,-0.6257],[199,-0.6257],[201,-0.3794],[205,-0.3794],[213,0],[242,0.2623],[269,0.2405],[288,0],[314,0],[320,4.4086],[324,7.2525],[330,0],[333,-11.4668],[335,-8.9167],[337,-3.5332],[341,11.201],[346,5.959],[351,-2.1452],[359,0],[658,0]]
       }
       dyn_wings_metacarpal_03_l_ctrl.Rotation.Y {
        type=MovieSceneScriptingFloatChannel
        default=0
        thinned_from=40
        keys [[-273,-1.369],[-253,2.3222],[-224,0],[19,0],[21,-0.03],[29,-0.03],[31,-0.0474],[37,-0.0474],[39,-0.038],[43,-0.038],[44,-0.0237],[45,-0.0237],[46,-0.0138],[51,-0.0138],[53,0],[125,0],[145,-0.0469],[199,-0.0469],[201,-0.0284],[205,-0.0284],[213,0],[242,-0.0106],[269,-0.0106],[288,0],[314,0],[320,1.6091],[324,1.189],[330,0],[333,-0.2101],[335,0.5075],[337,1.2747],[341,4.0883],[346,0.4778],[351,-0.172],[359,0],[658,0]]
       }
       dyn_wings_metacarpal_03_l_ctrl.Rotation.Z {
        type=MovieSceneScriptingFloatChannel
        default=0
        thinned_from=38
        keys [[-273,8.2635],[-253,-12.7716],[-224,0],[19,0],[21,-5.5521],[29,-5.5521],[31,-8.7688],[37,-8.7688],[39,-7.0303],[43,-7.0303],[44,-4.3779],[45,-4.3779],[46,-2.5535],[51,-2.5535],[53,0],[125,0],[145,-6.5315],[199,-6.5315],[201,-3.9608],[205,-3.9608],[213,0],[242,4.6072],[269,4.3631],[288,0],[314,0],[320,-0.1004],[324,-0.2672],[330,0],[333,7.3332],[335,6.3263],[337,5.4687],[341,1.266],[351,-3.3197],[359,-2.5836],[364,0],[658,0]]
       }
       dyn_wings_metacarpal_03_l_ctrl.Scale.X {type=MovieSceneScriptingFloatChannel, default=0, const [0,1]}
       dyn_wings_metacarpal_03_l_ctrl.Scale.Y {type=MovieSceneScriptingFloatChannel, default=0, const [0,1]}
       dyn_wings_metacarpal_03_l_ctrl.Scale.Z {type=MovieSceneScriptingFloatChannel, default=0, const [0,1]}
       dyn_wings_phalange_03_l_ctrl.Location.X {
        type=MovieSceneScriptingFloatChannel
        default=0
        thinned_from=40
        keys [[-273,-0.5278],[-253,0],[330,0],[333,-0.4266],[335,-0.3766],[337,-0.2733],[341,0],[658,0]]
       }
       dyn_wings_phalange_03_l_ctrl.Location.Y {
        type=MovieSceneScriptingFloatChannel
        default=0
        thinned_from=40
        keys [[-273,-0.5278],[-253,0],[330,0],[333,-0.4266],[335,-0.3766],[337,-0.2733],[341,0],[658,0]]
       }
       dyn_wings_phalange_03_l_ctrl.Location.Z {
        type=MovieSceneScriptingFloatChannel
        default=0
        thinned_from=40
        keys [[-273,-0.5278],[-253,0],[330,0],[333,-0.4266],[335,-0.3766],[337,-0.2733],[341,0],[658,0]]
       }
       dyn_wings_phalange_03_l_ctrl.Rotation.X {
        type=MovieSceneScriptingFloatChannel
        default=0
        thinned_from=40
        keys [[-273,0.9139],[-253,-1.858],[-224,0],[19,0],[21,9.4525],[29,9.4525],[31,14.9288],[37,14.9288],[39,11.9692],[43,11.9692],[44,7.4534],[45,7.4534],[46,4.3473],[51,4.3473],[53,0],[125,0],[145,0.5313],[199,0.5313],[201,0.3222],[205,0.3222],[213,0],[242,0.8327],[269,0.8324],[288,0],[314,0],[320,-1.7868],[324,-0.8042],[330,0],[333,-0.2506],[335,-0.7268],[337,-1.7392],[341,-4.5398],[346,-0.3903],[351,0.1405],[359,0],[658,0]]
       }
       dyn_wings_phalange_03_l_ctrl.Rotation.Y {
        type=MovieSceneScriptingFloatChannel
        default=0
        thinned_from=40
        keys [[-273,-3.2404],[-253,-0.6472],[-224,0],[19,0],[21,4.0395],[29,4.0395],[31,6.3797],[37,6.3797],[39,5.1149],[43,5.1149],[44,3.1852],[45,3.1852],[46,1.8578],[51,1.8578],[53,0],[125,0],[145,0.1534],[199,0.1534],[201,0.093],[205,0.093],[213,0],[242,-0.0628],[269,-0.0575],[288,0],[314,0],[320,1.1017],[324,1.5251],[330,0],[333,-1.9999],[335,-1.464],[337,-0.3283],[341,2.7991],[346,0.5082],[351,-0.183],[359,0],[658,0]]
       }
       dyn_wings_phalange_03_l_ctrl.Rotation.Z {
        type=MovieSceneScriptingFloatChannel
        default=0
        thinned_from=40
        keys [[-273,23.9023],[-253,-24.6122],[-224,0],[19,0],[21,-30.3416],[29,-30.3416],[31,-47.9202],[37,-47.9202],[39,-38.42],[43,-38.42],[44,-23.9248],[45,-23.9248],[46,-13.9544],[51,-13.9544],[53,0],[125,0],[145,-32.2206],[199,-32.2206],[201,-19.5391],[205,-19.5391],[213,0],[242,6.2307],[269,5.1683],[288,0],[314,0],[320,-18.4661],[324,-17.4066],[330,0],[333,8.9422],[335,1.7358],[337,-10.2455],[341,-46.9177],[346,-5.99],[351,16.3094],[355,1.0782],[359,-2.5836],[364,0],[658,0]]
       }
       dyn_wings_phalange_03_l_ctrl.Scale.X {type=MovieSceneScriptingFloatChannel, default=0, const [0,1]}
       dyn_wings_phalange_03_l_ctrl.Scale.Y {type=MovieSceneScriptingFloatChannel, default=0, const [0,1]}
       dyn_wings_phalange_03_l_ctrl.Scale.Z {type=MovieSceneScriptingFloatChannel, default=0, const [0,1]}
       dyn_wings_metacarpal_04_l_ctrl.Location.X {
        type=MovieSceneScriptingFloatChannel
        default=0
        thinned_from=40
        keys [[-273,-0.5278],[-253,0],[330,0],[333,-0.3714],[335,-0.3705],[337,-0.3608],[341,0],[658,0]]
       }
       dyn_wings_metacarpal_04_l_ctrl.Location.Y {
        type=MovieSceneScriptingFloatChannel
        default=0
        thinned_from=40
        keys [[-273,-0.5278],[-253,0],[330,0],[333,-0.3714],[335,-0.3705],[337,-0.3608],[341,0],[658,0]]
       }
       dyn_wings_metacarpal_04_l_ctrl.Location.Z {
        type=MovieSceneScriptingFloatChannel
        default=0
        thinned_from=40
        keys [[-273,-0.5278],[-253,0],[330,0],[333,-0.3714],[335,-0.3705],[337,-0.3608],[341,0],[658,0]]
       }
       dyn_wings_metacarpal_04_l_ctrl.Rotation.X {
        type=MovieSceneScriptingFloatChannel
        default=0
        thinned_from=40
        keys [[-273,-37.4175],[-253,33.0271],[-224,0],[19,0],[21,18.6569],[29,18.6569],[31,29.4659],[37,29.4659],[39,23.6243],[43,23.6243],[44,14.7112],[45,14.7112],[46,8.5805],[51,8.5805],[53,0],[125,0],[145,-0.1585],[199,-0.1585],[201,-0.0961],[205,-0.0961],[213,0],[242,-4.1075],[269,-4.1062],[288,0],[314,0],[320,-0.7812],[324,-0.3516],[330,0],[333,-35.3444],[335,-39.8986],[337,-31.894],[341,-9.3155],[346,25.6917],[351,-9.249],[359,0],[658,0]]
       }
       dyn_wings_metacarpal_04_l_ctrl.Rotation.Y {
        type=MovieSceneScriptingFloatChannel
        default=0
        thinned_from=40
        keys [[-273,-6.75],[-253,-1.4312],[-224,0],[19,0],[21,-4.0561],[29,-4.0561],[31,-6.406],[37,-6.406],[39,-5.136],[43,-5.136],[44,-3.1983],[45,-3.1983],[46,-1.8654],[51,-1.8654],[53,0],[125,0],[145,-0.009],[199,-0.009],[201,-0.0055],[205,-0.0055],[213,0],[242,0.7145],[269,0.7142],[288,0],[314,0],[320,4.6719],[324,2.1027],[330,0],[333,2.0488],[337,1.9904],[341,11.8702],[346,0.8124],[351,-0.2925],[359,0],[658,0]]
       }
       dyn_wings_metacarpal_04_l_ctrl.Rotation.Z {
        type=MovieSceneScriptingFloatChannel
        default=0
        thinned_from=40
        keys [[-273,42.4672],[-253,-31.5887],[-224,0],[19,0],[21,-20.0818],[29,-20.0818],[31,-31.7163],[37,-31.7163],[39,-25.4286],[43,-25.4286],[44,-15.8348],[45,-15.8348],[46,-9.2359],[51,-9.2359],[53,0],[125,0],[145,-6.5035],[199,-6.5035],[201,-3.9438],[205,-3.9438],[213,0],[242,14.3104],[269,13.9837],[288,0],[314,0],[320,-12.2937],[324,-19.968],[330,0],[333,31.2045],[335,38.0608],[337,23.7186],[341,-31.2351],[346,-8.2205],[351,17.1124],[355,1.4797],[359,-2.5836],[364,0],[658,0]]
       }
       dyn_wings_metacarpal_04_l_ctrl.Scale.X {type=MovieSceneScriptingFloatChannel, default=0, const [0,1]}
       dyn_wings_metacarpal_04_l_ctrl.Scale.Y {type=MovieSceneScriptingFloatChannel, default=0, const [0,1]}
       dyn_wings_metacarpal_04_l_ctrl.Scale.Z {type=MovieSceneScriptingFloatChannel, default=0, const [0,1]}
       dyn_wings_phalange_04_l_ctrl.Location.X {
        type=MovieSceneScriptingFloatChannel
        default=0
        thinned_from=40
        keys [[-273,-0.5278],[-253,0],[330,0],[333,-0.3714],[335,-0.3705],[337,-0.3608],[341,0],[658,0]]
       }
       dyn_wings_phalange_04_l_ctrl.Location.Y {
        type=MovieSceneScriptingFloatChannel
        default=0
        thinned_from=40
        keys [[-273,-0.5278],[-253,0],[330,0],[333,-0.3714],[335,-0.3705],[337,-0.3608],[341,0],[658,0]]
       }
       dyn_wings_phalange_04_l_ctrl.Location.Z {
        type=MovieSceneScriptingFloatChannel
        default=0
        thinned_from=40
        keys [[-273,-0.5278],[-253,0],[330,0],[333,-0.3714],[335,-0.3705],[337,-0.3608],[341,0],[658,0]]
       }
       dyn_wings_phalange_04_l_ctrl.Rotation.X {
        type=MovieSceneScriptingFloatChannel
        default=0
        thinned_from=40
        keys [[-273,-2.1328],[-253,12.1651],[-224,0],[19,0],[21,-17.0824],[29,-17.0824],[31,-26.9792],[37,-26.9792],[39,-21.6306],[43,-21.6306],[44,-13.4697],[45,-13.4697],[46,-7.8564],[51,-7.8564],[53,0],[125,0],[145,-1.2646],[199,-1.2646],[201,-0.7669],[205,-0.7669],[213,0],[242,-1.163],[269,-1.1626],[288,0],[314,0],[320,3.7285],[324,1.6781],[330,0],[333,1.5397],[337,1.4958],[341,9.4732],[346,11.0078],[351,-3.9628],[359,0],[658,0]]
       }
       dyn_wings_phalange_04_l_ctrl.Rotation.Y {
        type=MovieSceneScriptingFloatChannel
        default=0
        thinned_from=40
        keys [[-273,1.772],[-253,-7.014],[-224,0],[19,0],[21,-4.4087],[29,-4.4087],[31,-6.963],[37,-6.963],[39,-5.5825],[43,-5.5825],[44,-3.4763],[45,-3.4763],[46,-2.0276],[51,-2.0276],[53,0],[125,0],[145,2.8048],[199,2.8048],[201,1.7009],[205,1.7009],[213,0],[242,-1.1986],[269,-1.1005],[288,0],[314,0],[320,5.1947],[324,2.338],[330,0],[333,5.5768],[337,5.4179],[341,13.1983],[346,0.0476],[351,-0.0171],[359,0],[658,0]]
       }
       dyn_wings_phalange_04_l_ctrl.Rotation.Z {
        type=MovieSceneScriptingFloatChannel
        default=0
        keys [[-273,29.1581],[-253,-29.8027],[-224,0.5478],[1,0],[19,0],[21,-32.9156],[29,-32.9156],[31,-51.9009],[37,-51.9009],[39,-41.6115],[43,-41.6115],[44,-25.9122],[45,-25.9122],[46,-15.1136],[51,-15.1136],[53,0],[55,0],[125,0.6804],[145,-31.4595],[199,-31.4595],[201,-19.0775],[205,-19.0775],[213,0],[242,15.0716],[269,13.9873],[288,0],[314,0],[320,-30.809],[324,-15.5001],[330,0],[333,0.7385],[335,0.735],[337,-2.4353],[341,-77.7727],[346,-12.6337],[351,19.4462],[355,2.9205],[359,-2.0358],[364,0.5478],[658,0.5478]]
       }
       dyn_wings_phalange_04_l_ctrl.Scale.X {type=MovieSceneScriptingFloatChannel, default=0, const [0,1]}
       dyn_wings_phalange_04_l_ctrl.Scale.Y {type=MovieSceneScriptingFloatChannel, default=0, const [0,1]}
       dyn_wings_phalange_04_l_ctrl.Scale.Z {type=MovieSceneScriptingFloatChannel, default=0, const [0,1]}
       dyn_wings_humelus_r_ctrl.Location.X {
        type=MovieSceneScriptingFloatChannel
        default=0
        thinned_from=43
        keys [[50,0],[256,0],[315,-0.0049],[424,-0.0173],[449,0],[465,-0.0044],[513,-0.0285],[596,-0.0285],[616,0],[751,0]]
       }
       dyn_wings_humelus_r_ctrl.Location.Y {
        type=MovieSceneScriptingFloatChannel
        default=0
        thinned_from=43
        keys [[50,0],[256,0],[315,-0.0049],[424,-0.0173],[449,0],[465,-0.0044],[513,-0.0285],[596,-0.0285],[616,0],[751,0]]
       }
       dyn_wings_humelus_r_ctrl.Location.Z {
        type=MovieSceneScriptingFloatChannel
        default=0
        thinned_from=43
        keys [[50,0],[256,0],[315,-0.0049],[424,-0.0173],[449,0],[465,-0.0044],[513,-0.0285],[596,-0.0285],[616,0],[751,0]]
       }
       dyn_wings_humelus_r_ctrl.Rotation.X {
        type=MovieSceneScriptingFloatChannel
        default=0
        thinned_from=48
        keys [[50,0],[58,0],[66,-3.421],[71,-9.0538],[78,0],[79,3.2235],[81,10.5887],[85,24.1586],[88,17.0087],[91,-9.2681],[95,-15.2664],[98,-13.1495],[102,-3.9088],[107,1.309],[112,0],[189,0],[200,8.0537],[220,10.7689],[231,10.7689],[233,6.9783],[256,0],[315,0.004],[424,-0.0173],[430,-34.2777],[435,-40.7798],[444,-1.6271],[449,0.8444],[465,-0.0044],[513,-0.0285],[544,-7.7822],[555,-7.7822],[557,-5.0529],[570,-5.0529],[596,-0.0285],[616,0],[620,3.4871],[629,3.4871],[632,8.0537],[650,8.0537],[652,10.7689],[663,10.7689],[665,6.9783],[678,6.9783],[680,4.5219],[682,4.5219],[685,0],[751,0]]
       }
       dyn_wings_humelus_r_ctrl.Rotation.Y {
        type=MovieSceneScriptingFloatChannel
        default=0
        thinned_from=44
        keys [[50,0],[58,0.5677],[66,-3.1519],[71,-4.5323],[78,-4.8534],[81,-4.1646],[85,1.6797],[87,14.5697],[89,20.2523],[92,18.7779],[102,4.2161],[113,0],[189,0],[200,-0.8609],[220,-2.9852],[231,-2.9852],[233,-1.9344],[256,0],[315,0.004],[424,-0.0173],[430,-14.0786],[435,-17.6685],[449,0],[465,-0.0044],[513,-0.0285],[544,2.1209],[555,2.1209],[557,1.3643],[570,1.3643],[596,-0.0285],[616,0],[620,-0.3719],[629,-0.3719],[632,-0.8609],[650,-0.8609],[652,-2.9852],[663,-2.9852],[665,-1.9344],[678,-1.9344],[680,-1.2535],[682,-1.2535],[685,0],[751,0]]
       }
       dyn_wings_humelus_r_ctrl.Rotation.Z {
        type=MovieSceneScriptingFloatChannel
        default=0
        thinned_from=45
        keys [[50,0],[58,0],[66,4.8999],[71,-9.3948],[77,-18.3224],[79,-17.1057],[81,-2.8603],[85,24.9525],[87,27.0497],[93,-0.3227],[98,-8.4408],[107,-1.5215],[113,0],[189,0],[200,-6.236],[220,-8.0712],[231,-8.0712],[233,-5.2301],[256,0],[315,-0.0049],[424,-0.0173],[430,13.1765],[435,22.9795],[449,0],[465,-0.0263],[513,-0.0285],[544,5.7829],[555,5.7829],[557,3.7373],[570,3.7373],[596,-0.0285],[616,0],[620,-2.694],[629,-2.694],[632,-6.236],[650,-6.236],[652,-8.0712],[663,-8.0712],[665,-5.2301],[678,-5.2301],[680,-3.3891],[682,-3.3891],[685,0],[751,0]]
       }
       dyn_wings_humelus_r_ctrl.Scale.X {type=MovieSceneScriptingFloatChannel, default=0, const [0,7]}
       dyn_wings_humelus_r_ctrl.Scale.Y {type=MovieSceneScriptingFloatChannel, default=0, const [0,7]}
       dyn_wings_humelus_r_ctrl.Scale.Z {type=MovieSceneScriptingFloatChannel, default=0, const [0,7]}
       dyn_wings_radius_r_ctrl.Location.X {
        type=MovieSceneScriptingFloatChannel
        default=0
        thinned_from=44
        keys [[50,0],[256,0],[315,-0.0046],[424,-0.0163],[449,0],[524,-0.0285],[572,-0.0285],[632,0],[751,0]]
       }
       dyn_wings_radius_r_ctrl.Location.Y {
        type=MovieSceneScriptingFloatChannel
        default=0
        thinned_from=44
        keys [[50,0],[256,0],[315,-0.0046],[424,-0.0163],[449,0],[524,-0.0285],[572,-0.0285],[632,0],[751,0]]
       }
       dyn_wings_radius_r_ctrl.Location.Z {
        type=MovieSceneScriptingFloatChannel
        default=0
        thinned_from=44
        keys [[50,0],[256,0],[315,-0.0046],[424,-0.0163],[449,0],[524,-0.0285],[572,-0.0285],[632,0],[751,0]]
       }
       dyn_wings_radius_r_ctrl.Rotation.X {
        type=MovieSceneScriptingFloatChannel
        default=0
        thinned_from=44
        keys [[50,0],[58,0],[66,-0.0193],[71,-0.0577],[78,0],[256,0],[315,-0.0046],[424,-0.0163],[449,0],[524,-0.0285],[572,-0.0285],[632,0],[751,0]]
       }
       dyn_wings_radius_r_ctrl.Rotation.Y {
        type=MovieSceneScriptingFloatChannel
        default=0
        thinned_from=44
        keys [[50,0],[58,0],[66,-0.9966],[71,-2.932],[78,-1.0625],[81,5.5165],[85,10.194],[88,-10.6523],[94,-25.756],[101,-3.82],[106,0],[256,0],[315,-0.0046],[424,-0.0163],[449,0],[524,-0.0285],[572,-0.0285],[632,0],[751,0]]
       }
       dyn_wings_radius_r_ctrl.Rotation.Z {
        type=MovieSceneScriptingFloatChannel
        default=0
        thinned_from=44
        keys [[50,0],[58,0],[66,0.8005],[71,2.3969],[78,1.1131],[81,-0.9935],[85,-2.8127],[88,0],[93,1.1603],[101,0],[256,0],[315,-0.0046],[424,-0.0163],[449,0],[524,-0.0285],[572,-0.0285],[632,0],[751,0]]
       }
       dyn_wings_radius_r_ctrl.Scale.X {type=MovieSceneScriptingFloatChannel, default=0, const [0,1]}
       dyn_wings_radius_r_ctrl.Scale.Y {type=MovieSceneScriptingFloatChannel, default=0, const [0,1]}
       dyn_wings_radius_r_ctrl.Scale.Z {type=MovieSceneScriptingFloatChannel, default=0, const [0,1]}
       dyn_wings_metacarpal_01_r_ctrl.Location.X {
        type=MovieSceneScriptingFloatChannel
        default=0
        thinned_from=44
        keys [[50,0],[256,0],[315,-0.0046],[424,-0.0163],[439,0.1196],[449,0],[524,-0.0285],[572,-0.0285],[632,0],[751,0]]
       }
       dyn_wings_metacarpal_01_r_ctrl.Location.Y {
        type=MovieSceneScriptingFloatChannel
        default=0
        thinned_from=44
        keys [[50,0],[256,0],[315,-0.0046],[424,-0.0163],[439,0.1196],[449,0],[524,-0.0285],[572,-0.0285],[632,0],[751,0]]
       }
       dyn_wings_metacarpal_01_r_ctrl.Location.Z {
        type=MovieSceneScriptingFloatChannel
        default=0
        thinned_from=44
        keys [[50,0],[256,0],[315,-0.0046],[424,-0.0163],[439,0.1196],[449,0],[524,-0.0285],[572,-0.0285],[632,0],[751,0]]
       }
       dyn_wings_metacarpal_01_r_ctrl.Rotation.X {
        type=MovieSceneScriptingFloatChannel
        default=0
        thinned_from=44
        keys [[50,0],[256,0],[315,-0.0046],[424,-0.0163],[439,0.1196],[449,0],[524,-0.0285],[572,-0.0285],[632,0],[751,0]]
       }
       dyn_wings_metacarpal_01_r_ctrl.Rotation.Y {
        type=MovieSceneScriptingFloatChannel
        default=0
        thinned_from=41
        keys [[50,0],[58,0],[71,-0.0005],[78,0],[200,0],[206,2.5849],[209,2.5849],[212,5.0098],[217,6.5675],[233,6.5675],[235,4.2557],[256,0],[315,-0.0417],[424,-0.036],[439,0.1196],[449,0],[524,-0.0285],[530,-1.8896],[533,-1.8896],[536,-3.6356],[541,-4.7571],[557,-4.7571],[559,-3.0926],[570,-3.0926],[572,-0.0285],[632,0],[638,2.5849],[641,2.5849],[644,5.0098],[649,6.5675],[665,6.5675],[667,4.2557],[678,4.2557],[680,0],[751,0]]
       }
       dyn_wings_metacarpal_01_r_ctrl.Rotation.Z {
        type=MovieSceneScriptingFloatChannel
        default=0
        thinned_from=41
        keys [[50,0],[58,0],[71,-0.0005],[78,0],[200,0],[206,4.352],[209,4.352],[212,8.4348],[217,11.0573],[233,11.0573],[235,7.1651],[256,0],[315,-0.0417],[424,-0.036],[439,0.1196],[449,0],[524,-0.0285],[530,-3.162],[533,-3.162],[536,-6.1016],[541,-7.9899],[557,-7.9899],[559,-5.1875],[570,-5.1875],[572,-0.0285],[632,0],[638,4.352],[641,4.352],[644,8.4348],[649,11.0573],[665,11.0573],[667,7.1651],[678,7.1651],[680,0],[751,0]]
       }
       dyn_wings_metacarpal_01_r_ctrl.Scale.X {type=MovieSceneScriptingFloatChannel, default=0, const [0,1]}
       dyn_wings_metacarpal_01_r_ctrl.Scale.Y {type=MovieSceneScriptingFloatChannel, default=0, const [0,1]}
       dyn_wings_metacarpal_01_r_ctrl.Scale.Z {type=MovieSceneScriptingFloatChannel, default=0, const [0,1]}
       dyn_wings_phalange_01_r_ctrl.Location.X {
        type=MovieSceneScriptingFloatChannel
        default=0
        thinned_from=44
        keys [[50,0],[256,0],[315,-0.0046],[424,-0.0163],[439,0.1196],[449,0],[524,-0.0285],[572,-0.0285],[632,0],[751,0]]
       }
       dyn_wings_phalange_01_r_ctrl.Location.Y {
        type=MovieSceneScriptingFloatChannel
        default=0
        thinned_from=44
        keys [[50,0],[256,0],[315,-0.0046],[424,-0.0163],[439,0.1196],[449,0],[524,-0.0285],[572,-0.0285],[632,0],[751,0]]
       }
       dyn_wings_phalange_01_r_ctrl.Location.Z {
        type=MovieSceneScriptingFloatChannel
        default=0
        thinned_from=44
        keys [[50,0],[256,0],[315,-0.0046],[424,-0.0163],[439,0.1196],[449,0],[524,-0.0285],[572,-0.0285],[632,0],[751,0]]
       }
       dyn_wings_phalange_01_r_ctrl.Rotation.X {
        type=MovieSceneScriptingFloatChannel
        default=0
        thinned_from=44
        keys [[50,0],[256,0],[315,-0.0046],[424,-0.0163],[439,0.1196],[449,0],[524,-0.0285],[572,-0.0285],[632,0],[751,0]]
       }
       dyn_wings_phalange_01_r_ctrl.Rotation.Y {
        type=MovieSceneScriptingFloatChannel
        default=0
        thinned_from=41
        keys [[50,0],[58,0],[71,-0.0005],[78,0],[200,0],[206,2.5849],[209,2.5849],[212,5.0098],[217,6.5675],[233,6.5675],[235,4.2557],[256,0],[315,-0.0417],[424,-0.036],[439,0.1196],[449,0],[524,-0.0285],[530,-1.8896],[533,-1.8896],[536,-3.6356],[541,-4.7571],[557,-4.7571],[559,-3.0926],[570,-3.0926],[572,-0.0285],[632,0],[638,2.5849],[641,2.5849],[644,5.0098],[649,6.5675],[665,6.5675],[667,4.2557],[678,4.2557],[680,0],[751,0]]
       }
       dyn_wings_phalange_01_r_ctrl.Rotation.Z {
        type=MovieSceneScriptingFloatChannel
        default=0
        thinned_from=41
        keys [[50,0],[58,0],[71,-0.0005],[78,0],[200,0],[206,4.352],[209,4.352],[212,8.4348],[217,11.0573],[233,11.0573],[235,7.1651],[256,0],[315,-0.0417],[424,-0.036],[439,0.1196],[449,0],[524,-0.0285],[530,-3.162],[533,-3.162],[536,-6.1016],[541,-7.9899],[557,-7.9899],[559,-5.1875],[570,-5.1875],[572,-0.0285],[632,0],[638,4.352],[641,4.352],[644,8.4348],[649,11.0573],[665,11.0573],[667,7.1651],[678,7.1651],[680,0],[751,0]]
       }
       dyn_wings_phalange_01_r_ctrl.Scale.X {type=MovieSceneScriptingFloatChannel, default=0, const [0,1]}
       dyn_wings_phalange_01_r_ctrl.Scale.Y {type=MovieSceneScriptingFloatChannel, default=0, const [0,1]}
       dyn_wings_phalange_01_r_ctrl.Scale.Z {type=MovieSceneScriptingFloatChannel, default=0, const [0,1]}
       dyn_wings_carpals_01_r_ctrl.Location.X {
        type=MovieSceneScriptingFloatChannel
        default=0
        thinned_from=47
        keys [[50,0],[281,0],[315,-0.0046],[424,-0.0163],[439,0.1196],[449,0],[524,-0.0285],[572,-0.0285],[632,0],[751,0]]
       }
       dyn_wings_carpals_01_r_ctrl.Location.Y {
        type=MovieSceneScriptingFloatChannel
        default=0
        thinned_from=47
        keys [[50,0],[281,0],[315,-0.0046],[424,-0.0163],[439,0.1196],[449,0],[524,-0.0285],[572,-0.0285],[632,0],[751,0]]
       }
       dyn_wings_carpals_01_r_ctrl.Location.Z {
        type=MovieSceneScriptingFloatChannel
        default=0
        thinned_from=47
        keys [[50,0],[281,0],[315,-0.0046],[424,-0.0163],[439,0.1196],[449,0],[524,-0.0285],[572,-0.0285],[632,0],[751,0]]
       }
       dyn_wings_carpals_01_r_ctrl.Rotation.X {
        type=MovieSceneScriptingFloatChannel
        default=0
        thinned_from=50
        keys [[50,0],[66,0],[71,-0.5345],[78,0],[81,-1.5349],[85,-0.5178],[93,-4.2701],[100,-0.5471],[110,0],[226,0],[229,16.3519],[240,16.3519],[243,0],[281,0],[315,-0.0046],[424,-0.0163],[431,2.0344],[439,0.4693],[449,0],[524,-0.0285],[572,-0.0285],[632,0],[751,0]]
       }
       dyn_wings_carpals_01_r_ctrl.Rotation.Y {
        type=MovieSceneScriptingFloatChannel
        default=0
        thinned_from=50
        keys [[50,0],[66,0],[71,-2.2865],[78,0],[81,4.4023],[85,0.8536],[93,-1.5908],[100,-7.2069],[110,0],[226,0],[229,2.726],[233,2.8618],[240,2.726],[243,0],[281,0],[315,-0.0046],[424,-0.0163],[431,0.1439],[439,0.1468],[449,0],[524,-0.0285],[572,-0.0285],[632,0],[751,0]]
       }
       dyn_wings_carpals_01_r_ctrl.Rotation.Z {
        type=MovieSceneScriptingFloatChannel
        default=0
        thinned_from=50
        keys [[50,0],[58,0],[66,1.4919],[71,2.7055],[78,0],[81,-9.7248],[85,-18.7075],[93,13.3009],[100,1.7702],[110,0],[226,0],[229,-24.7253],[240,-24.7253],[243,0],[281,0],[315,-0.0046],[424,-0.0163],[431,5.1047],[439,-5.4008],[449,0],[524,-0.0285],[572,-0.0285],[632,0],[751,0]]
       }
       dyn_wings_carpals_01_r_ctrl.Scale.X {type=MovieSceneScriptingFloatChannel, default=0, const [0,7]}
       dyn_wings_carpals_01_r_ctrl.Scale.Y {type=MovieSceneScriptingFloatChannel, default=0, const [0,7]}
       dyn_wings_carpals_01_r_ctrl.Scale.Z {type=MovieSceneScriptingFloatChannel, default=0, const [0,7]}
       dyn_wings_metacarpal_02_r_ctrl.Location.X {
        type=MovieSceneScriptingFloatChannel
        default=0
        thinned_from=44
        keys [[50,0],[256,0],[315,-0.0046],[424,-0.0163],[439,0.1196],[449,0],[524,-0.0285],[578,-0.0285],[632,0],[751,0]]
       }
       dyn_wings_metacarpal_02_r_ctrl.Location.Y {
        type=MovieSceneScriptingFloatChannel
        default=0
        thinned_from=44
        keys [[50,0],[256,0],[315,-0.0046],[424,-0.0163],[439,0.1196],[449,0],[524,-0.0285],[578,-0.0285],[632,0],[751,0]]
       }
       dyn_wings_metacarpal_02_r_ctrl.Location.Z {
        type=MovieSceneScriptingFloatChannel
        default=0
        thinned_from=44
        keys [[50,0],[256,0],[315,-0.0046],[424,-0.0163],[439,0.1196],[449,0],[524,-0.0285],[578,-0.0285],[632,0],[751,0]]
       }
       dyn_wings_metacarpal_02_r_ctrl.Rotation.X {
        type=MovieSceneScriptingFloatChannel
        default=0
        keys [[50,0],[58,0],[66,0.0485],[71,0.0484],[78,0.0507],[81,-0.0251],[85,9.6396],[88,0],[93,0.1209],[102,0],[200,0],[206,-1.3278],[209,-1.3278],[212,-2.5734],[216,-2.5734],[217,-3.3736],[233,-3.3736],[235,-2.1861],[256,0],[315,-0.0046],[424,-0.0163],[439,0.2439],[449,0],[524,-0.0285],[530,0.9276],[533,0.9276],[536,1.8245],[540,1.8245],[541,2.4006],[557,2.4006],[559,1.5455],[570,1.5455],[578,-0.0285],[632,0],[638,-1.3278],[641,-1.3278],[644,-2.5734],[648,-2.5734],[649,-3.3736],[665,-3.3736],[667,-2.1861],[678,-2.1861],[686,0],[751,0]]
       }
       dyn_wings_metacarpal_02_r_ctrl.Rotation.Y {
        type=MovieSceneScriptingFloatChannel
        default=0
        keys [[50,0],[58,0],[66,-0.0027],[71,-0.003],[78,-0.0026],[81,-0.0006],[85,-5.5383],[88,0],[93,-0.0151],[102,0],[200,0],[206,0.2092],[209,0.2092],[212,0.4055],[216,0.4055],[217,0.5316],[233,0.5316],[235,0.3445],[256,0],[315,-0.0417],[424,-0.036],[439,0.1048],[449,0],[524,-0.0285],[530,-0.1791],[533,-0.1791],[536,-0.3204],[540,-0.3204],[541,-0.4112],[557,-0.4112],[559,-0.2765],[570,-0.2765],[578,-0.0285],[632,0],[638,0.2092],[641,0.2092],[644,0.4055],[648,0.4055],[649,0.5316],[665,0.5316],[667,0.3445],[678,0.3445],[686,0],[751,0]]
       }
       dyn_wings_metacarpal_02_r_ctrl.Rotation.Z {
        type=MovieSceneScriptingFloatChannel
        default=0
        keys [[50,0],[58,0],[66,5.6706],[71,6.2516],[78,5.9241],[81,-2.9278],[85,-15.1214],[88,0],[93,14.2583],[102,0],[200,0],[206,6.1664],[209,6.1664],[212,11.9513],[216,11.9513],[217,15.6672],[233,15.6672],[235,10.1524],[256,0],[315,-0.0241],[424,-0.036],[439,-3.8864],[449,0],[524,-0.0285],[530,-4.4683],[533,-4.4683],[536,-8.6336],[540,-8.6336],[541,-11.3091],[557,-11.3091],[559,-7.3383],[570,-7.3383],[578,-0.0285],[632,0],[638,6.1664],[641,6.1664],[644,11.9513],[648,11.9513],[649,15.6672],[665,15.6672],[667,10.1524],[678,10.1524],[686,0],[751,0]]
       }
       dyn_wings_metacarpal_02_r_ctrl.Scale.X {type=MovieSceneScriptingFloatChannel, default=0, const [0,1]}
       dyn_wings_metacarpal_02_r_ctrl.Scale.Y {type=MovieSceneScriptingFloatChannel, default=0, const [0,1]}
       dyn_wings_metacarpal_02_r_ctrl.Scale.Z {type=MovieSceneScriptingFloatChannel, default=0, const [0,1]}
       dyn_wings_phalange_02_r_ctrl.Location.X {
        type=MovieSceneScriptingFloatChannel
        default=0
        thinned_from=44
        keys [[50,0],[256,0],[315,-0.0046],[424,-0.0163],[439,0.1196],[449,0],[524,-0.0285],[578,-0.0285],[632,0],[751,0]]
       }
       dyn_wings_phalange_02_r_ctrl.Location.Y {
        type=MovieSceneScriptingFloatChannel
        default=0
        thinned_from=44
        keys [[50,0],[256,0],[315,-0.0046],[424,-0.0163],[439,0.1196],[449,0],[524,-0.0285],[578,-0.0285],[632,0],[751,0]]
       }
       dyn_wings_phalange_02_r_ctrl.Location.Z {
        type=MovieSceneScriptingFloatChannel
        default=0
        thinned_from=44
        keys [[50,0],[256,0],[315,-0.0046],[424,-0.0163],[439,0.1196],[449,0],[524,-0.0285],[578,-0.0285],[632,0],[751,0]]
       }
       dyn_wings_phalange_02_r_ctrl.Rotation.X {
        type=MovieSceneScriptingFloatChannel
        default=0
        keys [[50,0],[58,0],[66,-0.4533],[71,-0.8752],[78,-0.4732],[81,0.2342],[85,10.9797],[88,0],[93,-1.1291],[102,0],[200,0],[206,0.988],[209,0.988],[212,1.9148],[216,1.9148],[217,2.5101],[233,2.5101],[235,1.6266],[256,0],[315,-0.0331],[424,-0.036],[439,-0.1169],[449,0],[524,-0.0285],[530,-0.7398],[533,-0.7398],[536,-1.4071],[540,-1.4071],[541,-1.8358],[557,-1.8358],[559,-1.1996],[570,-1.1996],[578,-0.0285],[632,0],[638,0.988],[641,0.988],[644,1.9148],[648,1.9148],[649,2.5101],[665,2.5101],[667,1.6266],[678,1.6266],[686,0],[751,0]]
       }
       dyn_wings_phalange_02_r_ctrl.Rotation.Y {
        type=MovieSceneScriptingFloatChannel
        default=0
        keys [[50,0],[58,0],[66,0.0225],[71,0.0224],[78,0.0245],[81,0.006],[85,-5.3611],[88,0],[93,0.1412],[102,0],[200,0],[206,-0.1328],[209,-0.1328],[212,-0.2574],[216,-0.2574],[217,-0.3375],[233,-0.3375],[235,-0.2187],[256,0],[315,-0.0046],[424,-0.0163],[439,0.1062],[449,0],[524,-0.0285],[530,0.0672],[533,0.0672],[536,0.1569],[540,0.1569],[541,0.2145],[557,0.2145],[559,0.129],[570,0.129],[578,-0.0285],[632,0],[638,-0.1328],[641,-0.1328],[644,-0.2574],[648,-0.2574],[649,-0.3375],[665,-0.3375],[667,-0.2187],[678,-0.2187],[686,0],[751,0]]
       }
       dyn_wings_phalange_02_r_ctrl.Rotation.Z {
        type=MovieSceneScriptingFloatChannel
        default=0
        keys [[50,0],[58,0],[66,5.6884],[71,6.2694],[78,5.9427],[81,-2.937],[85,-15.2963],[88,0],[93,14.3015],[102,0],[200,0],[206,6.1048],[209,6.1048],[212,11.8319],[216,11.8319],[217,15.5106],[233,15.5106],[235,10.0509],[256,0],[315,-0.0241],[424,-0.036],[439,-3.8914],[449,0],[524,-0.0285],[530,-4.424],[533,-4.424],[536,-8.5476],[540,-8.5476],[541,-11.1963],[557,-11.1963],[559,-7.2652],[570,-7.2652],[578,-0.0285],[632,0],[638,6.1048],[641,6.1048],[644,11.8319],[648,11.8319],[649,15.5106],[665,15.5106],[667,10.0509],[678,10.0509],[686,0],[751,0]]
       }
       dyn_wings_phalange_02_r_ctrl.Scale.X {type=MovieSceneScriptingFloatChannel, default=0, const [0,1]}
       dyn_wings_phalange_02_r_ctrl.Scale.Y {type=MovieSceneScriptingFloatChannel, default=0, const [0,1]}
       dyn_wings_phalange_02_r_ctrl.Scale.Z {type=MovieSceneScriptingFloatChannel, default=0, const [0,1]}
       dyn_wings_carpals_02_r_ctrl.Location.X {
        type=MovieSceneScriptingFloatChannel
        default=0
        thinned_from=50
        keys [[50,0],[424,0],[439,0.1196],[449,0],[465,0],[524,-0.0285],[586,-0.0285],[648,0],[751,0]]
       }
       dyn_wings_carpals_02_r_ctrl.Location.Y {
        type=MovieSceneScriptingFloatChannel
        default=0
        thinned_from=50
        keys [[50,0],[424,0],[439,0.1196],[449,0],[465,0],[524,-0.0285],[586,-0.0285],[648,0],[751,0]]
       }
       dyn_wings_carpals_02_r_ctrl.Location.Z {
        type=MovieSceneScriptingFloatChannel
        default=0
        thinned_from=50
        keys [[50,0],[424,0],[439,0.1196],[449,0],[465,0],[524,-0.0285],[586,-0.0285],[648,0],[751,0]]
       }
       dyn_wings_carpals_02_r_ctrl.Rotation.X {
        type=MovieSceneScriptingFloatChannel
        default=0
        thinned_from=52
        keys [[50,0],[58,0],[66,-0.4522],[71,-0.9917],[78,-1.7518],[81,-0.0654],[85,3.4696],[93,-17.2357],[100,-2.2256],[110,0],[194,0],[195,-7.2408],[213,-7.2408],[217,-7.7272],[222,-7.7272],[223,-9.2745],[227,-9.2745],[228,-1.2326],[242,-1.2326],[244,-15.712],[247,-15.712],[248,-11.6389],[271,-11.6389],[275,0],[315,0.0025],[424,-0.0111],[431,-3.2169],[439,0.6219],[449,0],[465,0],[524,-0.0285],[530,6.8428],[533,6.8428],[536,13.289],[540,13.289],[541,17.4297],[558,17.4297],[560,11.2844],[570,11.2844],[586,-0.0285],[648,0],[649,-2.89],[651,-2.89],[653,-18.4961],[657,-18.4961],[658,-24.2469],[666,-24.2469],[668,-15.712],[678,-15.712],[694,0],[751,0]]
       }
       dyn_wings_carpals_02_r_ctrl.Rotation.Y {
        type=MovieSceneScriptingFloatChannel
        default=0
        keys [[50,0],[58,0],[66,0.3759],[71,-1.4195],[78,0.0126],[81,-4.9465],[85,-10.9476],[93,-13.8997],[100,-7.5068],[110,-0.0006],[193,-0.0438],[194,-0.0418],[195,-2.6025],[210,-2.6025],[217,-4.4132],[222,-4.4132],[223,-2.3185],[227,-2.3185],[228,4.8064],[242,4.8064],[244,3.1145],[247,3.1145],[248,2.3103],[271,2.3103],[275,0],[315,-0.0209],[424,-0.018],[431,0.1991],[439,0.1587],[449,0],[465,0],[524,-0.0285],[530,-1.3905],[533,-1.3905],[536,-2.6683],[540,-2.6683],[541,-3.4891],[558,-3.4891],[560,-2.271],[570,-2.271],[586,-0.0285],[648,0],[649,0.5746],[651,0.5746],[653,3.6664],[657,3.6664],[658,4.8064],[666,4.8064],[668,3.1145],[678,3.1145],[694,0],[751,0]]
       }
       dyn_wings_carpals_02_r_ctrl.Rotation.Z {
        type=MovieSceneScriptingFloatChannel
        default=0
        keys [[50,0],[58,0],[66,5.1354],[71,8.7733],[78,6.118],[81,-0.5646],[85,-8.4562],[93,8.4499],[100,1.4394],[110,-0.0006],[193,-0.0439],[194,-0.0419],[195,-17.2233],[212,-17.2233],[213,-29.7093],[220,-29.7093],[223,-15.6079],[229,-15.6079],[230,-5.0969],[242,-5.0969],[244,2.4437],[247,2.4437],[248,-2.992],[271,-2.992],[275,0],[315,-0.0006],[424,-0.0005],[431,1.979],[439,-5.4125],[449,0],[465,0],[524,-0.0285],[530,-9.1978],[533,-9.1978],[536,-17.7999],[540,-17.7999],[541,-23.3254],[558,-23.3254],[560,-15.1249],[570,-15.1249],[586,-0.0285],[648,0],[649,3.8583],[651,3.8583],[653,24.6821],[657,24.6821],[658,32.3562],[666,32.3562],[668,20.9668],[678,20.9668],[694,0],[751,0]]
       }
       dyn_wings_carpals_02_r_ctrl.Scale.X {type=MovieSceneScriptingFloatChannel, default=0, thinned_from=13, keys [[193,0],[751,0]]}
       dyn_wings_carpals_02_r_ctrl.Scale.Y {type=MovieSceneScriptingFloatChannel, default=0, thinned_from=13, keys [[193,0],[751,0]]}
       dyn_wings_carpals_02_r_ctrl.Scale.Z {type=MovieSceneScriptingFloatChannel, default=0, thinned_from=13, keys [[193,0],[751,0]]}
       dyn_wings_metacarpal_03_r_ctrl.Location.X {
        type=MovieSceneScriptingFloatChannel
        default=0
        thinned_from=47
        keys [[50,0],[424,0],[439,0.1196],[449,0],[465,0],[524,-0.0285],[580,-0.0285],[632,0],[751,0]]
       }
       dyn_wings_metacarpal_03_r_ctrl.Location.Y {
        type=MovieSceneScriptingFloatChannel
        default=0
        thinned_from=47
        keys [[50,0],[424,0],[439,0.1196],[449,0],[465,0],[524,-0.0285],[580,-0.0285],[632,0],[751,0]]
       }
       dyn_wings_metacarpal_03_r_ctrl.Location.Z {
        type=MovieSceneScriptingFloatChannel
        default=0
        thinned_from=47
        keys [[50,0],[424,0],[439,0.1196],[449,0],[465,0],[524,-0.0285],[580,-0.0285],[632,0],[751,0]]
       }
       dyn_wings_metacarpal_03_r_ctrl.Rotation.X {
        type=MovieSceneScriptingFloatChannel
        default=0
        thinned_from=48
        keys [[50,0],[66,0],[71,-0.0015],[78,-0.0065],[81,0.0018],[85,0.0208],[88,0],[93,1.2536],[102,0],[200,0],[201,-0.18],[204,-0.18],[207,-0.3489],[211,-0.3489],[212,-0.4574],[229,-0.4574],[231,-0.2964],[237,-0.2964],[239,-0.1482],[253,-0.1482],[255,0],[424,-0.0111],[431,-1.2618],[439,0.1456],[449,0],[465,0],[524,-0.0285],[530,0.1012],[533,0.1012],[536,0.2228],[540,0.2228],[541,0.3009],[558,0.3009],[560,0.1849],[578,0.1849],[580,-0.0285],[632,0],[635,-0.18],[641,-0.18],[644,-0.3489],[648,-0.3489],[649,-0.4574],[666,-0.4574],[668,-0.2964],[678,-0.2964],[694,0],[751,0]]
       }
       dyn_wings_metacarpal_03_r_ctrl.Rotation.Y {
        type=MovieSceneScriptingFloatChannel
        default=0
        thinned_from=48
        keys [[50,0],[71,0],[78,0.0004],[81,0],[85,0.0062],[88,0],[93,4.179],[102,0],[200,0],[201,0.0183],[204,0.0183],[207,0.0355],[211,0.0355],[212,0.0465],[229,0.0465],[231,0.0301],[237,0.0301],[239,0.0156],[253,0.0156],[255,0],[424,-0.0107],[431,-1.2948],[439,0.121],[449,0],[465,0],[524,-0.0285],[530,-0.0416],[533,-0.0416],[536,-0.054],[540,-0.054],[541,-0.0619],[558,-0.0619],[560,-0.0502],[578,-0.0502],[580,-0.0285],[632,0],[635,0.0183],[641,0.0183],[644,0.0355],[648,0.0355],[649,0.0465],[666,0.0465],[668,0.0301],[678,0.0301],[694,0],[751,0]]
       }
       dyn_wings_metacarpal_03_r_ctrl.Rotation.Z {
        type=MovieSceneScriptingFloatChannel
        default=0
        thinned_from=48
        keys [[50,0],[66,-0.0001],[71,0.4904],[78,2.113],[81,-0.5671],[85,-7.1333],[88,0],[93,5.5652],[102,0],[200,0],[201,4.5731],[204,4.5731],[207,8.8634],[211,8.8634],[212,11.6191],[229,11.6191],[231,7.5292],[237,7.5292],[239,3.7646],[253,3.7646],[255,0],[424,-0.0005],[431,14.5438],[439,-3.8713],[449,0],[465,0],[524,-0.0285],[530,-3.3212],[533,-3.3212],[536,-6.4102],[540,-6.4102],[541,-8.3944],[558,-8.3944],[560,-5.4496],[578,-5.4496],[580,-0.0285],[632,0],[635,4.5731],[641,4.5731],[644,8.8634],[648,8.8634],[649,11.6191],[666,11.6191],[668,7.5292],[678,7.5292],[694,0],[751,0]]
       }
       dyn_wings_metacarpal_03_r_ctrl.Scale.X {type=MovieSceneScriptingFloatChannel, default=0, const [0,4]}
       dyn_wings_metacarpal_03_r_ctrl.Scale.Y {type=MovieSceneScriptingFloatChannel, default=0, const [0,4]}
       dyn_wings_metacarpal_03_r_ctrl.Scale.Z {type=MovieSceneScriptingFloatChannel, default=0, const [0,4]}
       dyn_wings_phalange_03_r_ctrl.Location.X {
        type=MovieSceneScriptingFloatChannel
        default=0
        thinned_from=47
        keys [[50,0],[424,0],[439,0.1196],[449,0],[465,0],[524,-0.0285],[584,-0.0285],[632,0],[751,0]]
       }
       dyn_wings_phalange_03_r_ctrl.Location.Y {
        type=MovieSceneScriptingFloatChannel
        default=0
        thinned_from=47
        keys [[50,0],[424,0],[439,0.1196],[449,0],[465,0],[524,-0.0285],[584,-0.0285],[632,0],[751,0]]
       }
       dyn_wings_phalange_03_r_ctrl.Location.Z {
        type=MovieSceneScriptingFloatChannel
        default=0
        thinned_from=47
        keys [[50,0],[424,0],[439,0.1196],[449,0],[465,0],[524,-0.0285],[584,-0.0285],[632,0],[751,0]]
       }
       dyn_wings_phalange_03_r_ctrl.Rotation.X {
        type=MovieSceneScriptingFloatChannel
        default=0
        thinned_from=48
        keys [[50,0],[66,0],[71,-0.0005],[78,-0.002],[81,0.0009],[85,0.0098],[88,0],[93,-0.0135],[102,0],[188,0],[189,-2.4088],[201,-2.4088],[204,-4.6687],[215,-4.6687],[216,-6.1202],[228,-6.1202],[230,-3.9659],[237,-3.9659],[239,-1.9845],[253,-1.9845],[255,-0.0032],[424,-0.0111],[431,9.0388],[439,0.1091],[449,0],[465,0],[524,-0.0285],[530,1.7059],[533,1.7059],[536,3.333],[540,3.333],[541,4.3782],[562,4.3782],[564,2.827],[582,2.827],[584,-0.0285],[632,0],[635,-2.4088],[641,-2.4088],[644,-4.6687],[648,-4.6687],[649,-6.1202],[666,-6.1202],[668,-3.9659],[678,-3.9659],[694,0],[751,0]]
       }
       dyn_wings_phalange_03_r_ctrl.Rotation.Y {
        type=MovieSceneScriptingFloatChannel
        default=0
        thinned_from=48
        keys [[50,0],[71,-0.0001],[78,0.0001],[81,0],[85,0.0029],[88,0],[93,0.0062],[102,0],[188,0],[189,0.6994],[201,0.6994],[204,1.3555],[215,1.3555],[216,1.7769],[228,1.7769],[230,1.1514],[237,1.1514],[239,0.5703],[253,0.5703],[255,-0.0117],[424,-0.018],[431,-3.2561],[439,0.1181],[449,0],[465,0],[524,-0.0285],[530,-0.532],[533,-0.532],[536,-1.0044],[540,-1.0044],[541,-1.3078],[562,-1.3078],[564,-0.8575],[582,-0.8575],[584,-0.0285],[632,0],[635,0.6994],[641,0.6994],[644,1.3554],[648,1.3554],[649,1.7769],[666,1.7769],[668,1.1514],[678,1.1514],[694,0],[751,0]]
       }
       dyn_wings_phalange_03_r_ctrl.Rotation.Z {
        type=MovieSceneScriptingFloatChannel
        default=0
        thinned_from=48
        keys [[50,0],[66,-0.0001],[71,0.3557],[78,1.5325],[81,-0.6921],[85,-7.8338],[88,0],[93,11.5551],[102,0],[188,0],[189,12.935],[201,12.935],[204,25.0698],[215,25.0698],[216,32.8644],[228,32.8644],[230,21.2962],[237,21.2962],[239,10.6478],[253,10.6478],[255,-0.0006],[424,-0.0005],[431,14.0937],[439,-9.7928],[449,0],[465,0],[524,-0.0285],[530,-9.3418],[533,-9.3418],[536,-18.0791],[540,-18.0791],[541,-23.6913],[558,-23.6913],[560,-15.362],[578,-15.362],[584,-0.0285],[632,0],[635,12.935],[641,12.935],[644,25.0698],[648,25.0698],[649,32.8644],[666,32.8644],[668,21.2962],[678,21.2962],[694,0],[751,0]]
       }
       dyn_wings_phalange_03_r_ctrl.Scale.X {type=MovieSceneScriptingFloatChannel, default=0, const [0,4]}
       dyn_wings_phalange_03_r_ctrl.Scale.Y {type=MovieSceneScriptingFloatChannel, default=0, const [0,4]}
       dyn_wings_phalange_03_r_ctrl.Scale.Z {type=MovieSceneScriptingFloatChannel, default=0, const [0,4]}
       dyn_wings_metacarpal_04_r_ctrl.Location.X {
        type=MovieSceneScriptingFloatChannel
        default=0
        thinned_from=45
        keys [[50,0],[424,0],[439,0.1196],[449,0],[465,0],[524,-0.0285],[580,-0.0285],[643,0],[751,0]]
       }
       dyn_wings_metacarpal_04_r_ctrl.Location.Y {
        type=MovieSceneScriptingFloatChannel
        default=0
        thinned_from=45
        keys [[50,0],[424,0],[439,0.1196],[449,0],[465,0],[524,-0.0285],[580,-0.0285],[643,0],[751,0]]
       }
       dyn_wings_metacarpal_04_r_ctrl.Location.Z {
        type=MovieSceneScriptingFloatChannel
        default=0
        thinned_from=45
        keys [[50,0],[424,0],[439,0.1196],[449,0],[465,0],[524,-0.0285],[580,-0.0285],[643,0],[751,0]]
       }
       dyn_wings_metacarpal_04_r_ctrl.Rotation.X {
        type=MovieSceneScriptingFloatChannel
        default=0
        keys [[50,0],[58,0],[66,-1.5847],[71,-3.3149],[78,0.0016],[81,1.3238],[85,5.7417],[88,0],[93,-18.9922],[102,0],[200,0],[201,-4.5475],[204,-4.5475],[207,-8.8137],[211,-8.8137],[212,-11.5541],[229,-11.5541],[231,-7.487],[240,-7.487],[242,-3.7435],[253,-3.7435],[255,0],[424,-0.0111],[431,-5.7466],[439,0.8728],[449,0],[465,0],[524,-0.0285],[530,3.2458],[533,3.2458],[536,6.3176],[540,6.3176],[541,8.2906],[558,8.2906],[560,5.3623],[578,5.3623],[580,-0.0285],[643,0],[647,-8.8137],[648,-8.8137],[649,-11.5541],[666,-11.5541],[668,-7.487],[678,-7.487],[694,0],[751,0]]
       }
       dyn_wings_metacarpal_04_r_ctrl.Rotation.Y {
        type=MovieSceneScriptingFloatChannel
        default=0
        keys [[45,0.1611],[50,0],[58,0],[66,0.294],[71,0.8082],[78,-0.0001],[81,-2.9065],[85,-4.3697],[88,0],[93,-0.0019],[102,0],[200,0],[201,1.2175],[204,1.2175],[207,2.3597],[211,2.3597],[212,3.0934],[229,3.0934],[231,2.0045],[240,2.0045],[242,1.0037],[253,1.0037],[255,0],[424,-0.018],[431,1.1013],[439,0.2908],[449,0],[465,0],[524,-0.0285],[530,-0.9051],[533,-0.9051],[536,-1.7275],[540,-1.7275],[541,-2.2558],[558,-2.2558],[560,-1.4717],[578,-1.4717],[580,-0.0285],[643,0],[647,2.3597],[648,2.3597],[649,3.0934],[666,3.0934],[668,2.0045],[678,2.0045],[694,0],[751,0]]
       }
       dyn_wings_metacarpal_04_r_ctrl.Rotation.Z {
        type=MovieSceneScriptingFloatChannel
        default=0
        keys [[50,0],[58,0],[66,-1.3516],[71,4.1453],[78,2.9456],[81,-5.3431],[85,-13.4506],[88,0],[93,11.5949],[102,0],[200,0],[201,12.5352],[204,12.5352],[207,24.2951],[211,24.2951],[212,31.8488],[229,31.8488],[231,20.638],[240,20.638],[242,10.3191],[253,10.3191],[255,0],[424,-0.0005],[431,13.2921],[433,-6.1105],[439,-15.7602],[449,0],[465,0],[524,-0.0285],[530,-9.054],[533,-9.054],[536,-17.5212],[540,-17.5212],[541,-22.96],[558,-22.96],[560,-14.8881],[578,-14.8881],[580,-0.0285],[643,0],[647,24.2951],[648,24.2951],[649,31.8488],[666,31.8488],[668,20.638],[678,20.638],[694,0],[751,0]]
       }
       dyn_wings_metacarpal_04_r_ctrl.Scale.X {type=MovieSceneScriptingFloatChannel, default=0, const [0,4]}
       dyn_wings_metacarpal_04_r_ctrl.Scale.Y {type=MovieSceneScriptingFloatChannel, default=0, const [0,4]}
       dyn_wings_metacarpal_04_r_ctrl.Scale.Z {type=MovieSceneScriptingFloatChannel, default=0, const [0,4]}
       dyn_wings_phalange_04_r_ctrl.Location.X {
        type=MovieSceneScriptingFloatChannel
        default=0
        thinned_from=45
        keys [[50,0],[424,0],[439,0.1196],[449,0],[465,0],[524,-0.0285],[584,-0.0285],[643,0],[751,0]]
       }
       dyn_wings_phalange_04_r_ctrl.Location.Y {
        type=MovieSceneScriptingFloatChannel
        default=0
        thinned_from=45
        keys [[50,0],[424,0],[439,0.1196],[449,0],[465,0],[524,-0.0285],[584,-0.0285],[643,0],[751,0]]
       }
       dyn_wings_phalange_04_r_ctrl.Location.Z {
        type=MovieSceneScriptingFloatChannel
        default=0
        thinned_from=45
        keys [[50,0],[424,0],[439,0.1196],[449,0],[465,0],[524,-0.0285],[584,-0.0285],[643,0],[751,0]]
       }
       dyn_wings_phalange_04_r_ctrl.Rotation.X {
        type=MovieSceneScriptingFloatChannel
        default=0
        keys [[50,0],[58,0],[66,0.0938],[71,-0.2068],[78,-0.302],[81,0.1281],[85,0.319],[88,0],[93,-0.3394],[102,0],[188,0],[189,4.4184],[201,4.4184],[204,8.5635],[215,8.5635],[216,11.2261],[228,11.2261],[230,7.2745],[245,7.2745],[247,3.6371],[264,3.6371],[266,-0.0003],[424,-0.0005],[431,-3.9682],[439,-0.0076],[449,0],[465,0],[524,-0.0285],[530,-3.2098],[533,-3.2098],[536,-6.1943],[540,-6.1943],[541,-8.1114],[562,-8.1114],[564,-5.2662],[582,-5.2662],[584,-0.0285],[643,0],[647,8.5635],[648,8.5635],[649,11.2261],[666,11.2261],[668,7.2745],[678,7.2745],[694,0],[751,0]]
       }
       dyn_wings_phalange_04_r_ctrl.Rotation.Y {
        type=MovieSceneScriptingFloatChannel
        default=0
        keys [[50,0],[58,0],[66,0.0076],[71,0.0379],[78,0.0842],[81,0.0142],[85,0.0949],[88,0],[93,0.1089],[102,0],[188,0],[189,-1.1827],[201,-1.1827],[204,-2.2923],[215,-2.2923],[216,-3.0051],[228,-3.0051],[230,-1.9473],[245,-1.9473],[247,-0.9736],[264,-0.9736],[266,0],[424,0],[431,1.1384],[439,0.1],[449,0],[465,0],[524,-0.0285],[530,0.8231],[533,0.8231],[536,1.6221],[540,1.6221],[541,2.1352],[562,2.1352],[564,1.3736],[582,1.3736],[584,-0.0285],[643,0],[647,-2.2923],[648,-2.2923],[649,-3.0051],[666,-3.0051],[668,-1.9473],[678,-1.9473],[694,0],[751,0]]
       }
       dyn_wings_phalange_04_r_ctrl.Rotation.Z {
        type=MovieSceneScriptingFloatChannel
        default=0
        keys [[50,0],[58,0],[66,-2.1905],[71,4.9083],[78,7.3741],[81,-3.0002],[85,-7.8396],[88,0],[93,8.4151],[102,0],[188,0],[189,12.4933],[201,12.4933],[204,24.2138],[215,24.2138],[216,31.7422],[228,31.7422],[230,20.569],[245,20.569],[247,10.2842],[264,10.2842],[266,-0.0006],[424,-0.0005],[431,17.8711],[439,-10.6963],[449,0],[465,0],[524,-0.0285],[530,-9.0238],[533,-9.0238],[536,-17.4627],[540,-17.4627],[541,-22.8833],[558,-22.8833],[564,-14.8384],[582,-14.8384],[584,-0.0285],[643,0],[647,24.2138],[648,24.2138],[649,31.7422],[666,31.7422],[668,20.569],[678,20.569],[694,0],[751,0]]
       }
       dyn_wings_phalange_04_r_ctrl.Scale.X {type=MovieSceneScriptingFloatChannel, default=0, const [0,4]}
       dyn_wings_phalange_04_r_ctrl.Scale.Y {type=MovieSceneScriptingFloatChannel, default=0, const [0,4]}
       dyn_wings_phalange_04_r_ctrl.Scale.Z {type=MovieSceneScriptingFloatChannel, default=0, const [0,4]}
      ]
     }
     {
      type=MovieSceneControlRigParameterSection
      end_frame=752
      start_frame=0
      channels [
       pack_ctrl.Location.X {type=MovieSceneScriptingFloatChannel, default=0}
       pack_ctrl.Location.Y {type=MovieSceneScriptingFloatChannel, default=0}
       pack_ctrl.Location.Z {type=MovieSceneScriptingFloatChannel, default=0}
       pack_ctrl.Rotation.X {type=MovieSceneScriptingFloatChannel, default=0}
       pack_ctrl.Rotation.Y {type=MovieSceneScriptingFloatChannel, default=0}
       pack_ctrl.Rotation.Z {type=MovieSceneScriptingFloatChannel, default=0}
       pack_ctrl.Scale.X {type=MovieSceneScriptingFloatChannel, default=0}
       pack_ctrl.Scale.Y {type=MovieSceneScriptingFloatChannel, default=0}
       pack_ctrl.Scale.Z {type=MovieSceneScriptingFloatChannel, default=0}
       dyn_wings_humelus_l_ctrl.Location.X {type=MovieSceneScriptingFloatChannel, default=0}
       dyn_wings_humelus_l_ctrl.Location.Y {type=MovieSceneScriptingFloatChannel, default=0}
       dyn_wings_humelus_l_ctrl.Location.Z {type=MovieSceneScriptingFloatChannel, default=0}
       dyn_wings_humelus_l_ctrl.Rotation.X {type=MovieSceneScriptingFloatChannel, default=0}
       dyn_wings_humelus_l_ctrl.Rotation.Y {type=MovieSceneScriptingFloatChannel, default=0}
       dyn_wings_humelus_l_ctrl.Rotation.Z {type=MovieSceneScriptingFloatChannel, default=0}
       dyn_wings_humelus_l_ctrl.Scale.X {type=MovieSceneScriptingFloatChannel, default=0}
       dyn_wings_humelus_l_ctrl.Scale.Y {type=MovieSceneScriptingFloatChannel, default=0}
       dyn_wings_humelus_l_ctrl.Scale.Z {type=MovieSceneScriptingFloatChannel, default=0}
       dyn_wings_radius_l_ctrl.Location.X {type=MovieSceneScriptingFloatChannel, default=0}
       dyn_wings_radius_l_ctrl.Location.Y {type=MovieSceneScriptingFloatChannel, default=0}
       dyn_wings_radius_l_ctrl.Location.Z {type=MovieSceneScriptingFloatChannel, default=0}
       dyn_wings_radius_l_ctrl.Rotation.X {type=MovieSceneScriptingFloatChannel, default=0}
       dyn_wings_radius_l_ctrl.Rotation.Y {type=MovieSceneScriptingFloatChannel, default=0}
       dyn_wings_radius_l_ctrl.Rotation.Z {type=MovieSceneScriptingFloatChannel, default=0}
       dyn_wings_radius_l_ctrl.Scale.X {type=MovieSceneScriptingFloatChannel, default=0}
       dyn_wings_radius_l_ctrl.Scale.Y {type=MovieSceneScriptingFloatChannel, default=0}
       dyn_wings_radius_l_ctrl.Scale.Z {type=MovieSceneScriptingFloatChannel, default=0}
       dyn_wings_metacarpal_01_l_ctrl.Location.X {type=MovieSceneScriptingFloatChannel, default=0}
       dyn_wings_metacarpal_01_l_ctrl.Location.Y {type=MovieSceneScriptingFloatChannel, default=0}
       dyn_wings_metacarpal_01_l_ctrl.Location.Z {type=MovieSceneScriptingFloatChannel, default=0}
       dyn_wings_metacarpal_01_l_ctrl.Rotation.X {type=MovieSceneScriptingFloatChannel, default=0}
       dyn_wings_metacarpal_01_l_ctrl.Rotation.Y {type=MovieSceneScriptingFloatChannel, default=0}
       dyn_wings_metacarpal_01_l_ctrl.Rotation.Z {type=MovieSceneScriptingFloatChannel, default=0}
       dyn_wings_metacarpal_01_l_ctrl.Scale.X {type=MovieSceneScriptingFloatChannel, default=0}
       dyn_wings_metacarpal_01_l_ctrl.Scale.Y {type=MovieSceneScriptingFloatChannel, default=0}
       dyn_wings_metacarpal_01_l_ctrl.Scale.Z {type=MovieSceneScriptingFloatChannel, default=0}
       dyn_wings_phalange_01_l_ctrl.Location.X {type=MovieSceneScriptingFloatChannel, default=0}
       dyn_wings_phalange_01_l_ctrl.Location.Y {type=MovieSceneScriptingFloatChannel, default=0}
       dyn_wings_phalange_01_l_ctrl.Location.Z {type=MovieSceneScriptingFloatChannel, default=0}
       dyn_wings_phalange_01_l_ctrl.Rotation.X {type=MovieSceneScriptingFloatChannel, default=0}
       dyn_wings_phalange_01_l_ctrl.Rotation.Y {type=MovieSceneScriptingFloatChannel, default=0}
       dyn_wings_phalange_01_l_ctrl.Rotation.Z {type=MovieSceneScriptingFloatChannel, default=0}
       dyn_wings_phalange_01_l_ctrl.Scale.X {type=MovieSceneScriptingFloatChannel, default=0}
       dyn_wings_phalange_01_l_ctrl.Scale.Y {type=MovieSceneScriptingFloatChannel, default=0}
       dyn_wings_phalange_01_l_ctrl.Scale.Z {type=MovieSceneScriptingFloatChannel, default=0}
       dyn_wings_carpals_01_l_ctrl.Location.X {type=MovieSceneScriptingFloatChannel, default=0}
       dyn_wings_carpals_01_l_ctrl.Location.Y {type=MovieSceneScriptingFloatChannel, default=0}
       dyn_wings_carpals_01_l_ctrl.Location.Z {type=MovieSceneScriptingFloatChannel, default=0}
       dyn_wings_carpals_01_l_ctrl.Rotation.X {type=MovieSceneScriptingFloatChannel, default=0}
       dyn_wings_carpals_01_l_ctrl.Rotation.Y {type=MovieSceneScriptingFloatChannel, default=0}
       dyn_wings_carpals_01_l_ctrl.Rotation.Z {type=MovieSceneScriptingFloatChannel, default=0}
       dyn_wings_carpals_01_l_ctrl.Scale.X {type=MovieSceneScriptingFloatChannel, default=0}
       dyn_wings_carpals_01_l_ctrl.Scale.Y {type=MovieSceneScriptingFloatChannel, default=0}
       dyn_wings_carpals_01_l_ctrl.Scale.Z {type=MovieSceneScriptingFloatChannel, default=0}
       dyn_wings_metacarpal_02_l_ctrl.Location.X {type=MovieSceneScriptingFloatChannel, default=0}
       dyn_wings_metacarpal_02_l_ctrl.Location.Y {type=MovieSceneScriptingFloatChannel, default=0}
       dyn_wings_metacarpal_02_l_ctrl.Location.Z {type=MovieSceneScriptingFloatChannel, default=0}
       dyn_wings_metacarpal_02_l_ctrl.Rotation.X {type=MovieSceneScriptingFloatChannel, default=0}
       dyn_wings_metacarpal_02_l_ctrl.Rotation.Y {type=MovieSceneScriptingFloatChannel, default=0}
       dyn_wings_metacarpal_02_l_ctrl.Rotation.Z {type=MovieSceneScriptingFloatChannel, default=0}
       dyn_wings_metacarpal_02_l_ctrl.Scale.X {type=MovieSceneScriptingFloatChannel, default=0}
       dyn_wings_metacarpal_02_l_ctrl.Scale.Y {type=MovieSceneScriptingFloatChannel, default=0}
       dyn_wings_metacarpal_02_l_ctrl.Scale.Z {type=MovieSceneScriptingFloatChannel, default=0}
       dyn_wings_phalange_02_l_ctrl.Location.X {type=MovieSceneScriptingFloatChannel, default=0}
       dyn_wings_phalange_02_l_ctrl.Location.Y {type=MovieSceneScriptingFloatChannel, default=0}
       dyn_wings_phalange_02_l_ctrl.Location.Z {type=MovieSceneScriptingFloatChannel, default=0}
       dyn_wings_phalange_02_l_ctrl.Rotation.X {type=MovieSceneScriptingFloatChannel, default=0}
       dyn_wings_phalange_02_l_ctrl.Rotation.Y {type=MovieSceneScriptingFloatChannel, default=0}
       dyn_wings_phalange_02_l_ctrl.Rotation.Z {type=MovieSceneScriptingFloatChannel, default=0}
       dyn_wings_phalange_02_l_ctrl.Scale.X {type=MovieSceneScriptingFloatChannel, default=0}
       dyn_wings_phalange_02_l_ctrl.Scale.Y {type=MovieSceneScriptingFloatChannel, default=0}
       dyn_wings_phalange_02_l_ctrl.Scale.Z {type=MovieSceneScriptingFloatChannel, default=0}
       dyn_wings_carpals_02_l_ctrl.Location.X {type=MovieSceneScriptingFloatChannel, default=0}
       dyn_wings_carpals_02_l_ctrl.Location.Y {type=MovieSceneScriptingFloatChannel, default=0}
       dyn_wings_carpals_02_l_ctrl.Location.Z {type=MovieSceneScriptingFloatChannel, default=0}
       dyn_wings_carpals_02_l_ctrl.Rotation.X {type=MovieSceneScriptingFloatChannel, default=0}
       dyn_wings_carpals_02_l_ctrl.Rotation.Y {type=MovieSceneScriptingFloatChannel, default=0}
       dyn_wings_carpals_02_l_ctrl.Rotation.Z {type=MovieSceneScriptingFloatChannel, default=0}
       dyn_wings_carpals_02_l_ctrl.Scale.X {type=MovieSceneScriptingFloatChannel, default=0}
       dyn_wings_carpals_02_l_ctrl.Scale.Y {type=MovieSceneScriptingFloatChannel, default=0}
       dyn_wings_carpals_02_l_ctrl.Scale.Z {type=MovieSceneScriptingFloatChannel, default=0}
       dyn_wings_metacarpal_03_l_ctrl.Location.X {type=MovieSceneScriptingFloatChannel, default=0}
       dyn_wings_metacarpal_03_l_ctrl.Location.Y {type=MovieSceneScriptingFloatChannel, default=0}
       dyn_wings_metacarpal_03_l_ctrl.Location.Z {type=MovieSceneScriptingFloatChannel, default=0}
       dyn_wings_metacarpal_03_l_ctrl.Rotation.X {type=MovieSceneScriptingFloatChannel, default=0}
       dyn_wings_metacarpal_03_l_ctrl.Rotation.Y {type=MovieSceneScriptingFloatChannel, default=0}
       dyn_wings_metacarpal_03_l_ctrl.Rotation.Z {type=MovieSceneScriptingFloatChannel, default=0}
       dyn_wings_metacarpal_03_l_ctrl.Scale.X {type=MovieSceneScriptingFloatChannel, default=0}
       dyn_wings_metacarpal_03_l_ctrl.Scale.Y {type=MovieSceneScriptingFloatChannel, default=0}
       dyn_wings_metacarpal_03_l_ctrl.Scale.Z {type=MovieSceneScriptingFloatChannel, default=0}
       dyn_wings_phalange_03_l_ctrl.Location.X {type=MovieSceneScriptingFloatChannel, default=0}
       dyn_wings_phalange_03_l_ctrl.Location.Y {type=MovieSceneScriptingFloatChannel, default=0}
       dyn_wings_phalange_03_l_ctrl.Location.Z {type=MovieSceneScriptingFloatChannel, default=0}
       dyn_wings_phalange_03_l_ctrl.Rotation.X {type=MovieSceneScriptingFloatChannel, default=0}
       dyn_wings_phalange_03_l_ctrl.Rotation.Y {type=MovieSceneScriptingFloatChannel, default=0}
       dyn_wings_phalange_03_l_ctrl.Rotation.Z {type=MovieSceneScriptingFloatChannel, default=0}
       dyn_wings_phalange_03_l_ctrl.Scale.X {type=MovieSceneScriptingFloatChannel, default=0}
       dyn_wings_phalange_03_l_ctrl.Scale.Y {type=MovieSceneScriptingFloatChannel, default=0}
       dyn_wings_phalange_03_l_ctrl.Scale.Z {type=MovieSceneScriptingFloatChannel, default=0}
       dyn_wings_metacarpal_04_l_ctrl.Location.X {type=MovieSceneScriptingFloatChannel, default=0}
       dyn_wings_metacarpal_04_l_ctrl.Location.Y {type=MovieSceneScriptingFloatChannel, default=0}
       dyn_wings_metacarpal_04_l_ctrl.Location.Z {type=MovieSceneScriptingFloatChannel, default=0}
       dyn_wings_metacarpal_04_l_ctrl.Rotation.X {type=MovieSceneScriptingFloatChannel, default=0}
       dyn_wings_metacarpal_04_l_ctrl.Rotation.Y {type=MovieSceneScriptingFloatChannel, default=0}
       dyn_wings_metacarpal_04_l_ctrl.Rotation.Z {type=MovieSceneScriptingFloatChannel, default=0}
       dyn_wings_metacarpal_04_l_ctrl.Scale.X {type=MovieSceneScriptingFloatChannel, default=0}
       dyn_wings_metacarpal_04_l_ctrl.Scale.Y {type=MovieSceneScriptingFloatChannel, default=0}
       dyn_wings_metacarpal_04_l_ctrl.Scale.Z {type=MovieSceneScriptingFloatChannel, default=0}
       dyn_wings_phalange_04_l_ctrl.Location.X {type=MovieSceneScriptingFloatChannel, default=0}
       dyn_wings_phalange_04_l_ctrl.Location.Y {type=MovieSceneScriptingFloatChannel, default=0}
       dyn_wings_phalange_04_l_ctrl.Location.Z {type=MovieSceneScriptingFloatChannel, default=0}
       dyn_wings_phalange_04_l_ctrl.Rotation.X {type=MovieSceneScriptingFloatChannel, default=0}
       dyn_wings_phalange_04_l_ctrl.Rotation.Y {type=MovieSceneScriptingFloatChannel, default=0}
       dyn_wings_phalange_04_l_ctrl.Rotation.Z {type=MovieSceneScriptingFloatChannel, default=0}
       dyn_wings_phalange_04_l_ctrl.Scale.X {type=MovieSceneScriptingFloatChannel, default=0}
       dyn_wings_phalange_04_l_ctrl.Scale.Y {type=MovieSceneScriptingFloatChannel, default=0}
       dyn_wings_phalange_04_l_ctrl.Scale.Z {type=MovieSceneScriptingFloatChannel, default=0}
       dyn_wings_humelus_r_ctrl.Location.X {type=MovieSceneScriptingFloatChannel, default=0.0008, thinned_from=29, keys [[19,0.0008],[470,0.0008],[476,0.0002],[488,0]]}
       dyn_wings_humelus_r_ctrl.Location.Y {type=MovieSceneScriptingFloatChannel, default=0.0008, thinned_from=29, keys [[19,0.0008],[470,0.0008],[476,0.0002],[488,0]]}
       dyn_wings_humelus_r_ctrl.Location.Z {type=MovieSceneScriptingFloatChannel, default=0.0008, thinned_from=29, keys [[19,0.0008],[470,0.0008],[476,0.0002],[488,0]]}
       dyn_wings_humelus_r_ctrl.Rotation.X {
        type=MovieSceneScriptingFloatChannel
        default=4.989
        keys [[19,0],[22,5.1942],[26,13.1971],[35,13.1971],[37,12.2748],[41,12.2809],[44,12.0326],[50,12.0447],[58,2.3174],[120,2.3174],[123,-2.1416],[165,-2.1416],[169,6.9516],[197,6.9516],[231,-10.8089],[248,-10.0341],[261,-5.1601],[276,-0.0775],[307,-0.0775],[310,-11.6761],[312,-11.6851],[315,-27.1077],[347,-27.1077],[350,-7.3906],[367,-7.3779],[370,0.0419],[390,0.0419],[394,6.5168],[400,6.5171],[405,11.0647],[407,11.0308],[411,22.5099],[471,22.5099],[473,13.123],[474,13.1159],[476,3.5839],[486,3.5769],[488,0]]
       }
       dyn_wings_humelus_r_ctrl.Rotation.Y {type=MovieSceneScriptingFloatChannel, default=3.5889, const [0,29]}
       dyn_wings_humelus_r_ctrl.Rotation.Z {type=MovieSceneScriptingFloatChannel, default=19.6242, const [0,29]}
       dyn_wings_humelus_r_ctrl.Scale.X {type=MovieSceneScriptingFloatChannel, default=0, const [0,29]}
       dyn_wings_humelus_r_ctrl.Scale.Y {type=MovieSceneScriptingFloatChannel, default=0, const [0,29]}
       dyn_wings_humelus_r_ctrl.Scale.Z {type=MovieSceneScriptingFloatChannel, default=0, const [0,29]}
       dyn_wings_radius_r_ctrl.Location.X {type=MovieSceneScriptingFloatChannel, default=0}
       dyn_wings_radius_r_ctrl.Location.Y {type=MovieSceneScriptingFloatChannel, default=0}
       dyn_wings_radius_r_ctrl.Location.Z {type=MovieSceneScriptingFloatChannel, default=0}
       dyn_wings_radius_r_ctrl.Rotation.X {type=MovieSceneScriptingFloatChannel, default=0}
       dyn_wings_radius_r_ctrl.Rotation.Y {type=MovieSceneScriptingFloatChannel, default=0}
       dyn_wings_radius_r_ctrl.Rotation.Z {type=MovieSceneScriptingFloatChannel, default=0}
       dyn_wings_radius_r_ctrl.Scale.X {type=MovieSceneScriptingFloatChannel, default=0}
       dyn_wings_radius_r_ctrl.Scale.Y {type=MovieSceneScriptingFloatChannel, default=0}
       dyn_wings_radius_r_ctrl.Scale.Z {type=MovieSceneScriptingFloatChannel, default=0}
       dyn_wings_metacarpal_01_r_ctrl.Location.X {type=MovieSceneScriptingFloatChannel, default=0}
       dyn_wings_metacarpal_01_r_ctrl.Location.Y {type=MovieSceneScriptingFloatChannel, default=0}
       dyn_wings_metacarpal_01_r_ctrl.Location.Z {type=MovieSceneScriptingFloatChannel, default=0}
       dyn_wings_metacarpal_01_r_ctrl.Rotation.X {type=MovieSceneScriptingFloatChannel, default=0}
       dyn_wings_metacarpal_01_r_ctrl.Rotation.Y {type=MovieSceneScriptingFloatChannel, default=0}
       dyn_wings_metacarpal_01_r_ctrl.Rotation.Z {type=MovieSceneScriptingFloatChannel, default=0}
       dyn_wings_metacarpal_01_r_ctrl.Scale.X {type=MovieSceneScriptingFloatChannel, default=0}
       dyn_wings_metacarpal_01_r_ctrl.Scale.Y {type=MovieSceneScriptingFloatChannel, default=0}
       dyn_wings_metacarpal_01_r_ctrl.Scale.Z {type=MovieSceneScriptingFloatChannel, default=0}
       dyn_wings_phalange_01_r_ctrl.Location.X {type=MovieSceneScriptingFloatChannel, default=0}
       dyn_wings_phalange_01_r_ctrl.Location.Y {type=MovieSceneScriptingFloatChannel, default=0}
       dyn_wings_phalange_01_r_ctrl.Location.Z {type=MovieSceneScriptingFloatChannel, default=0}
       dyn_wings_phalange_01_r_ctrl.Rotation.X {type=MovieSceneScriptingFloatChannel, default=0}
       dyn_wings_phalange_01_r_ctrl.Rotation.Y {type=MovieSceneScriptingFloatChannel, default=0}
       dyn_wings_phalange_01_r_ctrl.Rotation.Z {type=MovieSceneScriptingFloatChannel, default=0}
       dyn_wings_phalange_01_r_ctrl.Scale.X {type=MovieSceneScriptingFloatChannel, default=0}
       dyn_wings_phalange_01_r_ctrl.Scale.Y {type=MovieSceneScriptingFloatChannel, default=0}
       dyn_wings_phalange_01_r_ctrl.Scale.Z {type=MovieSceneScriptingFloatChannel, default=0}
       dyn_wings_carpals_01_r_ctrl.Location.X {type=MovieSceneScriptingFloatChannel, default=0}
       dyn_wings_carpals_01_r_ctrl.Location.Y {type=MovieSceneScriptingFloatChannel, default=0}
       dyn_wings_carpals_01_r_ctrl.Location.Z {type=MovieSceneScriptingFloatChannel, default=0}
       dyn_wings_carpals_01_r_ctrl.Rotation.X {type=MovieSceneScriptingFloatChannel, default=0}
       dyn_wings_carpals_01_r_ctrl.Rotation.Y {type=MovieSceneScriptingFloatChannel, default=0}
       dyn_wings_carpals_01_r_ctrl.Rotation.Z {type=MovieSceneScriptingFloatChannel, default=0}
       dyn_wings_carpals_01_r_ctrl.Scale.X {type=MovieSceneScriptingFloatChannel, default=0}
       dyn_wings_carpals_01_r_ctrl.Scale.Y {type=MovieSceneScriptingFloatChannel, default=0}
       dyn_wings_carpals_01_r_ctrl.Scale.Z {type=MovieSceneScriptingFloatChannel, default=0}
       dyn_wings_metacarpal_02_r_ctrl.Location.X {type=MovieSceneScriptingFloatChannel, default=0}
       dyn_wings_metacarpal_02_r_ctrl.Location.Y {type=MovieSceneScriptingFloatChannel, default=0}
       dyn_wings_metacarpal_02_r_ctrl.Location.Z {type=MovieSceneScriptingFloatChannel, default=0}
       dyn_wings_metacarpal_02_r_ctrl.Rotation.X {type=MovieSceneScriptingFloatChannel, default=0}
       dyn_wings_metacarpal_02_r_ctrl.Rotation.Y {type=MovieSceneScriptingFloatChannel, default=0}
       dyn_wings_metacarpal_02_r_ctrl.Rotation.Z {type=MovieSceneScriptingFloatChannel, default=0}
       dyn_wings_metacarpal_02_r_ctrl.Scale.X {type=MovieSceneScriptingFloatChannel, default=0}
       dyn_wings_metacarpal_02_r_ctrl.Scale.Y {type=MovieSceneScriptingFloatChannel, default=0}
       dyn_wings_metacarpal_02_r_ctrl.Scale.Z {type=MovieSceneScriptingFloatChannel, default=0}
       dyn_wings_phalange_02_r_ctrl.Location.X {type=MovieSceneScriptingFloatChannel, default=0}
       dyn_wings_phalange_02_r_ctrl.Location.Y {type=MovieSceneScriptingFloatChannel, default=0}
       dyn_wings_phalange_02_r_ctrl.Location.Z {type=MovieSceneScriptingFloatChannel, default=0}
       dyn_wings_phalange_02_r_ctrl.Rotation.X {type=MovieSceneScriptingFloatChannel, default=0}
       dyn_wings_phalange_02_r_ctrl.Rotation.Y {type=MovieSceneScriptingFloatChannel, default=0}
       dyn_wings_phalange_02_r_ctrl.Rotation.Z {type=MovieSceneScriptingFloatChannel, default=0}
       dyn_wings_phalange_02_r_ctrl.Scale.X {type=MovieSceneScriptingFloatChannel, default=0}
       dyn_wings_phalange_02_r_ctrl.Scale.Y {type=MovieSceneScriptingFloatChannel, default=0}
       dyn_wings_phalange_02_r_ctrl.Scale.Z {type=MovieSceneScriptingFloatChannel, default=0}
       dyn_wings_carpals_02_r_ctrl.Location.X {type=MovieSceneScriptingFloatChannel, default=0}
       dyn_wings_carpals_02_r_ctrl.Location.Y {type=MovieSceneScriptingFloatChannel, default=0}
       dyn_wings_carpals_02_r_ctrl.Location.Z {type=MovieSceneScriptingFloatChannel, default=0}
       dyn_wings_carpals_02_r_ctrl.Rotation.X {type=MovieSceneScriptingFloatChannel, default=0}
       dyn_wings_carpals_02_r_ctrl.Rotation.Y {type=MovieSceneScriptingFloatChannel, default=0}
       dyn_wings_carpals_02_r_ctrl.Rotation.Z {type=MovieSceneScriptingFloatChannel, default=0}
       dyn_wings_carpals_02_r_ctrl.Scale.X {type=MovieSceneScriptingFloatChannel, default=0}
       dyn_wings_carpals_02_r_ctrl.Scale.Y {type=MovieSceneScriptingFloatChannel, default=0}
       dyn_wings_carpals_02_r_ctrl.Scale.Z {type=MovieSceneScriptingFloatChannel, default=0}
       dyn_wings_metacarpal_03_r_ctrl.Location.X {type=MovieSceneScriptingFloatChannel, default=0}
       dyn_wings_metacarpal_03_r_ctrl.Location.Y {type=MovieSceneScriptingFloatChannel, default=0}
       dyn_wings_metacarpal_03_r_ctrl.Location.Z {type=MovieSceneScriptingFloatChannel, default=0}
       dyn_wings_metacarpal_03_r_ctrl.Rotation.X {type=MovieSceneScriptingFloatChannel, default=0}
       dyn_wings_metacarpal_03_r_ctrl.Rotation.Y {type=MovieSceneScriptingFloatChannel, default=0}
       dyn_wings_metacarpal_03_r_ctrl.Rotation.Z {type=MovieSceneScriptingFloatChannel, default=0}
       dyn_wings_metacarpal_03_r_ctrl.Scale.X {type=MovieSceneScriptingFloatChannel, default=0}
       dyn_wings_metacarpal_03_r_ctrl.Scale.Y {type=MovieSceneScriptingFloatChannel, default=0}
       dyn_wings_metacarpal_03_r_ctrl.Scale.Z {type=MovieSceneScriptingFloatChannel, default=0}
       dyn_wings_phalange_03_r_ctrl.Location.X {type=MovieSceneScriptingFloatChannel, default=0}
       dyn_wings_phalange_03_r_ctrl.Location.Y {type=MovieSceneScriptingFloatChannel, default=0}
       dyn_wings_phalange_03_r_ctrl.Location.Z {type=MovieSceneScriptingFloatChannel, default=0}
       dyn_wings_phalange_03_r_ctrl.Rotation.X {type=MovieSceneScriptingFloatChannel, default=0}
       dyn_wings_phalange_03_r_ctrl.Rotation.Y {type=MovieSceneScriptingFloatChannel, default=0}
       dyn_wings_phalange_03_r_ctrl.Rotation.Z {type=MovieSceneScriptingFloatChannel, default=0}
       dyn_wings_phalange_03_r_ctrl.Scale.X {type=MovieSceneScriptingFloatChannel, default=0}
       dyn_wings_phalange_03_r_ctrl.Scale.Y {type=MovieSceneScriptingFloatChannel, default=0}
       dyn_wings_phalange_03_r_ctrl.Scale.Z {type=MovieSceneScriptingFloatChannel, default=0}
       dyn_wings_metacarpal_04_r_ctrl.Location.X {type=MovieSceneScriptingFloatChannel, default=0}
       dyn_wings_metacarpal_04_r_ctrl.Location.Y {type=MovieSceneScriptingFloatChannel, default=0}
       dyn_wings_metacarpal_04_r_ctrl.Location.Z {type=MovieSceneScriptingFloatChannel, default=0}
       dyn_wings_metacarpal_04_r_ctrl.Rotation.X {type=MovieSceneScriptingFloatChannel, default=0}
       dyn_wings_metacarpal_04_r_ctrl.Rotation.Y {type=MovieSceneScriptingFloatChannel, default=0}
       dyn_wings_metacarpal_04_r_ctrl.Rotation.Z {type=MovieSceneScriptingFloatChannel, default=0}
       dyn_wings_metacarpal_04_r_ctrl.Scale.X {type=MovieSceneScriptingFloatChannel, default=0}
       dyn_wings_metacarpal_04_r_ctrl.Scale.Y {type=MovieSceneScriptingFloatChannel, default=0}
       dyn_wings_metacarpal_04_r_ctrl.Scale.Z {type=MovieSceneScriptingFloatChannel, default=0}
       dyn_wings_phalange_04_r_ctrl.Location.X {type=MovieSceneScriptingFloatChannel, default=0}
       dyn_wings_phalange_04_r_ctrl.Location.Y {type=MovieSceneScriptingFloatChannel, default=0}
       dyn_wings_phalange_04_r_ctrl.Location.Z {type=MovieSceneScriptingFloatChannel, default=0}
       dyn_wings_phalange_04_r_ctrl.Rotation.X {type=MovieSceneScriptingFloatChannel, default=0}
       dyn_wings_phalange_04_r_ctrl.Rotation.Y {type=MovieSceneScriptingFloatChannel, default=0}
       dyn_wings_phalange_04_r_ctrl.Rotation.Z {type=MovieSceneScriptingFloatChannel, default=0}
       dyn_wings_phalange_04_r_ctrl.Scale.X {type=MovieSceneScriptingFloatChannel, default=0}
       dyn_wings_phalange_04_r_ctrl.Scale.Y {type=MovieSceneScriptingFloatChannel, default=0}
       dyn_wings_phalange_04_r_ctrl.Scale.Z {type=MovieSceneScriptingFloatChannel, default=0}
      ]
     }
     {
      type=MovieSceneControlRigParameterSection
      end_frame=751
      start_frame=0
      channels [
       pack_ctrl.Location.X {type=MovieSceneScriptingFloatChannel, default=0}
       pack_ctrl.Location.Y {type=MovieSceneScriptingFloatChannel, default=0}
       pack_ctrl.Location.Z {type=MovieSceneScriptingFloatChannel, default=0}
       pack_ctrl.Rotation.X {type=MovieSceneScriptingFloatChannel, default=0}
       pack_ctrl.Rotation.Y {type=MovieSceneScriptingFloatChannel, default=0}
       pack_ctrl.Rotation.Z {type=MovieSceneScriptingFloatChannel, default=0}
       pack_ctrl.Scale.X {type=MovieSceneScriptingFloatChannel, default=0}
       pack_ctrl.Scale.Y {type=MovieSceneScriptingFloatChannel, default=0}
       pack_ctrl.Scale.Z {type=MovieSceneScriptingFloatChannel, default=0}
       dyn_wings_humelus_l_ctrl.Location.X {type=MovieSceneScriptingFloatChannel, default=0, const [0,4]}
       dyn_wings_humelus_l_ctrl.Location.Y {type=MovieSceneScriptingFloatChannel, default=0, const [0,4]}
       dyn_wings_humelus_l_ctrl.Location.Z {type=MovieSceneScriptingFloatChannel, default=0, const [0,4]}
       dyn_wings_humelus_l_ctrl.Rotation.X {type=MovieSceneScriptingFloatChannel, default=0, keys [[322.3625,0],[331.5175,17.4662],[344.6788,17.4662],[546,0]]}
       dyn_wings_humelus_l_ctrl.Rotation.Y {type=MovieSceneScriptingFloatChannel, default=0, keys [[322.3625,0.0001],[331.5175,-7.5782],[344.6788,-7.5782],[546,0]]}
       dyn_wings_humelus_l_ctrl.Rotation.Z {type=MovieSceneScriptingFloatChannel, default=0, keys [[322.3625,0.0018],[331.5175,41.7925],[344.6788,41.7925],[546,0]]}
       dyn_wings_humelus_l_ctrl.Scale.X {type=MovieSceneScriptingFloatChannel, default=0, const [0,4]}
       dyn_wings_humelus_l_ctrl.Scale.Y {type=MovieSceneScriptingFloatChannel, default=0, const [0,4]}
       dyn_wings_humelus_l_ctrl.Scale.Z {type=MovieSceneScriptingFloatChannel, default=0, const [0,4]}
       dyn_wings_radius_l_ctrl.Location.X {type=MovieSceneScriptingFloatChannel, default=0}
       dyn_wings_radius_l_ctrl.Location.Y {type=MovieSceneScriptingFloatChannel, default=0}
       dyn_wings_radius_l_ctrl.Location.Z {type=MovieSceneScriptingFloatChannel, default=0}
       dyn_wings_radius_l_ctrl.Rotation.X {type=MovieSceneScriptingFloatChannel, default=0}
       dyn_wings_radius_l_ctrl.Rotation.Y {type=MovieSceneScriptingFloatChannel, default=0}
       dyn_wings_radius_l_ctrl.Rotation.Z {type=MovieSceneScriptingFloatChannel, default=0}
       dyn_wings_radius_l_ctrl.Scale.X {type=MovieSceneScriptingFloatChannel, default=0}
       dyn_wings_radius_l_ctrl.Scale.Y {type=MovieSceneScriptingFloatChannel, default=0}
       dyn_wings_radius_l_ctrl.Scale.Z {type=MovieSceneScriptingFloatChannel, default=0}
       dyn_wings_metacarpal_01_l_ctrl.Location.X {type=MovieSceneScriptingFloatChannel, default=0}
       dyn_wings_metacarpal_01_l_ctrl.Location.Y {type=MovieSceneScriptingFloatChannel, default=0}
       dyn_wings_metacarpal_01_l_ctrl.Location.Z {type=MovieSceneScriptingFloatChannel, default=0}
       dyn_wings_metacarpal_01_l_ctrl.Rotation.X {type=MovieSceneScriptingFloatChannel, default=0}
       dyn_wings_metacarpal_01_l_ctrl.Rotation.Y {type=MovieSceneScriptingFloatChannel, default=0}
       dyn_wings_metacarpal_01_l_ctrl.Rotation.Z {type=MovieSceneScriptingFloatChannel, default=0}
       dyn_wings_metacarpal_01_l_ctrl.Scale.X {type=MovieSceneScriptingFloatChannel, default=0}
       dyn_wings_metacarpal_01_l_ctrl.Scale.Y {type=MovieSceneScriptingFloatChannel, default=0}
       dyn_wings_metacarpal_01_l_ctrl.Scale.Z {type=MovieSceneScriptingFloatChannel, default=0}
       dyn_wings_phalange_01_l_ctrl.Location.X {type=MovieSceneScriptingFloatChannel, default=0}
       dyn_wings_phalange_01_l_ctrl.Location.Y {type=MovieSceneScriptingFloatChannel, default=0}
       dyn_wings_phalange_01_l_ctrl.Location.Z {type=MovieSceneScriptingFloatChannel, default=0}
       dyn_wings_phalange_01_l_ctrl.Rotation.X {type=MovieSceneScriptingFloatChannel, default=0}
       dyn_wings_phalange_01_l_ctrl.Rotation.Y {type=MovieSceneScriptingFloatChannel, default=0}
       dyn_wings_phalange_01_l_ctrl.Rotation.Z {type=MovieSceneScriptingFloatChannel, default=0}
       dyn_wings_phalange_01_l_ctrl.Scale.X {type=MovieSceneScriptingFloatChannel, default=0}
       dyn_wings_phalange_01_l_ctrl.Scale.Y {type=MovieSceneScriptingFloatChannel, default=0}
       dyn_wings_phalange_01_l_ctrl.Scale.Z {type=MovieSceneScriptingFloatChannel, default=0}
       dyn_wings_carpals_01_l_ctrl.Location.X {type=MovieSceneScriptingFloatChannel, default=0}
       dyn_wings_carpals_01_l_ctrl.Location.Y {type=MovieSceneScriptingFloatChannel, default=0}
       dyn_wings_carpals_01_l_ctrl.Location.Z {type=MovieSceneScriptingFloatChannel, default=0}
       dyn_wings_carpals_01_l_ctrl.Rotation.X {type=MovieSceneScriptingFloatChannel, default=0}
       dyn_wings_carpals_01_l_ctrl.Rotation.Y {type=MovieSceneScriptingFloatChannel, default=0}
       dyn_wings_carpals_01_l_ctrl.Rotation.Z {type=MovieSceneScriptingFloatChannel, default=0}
       dyn_wings_carpals_01_l_ctrl.Scale.X {type=MovieSceneScriptingFloatChannel, default=0}
       dyn_wings_carpals_01_l_ctrl.Scale.Y {type=MovieSceneScriptingFloatChannel, default=0}
       dyn_wings_carpals_01_l_ctrl.Scale.Z {type=MovieSceneScriptingFloatChannel, default=0}
       dyn_wings_metacarpal_02_l_ctrl.Location.X {type=MovieSceneScriptingFloatChannel, default=0}
       dyn_wings_metacarpal_02_l_ctrl.Location.Y {type=MovieSceneScriptingFloatChannel, default=0}
       dyn_wings_metacarpal_02_l_ctrl.Location.Z {type=MovieSceneScriptingFloatChannel, default=0}
       dyn_wings_metacarpal_02_l_ctrl.Rotation.X {type=MovieSceneScriptingFloatChannel, default=0}
       dyn_wings_metacarpal_02_l_ctrl.Rotation.Y {type=MovieSceneScriptingFloatChannel, default=0}
       dyn_wings_metacarpal_02_l_ctrl.Rotation.Z {type=MovieSceneScriptingFloatChannel, default=0}
       dyn_wings_metacarpal_02_l_ctrl.Scale.X {type=MovieSceneScriptingFloatChannel, default=0}
       dyn_wings_metacarpal_02_l_ctrl.Scale.Y {type=MovieSceneScriptingFloatChannel, default=0}
       dyn_wings_metacarpal_02_l_ctrl.Scale.Z {type=MovieSceneScriptingFloatChannel, default=0}
       dyn_wings_phalange_02_l_ctrl.Location.X {type=MovieSceneScriptingFloatChannel, default=0}
       dyn_wings_phalange_02_l_ctrl.Location.Y {type=MovieSceneScriptingFloatChannel, default=0}
       dyn_wings_phalange_02_l_ctrl.Location.Z {type=MovieSceneScriptingFloatChannel, default=0}
       dyn_wings_phalange_02_l_ctrl.Rotation.X {type=MovieSceneScriptingFloatChannel, default=0}
       dyn_wings_phalange_02_l_ctrl.Rotation.Y {type=MovieSceneScriptingFloatChannel, default=0}
       dyn_wings_phalange_02_l_ctrl.Rotation.Z {type=MovieSceneScriptingFloatChannel, default=0}
       dyn_wings_phalange_02_l_ctrl.Scale.X {type=MovieSceneScriptingFloatChannel, default=0}
       dyn_wings_phalange_02_l_ctrl.Scale.Y {type=MovieSceneScriptingFloatChannel, default=0}
       dyn_wings_phalange_02_l_ctrl.Scale.Z {type=MovieSceneScriptingFloatChannel, default=0}
       dyn_wings_carpals_02_l_ctrl.Location.X {type=MovieSceneScriptingFloatChannel, default=0}
       dyn_wings_carpals_02_l_ctrl.Location.Y {type=MovieSceneScriptingFloatChannel, default=0}
       dyn_wings_carpals_02_l_ctrl.Location.Z {type=MovieSceneScriptingFloatChannel, default=0}
       dyn_wings_carpals_02_l_ctrl.Rotation.X {type=MovieSceneScriptingFloatChannel, default=0}
       dyn_wings_carpals_02_l_ctrl.Rotation.Y {type=MovieSceneScriptingFloatChannel, default=0}
       dyn_wings_carpals_02_l_ctrl.Rotation.Z {type=MovieSceneScriptingFloatChannel, default=0}
       dyn_wings_carpals_02_l_ctrl.Scale.X {type=MovieSceneScriptingFloatChannel, default=0}
       dyn_wings_carpals_02_l_ctrl.Scale.Y {type=MovieSceneScriptingFloatChannel, default=0}
       dyn_wings_carpals_02_l_ctrl.Scale.Z {type=MovieSceneScriptingFloatChannel, default=0}
       dyn_wings_metacarpal_03_l_ctrl.Location.X {type=MovieSceneScriptingFloatChannel, default=0}
       dyn_wings_metacarpal_03_l_ctrl.Location.Y {type=MovieSceneScriptingFloatChannel, default=0}
       dyn_wings_metacarpal_03_l_ctrl.Location.Z {type=MovieSceneScriptingFloatChannel, default=0}
       dyn_wings_metacarpal_03_l_ctrl.Rotation.X {type=MovieSceneScriptingFloatChannel, default=0}
       dyn_wings_metacarpal_03_l_ctrl.Rotation.Y {type=MovieSceneScriptingFloatChannel, default=0}
       dyn_wings_metacarpal_03_l_ctrl.Rotation.Z {type=MovieSceneScriptingFloatChannel, default=0}
       dyn_wings_metacarpal_03_l_ctrl.Scale.X {type=MovieSceneScriptingFloatChannel, default=0}
       dyn_wings_metacarpal_03_l_ctrl.Scale.Y {type=MovieSceneScriptingFloatChannel, default=0}
       dyn_wings_metacarpal_03_l_ctrl.Scale.Z {type=MovieSceneScriptingFloatChannel, default=0}
       dyn_wings_phalange_03_l_ctrl.Location.X {type=MovieSceneScriptingFloatChannel, default=0}
       dyn_wings_phalange_03_l_ctrl.Location.Y {type=MovieSceneScriptingFloatChannel, default=0}
       dyn_wings_phalange_03_l_ctrl.Location.Z {type=MovieSceneScriptingFloatChannel, default=0}
       dyn_wings_phalange_03_l_ctrl.Rotation.X {type=MovieSceneScriptingFloatChannel, default=0}
       dyn_wings_phalange_03_l_ctrl.Rotation.Y {type=MovieSceneScriptingFloatChannel, default=0}
       dyn_wings_phalange_03_l_ctrl.Rotation.Z {type=MovieSceneScriptingFloatChannel, default=0}
       dyn_wings_phalange_03_l_ctrl.Scale.X {type=MovieSceneScriptingFloatChannel, default=0}
       dyn_wings_phalange_03_l_ctrl.Scale.Y {type=MovieSceneScriptingFloatChannel, default=0}
       dyn_wings_phalange_03_l_ctrl.Scale.Z {type=MovieSceneScriptingFloatChannel, default=0}
       dyn_wings_metacarpal_04_l_ctrl.Location.X {type=MovieSceneScriptingFloatChannel, default=0}
       dyn_wings_metacarpal_04_l_ctrl.Location.Y {type=MovieSceneScriptingFloatChannel, default=0}
       dyn_wings_metacarpal_04_l_ctrl.Location.Z {type=MovieSceneScriptingFloatChannel, default=0}
       dyn_wings_metacarpal_04_l_ctrl.Rotation.X {type=MovieSceneScriptingFloatChannel, default=0}
       dyn_wings_metacarpal_04_l_ctrl.Rotation.Y {type=MovieSceneScriptingFloatChannel, default=0}
       dyn_wings_metacarpal_04_l_ctrl.Rotation.Z {type=MovieSceneScriptingFloatChannel, default=0}
       dyn_wings_metacarpal_04_l_ctrl.Scale.X {type=MovieSceneScriptingFloatChannel, default=0}
       dyn_wings_metacarpal_04_l_ctrl.Scale.Y {type=MovieSceneScriptingFloatChannel, default=0}
       dyn_wings_metacarpal_04_l_ctrl.Scale.Z {type=MovieSceneScriptingFloatChannel, default=0}
       dyn_wings_phalange_04_l_ctrl.Location.X {type=MovieSceneScriptingFloatChannel, default=0}
       dyn_wings_phalange_04_l_ctrl.Location.Y {type=MovieSceneScriptingFloatChannel, default=0}
       dyn_wings_phalange_04_l_ctrl.Location.Z {type=MovieSceneScriptingFloatChannel, default=0}
       dyn_wings_phalange_04_l_ctrl.Rotation.X {type=MovieSceneScriptingFloatChannel, default=0}
       dyn_wings_phalange_04_l_ctrl.Rotation.Y {type=MovieSceneScriptingFloatChannel, default=0}
       dyn_wings_phalange_04_l_ctrl.Rotation.Z {type=MovieSceneScriptingFloatChannel, default=0}
       dyn_wings_phalange_04_l_ctrl.Scale.X {type=MovieSceneScriptingFloatChannel, default=0}
       dyn_wings_phalange_04_l_ctrl.Scale.Y {type=MovieSceneScriptingFloatChannel, default=0}
       dyn_wings_phalange_04_l_ctrl.Scale.Z {type=MovieSceneScriptingFloatChannel, default=0}
       dyn_wings_humelus_r_ctrl.Location.X {type=MovieSceneScriptingFloatChannel, default=0}
       dyn_wings_humelus_r_ctrl.Location.Y {type=MovieSceneScriptingFloatChannel, default=0}
       dyn_wings_humelus_r_ctrl.Location.Z {type=MovieSceneScriptingFloatChannel, default=0}
       dyn_wings_humelus_r_ctrl.Rotation.X {type=MovieSceneScriptingFloatChannel, default=0}
       dyn_wings_humelus_r_ctrl.Rotation.Y {type=MovieSceneScriptingFloatChannel, default=0}
       dyn_wings_humelus_r_ctrl.Rotation.Z {type=MovieSceneScriptingFloatChannel, default=0}
       dyn_wings_humelus_r_ctrl.Scale.X {type=MovieSceneScriptingFloatChannel, default=0}
       dyn_wings_humelus_r_ctrl.Scale.Y {type=MovieSceneScriptingFloatChannel, default=0}
       dyn_wings_humelus_r_ctrl.Scale.Z {type=MovieSceneScriptingFloatChannel, default=0}
       dyn_wings_radius_r_ctrl.Location.X {type=MovieSceneScriptingFloatChannel, default=0}
       dyn_wings_radius_r_ctrl.Location.Y {type=MovieSceneScriptingFloatChannel, default=0}
       dyn_wings_radius_r_ctrl.Location.Z {type=MovieSceneScriptingFloatChannel, default=0}
       dyn_wings_radius_r_ctrl.Rotation.X {type=MovieSceneScriptingFloatChannel, default=0}
       dyn_wings_radius_r_ctrl.Rotation.Y {type=MovieSceneScriptingFloatChannel, default=0}
       dyn_wings_radius_r_ctrl.Rotation.Z {type=MovieSceneScriptingFloatChannel, default=0}
       dyn_wings_radius_r_ctrl.Scale.X {type=MovieSceneScriptingFloatChannel, default=0}
       dyn_wings_radius_r_ctrl.Scale.Y {type=MovieSceneScriptingFloatChannel, default=0}
       dyn_wings_radius_r_ctrl.Scale.Z {type=MovieSceneScriptingFloatChannel, default=0}
       dyn_wings_metacarpal_01_r_ctrl.Location.X {type=MovieSceneScriptingFloatChannel, default=0}
       dyn_wings_metacarpal_01_r_ctrl.Location.Y {type=MovieSceneScriptingFloatChannel, default=0}
       dyn_wings_metacarpal_01_r_ctrl.Location.Z {type=MovieSceneScriptingFloatChannel, default=0}
       dyn_wings_metacarpal_01_r_ctrl.Rotation.X {type=MovieSceneScriptingFloatChannel, default=0}
       dyn_wings_metacarpal_01_r_ctrl.Rotation.Y {type=MovieSceneScriptingFloatChannel, default=0}
       dyn_wings_metacarpal_01_r_ctrl.Rotation.Z {type=MovieSceneScriptingFloatChannel, default=0}
       dyn_wings_metacarpal_01_r_ctrl.Scale.X {type=MovieSceneScriptingFloatChannel, default=0}
       dyn_wings_metacarpal_01_r_ctrl.Scale.Y {type=MovieSceneScriptingFloatChannel, default=0}
       dyn_wings_metacarpal_01_r_ctrl.Scale.Z {type=MovieSceneScriptingFloatChannel, default=0}
       dyn_wings_phalange_01_r_ctrl.Location.X {type=MovieSceneScriptingFloatChannel, default=0}
       dyn_wings_phalange_01_r_ctrl.Location.Y {type=MovieSceneScriptingFloatChannel, default=0}
       dyn_wings_phalange_01_r_ctrl.Location.Z {type=MovieSceneScriptingFloatChannel, default=0}
       dyn_wings_phalange_01_r_ctrl.Rotation.X {type=MovieSceneScriptingFloatChannel, default=0}
       dyn_wings_phalange_01_r_ctrl.Rotation.Y {type=MovieSceneScriptingFloatChannel, default=0}
       dyn_wings_phalange_01_r_ctrl.Rotation.Z {type=MovieSceneScriptingFloatChannel, default=0}
       dyn_wings_phalange_01_r_ctrl.Scale.X {type=MovieSceneScriptingFloatChannel, default=0}
       dyn_wings_phalange_01_r_ctrl.Scale.Y {type=MovieSceneScriptingFloatChannel, default=0}
       dyn_wings_phalange_01_r_ctrl.Scale.Z {type=MovieSceneScriptingFloatChannel, default=0}
       dyn_wings_carpals_01_r_ctrl.Location.X {type=MovieSceneScriptingFloatChannel, default=0}
       dyn_wings_carpals_01_r_ctrl.Location.Y {type=MovieSceneScriptingFloatChannel, default=0}
       dyn_wings_carpals_01_r_ctrl.Location.Z {type=MovieSceneScriptingFloatChannel, default=0}
       dyn_wings_carpals_01_r_ctrl.Rotation.X {type=MovieSceneScriptingFloatChannel, default=0}
       dyn_wings_carpals_01_r_ctrl.Rotation.Y {type=MovieSceneScriptingFloatChannel, default=0}
       dyn_wings_carpals_01_r_ctrl.Rotation.Z {type=MovieSceneScriptingFloatChannel, default=0}
       dyn_wings_carpals_01_r_ctrl.Scale.X {type=MovieSceneScriptingFloatChannel, default=0}
       dyn_wings_carpals_01_r_ctrl.Scale.Y {type=MovieSceneScriptingFloatChannel, default=0}
       dyn_wings_carpals_01_r_ctrl.Scale.Z {type=MovieSceneScriptingFloatChannel, default=0}
       dyn_wings_metacarpal_02_r_ctrl.Location.X {type=MovieSceneScriptingFloatChannel, default=0}
       dyn_wings_metacarpal_02_r_ctrl.Location.Y {type=MovieSceneScriptingFloatChannel, default=0}
       dyn_wings_metacarpal_02_r_ctrl.Location.Z {type=MovieSceneScriptingFloatChannel, default=0}
       dyn_wings_metacarpal_02_r_ctrl.Rotation.X {type=MovieSceneScriptingFloatChannel, default=0}
       dyn_wings_metacarpal_02_r_ctrl.Rotation.Y {type=MovieSceneScriptingFloatChannel, default=0}
       dyn_wings_metacarpal_02_r_ctrl.Rotation.Z {type=MovieSceneScriptingFloatChannel, default=0}
       dyn_wings_metacarpal_02_r_ctrl.Scale.X {type=MovieSceneScriptingFloatChannel, default=0}
       dyn_wings_metacarpal_02_r_ctrl.Scale.Y {type=MovieSceneScriptingFloatChannel, default=0}
       dyn_wings_metacarpal_02_r_ctrl.Scale.Z {type=MovieSceneScriptingFloatChannel, default=0}
       dyn_wings_phalange_02_r_ctrl.Location.X {type=MovieSceneScriptingFloatChannel, default=0}
       dyn_wings_phalange_02_r_ctrl.Location.Y {type=MovieSceneScriptingFloatChannel, default=0}
       dyn_wings_phalange_02_r_ctrl.Location.Z {type=MovieSceneScriptingFloatChannel, default=0}
       dyn_wings_phalange_02_r_ctrl.Rotation.X {type=MovieSceneScriptingFloatChannel, default=0}
       dyn_wings_phalange_02_r_ctrl.Rotation.Y {type=MovieSceneScriptingFloatChannel, default=0}
       dyn_wings_phalange_02_r_ctrl.Rotation.Z {type=MovieSceneScriptingFloatChannel, default=0}
       dyn_wings_phalange_02_r_ctrl.Scale.X {type=MovieSceneScriptingFloatChannel, default=0}
       dyn_wings_phalange_02_r_ctrl.Scale.Y {type=MovieSceneScriptingFloatChannel, default=0}
       dyn_wings_phalange_02_r_ctrl.Scale.Z {type=MovieSceneScriptingFloatChannel, default=0}
       dyn_wings_carpals_02_r_ctrl.Location.X {type=MovieSceneScriptingFloatChannel, default=0}
       dyn_wings_carpals_02_r_ctrl.Location.Y {type=MovieSceneScriptingFloatChannel, default=0}
       dyn_wings_carpals_02_r_ctrl.Location.Z {type=MovieSceneScriptingFloatChannel, default=0}
       dyn_wings_carpals_02_r_ctrl.Rotation.X {type=MovieSceneScriptingFloatChannel, default=0}
       dyn_wings_carpals_02_r_ctrl.Rotation.Y {type=MovieSceneScriptingFloatChannel, default=0}
       dyn_wings_carpals_02_r_ctrl.Rotation.Z {type=MovieSceneScriptingFloatChannel, default=0}
       dyn_wings_carpals_02_r_ctrl.Scale.X {type=MovieSceneScriptingFloatChannel, default=0}
       dyn_wings_carpals_02_r_ctrl.Scale.Y {type=MovieSceneScriptingFloatChannel, default=0}
       dyn_wings_carpals_02_r_ctrl.Scale.Z {type=MovieSceneScriptingFloatChannel, default=0}
       dyn_wings_metacarpal_03_r_ctrl.Location.X {type=MovieSceneScriptingFloatChannel, default=0}
       dyn_wings_metacarpal_03_r_ctrl.Location.Y {type=MovieSceneScriptingFloatChannel, default=0}
       dyn_wings_metacarpal_03_r_ctrl.Location.Z {type=MovieSceneScriptingFloatChannel, default=0}
       dyn_wings_metacarpal_03_r_ctrl.Rotation.X {type=MovieSceneScriptingFloatChannel, default=0}
       dyn_wings_metacarpal_03_r_ctrl.Rotation.Y {type=MovieSceneScriptingFloatChannel, default=0}
       dyn_wings_metacarpal_03_r_ctrl.Rotation.Z {type=MovieSceneScriptingFloatChannel, default=0}
       dyn_wings_metacarpal_03_r_ctrl.Scale.X {type=MovieSceneScriptingFloatChannel, default=0}
       dyn_wings_metacarpal_03_r_ctrl.Scale.Y {type=MovieSceneScriptingFloatChannel, default=0}
       dyn_wings_metacarpal_03_r_ctrl.Scale.Z {type=MovieSceneScriptingFloatChannel, default=0}
       dyn_wings_phalange_03_r_ctrl.Location.X {type=MovieSceneScriptingFloatChannel, default=0}
       dyn_wings_phalange_03_r_ctrl.Location.Y {type=MovieSceneScriptingFloatChannel, default=0}
       dyn_wings_phalange_03_r_ctrl.Location.Z {type=MovieSceneScriptingFloatChannel, default=0}
       dyn_wings_phalange_03_r_ctrl.Rotation.X {type=MovieSceneScriptingFloatChannel, default=0}
       dyn_wings_phalange_03_r_ctrl.Rotation.Y {type=MovieSceneScriptingFloatChannel, default=0}
       dyn_wings_phalange_03_r_ctrl.Rotation.Z {type=MovieSceneScriptingFloatChannel, default=0}
       dyn_wings_phalange_03_r_ctrl.Scale.X {type=MovieSceneScriptingFloatChannel, default=0}
       dyn_wings_phalange_03_r_ctrl.Scale.Y {type=MovieSceneScriptingFloatChannel, default=0}
       dyn_wings_phalange_03_r_ctrl.Scale.Z {type=MovieSceneScriptingFloatChannel, default=0}
       dyn_wings_metacarpal_04_r_ctrl.Location.X {type=MovieSceneScriptingFloatChannel, default=0}
       dyn_wings_metacarpal_04_r_ctrl.Location.Y {type=MovieSceneScriptingFloatChannel, default=0}
       dyn_wings_metacarpal_04_r_ctrl.Location.Z {type=MovieSceneScriptingFloatChannel, default=0}
       dyn_wings_metacarpal_04_r_ctrl.Rotation.X {type=MovieSceneScriptingFloatChannel, default=0}
       dyn_wings_metacarpal_04_r_ctrl.Rotation.Y {type=MovieSceneScriptingFloatChannel, default=0}
       dyn_wings_metacarpal_04_r_ctrl.Rotation.Z {type=MovieSceneScriptingFloatChannel, default=0}
       dyn_wings_metacarpal_04_r_ctrl.Scale.X {type=MovieSceneScriptingFloatChannel, default=0}
       dyn_wings_metacarpal_04_r_ctrl.Scale.Y {type=MovieSceneScriptingFloatChannel, default=0}
       dyn_wings_metacarpal_04_r_ctrl.Scale.Z {type=MovieSceneScriptingFloatChannel, default=0}
       dyn_wings_phalange_04_r_ctrl.Location.X {type=MovieSceneScriptingFloatChannel, default=0}
       dyn_wings_phalange_04_r_ctrl.Location.Y {type=MovieSceneScriptingFloatChannel, default=0}
       dyn_wings_phalange_04_r_ctrl.Location.Z {type=MovieSceneScriptingFloatChannel, default=0}
       dyn_wings_phalange_04_r_ctrl.Rotation.X {type=MovieSceneScriptingFloatChannel, default=0}
       dyn_wings_phalange_04_r_ctrl.Rotation.Y {type=MovieSceneScriptingFloatChannel, default=0}
       dyn_wings_phalange_04_r_ctrl.Rotation.Z {type=MovieSceneScriptingFloatChannel, default=0}
       dyn_wings_phalange_04_r_ctrl.Scale.X {type=MovieSceneScriptingFloatChannel, default=0}
       dyn_wings_phalange_04_r_ctrl.Scale.Y {type=MovieSceneScriptingFloatChannel, default=0}
       dyn_wings_phalange_04_r_ctrl.Scale.Z {type=MovieSceneScriptingFloatChannel, default=0}
      ]
     }
    ]
   }]
 }
 FortPlayerMannequin0 {
  id=#2
  possessed_class=FortPlayerMannequin
  tracks [
   {
    type=MovieScene3DTransformTrack
    sections [{
      type=MovieScene3DTransformSection
      channels [
       Location.X {type=MovieSceneScriptingDoubleChannel, default=-504}
       Location.Y {type=MovieSceneScriptingDoubleChannel, default=-1800}
       Location.Z {type=MovieSceneScriptingDoubleChannel, default=2136}
       Rotation.X {type=MovieSceneScriptingDoubleChannel, default=0}
       Rotation.Y {type=MovieSceneScriptingDoubleChannel, default=0}
       Rotation.Z {type=MovieSceneScriptingDoubleChannel, default=0}
       Scale.X {type=MovieSceneScriptingDoubleChannel, default=1}
       Scale.Y {type=MovieSceneScriptingDoubleChannel, default=1}
       Scale.Z {type=MovieSceneScriptingDoubleChannel, default=1}
      ]
     }]
   }
   {type=MovieSceneSkeletalAnimationTrack, sections []}
   {type=MovieSceneSpawnTrack, sections [{type=MovieSceneSpawnSection, channels [None {type=MovieSceneScriptingBoolChannel, default=true}]}]}
  ]
 }
]
display_rate {denominator=1, numerator=30}
tick_resolution {denominator=1, numerator=24000}
tracks []
