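# Animation Blueprint for the TimeSquare winged backpack cosmetic, adding physical secondary motion to its wings.
# An idle sequence feeds additive bone edits that shift the pack and yaw both upper wing bones in opposite directions.
# A serial chain of root-relative Anim Dynamics nodes with angular springs and gravity then sways each metacarpal and phalange bone, blended more strongly on the phalanges.
# The Event Graph holds no logic.
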AnimBlueprint /BRCosmetics/Accessories/FORT_Backpacks/Backpack_F_MED_TimeSquare/Meshes/F_MED_TimeSquare_Pack_AnimBP
target_skeleton=&/F_MED_TimeSquare_Pack_Skeleton
graphs [
 AnimGraph {nodes [
   n1 = "Transform (Modify) Bone - Bone: pack"(
    ComponentPose=Local To Component(LocalPose=Sequence Player 'TimeSquare_Idle_PuppetAnimTest_PACK'() {sequence=/BRCosmetics/Accessories/FORT_Backpacks/Backpack_F_MED_TimeSquare/Anim/TimeSquare_Idle_PuppetAnimTest_PACK})
    Translation="0,4,0"
    Rotation="0,0,0"
   ) {translationMode=BMM_Additive, alphaBoolBlend {blendOption=Linear}, boneToModify {boneName=pack}, translation {y=4}}
   n2 = "Transform (Modify) Bone - Bone: dyn_wings_humelus_l"(ComponentPose=@n1, Alpha=1, Translation="0,0,0", Rotation="0,-10,0", Scale="1,1,1") {rotationMode=BMM_Additive, alphaBoolBlend {blendOption=Linear}, boneToModify {boneName=dyn_wings_humelus_l}, rotation {yaw=-10}}
   n3 = "Transform (Modify) Bone - Bone: dyn_wings_humelus_r"(ComponentPose=@n2, Alpha=1, Translation="0,0,0", Rotation="0,10,0", Scale="1,1,1") {rotationMode=BMM_Additive, alphaBoolBlend {blendOption=Linear}, boneToModify {boneName=dyn_wings_humelus_r}, rotation {yaw=10}}
   n4 = Anim Dynamics - Bone: dyn_wings_metacarpal_01_l(ComponentPose=@n3, Alpha=0.5, ExternalForce="0,0,0") {
    angularDampingOverride=1
    angularSpringConstant=4000
    bAngularSpring=true
    bGravityOverrideInSimSpace=true
    bOverrideAngularDamping=true
    bUseGravityOverride=true
    simulationSpace=RootRelative
    alphaBoolBlend {blendOption=Linear}
    boundBone {boneName=dyn_wings_metacarpal_01_l}
    componentLinearAccScale=<0.4 0.4 0.4>
    gravityOverride {z=-980}
    physicsBodyDefinitions [{
      boundBone {boneName=dyn_wings_metacarpal_01_l}
      boxExtents=<10 20 20>
      constraintSetup {twistAxis=AxisY, angularLimitsMax=<0 0 30>, angularLimitsMin=<0 0 -35>, angularTarget=<1 0 0>}
      localJointOffset=<3 0 0>
     }]
   }
   n5 = Anim Dynamics - Bone: dyn_wings_metacarpal_02_l(ComponentPose=@n4, Alpha=0.5, ExternalForce="0,0,0") {
    angularDampingOverride=1
    angularSpringConstant=4000
    bAngularSpring=true
    bGravityOverrideInSimSpace=true
    bOverrideAngularDamping=true
    bUseGravityOverride=true
    simulationSpace=RootRelative
    alphaBoolBlend {blendOption=Linear}
    boundBone {boneName=dyn_wings_metacarpal_02_l}
    componentLinearAccScale=<0.4 0.4 0.4>
    gravityOverride {z=-980}
    physicsBodyDefinitions [{
      boundBone {boneName=dyn_wings_metacarpal_02_l}
      boxExtents=<10 20 20>
      constraintSetup {twistAxis=AxisY, angularLimitsMax=<0 0 30>, angularLimitsMin=<0 0 -35>, angularTarget=<1 0 0>}
      localJointOffset=<3 0 0>
     }]
   }
   n6 = Anim Dynamics - Bone: dyn_wings_metacarpal_03_l(ComponentPose=@n5, Alpha=0.5, ExternalForce="0,0,0") {
    angularDampingOverride=1
    angularSpringConstant=4000
    bAngularSpring=true
    bGravityOverrideInSimSpace=true
    bOverrideAngularDamping=true
    bUseGravityOverride=true
    simulationSpace=RootRelative
    alphaBoolBlend {blendOption=Linear}
    boundBone {boneName=dyn_wings_metacarpal_03_l}
    componentLinearAccScale=<0.4 0.4 0.4>
    gravityOverride {z=-980}
    physicsBodyDefinitions [{
      boundBone {boneName=dyn_wings_metacarpal_03_l}
      boxExtents=<10 20 20>
      constraintSetup {twistAxis=AxisY, angularLimitsMax=<0 0 30>, angularLimitsMin=<0 0 -35>, angularTarget=<1 0 0>}
      localJointOffset=<3 0 0>
     }]
   }
   n7 = Anim Dynamics - Bone: dyn_wings_metacarpal_04_l(ComponentPose=@n6, Alpha=0.5, ExternalForce="0,0,0") {
    angularDampingOverride=1
    angularSpringConstant=4000
    bAngularSpring=true
    bGravityOverrideInSimSpace=true
    bOverrideAngularDamping=true
    bUseGravityOverride=true
    simulationSpace=RootRelative
    alphaBoolBlend {blendOption=Linear}
    boundBone {boneName=dyn_wings_metacarpal_04_l}
    componentLinearAccScale=<0.4 0.4 0.4>
    gravityOverride {z=-980}
    physicsBodyDefinitions [{
      boundBone {boneName=dyn_wings_metacarpal_04_l}
      boxExtents=<10 20 20>
      constraintSetup {twistAxis=AxisY, angularLimitsMax=<0 0 30>, angularLimitsMin=<0 0 -35>, angularTarget=<1 0 0>}
      localJointOffset=<3 0 0>
     }]
   }
   n8 = Anim Dynamics - Bone: dyn_wings_phalange_01_l(ComponentPose=@n7, Alpha=0.7, ExternalForce="0,0,0") {
    angularDampingOverride=1
    angularSpringConstant=4000
    bAngularSpring=true
    bGravityOverrideInSimSpace=true
    bOverrideAngularDamping=true
    bUseGravityOverride=true
    simulationSpace=RootRelative
    alphaBoolBlend {blendOption=Linear}
    boundBone {boneName=dyn_wings_phalange_01_l}
    componentLinearAccScale=<0.4 0.4 0.4>
    gravityOverride {z=-980}
    physicsBodyDefinitions [{
      boundBone {boneName=dyn_wings_phalange_01_l}
      boxExtents=<10 20 20>
      constraintSetup {twistAxis=AxisY, angularLimitsMax=<0 0 30>, angularLimitsMin=<0 0 -35>, angularTarget=<1 0 0>}
      localJointOffset=<3 0 0>
     }]
   }
   n9 = Anim Dynamics - Bone: dyn_wings_phalange_02_l(ComponentPose=@n8, Alpha=0.7, ExternalForce="0,0,0") {
    angularDampingOverride=1
    angularSpringConstant=4000
    bAngularSpring=true
    bGravityOverrideInSimSpace=true
    bOverrideAngularDamping=true
    bUseGravityOverride=true
    simulationSpace=RootRelative
    alphaBoolBlend {blendOption=Linear}
    boundBone {boneName=dyn_wings_phalange_02_l}
    componentLinearAccScale=<0.4 0.4 0.4>
    gravityOverride {z=-980}
    physicsBodyDefinitions [{
      boundBone {boneName=dyn_wings_phalange_02_l}
      boxExtents=<10 20 20>
      constraintSetup {twistAxis=AxisY, angularLimitsMax=<0 0 30>, angularLimitsMin=<0 0 -35>, angularTarget=<1 0 0>}
      localJointOffset=<3 0 0>
     }]
   }
   n10 = Anim Dynamics - Bone: dyn_wings_phalange_03_l(ComponentPose=@n9, Alpha=0.7, ExternalForce="0,0,0") {
    angularDampingOverride=1
    angularSpringConstant=4000
    bAngularSpring=true
    bGravityOverrideInSimSpace=true
    bOverrideAngularDamping=true
    bUseGravityOverride=true
    simulationSpace=RootRelative
    alphaBoolBlend {blendOption=Linear}
    boundBone {boneName=dyn_wings_phalange_03_l}
    componentLinearAccScale=<0.4 0.4 0.4>
    gravityOverride {z=-980}
    physicsBodyDefinitions [{
      boundBone {boneName=dyn_wings_phalange_03_l}
      boxExtents=<10 20 20>
      constraintSetup {twistAxis=AxisY, angularLimitsMax=<0 0 30>, angularLimitsMin=<0 0 -35>, angularTarget=<1 0 0>}
      localJointOffset=<3 0 0>
     }]
   }
   n11 = Anim Dynamics - Bone: dyn_wings_phalange_04_l(ComponentPose=@n10, Alpha=0.7, ExternalForce="0,0,0") {
    angularDampingOverride=1
    angularSpringConstant=4000
    bAngularSpring=true
    bGravityOverrideInSimSpace=true
    bOverrideAngularDamping=true
    bUseGravityOverride=true
    simulationSpace=RootRelative
    alphaBoolBlend {blendOption=Linear}
    boundBone {boneName=dyn_wings_phalange_04_l}
    componentLinearAccScale=<0.4 0.4 0.4>
    gravityOverride {z=-980}
    physicsBodyDefinitions [{
      boundBone {boneName=dyn_wings_phalange_04_l}
      boxExtents=<10 20 20>
      constraintSetup {twistAxis=AxisY, angularLimitsMax=<0 0 30>, angularLimitsMin=<0 0 -35>, angularTarget=<1 0 0>}
      localJointOffset=<3 0 0>
     }]
   }
   n12 = Anim Dynamics - Bone: dyn_wings_metacarpal_01_r(ComponentPose=@n11, Alpha=0.5, ExternalForce="0,0,0") {
    angularDampingOverride=1
    angularSpringConstant=4000
    bAngularSpring=true
    bGravityOverrideInSimSpace=true
    bOverrideAngularDamping=true
    bUseGravityOverride=true
    simulationSpace=RootRelative
    alphaBoolBlend {blendOption=Linear}
    boundBone {boneName=dyn_wings_metacarpal_01_r}
    componentLinearAccScale=<0.4 0.4 0.4>
    gravityOverride {z=-980}
    physicsBodyDefinitions [{
      boundBone {boneName=dyn_wings_metacarpal_01_r}
      boxExtents=<10 20 20>
      constraintSetup {twistAxis=AxisY, angularLimitsMax=<0 0 30>, angularLimitsMin=<0 0 -35>, angularTarget=<1 0 0>}
      localJointOffset=<3 0 0>
     }]
   }
   n13 = Anim Dynamics - Bone: dyn_wings_metacarpal_02_r(ComponentPose=@n12, Alpha=0.5, ExternalForce="0,0,0") {
    angularDampingOverride=1
    angularSpringConstant=4000
    bAngularSpring=true
    bGravityOverrideInSimSpace=true
    bOverrideAngularDamping=true
    bUseGravityOverride=true
    simulationSpace=RootRelative
    alphaBoolBlend {blendOption=Linear}
    boundBone {boneName=dyn_wings_metacarpal_02_r}
    componentLinearAccScale=<0.4 0.4 0.4>
    gravityOverride {z=-980}
    physicsBodyDefinitions [{
      boundBone {boneName=dyn_wings_metacarpal_02_r}
      boxExtents=<10 20 20>
      constraintSetup {twistAxis=AxisY, angularLimitsMax=<0 0 30>, angularLimitsMin=<0 0 -35>, angularTarget=<1 0 0>}
      localJointOffset=<3 0 0>
     }]
   }
   n14 = Anim Dynamics - Bone: dyn_wings_metacarpal_03_r(ComponentPose=@n13, Alpha=0.5, ExternalForce="0,0,0") {
    angularDampingOverride=1
    angularSpringConstant=4000
    bAngularSpring=true
    bGravityOverrideInSimSpace=true
    bOverrideAngularDamping=true
    bUseGravityOverride=true
    simulationSpace=RootRelative
    alphaBoolBlend {blendOption=Linear}
    boundBone {boneName=dyn_wings_metacarpal_03_r}
    componentLinearAccScale=<0.4 0.4 0.4>
    gravityOverride {z=-980}
    physicsBodyDefinitions [{
      boundBone {boneName=dyn_wings_metacarpal_03_r}
      boxExtents=<10 20 20>
      constraintSetup {twistAxis=AxisY, angularLimitsMax=<0 0 30>, angularLimitsMin=<0 0 -35>, angularTarget=<1 0 0>}
      localJointOffset=<3 0 0>
     }]
   }
   n15 = Anim Dynamics - Bone: dyn_wings_metacarpal_04_r(ComponentPose=@n14, Alpha=0.5, ExternalForce="0,0,0") {
    angularDampingOverride=1
    angularSpringConstant=4000
    bAngularSpring=true
    bGravityOverrideInSimSpace=true
    bOverrideAngularDamping=true
    bUseGravityOverride=true
    simulationSpace=RootRelative
    alphaBoolBlend {blendOption=Linear}
    boundBone {boneName=dyn_wings_metacarpal_04_r}
    componentLinearAccScale=<0.4 0.4 0.4>
    gravityOverride {z=-980}
    physicsBodyDefinitions [{
      boundBone {boneName=dyn_wings_metacarpal_04_r}
      boxExtents=<10 20 20>
      constraintSetup {twistAxis=AxisY, angularLimitsMax=<0 0 30>, angularLimitsMin=<0 0 -35>, angularTarget=<1 0 0>}
      localJointOffset=<3 0 0>
     }]
   }
   n16 = Anim Dynamics - Bone: dyn_wings_phalange_01_r(ComponentPose=@n15, Alpha=0.7, ExternalForce="0,0,0") {
    angularDampingOverride=1
    angularSpringConstant=4000
    bAngularSpring=true
    bGravityOverrideInSimSpace=true
    bOverrideAngularDamping=true
    bUseGravityOverride=true
    simulationSpace=RootRelative
    alphaBoolBlend {blendOption=Linear}
    boundBone {boneName=dyn_wings_phalange_01_r}
    componentLinearAccScale=<0.4 0.4 0.4>
    gravityOverride {z=-980}
    physicsBodyDefinitions [{
      boundBone {boneName=dyn_wings_phalange_01_r}
      boxExtents=<10 20 20>
      constraintSetup {twistAxis=AxisY, angularLimitsMax=<0 0 30>, angularLimitsMin=<0 0 -35>, angularTarget=<1 0 0>}
      localJointOffset=<3 0 0>
     }]
   }
   n17 = Anim Dynamics - Bone: dyn_wings_phalange_02_r(ComponentPose=@n16, Alpha=0.7, ExternalForce="0,0,0") {
    angularDampingOverride=1
    angularSpringConstant=4000
    bAngularSpring=true
    bGravityOverrideInSimSpace=true
    bOverrideAngularDamping=true
    bUseGravityOverride=true
    simulationSpace=RootRelative
    alphaBoolBlend {blendOption=Linear}
    boundBone {boneName=dyn_wings_phalange_02_r}
    componentLinearAccScale=<0.4 0.4 0.4>
    gravityOverride {z=-980}
    physicsBodyDefinitions [{
      boundBone {boneName=dyn_wings_phalange_02_r}
      boxExtents=<10 20 20>
      constraintSetup {twistAxis=AxisY, angularLimitsMax=<0 0 30>, angularLimitsMin=<0 0 -35>, angularTarget=<1 0 0>}
      localJointOffset=<3 0 0>
     }]
   }
   n18 = Anim Dynamics - Bone: dyn_wings_phalange_03_r(ComponentPose=@n17, Alpha=0.7, ExternalForce="0,0,0") {
    angularDampingOverride=1
    angularSpringConstant=4000
    bAngularSpring=true
    bGravityOverrideInSimSpace=true
    bOverrideAngularDamping=true
    bUseGravityOverride=true
    simulationSpace=RootRelative
    alphaBoolBlend {blendOption=Linear}
    boundBone {boneName=dyn_wings_phalange_03_r}
    componentLinearAccScale=<0.4 0.4 0.4>
    gravityOverride {z=-980}
    physicsBodyDefinitions [{
      boundBone {boneName=dyn_wings_phalange_03_r}
      boxExtents=<10 20 20>
      constraintSetup {twistAxis=AxisY, angularLimitsMax=<0 0 30>, angularLimitsMin=<0 0 -35>, angularTarget=<1 0 0>}
      localJointOffset=<3 0 0>
     }]
   }
   n19 = Anim Dynamics - Bone: dyn_wings_phalange_04_r(ComponentPose=@n18, Alpha=0.7, ExternalForce="0,0,0") {
    angularDampingOverride=1
    angularSpringConstant=4000
    bAngularSpring=true
    bGravityOverrideInSimSpace=true
    bOverrideAngularDamping=true
    bUseGravityOverride=true
    simulationSpace=RootRelative
    alphaBoolBlend {blendOption=Linear}
    boundBone {boneName=dyn_wings_phalange_04_r}
    componentLinearAccScale=<0.4 0.4 0.4>
    gravityOverride {z=-980}
    physicsBodyDefinitions [{
      boundBone {boneName=dyn_wings_phalange_04_r}
      boxExtents=<10 20 20>
      constraintSetup {twistAxis=AxisY, angularLimitsMax=<0 0 30>, angularLimitsMin=<0 0 -35>, angularTarget=<1 0 0>}
      localJointOffset=<3 0 0>
     }]
   }
   Output Pose(Result=Component To Local(ComponentPose=@n19)) {layerGroup=None}
  ]}
 EventGraph {nodes [
   TryGetPawnOwner()
   Event BlueprintUpdateAnimation()
  ]}
]
variables []
