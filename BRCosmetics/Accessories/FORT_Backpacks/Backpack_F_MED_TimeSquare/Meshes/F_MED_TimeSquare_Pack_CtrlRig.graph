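# Control Rig for the TimeSquare backpack cosmetic, exposing animatable controls for its pack bone and a pair of left and right bone-jointed wings.
# A control hierarchy mirrors the bone chain from humerus through radius, carpals, metacarpals and phalanges under pack_ctrl.
# The forward solve batch-reads global transforms from the control list and writes them onto the matching bone list, propagating to children, with no IK or procedural motion.
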ControlRigBlueprint /BRCosmetics/Accessories/FORT_Backpacks/Backpack_F_MED_TimeSquare/Meshes/F_MED_TimeSquare_Pack_CtrlRig
graphs [
 RigVMFunctionLibrary {links [], nodes []}
 RigVMModel {nodes [
   RigUnit_BeginExecution = &RigUnit_BeginExecution()
   RigUnit_ItemArray = &RigUnit_ItemArray(Items="((Type=Control,Name=\"pack_ctrl\"),(Type=Control,Name=\"dyn_wings_humelus_l_ctrl\"),(Type=Control,Name=\"dyn_wings_radius_l_ctrl\"),(Type=Control,Name=\"dyn_wings_metacarpal_01_l_ctrl\"),(Type=Control,Name=\"dyn_wings_phalange_01_l_ctrl\"),(Type=Control,Name=\"dyn_wi ...(truncated)")
   n1 = &RigUnit_ItemArray(Items="((Type=Bone,Name=\"pack\"),(Type=Bone,Name=\"dyn_wings_humelus_l\"),(Type=Bone,Name=\"dyn_wings_radius_l\"),(Type=Bone,Name=\"dyn_wings_metacarpal_01_l\"),(Type=Bone,Name=\"dyn_wings_phalange_01_l\"),(Type=Bone,Name=\"dyn_wings_carpals_01_l\"),(Type=Bone,Name=\"dyn_win ...(truncated)")
   /Script/ControlRig.RigUnit_SetTransformItemArray(
    Space=GlobalSpace
    Weight=1
    bInitial=False
    bPropagateToChildren=True
    Transforms=/Script/ControlRig.RigUnit_GetTransformItemArray(Space=GlobalSpace, bInitial=False, Items=@RigUnit_ItemArray.Items) {is=struct RigVMUnitNode}.Transforms
    ExecutePin=@RigUnit_BeginExecution.ExecutePin
    Items=@n1.Items
   ) {is=struct RigVMUnitNode}
  ]}
]
hierarchy [
 root {loc=<0 0 0>, rot=<0 0 0 1>}
 pack {parent=root, loc=<0 -0.7316 0>, rot=<0 0 0 1>}
 dyn_wings_humelus_l {parent=pack, loc=<5.0168 -0.4741 1.0999>, rot=<0.879 -0.4709 0.0244 0.0715>}
 dyn_wings_radius_l {parent=dyn_wings_humelus_l, loc=<7.1098 0 0>, rot=<0.0216 -0.4197 0.0499 -0.906>}
 dyn_wings_metacarpal_01_l {parent=dyn_wings_radius_l, loc=<7.7057 0.7749 2.253>, rot=<0.0231 -0.9599 -0.0067 0.2795>}
 dyn_wings_phalange_01_l {parent=dyn_wings_metacarpal_01_l, loc=<7.7519 0 0>, rot=<0 0 0 1>}
 dyn_wings_carpals_01_l {parent=dyn_wings_radius_l, loc=<23.8601 0.1464 -0.8203>, rot=<-0.1249 0.3999 0.2139 -0.8825>}
 dyn_wings_metacarpal_02_l {parent=dyn_wings_carpals_01_l, loc=<0.2091 0.0146 4.2214>, rot=<-0.0054 0.7071 -0.0054 -0.7071>}
 dyn_wings_phalange_02_l {parent=dyn_wings_metacarpal_02_l, loc=<15.1291 0 0>, rot=<0 0 0 1>}
 dyn_wings_carpals_02_l {parent=dyn_wings_carpals_01_l, loc=<0.7215 -0.0071 0>, rot=<-0.0274 0.1041 -0.016 -0.9941>}
 dyn_wings_metacarpal_03_l {parent=dyn_wings_carpals_02_l, loc=<5.67 0.0482 1.6882>, rot=<-0.0382 0.4384 -0.015 -0.8979>}
 dyn_wings_phalange_03_l {parent=dyn_wings_metacarpal_03_l, loc=<10.0356 0 0>, rot=<0 0 0 1>}
 dyn_wings_metacarpal_04_l {parent=dyn_wings_carpals_02_l, loc=<9.1629 -0.0188 -0.0667>, rot=<0.013 0.3539 0.0304 -0.9347>}
 dyn_wings_phalange_04_l {parent=dyn_wings_metacarpal_04_l, loc=<12.4184 0 0>, rot=<0 0 0 1>}
 dyn_wings_humelus_r {parent=pack, loc=<-5.0168 -0.4741 1.0999>, rot=<0.4709 -0.879 0.0715 0.0244>}
 dyn_wings_radius_r {parent=dyn_wings_humelus_r, loc=<7.1098 0 0>, rot=<-0.0216 -0.4197 -0.0499 -0.906>}
 dyn_wings_metacarpal_01_r {parent=dyn_wings_radius_r, loc=<7.7057 -0.7749 2.2529>, rot=<-0.0231 -0.9599 0.0067 0.2795>}
 dyn_wings_phalange_01_r {parent=dyn_wings_metacarpal_01_r, loc=<7.7519 -0.0001 -0.0001>, rot=<0 0 0 1>}
 dyn_wings_carpals_01_r {parent=dyn_wings_radius_r, loc=<23.8525 -0.121 -0.8309>, rot=<0.1187 0.4018 -0.2002 -0.8857>}
 dyn_wings_metacarpal_02_r {parent=dyn_wings_carpals_01_r, loc=<0.2096 -0.0367 4.2213>, rot=<-0.0055 0.7071 -0.0055 -0.7071>}
 dyn_wings_phalange_02_r {parent=dyn_wings_metacarpal_02_r, loc=<15.129 0 0>, rot=<0 0 0 1>}
 dyn_wings_carpals_02_r {parent=dyn_wings_carpals_01_r, loc=<0.7212 0.0008 0>, rot=<0.0258 0.1045 0.0007 -0.9942>}
 dyn_wings_metacarpal_03_r {parent=dyn_wings_carpals_02_r, loc=<5.67 -0.0481 1.6882>, rot=<0.0382 0.4384 0.015 -0.8979>}
 dyn_wings_phalange_03_r {parent=dyn_wings_metacarpal_03_r, loc=<10.0356 0 0>, rot=<0 0 0 1>}
 dyn_wings_metacarpal_04_r {parent=dyn_wings_carpals_02_r, loc=<9.1629 0.0188 -0.0667>, rot=<-0.013 0.3539 -0.0304 -0.9347>}
 dyn_wings_phalange_04_r {parent=dyn_wings_metacarpal_04_r, loc=<12.4185 -0.0001 0>, rot=<0 0 0 1>}
 pack_ctrl {controlType=EULER_TRANSFORM, type=Control, loc=<0 0 0>, rot=<0 0 0 1>}
 dyn_wings_humelus_l_ctrl {controlType=EULER_TRANSFORM, parent=pack_ctrl, type=Control, loc=<0 0 0>, rot=<0 0 0 1>}
 dyn_wings_radius_l_ctrl {controlType=EULER_TRANSFORM, parent=dyn_wings_humelus_l_ctrl, type=Control, loc=<0 0 0>, rot=<0 0 0 1>}
 dyn_wings_metacarpal_01_l_ctrl {controlType=EULER_TRANSFORM, parent=dyn_wings_radius_l_ctrl, type=Control, loc=<0 0 0>, rot=<0 0 0 1>}
 dyn_wings_phalange_01_l_ctrl {controlType=EULER_TRANSFORM, parent=dyn_wings_metacarpal_01_l_ctrl, type=Control, loc=<0 0 0>, rot=<0 0 0 1>}
 dyn_wings_carpals_01_l_ctrl {controlType=EULER_TRANSFORM, parent=dyn_wings_radius_l_ctrl, type=Control, loc=<0 0 0>, rot=<0 0 0 1>}
 dyn_wings_metacarpal_02_l_ctrl {controlType=EULER_TRANSFORM, parent=dyn_wings_carpals_01_l_ctrl, type=Control, loc=<0 0 0>, rot=<0 0 0 1>}
 dyn_wings_phalange_02_l_ctrl {controlType=EULER_TRANSFORM, parent=dyn_wings_metacarpal_02_l_ctrl, type=Control, loc=<0 0 0>, rot=<0 0 0 1>}
 dyn_wings_carpals_02_l_ctrl {controlType=EULER_TRANSFORM, parent=dyn_wings_carpals_01_l_ctrl, type=Control, loc=<0 0 0>, rot=<0 0 0 1>}
 dyn_wings_metacarpal_03_l_ctrl {controlType=EULER_TRANSFORM, parent=dyn_wings_carpals_02_l_ctrl, type=Control, loc=<0 0 0>, rot=<0 0 0 1>}
 dyn_wings_phalange_03_l_ctrl {controlType=EULER_TRANSFORM, parent=dyn_wings_metacarpal_03_l_ctrl, type=Control, loc=<0 0 0>, rot=<0 0 0 1>}
 dyn_wings_metacarpal_04_l_ctrl {controlType=EULER_TRANSFORM, parent=dyn_wings_carpals_02_l_ctrl, type=Control, loc=<0 0 0>, rot=<0 0 0 1>}
 dyn_wings_phalange_04_l_ctrl {controlType=EULER_TRANSFORM, parent=dyn_wings_metacarpal_04_l_ctrl, type=Control, loc=<0 0 0>, rot=<0 0 0 1>}
 dyn_wings_humelus_r_ctrl {controlType=EULER_TRANSFORM, parent=pack_ctrl, type=Control, loc=<0 0 0>, rot=<0 0 0 1>}
 dyn_wings_radius_r_ctrl {controlType=EULER_TRANSFORM, parent=dyn_wings_humelus_r_ctrl, type=Control, loc=<0 0 0>, rot=<0 0 0 1>}
 dyn_wings_metacarpal_01_r_ctrl {controlType=EULER_TRANSFORM, parent=dyn_wings_radius_r_ctrl, type=Control, loc=<0 0 0>, rot=<0 0 0 1>}
 dyn_wings_phalange_01_r_ctrl {controlType=EULER_TRANSFORM, parent=dyn_wings_metacarpal_01_r_ctrl, type=Control, loc=<0 0 0>, rot=<0 0 0 1>}
 dyn_wings_carpals_01_r_ctrl {controlType=EULER_TRANSFORM, parent=dyn_wings_radius_r_ctrl, type=Control, loc=<0 0 0>, rot=<0 0 0 1>}
 dyn_wings_metacarpal_02_r_ctrl {controlType=EULER_TRANSFORM, parent=dyn_wings_carpals_01_r_ctrl, type=Control, loc=<0 0 0>, rot=<0 0 0 1>}
 dyn_wings_phalange_02_r_ctrl {controlType=EULER_TRANSFORM, parent=dyn_wings_metacarpal_02_r_ctrl, type=Control, loc=<0 0 0>, rot=<0 0 0 1>}
 dyn_wings_carpals_02_r_ctrl {controlType=EULER_TRANSFORM, parent=dyn_wings_carpals_01_r_ctrl, type=Control, loc=<0 0 0>, rot=<0 0 0 1>}
 dyn_wings_metacarpal_03_r_ctrl {controlType=EULER_TRANSFORM, parent=dyn_wings_carpals_02_r_ctrl, type=Control, loc=<0 0 0>, rot=<0 0 0 1>}
 dyn_wings_phalange_03_r_ctrl {controlType=EULER_TRANSFORM, parent=dyn_wings_metacarpal_03_r_ctrl, type=Control, loc=<0 0 0>, rot=<0 0 0 1>}
 dyn_wings_metacarpal_04_r_ctrl {controlType=EULER_TRANSFORM, parent=dyn_wings_carpals_02_r_ctrl, type=Control, loc=<0 0 0>, rot=<0 0 0 1>}
 dyn_wings_phalange_04_r_ctrl {controlType=EULER_TRANSFORM, parent=dyn_wings_metacarpal_04_r_ctrl, type=Control, loc=<0 0 0>, rot=<0 0 0 1>}
]
